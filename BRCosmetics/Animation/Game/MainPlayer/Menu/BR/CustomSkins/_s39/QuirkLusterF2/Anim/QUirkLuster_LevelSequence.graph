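# Level Sequence for animating the QuirkLuster custom-skin main-menu player animation on a spawned FortPlayerMannequin.
# The mannequin carries a transform track, a skeletal animation section over frames 0–296, and a Fortnite_Biped_CtrlRig track.
# Its Face component has a Fortnite_ExtraJoint_CtrlRig track, while Body has none.
# Biped keys look baked, with both FK and IK leg controls keyed and dropping to zero after frame 296.
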LevelSequence /BRCosmetics/Animation/Game/MainPlayer/Menu/BR/CustomSkins/_s39/QuirkLusterF2/Anim/QUirkLuster_LevelSequence
bindings [
 Body {id=#1, parent_id=#2, possessed_class=SkeletalMeshComponentBudgeted, tracks []}
 Face {
  id=#3
  parent_id=#2
  possessed_class=SkeletalMeshComponentBudgeted
  tracks [{type=MovieSceneControlRigParameterTrack, display_name=Fortnite_ExtraJoint_CtrlRig, sections [{type=MovieSceneControlRigParameterSection}]}]
 }
 FortPlayerMannequin {
  id=#2
  possessed_class=FortPlayerMannequin
  tracks [
   {type=MovieSceneSpawnTrack, sections [{type=MovieSceneSpawnSection, channels [None {type=MovieSceneScriptingBoolChannel, default=true}]}]}
   {
    type=MovieScene3DTransformTrack
    sections [{
      type=MovieScene3DTransformSection
      channels [
       Location.X {type=MovieSceneScriptingDoubleChannel, default=0}
       Location.Y {type=MovieSceneScriptingDoubleChannel, default=0}
       Location.Z {type=MovieSceneScriptingDoubleChannel, default=0}
       Rotation.X {type=MovieSceneScriptingDoubleChannel, default=0}
       Rotation.Y {type=MovieSceneScriptingDoubleChannel, default=0}
       Rotation.Z {type=MovieSceneScriptingDoubleChannel, default=0}
       Scale.X {type=MovieSceneScriptingDoubleChannel, default=1}
       Scale.Y {type=MovieSceneScriptingDoubleChannel, default=1}
       Scale.Z {type=MovieSceneScriptingDoubleChannel, default=1}
      ]
     }]
   }
   {type=MovieSceneSkeletalAnimationTrack, sections [{type=MovieSceneSkeletalAnimationSection, end_frame=296, start_frame=0}]}
   {
    type=MovieSceneControlRigParameterTrack
    display_name=Fortnite_Biped_CtrlRig
    sections [{
      type=MovieSceneControlRigParameterSection
      channels [
       leg_l_stretch_switch {type=MovieSceneScriptingBoolChannel, default=false, const [false,368]}
       leg_r_stretch_switch {type=MovieSceneScriptingBoolChannel, default=false, const [false,368]}
       arm_l_fk_local {type=MovieSceneScriptingBoolChannel, default=true, const [true,368]}
       Auto Interpolate_3 {type=MovieSceneScriptingBoolChannel, default=false, const [false,368]}
       Auto Interpolate {type=MovieSceneScriptingBoolChannel, default=false, const [false,368]}
       arm_l_stretch_switch {type=MovieSceneScriptingBoolChannel, default=false, const [false,368]}
       Keep Aligned {type=MovieSceneScriptingBoolChannel, default=false, const [false,368]}
       arm_r_fk_local {type=MovieSceneScriptingBoolChannel, default=true, const [true,368]}
       Auto Interpolate_4 {type=MovieSceneScriptingBoolChannel, default=false, const [false,368]}
       Auto Interpolate_2 {type=MovieSceneScriptingBoolChannel, default=false, const [false,368]}
       arm_r_stretch_switch {type=MovieSceneScriptingBoolChannel, default=false, const [false,368]}
       Keep Aligned_2 {type=MovieSceneScriptingBoolChannel, default=false, const [false,368]}
       head_local {type=MovieSceneScriptingBoolChannel, default=true, const [true,368]}
       Stretch {type=MovieSceneScriptingBoolChannel, default=false, const [false,368]}
       head_stretch {type=MovieSceneScriptingBoolChannel, default=false, const [false,368]}
       spine_stretch_switch {type=MovieSceneScriptingBoolChannel, default=false, const [false,368]}
       arm_l_fk_ik_switch {
        type=MovieSceneScriptingBoolChannel
        default=false
        keys [[-109,true],[-56,false],[-16,true],[0,false],[1,false],[2,false],[3,false],[4,false],[5,false],[6,false],[7,false],[8,false],[9,false],[10,false],[11,false],[12,false],[13,false],[14,false],[15,false],[16,false],[17,false],[18,false],[19,false],[20,false],[21,false],[22,false],[23,false],[24,false],[25,false],[26,false],[27,false],[28,false],[29,false],[30,false],[31,false],[32,false],[33,false],[34,false],[35,false],[36,false],[37,false],[38,false],[39,false],[40,false],[41,false],[42,false],[43,false],[44,false],[45,false],[46,false],[47,false],[48,false],[49,false],[50,false],[51,false],[52,false],[53,false],[54,false],[55,false],[56,false],[57,false],[58,false],[59,false],[60,false],[61,false],[62,false],[63,false],[64,false],[65,false],[66,false],[67,false],[68,false],[69,false],[70,false],[71,false],[72,false],[73,false],[74,false],[75,false],[76,false],[77,false],[78,false],[79,false],[80,false],[81,false],[82,false],[83,false],[84,false],[85,false],[86,false],[87,false],[88,false],[89,false],[90,false],[91,false],[92,false],[93,false],[94,false],[95,false],[96,false],[97,false],[98,false],[99,false],[100,false],[101,false],[102,false],[103,false],[104,false],[105,false],[106,false],[107,false],[108,false],[109,false],[110,false],[111,false],[112,false],[113,false],[114,false],[115,false],[116,false],[117,false],[118,false],[119,false],[120,false],[121,false],[122,false],[123,false],[124,false],[125,false],[126,false],[127,false],[128,false],[129,false],[130,false],[131,false],[132,false],[133,false],[134,false],[135,false],[136,false],[137,false],[138,false],[139,false],[140,false],[141,false],[142,false],[143,false],[144,false],[145,false],[146,false],[147,false],[148,false],[149,false],[150,false],[151,false],[152,false],[153,false],[154,false],[155,false],[156,false],[157,false],[158,false],[159,false],[160,false],[161,false],[162,false],[163,false],[164,false],[165,false],[166,false],[167,false],[168,false],[169,false],[170,false],[171,false],[172,false],[173,false],[174,false],[175,false],[176,false],[177,false],[178,false],[179,false],[180,false],[181,false],[182,false],[183,false],[184,false],[185,false],[186,false],[187,false],[188,false],[189,false],[190,false],[191,false],[192,false],[193,false],[194,false],[195,false],[196,false],[197,false],[198,false],[199,false],[200,false],[201,false],[202,false],[203,false],[204,false],[205,false],[206,false],[207,false],[208,false],[209,false],[210,false],[211,false],[212,false],[213,false],[214,false],[215,false],[216,false],[217,false],[218,false],[219,false],[220,false],[221,false],[222,false],[223,false],[224,false],[225,false],[226,false],[227,false],[228,false],[229,false],[230,false],[231,false],[232,false],[233,false],[234,false],[235,false],[236,false],[237,false],[238,false],[239,false],[240,false],[241,false],[242,false],[243,false],[244,false],[245,false],[246,false],[247,false],[248,false],[249,false],[250,false],[251,false],[252,false],[253,false],[254,false],[255,false],[256,false],[257,false],[258,false],[259,false],[260,false],[261,false],[262,false],[263,false],[264,false],[265,false],[266,false],[267,false],[268,false],[269,false],[270,false],[271,false],[272,false],[273,false],[274,false],[275,false],[276,false],[277,false],[278,false],[279,false],[280,false],[281,false],[282,false],[283,false],[284,false],[285,false],[286,false],[287,false],[288,false],[289,false],[290,false],[291,false],[292,false],[293,false],[294,false],[295,false],[296,false],[297,false],[298,false],[299,false],[300,false],[301,false],[302,false],[303,false],[304,false],[305,false],[306,false],[307,false],[308,false],[309,false],[310,false],[311,false],[312,false],[313,false],[314,false],[315,false],[316,false],[317,false],[318,false],[319,false],[320,false],[321,false],[322,false],[323,false],[324,false],[325,false],[326,false],[327,false],[328,false],[329,false],[330,false],[331,false],[332,false],[333,false],[334,false],[335,false],[336,false],[337,false],[338,false],[339,false],[340,false],[341,false],[342,false],[343,false],[344,false],[345,false],[346,false],[347,false],[348,false],[349,false],[350,false],[351,false],[352,false],[353,false],[354,false],[355,false],[356,false],[357,false],[358,false],[359,false],[360,false],[361,false],[362,false],[363,false],[364,false],[365,false],[366,false],[367,false]]
       }
       leg_l_fk_ik_switch {type=MovieSceneScriptingBoolChannel, default=true, const [true,368]}
       arm_r_fk_ik_switch {type=MovieSceneScriptingBoolChannel, default=false, const [false,368]}
       leg_r_fk_ik_switch {type=MovieSceneScriptingBoolChannel, default=true, const [true,368]}
       ShowBodyControls {type=MovieSceneScriptingBoolChannel, default=true, const [true,368]}
       spine_fk_ik_switch {type=MovieSceneScriptingBoolChannel, default=false, const [false,368]}
       neck_fk_ik_switch {type=MovieSceneScriptingBoolChannel, default=false, const [false,368]}
       space {type=MovieSceneScriptingByteChannel, default=1, const [1,368]}
       global_ctrl.Location.X {type=MovieSceneScriptingFloatChannel, default=0, const [0,368]}
       global_ctrl.Location.Y {type=MovieSceneScriptingFloatChannel, default=0, const [0,368]}
       global_ctrl.Location.Z {type=MovieSceneScriptingFloatChannel, default=0, const [0,368]}
       global_ctrl.Rotation.X {type=MovieSceneScriptingFloatChannel, default=0, const [0,368]}
       global_ctrl.Rotation.Y {type=MovieSceneScriptingFloatChannel, default=0, const [0,368]}
       global_ctrl.Rotation.Z {type=MovieSceneScriptingFloatChannel, default=0, const [0,368]}
       root_ctrl.Location.X {type=MovieSceneScriptingFloatChannel, default=0, const [0,368]}
       root_ctrl.Location.Y {type=MovieSceneScriptingFloatChannel, default=0, const [0,368]}
       root_ctrl.Location.Z {type=MovieSceneScriptingFloatChannel, default=0, const [0,368]}
       root_ctrl.Rotation.X {type=MovieSceneScriptingFloatChannel, default=0, const [0,368]}
       root_ctrl.Rotation.Y {type=MovieSceneScriptingFloatChannel, default=0, const [0,368]}
       root_ctrl.Rotation.Z {type=MovieSceneScriptingFloatChannel, default=0, const [0,368]}
       root_ctrl.Scale.X {type=MovieSceneScriptingFloatChannel, default=1, const [1,368]}
       root_ctrl.Scale.Y {type=MovieSceneScriptingFloatChannel, default=1, const [1,368]}
       root_ctrl.Scale.Z {type=MovieSceneScriptingFloatChannel, default=1, const [1,368]}
       body_offset_ctrl.Location.X {type=MovieSceneScriptingFloatChannel, default=0, const [0,368]}
       body_offset_ctrl.Location.Y {type=MovieSceneScriptingFloatChannel, default=0, const [0,368]}
       body_offset_ctrl.Location.Z {type=MovieSceneScriptingFloatChannel, default=0, const [0,368]}
       body_offset_ctrl.Rotation.X {type=MovieSceneScriptingFloatChannel, default=0, const [0,368]}
       body_offset_ctrl.Rotation.Y {type=MovieSceneScriptingFloatChannel, default=0, const [0,368]}
       body_offset_ctrl.Rotation.Z {type=MovieSceneScriptingFloatChannel, default=0, const [0,368]}
       body_offset_ctrl.Scale.X {type=MovieSceneScriptingFloatChannel, default=1, const [1,368]}
       body_offset_ctrl.Scale.Y {type=MovieSceneScriptingFloatChannel, default=1, const [1,368]}
       body_offset_ctrl.Scale.Z {type=MovieSceneScriptingFloatChannel, default=1, const [1,368]}
       thigh_l_fk_ctrl.Location.X {type=MovieSceneScriptingFloatChannel, default=0, thinned_from=368, keys [[0,0],[367,0]]}
       thigh_l_fk_ctrl.Location.Y {type=MovieSceneScriptingFloatChannel, default=0, thinned_from=368, keys [[0,0],[367,0]]}
       thigh_l_fk_ctrl.Location.Z {type=MovieSceneScriptingFloatChannel, default=0, thinned_from=368, keys [[0,0],[367,0]]}
       thigh_l_fk_ctrl.Rotation.X {
        type=MovieSceneScriptingFloatChannel
        default=1.0553
        thinned_from=368
        keys [[0,27.9386],[47,27.9336],[49,27.7691],[51,27.4664],[54,26.8061],[61,24.9963],[65,24.1931],[68,23.7931],[98,23.2291],[100,23.2886],[103,23.5738],[106,24.0493],[120,27.1311],[124,27.7575],[173,29.0232],[174,28.9819],[175,28.8591],[176,28.6578],[177,28.3475],[178,27.9683],[179,27.5312],[181,26.4941],[187,22.8851],[189,21.8017],[191,20.8812],[192,20.4865],[193,20.135],[194,19.8361],[195,19.5844],[196,19.3808],[197,19.2225],[199,19.0582],[200,19.0327],[201,19.0481],[202,19.1236],[203,19.2568],[204,19.4489],[205,19.6963],[206,19.9939],[208,20.7246],[217,24.7519],[221,26.2826],[223,26.9293],[226,27.7016],[229,28.27],[231,28.5361],[233,28.6747],[235,28.6261],[240,27.9449],[243,27.7029],[296,27.9368],[297,0],[367,0]]
       }
       thigh_l_fk_ctrl.Rotation.Y {
        type=MovieSceneScriptingFloatChannel
        default=-1.0258
        thinned_from=368
        keys [[0,9.8597],[39,9.9883],[43,9.91],[51,9.4716],[53,9.3941],[56,9.3572],[100,9.5013],[123,9.9618],[173,9.8799],[174,9.9112],[175,9.9982],[176,10.1425],[177,10.4606],[179,11.2013],[181,11.9868],[182,12.349],[183,12.6584],[184,12.9077],[185,13.0899],[186,13.2007],[187,13.2444],[188,13.2307],[190,13.093],[192,12.9123],[193,12.8493],[194,12.8151],[197,12.8408],[199,12.7988],[200,12.7441],[201,12.6563],[202,12.51],[203,12.2972],[204,12.0276],[205,11.7138],[209,10.3218],[211,9.6737],[213,9.0992],[214,8.8409],[216,8.3711],[217,8.1628],[218,7.9726],[219,7.8004],[221,7.4999],[222,7.3639],[224,7.3387],[226,7.3623],[227,7.3918],[228,7.4372],[229,7.4997],[232,7.7779],[235,8.1159],[237,8.3601],[238,8.4664],[239,8.5528],[240,8.6164],[241,8.6569],[242,8.6799],[244,8.6769],[251,8.5225],[258,8.5417],[261,8.6127],[265,8.787],[276,9.4282],[283,9.7064],[291,9.8433],[296,9.8572],[297,0],[367,0]]
       }
       thigh_l_fk_ctrl.Rotation.Z {
        type=MovieSceneScriptingFloatChannel
        default=-0.0727
        thinned_from=368
        keys [[0,-5.2749],[5,-5.293],[25,-5.6056],[35,-5.6541],[46,-5.6179],[47,-5.6269],[49,-5.6811],[51,-5.7796],[54,-5.9856],[59,-6.4255],[60,-6.5057],[61,-6.572],[64,-6.6981],[68,-6.8006],[98,-6.8619],[99,-6.8547],[101,-6.8014],[103,-6.7139],[106,-6.5336],[119,-5.5215],[121,-5.3897],[123,-5.2802],[125,-5.1997],[137,-4.9975],[164,-4.8286],[173,-4.8292],[174,-4.8415],[175,-4.871],[176,-4.9182],[177,-5.2126],[179,-5.8542],[180,-6.2063],[181,-6.5776],[185,-8.1195],[186,-8.4899],[188,-9.1861],[191,-10.1448],[193,-10.7088],[194,-10.9585],[195,-11.1852],[196,-11.3791],[197,-11.5363],[198,-11.6553],[199,-11.731],[200,-11.7656],[201,-11.7569],[202,-11.7083],[203,-11.6292],[204,-11.5148],[205,-11.3588],[206,-11.1743],[207,-10.9564],[208,-10.7101],[210,-10.1444],[212,-9.5099],[215,-8.504],[218,-7.4624],[219,-7.1286],[220,-6.8096],[222,-6.0988],[224,-5.7993],[226,-5.5339],[228,-5.3096],[229,-5.2097],[231,-5.0424],[232,-4.9785],[233,-4.9325],[234,-4.9072],[235,-4.9044],[236,-4.9244],[239,-5.03],[241,-5.0712],[243,-5.0798],[251,-4.9839],[254,-4.9336],[256,-4.9191],[293,-5.2654],[296,-5.2757],[297,0],[367,0]]
       }
       thigh_l_fk_ctrl.Scale.X {type=MovieSceneScriptingFloatChannel, default=1, const [1,368]}
       thigh_l_fk_ctrl.Scale.Y {type=MovieSceneScriptingFloatChannel, default=1, const [1,368]}
       thigh_l_fk_ctrl.Scale.Z {type=MovieSceneScriptingFloatChannel, default=1, const [1,368]}
       calf_l_fk_ctrl.Location.X {
        type=MovieSceneScriptingFloatChannel
        default=0
        thinned_from=368
        keys [[0,-0.0832],[25,-0.0831],[26,-0.0945],[33,-0.0945],[34,-0.0907],[35,-0.0907],[36,-0.1002],[41,-0.1002],[42,-0.1097],[44,-0.1097],[46,-0.1287],[47,-0.1287],[50,-0.1571],[51,-0.1609],[55,-0.1913],[56,-0.1913],[57,-0.1989],[58,-0.1989],[59,-0.2065],[62,-0.2065],[63,-0.2141],[67,-0.2141],[68,-0.2146],[84,-0.2146],[85,-0.221],[101,-0.221],[102,-0.2178],[103,-0.2178],[104,-0.2048],[106,-0.2048],[107,-0.1918],[109,-0.1918],[110,-0.1788],[111,-0.1788],[112,-0.1658],[113,-0.1658],[114,-0.1528],[115,-0.1528],[116,-0.1398],[117,-0.1398],[118,-0.1268],[119,-0.1268],[121,-0.1127],[122,-0.1127],[123,-0.1056],[124,-0.1056],[125,-0.0986],[127,-0.0986],[128,-0.0915],[130,-0.0915],[131,-0.0845],[134,-0.0845],[135,-0.0774],[136,-0.0793],[137,-0.0793],[138,-0.0717],[143,-0.0717],[144,-0.0642],[149,-0.0642],[150,-0.0566],[152,-0.0566],[153,-0.0585],[154,-0.0585],[155,-0.054],[160,-0.054],[161,-0.0496],[168,-0.0496],[169,-0.0486],[176,-0.0486],[179,-0.0646],[183,-0.1071],[184,-0.123],[185,-0.1363],[186,-0.1598],[187,-0.1755],[189,-0.2225],[191,-0.2538],[194,-0.2538],[195,-0.246],[196,-0.2538],[197,-0.2773],[199,-0.34],[200,-0.3791],[201,-0.4106],[202,-0.4516],[205,-0.5131],[206,-0.5234],[208,-0.5234],[211,-0.4926],[213,-0.4516],[214,-0.4209],[215,-0.4004],[217,-0.3393],[219,-0.2723],[223,-0.119],[224,-0.0903],[225,-0.0663],[227,-0.0472],[228,-0.0424],[229,-0.0424],[230,-0.0376],[231,-0.0376],[232,-0.0424],[233,-0.0414],[234,-0.0414],[235,-0.0452],[236,-0.0452],[238,-0.0528],[241,-0.0755],[248,-0.1552],[249,-0.1609],[250,-0.175],[254,-0.2032],[255,-0.2032],[256,-0.2103],[260,-0.2103],[261,-0.2032],[263,-0.2032],[264,-0.1962],[265,-0.1932],[280,-0.1173],[281,-0.1154],[282,-0.1102],[283,-0.1102],[285,-0.0999],[286,-0.0999],[287,-0.0948],[288,-0.0948],[289,-0.0896],[291,-0.0896],[292,-0.0845],[296,-0.0845],[297,0],[367,0]]
       }
       calf_l_fk_ctrl.Location.Y {
        type=MovieSceneScriptingFloatChannel
        default=0
        thinned_from=368
        keys [[0,0.0111],[25,0.0111],[26,0.0126],[33,0.0126],[34,0.0121],[35,0.0121],[36,0.0133],[41,0.0133],[42,0.0146],[44,0.0146],[46,0.0171],[47,0.0172],[50,0.021],[51,0.0215],[55,0.0255],[56,0.0255],[57,0.0265],[58,0.0265],[59,0.0275],[62,0.0275],[63,0.0285],[84,0.0286],[85,0.0295],[101,0.0295],[102,0.0291],[103,0.0291],[104,0.0273],[106,0.0273],[107,0.0256],[109,0.0256],[110,0.0239],[111,0.0238],[112,0.0221],[113,0.0221],[114,0.0204],[115,0.0203],[116,0.0186],[117,0.0186],[118,0.0169],[119,0.0169],[121,0.015],[122,0.015],[123,0.0141],[124,0.0141],[125,0.0131],[127,0.0131],[128,0.0122],[130,0.0122],[131,0.0112],[134,0.0112],[135,0.0103],[136,0.0106],[137,0.0106],[138,0.0095],[143,0.0095],[144,0.0085],[149,0.0085],[150,0.0075],[152,0.0075],[153,0.0078],[154,0.0078],[155,0.0072],[160,0.0072],[161,0.0066],[168,0.0066],[176,0.0065],[179,0.0084],[183,0.014],[184,0.0162],[185,0.0179],[186,0.0211],[187,0.0232],[189,0.0295],[191,0.0337],[194,0.0337],[195,0.0327],[196,0.0337],[197,0.0369],[199,0.0452],[200,0.0505],[201,0.0547],[202,0.0601],[205,0.0683],[206,0.0697],[208,0.0697],[211,0.0655],[213,0.0601],[214,0.056],[215,0.0532],[217,0.0451],[219,0.0362],[223,0.0158],[224,0.012],[225,0.0088],[227,0.0063],[228,0.0056],[229,0.0056],[230,0.005],[231,0.005],[232,0.0056],[234,0.0055],[235,0.006],[236,0.006],[238,0.007],[241,0.01],[248,0.0207],[249,0.0215],[250,0.0233],[254,0.0271],[255,0.0271],[256,0.028],[260,0.028],[261,0.0271],[263,0.0271],[264,0.0262],[265,0.0258],[280,0.0156],[281,0.0154],[282,0.0147],[283,0.0147],[285,0.0133],[286,0.0133],[287,0.0126],[288,0.0126],[289,0.0119],[291,0.0119],[292,0.0112],[296,0.0112],[297,0],[367,0]]
       }
       calf_l_fk_ctrl.Location.Z {type=MovieSceneScriptingFloatChannel, default=0, thinned_from=368, keys [[0,0.0001],[367,0]]}
       calf_l_fk_ctrl.Rotation.X {
        type=MovieSceneScriptingFloatChannel
        default=0.0724
        thinned_from=368
        keys [[0,0.0055],[176,0.0055],[186,0.0023],[196,0.0012],[207,0.0019],[212,0.0029],[213,0.0032],[214,0.0033],[222,0.0055],[296,0.0055],[297,0],[367,0]]
       }
       calf_l_fk_ctrl.Rotation.Y {
        type=MovieSceneScriptingFloatChannel
        default=1.0549
        thinned_from=368
        keys [[0,-0.0084],[176,-0.0084],[186,-0.0038],[192,-0.0024],[200,-0.002],[212,-0.0047],[222,-0.0084],[296,-0.0084],[297,0],[367,0]]
       }
       calf_l_fk_ctrl.Rotation.Z {
        type=MovieSceneScriptingFloatChannel
        default=0.2163
        thinned_from=368
        keys [[0,-7.824],[176,-7.824],[178,-6.845],[179,-6.3981],[182,-5.015],[183,-4.5965],[184,-4.1994],[185,-3.8338],[186,-3.4976],[187,-3.194],[188,-2.927],[189,-2.6907],[190,-2.4869],[191,-2.3075],[192,-2.1669],[193,-2.0507],[194,-1.957],[195,-1.8836],[196,-1.8306],[197,-1.798],[198,-1.7837],[199,-1.7919],[200,-1.8286],[201,-1.8914],[202,-1.9959],[203,-2.1302],[204,-2.2945],[205,-2.496],[206,-2.7125],[207,-2.9513],[208,-3.2051],[209,-3.4738],[210,-3.7499],[212,-4.347],[214,-4.9739],[215,-5.2948],[216,-5.6531],[219,-6.6811],[220,-6.9994],[222,-7.824],[296,-7.824],[297,0],[367,0]]
       }
       calf_l_fk_ctrl.Scale.X {type=MovieSceneScriptingFloatChannel, default=1, const [1,368]}
       calf_l_fk_ctrl.Scale.Y {type=MovieSceneScriptingFloatChannel, default=1, const [1,368]}
       calf_l_fk_ctrl.Scale.Z {type=MovieSceneScriptingFloatChannel, default=1, const [1,368]}
       foot_l_fk_ctrl.Location.X {
        type=MovieSceneScriptingFloatChannel
        default=0
        thinned_from=368
        keys [[0,-0.0756],[16,-0.0756],[17,-0.0786],[23,-0.0786],[24,-0.0829],[30,-0.0829],[31,-0.0872],[33,-0.0872],[34,-0.0899],[38,-0.0899],[39,-0.0948],[40,-0.0948],[41,-0.0997],[42,-0.0997],[45,-0.1143],[46,-0.124],[47,-0.1289],[50,-0.1581],[53,-0.1824],[54,-0.1864],[55,-0.1945],[60,-0.2145],[62,-0.2145],[63,-0.2185],[66,-0.2185],[67,-0.2225],[68,-0.2182],[70,-0.2182],[71,-0.229],[84,-0.229],[101,-0.2286],[102,-0.2268],[103,-0.2199],[104,-0.2199],[107,-0.1993],[108,-0.1993],[114,-0.1581],[115,-0.1443],[116,-0.1443],[118,-0.1306],[119,-0.1219],[122,-0.1057],[126,-0.0913],[127,-0.0895],[128,-0.0859],[131,-0.0805],[132,-0.0769],[135,-0.0715],[136,-0.07],[138,-0.07],[139,-0.0614],[144,-0.0614],[145,-0.0528],[151,-0.0528],[152,-0.0443],[153,-0.0486],[154,-0.0486],[155,-0.0435],[160,-0.0435],[161,-0.0385],[168,-0.0385],[169,-0.0314],[175,-0.0314],[176,-0.0452],[178,-0.0452],[184,-0.128],[185,-0.1474],[186,-0.1649],[188,-0.2172],[190,-0.2609],[191,-0.274],[192,-0.2783],[193,-0.2783],[194,-0.274],[195,-0.274],[196,-0.2783],[197,-0.3045],[199,-0.3743],[200,-0.418],[201,-0.4554],[202,-0.4952],[203,-0.5236],[204,-0.5463],[205,-0.5634],[207,-0.5748],[208,-0.5748],[210,-0.552],[211,-0.535],[213,-0.4895],[215,-0.4327],[217,-0.3643],[218,-0.3314],[223,-0.1125],[224,-0.0797],[225,-0.0578],[227,-0.0359],[228,-0.0359],[229,-0.0249],[233,-0.0249],[234,-0.0337],[236,-0.0337],[237,-0.0424],[238,-0.0424],[242,-0.0773],[243,-0.0948],[244,-0.1035],[245,-0.121],[247,-0.1385],[248,-0.1559],[249,-0.1624],[250,-0.1777],[254,-0.2084],[255,-0.2084],[256,-0.2161],[260,-0.2161],[261,-0.2084],[263,-0.2084],[264,-0.2007],[265,-0.1989],[280,-0.113],[281,-0.113],[283,-0.1013],[284,-0.1013],[286,-0.0897],[287,-0.0897],[288,-0.0838],[290,-0.0838],[291,-0.078],[296,-0.078],[297,0],[367,0]]
       }
       foot_l_fk_ctrl.Location.Y {
        type=MovieSceneScriptingFloatChannel
        default=0.0627
        thinned_from=368
        keys [[0,-0.0003],[197,-0.0011],[208,-0.0021],[233,-0.0001],[296,-0.0003],[297,0],[367,0]]
       }
       foot_l_fk_ctrl.Location.Z {type=MovieSceneScriptingFloatChannel, default=-0.8453, thinned_from=368, keys [[0,0],[367,0]]}
       foot_l_fk_ctrl.Rotation.X {
        type=MovieSceneScriptingFloatChannel
        default=-0.9801
        thinned_from=368
        keys [[0,6.3619],[176,6.373],[177,6.386],[178,6.4299],[179,6.5068],[180,6.6163],[181,6.7539],[182,6.9149],[184,7.289],[187,7.8723],[188,8.0514],[190,8.3539],[191,8.479],[193,8.675],[194,8.7393],[196,8.8176],[198,8.8319],[199,8.8138],[200,8.7671],[201,8.701],[203,8.5203],[206,8.1606],[213,7.165],[215,6.9111],[217,6.6999],[218,6.6164],[219,6.5481],[220,6.5011],[221,6.4739],[231,6.4967],[257,6.3307],[296,6.3618],[297,0],[367,0]]
       }
       foot_l_fk_ctrl.Rotation.Y {
        type=MovieSceneScriptingFloatChannel
        default=0.1001
        thinned_from=368
        keys [[0,-4.9301],[32,-5.0174],[50,-5.1846],[62,-5.354],[72,-5.4174],[102,-5.4617],[107,-5.4147],[122,-5.131],[155,-4.8722],[173,-4.8069],[174,-4.8141],[175,-4.8319],[177,-4.8966],[181,-5.084],[187,-5.4462],[197,-6.1071],[199,-6.2135],[201,-6.2872],[202,-6.2895],[203,-6.25],[204,-6.1764],[205,-6.0784],[206,-5.9597],[208,-5.6779],[214,-4.7376],[216,-4.4494],[217,-4.3175],[218,-4.1994],[219,-4.0928],[220,-4.0013],[221,-3.9233],[222,-3.8602],[224,-3.7828],[225,-3.7515],[227,-3.7191],[228,-3.7191],[229,-3.7318],[230,-3.7531],[231,-3.7883],[233,-3.8825],[235,-4.0026],[242,-4.5157],[243,-4.5979],[246,-4.8032],[248,-4.918],[253,-5.1528],[255,-5.2061],[257,-5.2335],[261,-5.2362],[274,-5.1414],[292,-4.942],[296,-4.9282],[297,0],[367,0]]
       }
       foot_l_fk_ctrl.Rotation.Z {
        type=MovieSceneScriptingFloatChannel
        default=-0.1441
        thinned_from=368
        keys [[0,11.5156],[47,11.5512],[61,11.9769],[100,12.0827],[127,11.4747],[173,11.272],[176,11.3339],[178,10.962],[181,10.5433],[182,10.4289],[184,10.2665],[186,10.1882],[188,10.1903],[190,10.2748],[191,10.3454],[193,10.5425],[198,11.1191],[205,11.682],[209,11.8941],[220,12.2299],[222,12.3802],[226,11.9408],[230,11.6199],[233,11.473],[236,11.4113],[257,11.2614],[296,11.5155],[297,0],[367,0]]
       }
       foot_l_fk_ctrl.Scale.X {type=MovieSceneScriptingFloatChannel, default=1, const [1,368]}
       foot_l_fk_ctrl.Scale.Y {type=MovieSceneScriptingFloatChannel, default=1, const [1,368]}
       foot_l_fk_ctrl.Scale.Z {type=MovieSceneScriptingFloatChannel, default=1, const [1,368]}
       ball_l_fk_ctrl.Location.X {type=MovieSceneScriptingFloatChannel, default=0, const [0,368]}
       ball_l_fk_ctrl.Location.Y {type=MovieSceneScriptingFloatChannel, default=0, const [0,368]}
       ball_l_fk_ctrl.Location.Z {type=MovieSceneScriptingFloatChannel, default=0, const [0,368]}
       ball_l_fk_ctrl.Rotation.X {type=MovieSceneScriptingFloatChannel, default=0, const [0,368]}
       ball_l_fk_ctrl.Rotation.Y {type=MovieSceneScriptingFloatChannel, default=0, const [0,368]}
       ball_l_fk_ctrl.Rotation.Z {type=MovieSceneScriptingFloatChannel, default=0, const [0,368]}
       ball_l_fk_ctrl.Scale.X {type=MovieSceneScriptingFloatChannel, default=1, const [1,368]}
       ball_l_fk_ctrl.Scale.Y {type=MovieSceneScriptingFloatChannel, default=1, const [1,368]}
       ball_l_fk_ctrl.Scale.Z {type=MovieSceneScriptingFloatChannel, default=1, const [1,368]}
       foot_l_ik_ctrl.Location.X {
        type=MovieSceneScriptingFloatChannel
        default=0
        thinned_from=368
        keys [[0,-3.047],[7,-3.0443],[8,-3.0492],[9,-3.0449],[10,-3.0502],[22,-3.044],[23,-3.05],[26,-3.0446],[33,-3.0457],[34,-3.0516],[42,-3.0485],[43,-3.0441],[44,-3.0486],[45,-3.0447],[46,-3.0487],[47,-3.0454],[48,-3.0496],[82,-3.0496],[93,-3.0447],[97,-3.0484],[101,-3.0499],[127,-3.0468],[133,-3.0443],[140,-3.0498],[142,-3.0457],[158,-3.047],[159,-3.0436],[162,-3.0483],[170,-3.0447],[171,-3.0487],[176,-3.0436],[177,-3.0524],[179,-3.1023],[180,-3.1448],[181,-3.1953],[183,-3.329],[186,-3.5521],[188,-3.6917],[190,-3.8083],[192,-3.8986],[193,-3.934],[195,-3.9838],[197,-4.0068],[199,-4.0033],[201,-3.9673],[203,-3.9021],[207,-3.7062],[212,-3.4076],[215,-3.244],[217,-3.1524],[218,-3.1176],[220,-3.0647],[222,-3.046],[235,-3.0436],[236,-3.0492],[239,-3.0442],[241,-3.0513],[243,-3.0446],[244,-3.0504],[245,-3.0443],[250,-3.05],[252,-3.0432],[253,-3.0485],[258,-3.0442],[259,-3.0496],[261,-3.047],[262,-3.0506],[264,-3.0453],[265,-3.0481],[266,-3.0427],[269,-3.0504],[271,-3.0437],[272,-3.0522],[276,-3.0447],[285,-3.0475],[286,-3.0436],[288,-3.0499],[295,-3.0454],[296,-3.0497],[297,-0.0001],[367,0]]
       }
       foot_l_ik_ctrl.Location.Y {type=MovieSceneScriptingFloatChannel, default=0, thinned_from=368, keys [[0,11.0224],[296,11.0202],[297,-0.0004],[298,0],[367,0]]}
       foot_l_ik_ctrl.Location.Z {
        type=MovieSceneScriptingFloatChannel
        default=0
        thinned_from=368
        keys [[0,0.0002],[2,0],[4,0.0002],[5,0.0002],[7,0.0007],[8,0.0007],[9,0.0011],[14,0.0028],[15,0.0035],[16,0.004],[17,0.0016],[21,0.0044],[22,0.0055],[23,0.0062],[24,0.0028],[25,0.004],[26,-0.0062],[27,-0.0052],[30,-0.0016],[31,-0.0042],[33,-0.0012],[35,0.0029],[36,-0.0051],[37,-0.0024],[38,-0.0006],[39,-0.002],[40,0.0022],[41,0.0025],[42,-0.0012],[43,0.0019],[44,0.0055],[45,0.0023],[46,-0.0045],[47,0.0053],[48,0.0016],[49,-0.0005],[50,-0.0035],[51,0.0012],[52,0.0013],[53,0.0006],[54,0.0021],[55,-0.002],[56,0.0038],[57,0.0007],[58,0.0037],[59,-0.002],[60,-0.0013],[61,0.002],[62,0.0045],[63,-0.0052],[64,-0.0033],[66,0.0002],[67,-0.0021],[68,0.0032],[70,0.0067],[71,-0.0028],[73,-0.0004],[75,0.001],[76,0.0016],[77,0.0025],[80,0.0039],[81,0.0047],[83,0.0051],[84,0.0056],[85,-0.0004],[86,0.0001],[87,0.0003],[88,0.0009],[89,0.0008],[90,0.0016],[91,0.0015],[92,0.0017],[93,0.0023],[94,0.0022],[95,0.0025],[97,0.0023],[98,0.0026],[99,0.0021],[100,0.0001],[101,-0.003],[102,-0.0024],[103,-0.0016],[104,0.0038],[105,0.0027],[106,-0.0005],[107,0.0089],[108,-0.0026],[109,-0.0082],[110,-0.001],[111,-0.0078],[112,-0.0015],[113,-0.008],[114,-0.0016],[116,-0.001],[117,-0.0068],[118,0.0012],[119,-0.0017],[120,-0.0002],[121,0.0024],[122,-0.0014],[123,0.0007],[124,-0.0028],[125,0.0008],[126,-0.0011],[127,-0.004],[128,0.0016],[129,-0.0006],[130,-0.0031],[131,0.0016],[132,0.0011],[134,-0.0031],[135,0.0024],[136,-0.0017],[137,-0.005],[139,0.005],[140,0.0019],[142,-0.0034],[143,-0.0063],[144,-0.0012],[145,0.0049],[146,0.0023],[147,-0.0001],[148,-0.0021],[149,-0.0044],[150,0.0011],[151,-0.0008],[152,0.0058],[153,-0.0024],[154,-0.0042],[155,0.0038],[156,0.0022],[157,0],[160,-0.0039],[161,0.0039],[162,0.0026],[166,-0.0018],[167,-0.0025],[168,-0.0035],[169,0.0036],[172,0.0016],[173,0.0017],[174,0.0033],[175,0.0057],[176,-0.0029],[177,-0.0039],[178,0.002],[179,-0.0023],[180,-0.0056],[181,-0.0058],[182,-0.0021],[183,0],[184,0.0008],[185,0.002],[186,-0.0003],[187,0.0033],[188,0.0018],[189,0.0022],[190,0.0038],[191,0.0006],[192,0.0063],[193,0.0057],[194,0.0027],[195,0.0044],[196,0.0082],[197,0.0081],[198,0.0029],[199,0.0071],[200,0.0015],[201,0.0083],[202,-0.0027],[203,0.006],[204,0.0068],[205,0],[206,0.0024],[207,0.003],[208,-0.0019],[209,0.0045],[210,0.0015],[211,-0.0036],[212,0.0006],[213,-0.0019],[214,0.006],[215,-0.0011],[216,0.0015],[217,0],[218,-0.0043],[219,-0.0028],[220,-0.0001],[222,-0.0018],[223,0.0017],[224,-0.0012],[225,0.0031],[226,-0.0014],[227,0.0014],[228,-0.0045],[229,0.0016],[230,0.004],[231,0.0037],[232,0.0004],[233,0.0037],[234,-0.0017],[235,-0.0016],[236,0.0035],[237,-0.0023],[238,0.0029],[239,-0.0018],[240,-0.0026],[241,-0.0005],[242,0.0012],[243,-0.004],[244,-0.0001],[245,-0.0037],[246,0.0005],[247,0.004],[248,-0.0022],[249,0.0061],[250,-0.0039],[251,-0.0006],[252,0.0006],[253,-0.0012],[254,-0.0046],[255,0.0042],[256,-0.0053],[257,-0.0026],[258,-0.0024],[259,-0.0039],[260,-0.0062],[261,0.005],[262,-0.0003],[263,-0.0063],[264,0.0011],[265,-0.0025],[266,-0.0001],[267,0.0003],[268,0.0015],[269,0.0012],[270,0.0016],[271,0.0012],[272,0.0001],[273,0.0004],[275,-0.0004],[276,0.0004],[277,0.0006],[279,0.0035],[280,0.0043],[281,-0.0022],[282,0.0007],[283,-0.0011],[284,-0.0033],[285,0.0007],[286,0.0005],[287,-0.0001],[288,0.0005],[289,0.0013],[290,-0.0029],[291,-0.0005],[292,0.0015],[293,-0.0007],[294,-0.0026],[295,-0.0038],[296,-0.0041],[297,0],[367,0]]
       }
       foot_l_ik_ctrl.Rotation.X {
        type=MovieSceneScriptingFloatChannel
        default=0
        thinned_from=368
        keys [[0,-0.0007],[1,-0.0007],[2,0.0002],[3,0.0002],[4,0.0008],[5,0.0009],[6,-0.0006],[7,0.0007],[8,-0.0003],[9,0.0014],[10,-0.0014],[12,0.0006],[13,0.0001],[15,0.0009],[16,0.0018],[17,0.0002],[19,0.0006],[21,0.0006],[22,0.0003],[23,-0.0005],[24,0.0013],[25,-0.0012],[26,0.0009],[27,-0.0016],[28,0.0011],[29,0.001],[30,0.0015],[31,-0.0002],[32,-0.0012],[33,-0.0011],[34,0.0002],[35,-0.0002],[36,0.0004],[37,-0.0002],[38,0.001],[39,0.0011],[40,0.0016],[41,-0.0009],[42,-0.0001],[43,-0.0003],[44,-0.0013],[45,-0.0012],[47,0.0014],[48,0.0001],[49,-0.0009],[50,0.0008],[51,-0.0017],[53,0],[56,-0.0005],[57,-0.001],[58,-0.0006],[59,0.0016],[60,0.0002],[61,0.0014],[63,-0.0013],[64,-0.001],[65,0.0014],[66,0.0005],[67,-0.0016],[68,-0.0005],[69,-0.0006],[70,0.0008],[71,0.0016],[72,-0.0018],[73,-0.0018],[75,0],[76,0.0018],[77,-0.0004],[78,0.0009],[79,-0.0002],[80,0.0008],[81,-0.001],[82,-0.0019],[83,0.0011],[84,0.0001],[85,-0.0005],[86,-0.0001],[87,-0.0006],[88,-0.0015],[89,-0.002],[90,-0.0031],[92,-0.0023],[93,-0.0034],[96,-0.0026],[97,-0.0025],[98,-0.0015],[99,-0.0013],[100,0.0019],[101,0.0001],[102,0.0006],[103,-0.0003],[105,0.0008],[106,-0.0007],[107,-0.0002],[108,-0.0006],[109,0.0002],[110,0.0008],[111,0.0003],[112,0.0003],[113,-0.0009],[115,0.0003],[116,-0.0005],[117,-0.0004],[118,0.0003],[119,0.0003],[120,0.0012],[121,0.0003],[122,0.0001],[123,0.0006],[124,0.0004],[125,0.0018],[126,0.0014],[127,0.0003],[128,0.0014],[129,-0.0001],[131,0.0005],[132,-0.0001],[133,0.0014],[134,-0.0001],[135,-0.001],[137,-0.0004],[138,-0.0027],[139,0.0024],[140,0.0007],[141,0.0012],[142,-0.0006],[143,-0.002],[144,0.0002],[145,-0.0004],[146,0.002],[147,0.0004],[148,0.0005],[149,-0.0005],[150,-0.0006],[151,0.0004],[152,0.0002],[153,-0.0006],[154,-0.0006],[155,-0.0011],[156,-0.001],[157,-0.0002],[158,-0.001],[159,-0.0007],[160,0.001],[161,-0.0005],[162,-0.0014],[163,0.0012],[164,-0.0008],[165,-0.0006],[166,0.0019],[167,-0.0002],[168,0.0001],[169,-0.0002],[170,0],[171,0.0005],[172,0.0005],[173,0.001],[174,0.0017],[175,0.0007],[176,0.0011],[177,0.0011],[178,0],[179,0.0043],[180,0.0014],[181,0.0004],[182,0.0066],[183,0.008],[184,0.0073],[186,0.0074],[187,0.0115],[188,0.0078],[189,0.0125],[190,0.0123],[191,0.0129],[192,0.0166],[193,0.0143],[194,0.0171],[196,0.0179],[197,0.0162],[198,0.0155],[199,0.0185],[200,0.0158],[201,0.0189],[202,0.0145],[203,0.0125],[204,0.0128],[205,0.0132],[206,0.0112],[207,0.0086],[209,0.0063],[210,0.0095],[211,0.0052],[212,0.0076],[213,0.0034],[214,0.0049],[215,0.0022],[216,0.002],[217,0.0014],[218,0.0014],[219,0.0044],[220,0.0008],[221,-0.0014],[222,0.001],[223,-0.0026],[224,-0.0003],[226,-0.0023],[227,0.0002],[228,0.0017],[229,0.0021],[230,-0.0025],[231,-0.0014],[232,0.0022],[233,0.0007],[234,0.0003],[236,-0.0001],[237,0.0002],[238,0.0003],[239,0],[240,-0.0006],[241,0.0003],[242,-0.0007],[243,0.0001],[244,0.0003],[245,-0.0005],[246,0.0008],[247,-0.0006],[248,-0.0009],[249,-0.0005],[250,-0.0005],[251,0.0005],[252,0.0002],[253,0.0005],[254,-0.0003],[255,0.0008],[257,0.001],[258,-0.0001],[259,0.0007],[260,-0.0001],[261,0.0001],[262,-0.0013],[263,-0.0013],[264,0],[265,0.0004],[266,0.0006],[267,0.0002],[268,-0.001],[269,0],[270,-0.0004],[271,0.0005],[272,0],[273,0],[274,0.0008],[275,-0.0005],[276,0],[277,-0.0005],[278,0.0008],[279,-0.0001],[281,-0.0006],[282,0.0006],[283,0.0014],[284,-0.0006],[285,0.0006],[286,0.0023],[287,0.0017],[288,0.0006],[289,-0.0009],[290,-0.0002],[291,-0.0003],[292,0.0021],[293,0.0006],[294,0.0022],[295,0.0011],[296,0.0002],[297,0.0003],[298,0],[367,0]]
       }
       foot_l_ik_ctrl.Rotation.Y {
        type=MovieSceneScriptingFloatChannel
        default=0
        thinned_from=368
        keys [[0,-0.0002],[2,-0.0002],[3,0.0002],[4,-0.0001],[5,0.0007],[6,0.0007],[7,0.0016],[8,-0.0012],[9,0.0013],[10,-0.0016],[11,0.0005],[12,0.0002],[13,-0.0008],[14,0.001],[15,0.0007],[16,0.0013],[17,0.0004],[19,-0.0005],[21,-0.0006],[22,-0.0015],[23,0.001],[25,-0.0022],[26,0.0037],[27,0.0014],[28,0.0017],[29,0.0013],[31,0.0019],[32,0.0029],[33,-0.0025],[34,0.0003],[35,-0.001],[36,0.0021],[37,0.0023],[38,-0.002],[39,0.0008],[40,-0.0006],[41,-0.0029],[42,-0.0013],[43,-0.0004],[44,0.0025],[45,0.0024],[46,0.0001],[47,0.0002],[48,-0.001],[49,0.0019],[50,0.0002],[52,0.0001],[53,0.0005],[54,-0.0004],[55,0.0004],[56,-0.0005],[57,-0.0007],[58,0.0002],[59,0.0007],[60,-0.0002],[62,-0.0005],[64,0.0004],[65,-0.0001],[66,0.0007],[67,0.0005],[68,-0.0015],[69,0.0012],[70,-0.002],[71,-0.0015],[72,-0.0014],[73,0.0012],[74,0.0005],[75,-0.0009],[76,0.0011],[77,-0.0004],[78,0.0023],[79,0.0001],[80,0.0012],[81,-0.0001],[82,0.0009],[83,-0.0014],[84,-0.0001],[85,-0.0007],[86,0.0019],[87,-0.0029],[88,-0.0015],[89,-0.0012],[90,0.0009],[92,0.0031],[93,-0.0016],[94,-0.0013],[95,-0.0005],[96,-0.0003],[97,0.0007],[98,0.0025],[99,0.0018],[100,0],[101,0.0013],[102,-0.0003],[103,0.0005],[104,-0.0005],[106,0.0005],[107,0.0004],[108,0.0007],[109,0.0001],[110,0.0009],[111,0.0004],[112,-0.0006],[113,0.0002],[114,-0.0003],[115,-0.0002],[116,0.0004],[117,-0.0006],[118,0],[119,0.0004],[120,-0.001],[121,-0.0001],[122,0],[123,-0.0007],[124,0.0007],[125,-0.0003],[126,0.0002],[127,0.0002],[128,-0.0008],[130,0.0007],[131,-0.0003],[132,0.0009],[133,-0.0001],[134,-0.0005],[135,0.0002],[136,-0.0011],[137,-0.0005],[138,0.0004],[139,0.0003],[140,-0.0001],[141,-0.001],[142,-0.0003],[143,-0.0002],[144,0.0015],[145,-0.0003],[146,0.0009],[147,-0.001],[148,0.001],[149,-0.0009],[150,0.0006],[151,-0.0004],[152,0.0016],[153,0.0002],[154,0.0009],[155,0.0027],[156,-0.0016],[157,-0.002],[158,-0.0004],[159,0.0017],[160,0.0031],[161,-0.0016],[162,0.0003],[163,0.0029],[164,-0.0012],[165,-0.0002],[166,0.0022],[167,-0.0007],[168,0.0007],[169,0.0003],[170,0.0004],[171,0.0009],[172,-0.0008],[173,-0.0002],[174,-0.0002],[175,0.0001],[176,0.0006],[177,0.0003],[178,0.0007],[179,0.0009],[180,0.0017],[181,0.0043],[182,0.0058],[183,0.0067],[184,0.0088],[185,0.0112],[186,0.0122],[187,0.0158],[188,0.0161],[189,0.0188],[190,0.02],[191,0.0195],[192,0.0216],[193,0.0223],[194,0.0221],[195,0.0233],[196,0.0239],[197,0.0231],[198,0.0237],[199,0.0247],[200,0.0235],[201,0.0251],[202,0.0221],[203,0.0212],[204,0.0208],[205,0.0191],[206,0.0169],[207,0.0166],[208,0.0144],[209,0.0127],[210,0.0138],[211,0.0082],[212,0.0091],[213,0.0055],[214,0.0047],[216,0.0048],[217,0.0055],[218,0.0028],[219,0.0013],[220,-0.0005],[221,-0.0017],[222,0.0013],[223,-0.0014],[224,-0.0017],[225,0.0001],[226,-0.0019],[227,0.0004],[228,0.0004],[229,-0.0003],[230,0.0013],[231,0.001],[232,0],[233,0.0018],[234,-0.0037],[235,0.0035],[236,-0.0036],[237,0.0008],[238,-0.0007],[239,-0.0019],[240,-0.0009],[241,0.0004],[242,0.0037],[243,-0.0031],[244,0.0028],[245,-0.0005],[246,-0.0011],[247,0.0018],[248,0.0006],[249,0.0019],[250,-0.0016],[251,0.0019],[252,0.003],[253,-0.0041],[254,0.002],[255,0.0024],[256,0.0006],[257,-0.0006],[258,0.0007],[259,-0.001],[260,-0.0024],[261,-0.0031],[262,0.001],[263,-0.003],[264,0.0008],[265,-0.0001],[266,0.0015],[267,-0.0001],[268,0.0011],[269,-0.0005],[271,0.0013],[272,-0.0032],[273,-0.0008],[274,-0.0014],[275,-0.0008],[276,0.0009],[277,-0.002],[278,0.0018],[279,0.0013],[280,-0.0009],[281,-0.0005],[282,0.0023],[283,0],[284,0.0009],[285,-0.002],[286,0.0023],[287,0.0001],[288,-0.0022],[289,-0.0014],[290,-0.0018],[291,0.0009],[292,0.0026],[293,0.0013],[294,0.0008],[295,0.0016],[296,-0.0006],[297,-0.0001],[367,0]]
       }
       foot_l_ik_ctrl.Rotation.Z {
        type=MovieSceneScriptingFloatChannel
        default=0
        thinned_from=368
        keys [[0,8.0108],[176,8.0098],[177,7.9895],[178,7.9316],[179,7.8338],[180,7.6998],[181,7.5326],[182,7.3341],[184,6.8808],[187,6.1719],[188,5.9568],[189,5.7606],[190,5.58],[191,5.4241],[192,5.2925],[193,5.1806],[194,5.0927],[195,5.0265],[196,4.9763],[197,4.952],[198,4.9406],[199,4.9554],[200,5.0012],[201,5.0735],[202,5.1709],[203,5.286],[204,5.4215],[206,5.738],[208,6.0952],[212,6.8585],[214,7.2167],[216,7.529],[217,7.6654],[218,7.7818],[219,7.8789],[220,7.948],[221,7.9955],[296,8.0104],[297,0.0011],[298,0],[367,0]]
       }
       foot_l_ik_ctrl.Scale.X {type=MovieSceneScriptingFloatChannel, default=1, const [1,368]}
       foot_l_ik_ctrl.Scale.Y {type=MovieSceneScriptingFloatChannel, default=1, const [1,368]}
       foot_l_ik_ctrl.Scale.Z {type=MovieSceneScriptingFloatChannel, default=1, const [1,368]}
       ball_l_ik_ctrl.Location.X {type=MovieSceneScriptingFloatChannel, default=0, const [0,368]}
       ball_l_ik_ctrl.Location.Y {type=MovieSceneScriptingFloatChannel, default=0, const [0,368]}
       ball_l_ik_ctrl.Location.Z {type=MovieSceneScriptingFloatChannel, default=0, const [0,368]}
       ball_l_ik_ctrl.Rotation.X {type=MovieSceneScriptingFloatChannel, default=0, const [0,368]}
       ball_l_ik_ctrl.Rotation.Y {type=MovieSceneScriptingFloatChannel, default=0, const [0,368]}
       ball_l_ik_ctrl.Rotation.Z {type=MovieSceneScriptingFloatChannel, default=0, const [0,368]}
       ball_l_ik_ctrl.Scale.X {type=MovieSceneScriptingFloatChannel, default=1, const [1,368]}
       ball_l_ik_ctrl.Scale.Y {type=MovieSceneScriptingFloatChannel, default=1, const [1,368]}
       ball_l_ik_ctrl.Scale.Z {type=MovieSceneScriptingFloatChannel, default=1, const [1,368]}
       foot_bk1_l_ctrl.Location.X {type=MovieSceneScriptingFloatChannel, default=0, const [0,368]}
       foot_bk1_l_ctrl.Location.Y {type=MovieSceneScriptingFloatChannel, default=0, const [0,368]}
       foot_bk1_l_ctrl.Location.Z {type=MovieSceneScriptingFloatChannel, default=0, const [0,368]}
       foot_bk1_l_ctrl.Rotation.X {type=MovieSceneScriptingFloatChannel, default=0, const [0,368]}
       foot_bk1_l_ctrl.Rotation.Y {type=MovieSceneScriptingFloatChannel, default=0, const [0,368]}
       foot_bk1_l_ctrl.Rotation.Z {type=MovieSceneScriptingFloatChannel, default=0, const [0,368]}
       foot_bk1_l_ctrl.Scale.X {type=MovieSceneScriptingFloatChannel, default=1, const [1,368]}
       foot_bk1_l_ctrl.Scale.Y {type=MovieSceneScriptingFloatChannel, default=1, const [1,368]}
       foot_bk1_l_ctrl.Scale.Z {type=MovieSceneScriptingFloatChannel, default=1, const [1,368]}
       foot_roll_l_ctrl.Location.X {type=MovieSceneScriptingFloatChannel, default=0, const [0,368]}
       foot_roll_l_ctrl.Location.Y {type=MovieSceneScriptingFloatChannel, default=0, const [0,368]}
       foot_roll_l_ctrl.Location.Z {type=MovieSceneScriptingFloatChannel, default=0, const [0,368]}
       foot_roll_l_ctrl.Rotation.X {type=MovieSceneScriptingFloatChannel, default=0, const [0,368]}
       foot_roll_l_ctrl.Rotation.Y {type=MovieSceneScriptingFloatChannel, default=0, const [0,368]}
       foot_roll_l_ctrl.Rotation.Z {type=MovieSceneScriptingFloatChannel, default=0, const [0,368]}
       foot_roll_l_ctrl.Scale.X {type=MovieSceneScriptingFloatChannel, default=1, const [1,368]}
       foot_roll_l_ctrl.Scale.Y {type=MovieSceneScriptingFloatChannel, default=1, const [1,368]}
       foot_roll_l_ctrl.Scale.Z {type=MovieSceneScriptingFloatChannel, default=1, const [1,368]}
       roll_blend_l_ctrl {type=MovieSceneScriptingFloatChannel, default=1, const [1,368]}
       tip_l_ctrl.Location.X {type=MovieSceneScriptingFloatChannel, default=0, const [0,368]}
       tip_l_ctrl.Location.Y {type=MovieSceneScriptingFloatChannel, default=0, const [0,368]}
       tip_l_ctrl.Location.Z {type=MovieSceneScriptingFloatChannel, default=0, const [0,368]}
       tip_l_ctrl.Rotation.X {type=MovieSceneScriptingFloatChannel, default=0, const [0,368]}
       tip_l_ctrl.Rotation.Y {type=MovieSceneScriptingFloatChannel, default=0, const [0,368]}
       tip_l_ctrl.Rotation.Z {type=MovieSceneScriptingFloatChannel, default=0, const [0,368]}
       tip_l_ctrl.Scale.X {type=MovieSceneScriptingFloatChannel, default=1, const [1,368]}
       tip_l_ctrl.Scale.Y {type=MovieSceneScriptingFloatChannel, default=1, const [1,368]}
       tip_l_ctrl.Scale.Z {type=MovieSceneScriptingFloatChannel, default=1, const [1,368]}
       heel_l_ctrl.Location.X {type=MovieSceneScriptingFloatChannel, default=0, const [0,368]}
       heel_l_ctrl.Location.Y {type=MovieSceneScriptingFloatChannel, default=0, const [0,368]}
       heel_l_ctrl.Location.Z {type=MovieSceneScriptingFloatChannel, default=0, const [0,368]}
       heel_l_ctrl.Rotation.X {type=MovieSceneScriptingFloatChannel, default=0, const [0,368]}
       heel_l_ctrl.Rotation.Y {type=MovieSceneScriptingFloatChannel, default=0, const [0,368]}
       heel_l_ctrl.Rotation.Z {type=MovieSceneScriptingFloatChannel, default=0, const [0,368]}
       heel_l_ctrl.Scale.X {type=MovieSceneScriptingFloatChannel, default=1, const [1,368]}
       heel_l_ctrl.Scale.Y {type=MovieSceneScriptingFloatChannel, default=1, const [1,368]}
       heel_l_ctrl.Scale.Z {type=MovieSceneScriptingFloatChannel, default=1, const [1,368]}
       leg_l_softness {type=MovieSceneScriptingFloatChannel, default=0, const [0,368]}
       leg_l_pv_ik_ctrl.Location.X {
        type=MovieSceneScriptingFloatChannel
        default=0
        thinned_from=368
        keys [[0,-5.7692],[176,-5.7481],[177,-5.8091],[178,-5.8963],[179,-6.0075],[180,-6.1467],[181,-6.3098],[182,-6.4946],[187,-7.5241],[188,-7.7167],[190,-8.054],[191,-8.1986],[193,-8.4365],[194,-8.5272],[196,-8.6611],[197,-8.7034],[198,-8.7279],[199,-8.7336],[200,-8.7096],[201,-8.6635],[202,-8.5891],[203,-8.4939],[204,-8.3751],[206,-8.0864],[209,-7.573],[213,-6.849],[215,-6.5159],[217,-6.2183],[218,-6.0917],[219,-5.9805],[221,-5.8105],[222,-5.7528],[230,-5.6913],[296,-5.7688],[297,0],[367,0]]
       }
       leg_l_pv_ik_ctrl.Location.Y {
        type=MovieSceneScriptingFloatChannel
        default=0
        thinned_from=368
        keys [[0,-3.8679],[176,-3.8655],[178,-3.4172],[181,-2.8211],[182,-2.6281],[183,-2.4579],[184,-2.3004],[185,-2.1594],[186,-2.0328],[187,-1.9209],[188,-1.826],[189,-1.7435],[190,-1.6745],[191,-1.6148],[192,-1.57],[193,-1.535],[194,-1.5069],[195,-1.4852],[196,-1.4711],[197,-1.4633],[198,-1.4609],[199,-1.4657],[200,-1.4785],[201,-1.4995],[202,-1.5352],[203,-1.5812],[204,-1.6373],[205,-1.7079],[206,-1.7842],[207,-1.869],[208,-1.9601],[210,-2.1599],[212,-2.3864],[214,-2.6325],[215,-2.7621],[216,-2.9108],[219,-3.3495],[220,-3.4897],[222,-3.866],[296,-3.8679],[297,0],[367,0]]
       }
       leg_l_pv_ik_ctrl.Location.Z {
        type=MovieSceneScriptingFloatChannel
        default=0
        thinned_from=368
        keys [[0,-3.4023],[176,-3.401],[178,-3.046],[182,-2.3906],[183,-2.2423],[184,-2.1021],[185,-1.9736],[186,-1.8558],[187,-1.7494],[188,-1.6565],[189,-1.5744],[190,-1.5038],[191,-1.4418],[192,-1.3936],[193,-1.3543],[194,-1.3226],[195,-1.2976],[196,-1.2802],[197,-1.2697],[198,-1.2655],[199,-1.269],[200,-1.2821],[201,-1.3042],[202,-1.3414],[203,-1.3888],[204,-1.4466],[205,-1.5178],[206,-1.5941],[207,-1.6779],[208,-1.7668],[210,-1.9574],[212,-2.1662],[215,-2.4985],[216,-2.6253],[219,-2.9898],[220,-3.1033],[222,-3.4013],[296,-3.4022],[297,0],[367,0]]
       }
       leg_l_pv_ik_ctrl.Rotation.X {
        type=MovieSceneScriptingFloatChannel
        default=0
        thinned_from=368
        keys [[0,4.5878],[176,4.5862],[178,4.0791],[182,3.1544],[183,2.9472],[184,2.752],[185,2.5735],[186,2.4104],[187,2.2635],[188,2.1355],[189,2.0225],[190,1.9255],[191,1.8405],[192,1.7744],[193,1.7204],[194,1.677],[195,1.6427],[196,1.6189],[197,1.6044],[198,1.5985],[199,1.6032],[200,1.621],[201,1.6512],[202,1.7017],[203,1.7664],[204,1.8454],[205,1.9429],[206,2.0476],[207,2.1628],[208,2.2853],[210,2.5488],[212,2.8392],[214,3.146],[215,3.3043],[216,3.4826],[219,3.9986],[220,4.1604],[222,4.5866],[296,4.5878],[297,0],[367,0]]
       }
       leg_l_pv_ik_ctrl.Rotation.Y {
        type=MovieSceneScriptingFloatChannel
        default=0
        thinned_from=368
        keys [[0,-0.5995],[176,-0.5972],[177,-0.6005],[178,-0.6065],[179,-0.6153],[180,-0.6268],[181,-0.6407],[182,-0.6567],[184,-0.6932],[187,-0.7502],[188,-0.7681],[190,-0.7996],[191,-0.8132],[193,-0.8358],[195,-0.8515],[196,-0.8574],[198,-0.864],[199,-0.8647],[200,-0.8625],[201,-0.8583],[202,-0.8515],[203,-0.8427],[204,-0.8318],[206,-0.8053],[214,-0.6758],[216,-0.6475],[217,-0.6348],[218,-0.6238],[219,-0.6144],[220,-0.6069],[221,-0.6011],[222,-0.5977],[230,-0.5911],[296,-0.5994],[297,0],[367,0]]
       }
       leg_l_pv_ik_ctrl.Rotation.Z {
        type=MovieSceneScriptingFloatChannel
        default=0
        thinned_from=368
        keys [[0,7.4749],[176,7.4473],[177,7.4902],[178,7.5667],[179,7.6767],[180,7.8208],[181,7.9943],[182,8.1941],[184,8.6444],[187,9.3441],[188,9.5621],[190,9.944],[191,10.1079],[193,10.38],[195,10.568],[197,10.6876],[199,10.7245],[200,10.6982],[201,10.6469],[202,10.5645],[203,10.4588],[204,10.3263],[206,10.0047],[209,9.4294],[214,8.4228],[216,8.0738],[217,7.9168],[218,7.7807],[219,7.6631],[220,7.5703],[221,7.4975],[222,7.4534],[230,7.3733],[296,7.4743],[297,0],[367,0]]
       }
       leg_l_pv_ik_ctrl.Scale.X {type=MovieSceneScriptingFloatChannel, default=1, const [1,368]}
       leg_l_pv_ik_ctrl.Scale.Y {type=MovieSceneScriptingFloatChannel, default=1, const [1,368]}
       leg_l_pv_ik_ctrl.Scale.Z {type=MovieSceneScriptingFloatChannel, default=1, const [1,368]}
       thigh_r_fk_ctrl.Location.X {type=MovieSceneScriptingFloatChannel, default=0, thinned_from=368, keys [[0,-0.0002],[296,-0.0002],[367,0]]}
       thigh_r_fk_ctrl.Location.Y {type=MovieSceneScriptingFloatChannel, default=0, thinned_from=368, keys [[0,0],[367,0]]}
       thigh_r_fk_ctrl.Location.Z {type=MovieSceneScriptingFloatChannel, default=0, thinned_from=368, keys [[0,0],[367,0]]}
       thigh_r_fk_ctrl.Rotation.X {
        type=MovieSceneScriptingFloatChannel
        default=1.0528
        thinned_from=368
        keys [[0,17.4631],[46,17.5223],[47,17.5468],[48,17.6094],[50,17.8401],[52,18.194],[61,20.3876],[64,20.9878],[67,21.425],[68,21.5321],[98,22.064],[100,22.0049],[102,21.8389],[105,21.4267],[109,20.6583],[118,18.6368],[121,18.0654],[123,17.7596],[125,17.5328],[141,16.8134],[173,16.4169],[174,16.4551],[175,16.5685],[176,16.7546],[177,17.0095],[178,17.3284],[179,17.705],[180,18.1323],[182,19.1083],[188,22.3519],[190,23.3093],[192,24.1158],[193,24.4511],[194,24.7372],[195,24.9738],[197,25.2991],[199,25.4518],[201,25.4719],[203,25.3237],[205,24.9933],[207,24.4764],[209,23.7943],[217,20.4713],[220,19.3595],[222,18.7109],[224,18.1433],[226,17.6615],[228,17.2691],[230,16.9657],[232,16.763],[233,16.7155],[234,16.7154],[235,16.768],[236,16.8734],[240,17.4506],[242,17.6395],[245,17.7874],[252,17.7973],[296,17.4647],[297,0],[367,0]]
       }
       thigh_r_fk_ctrl.Rotation.Y {
        type=MovieSceneScriptingFloatChannel
        default=-1.026
        thinned_from=368
        keys [[0,-3.0234],[33,-3.1251],[37,-3.151],[39,-3.1482],[40,-3.139],[41,-3.1236],[42,-3.1013],[43,-3.071],[44,-3.0327],[45,-2.9849],[49,-2.7578],[50,-2.7082],[51,-2.6655],[52,-2.6322],[53,-2.61],[54,-2.6005],[55,-2.6031],[56,-2.6145],[57,-2.6316],[63,-2.7877],[66,-2.8492],[68,-2.8785],[99,-2.9552],[123,-3.1408],[172,-3.0035],[173,-3.0045],[174,-3.0368],[175,-3.1273],[176,-3.2777],[177,-3.4869],[178,-3.7497],[179,-4.0551],[181,-4.7258],[182,-5.0512],[183,-5.3453],[184,-5.5953],[185,-5.7926],[186,-5.9331],[187,-6.0179],[188,-6.0529],[189,-6.0498],[190,-6.0238],[192,-5.9428],[193,-5.9227],[194,-5.9249],[195,-5.9487],[197,-6.0312],[198,-6.0413],[199,-6.0246],[200,-5.9824],[201,-5.902],[202,-5.7587],[203,-5.5414],[204,-5.262],[205,-4.9335],[206,-4.5726],[209,-3.4383],[210,-3.0781],[211,-2.7372],[212,-2.4186],[213,-2.1247],[214,-1.8557],[215,-1.6122],[216,-1.3942],[217,-1.2014],[218,-1.034],[219,-0.8915],[220,-0.7741],[221,-0.6819],[222,-0.6154],[223,-0.5738],[224,-0.5492],[225,-0.5421],[226,-0.5499],[227,-0.5716],[228,-0.6106],[229,-0.6694],[230,-0.7495],[231,-0.8419],[232,-0.9386],[233,-1.0423],[234,-1.1549],[235,-1.2764],[236,-1.4042],[237,-1.5278],[238,-1.6399],[239,-1.731],[240,-1.7999],[241,-1.8474],[242,-1.8729],[243,-1.8821],[244,-1.8782],[245,-1.8644],[246,-1.8436],[249,-1.768],[250,-1.748],[251,-1.7343],[252,-1.7266],[254,-1.7237],[257,-1.7405],[258,-1.7499],[259,-1.7661],[260,-1.7892],[261,-1.8175],[262,-1.8529],[263,-1.8932],[264,-1.9385],[266,-2.0409],[274,-2.5056],[276,-2.6086],[278,-2.7015],[281,-2.8162],[283,-2.8779],[285,-2.9263],[288,-2.9788],[291,-3.0087],[296,-3.0231],[297,0],[367,0]]
       }
       thigh_r_fk_ctrl.Rotation.Z {
        type=MovieSceneScriptingFloatChannel
        default=-0.07
        thinned_from=368
        keys [[0,-2.7722],[2,-2.7724],[4,-2.7817],[7,-2.8103],[11,-2.8688],[23,-3.0812],[28,-3.1359],[33,-3.1575],[40,-3.1428],[45,-3.1126],[47,-3.1096],[49,-3.0851],[51,-3.037],[53,-2.968],[56,-2.847],[57,-2.8143],[59,-2.7418],[60,-2.7002],[61,-2.6529],[63,-2.5333],[65,-2.4276],[66,-2.3819],[67,-2.3424],[68,-2.3077],[71,-2.2549],[74,-2.2203],[76,-2.2019],[83,-2.1553],[92,-2.1186],[99,-2.1023],[101,-2.1051],[103,-2.1218],[105,-2.1494],[107,-2.1894],[110,-2.2681],[117,-2.4831],[121,-2.5873],[124,-2.6457],[128,-2.6929],[137,-2.7566],[144,-2.7843],[173,-2.8408],[174,-2.835],[175,-2.8079],[176,-2.7607],[177,-2.6958],[178,-2.6141],[179,-2.5197],[180,-2.4129],[181,-2.2965],[184,-1.9274],[185,-1.8074],[186,-1.6955],[187,-1.5933],[188,-1.5086],[189,-1.4451],[190,-1.4039],[191,-1.396],[192,-1.4108],[193,-1.4432],[194,-1.4911],[195,-1.551],[197,-1.6918],[198,-1.7582],[199,-1.8173],[200,-1.8691],[201,-1.9138],[202,-1.9633],[204,-2.0971],[209,-2.4592],[211,-2.5912],[213,-2.7093],[215,-2.8068],[216,-2.8475],[217,-2.8822],[218,-2.9093],[219,-2.9304],[220,-2.9433],[221,-2.9471],[222,-2.9425],[228,-2.8552],[233,-2.7596],[235,-2.7102],[237,-2.6477],[241,-2.5014],[245,-2.394],[248,-2.3417],[252,-2.3004],[254,-2.2903],[256,-2.2942],[260,-2.3341],[286,-2.7007],[292,-2.755],[295,-2.7678],[296,-2.7681],[297,0],[367,0]]
       }
       thigh_r_fk_ctrl.Scale.X {type=MovieSceneScriptingFloatChannel, default=1, const [1,368]}
       thigh_r_fk_ctrl.Scale.Y {type=MovieSceneScriptingFloatChannel, default=1, const [1,368]}
       thigh_r_fk_ctrl.Scale.Z {type=MovieSceneScriptingFloatChannel, default=1, const [1,368]}
       calf_r_fk_ctrl.Location.X {
        type=MovieSceneScriptingFloatChannel
        default=0
        thinned_from=368
        keys [[0,-0.7655],[16,-0.7655],[17,-0.7618],[23,-0.7618],[24,-0.7593],[31,-0.7593],[32,-0.7567],[39,-0.7556],[40,-0.7515],[41,-0.7515],[43,-0.7433],[44,-0.7433],[45,-0.7351],[46,-0.7309],[53,-0.6733],[54,-0.6693],[55,-0.6612],[56,-0.6572],[57,-0.6492],[59,-0.6412],[60,-0.6412],[61,-0.6371],[64,-0.6371],[65,-0.6331],[71,-0.6331],[72,-0.63],[79,-0.63],[80,-0.627],[101,-0.6275],[102,-0.6316],[103,-0.6316],[104,-0.6448],[106,-0.6448],[107,-0.6581],[109,-0.6581],[110,-0.6713],[111,-0.6713],[112,-0.6845],[113,-0.6845],[114,-0.6978],[115,-0.6978],[116,-0.711],[117,-0.711],[118,-0.7242],[119,-0.7273],[121,-0.7423],[122,-0.7423],[123,-0.7498],[124,-0.7498],[125,-0.7573],[126,-0.7573],[127,-0.7648],[129,-0.7648],[130,-0.7723],[133,-0.7723],[134,-0.7798],[137,-0.7783],[138,-0.786],[143,-0.786],[144,-0.7937],[150,-0.7937],[151,-0.8015],[152,-0.8015],[153,-0.7984],[154,-0.8025],[159,-0.8025],[160,-0.8066],[168,-0.8066],[174,-0.8076],[175,-0.8139],[177,-0.8388],[180,-0.9135],[181,-0.9446],[182,-0.9695],[183,-0.9882],[184,-1.0007],[185,-1.0022],[186,-0.996],[187,-0.9836],[188,-0.965],[189,-0.9402],[190,-0.9092],[191,-0.8658],[195,-0.6673],[196,-0.6363],[197,-0.6239],[199,-0.6115],[200,-0.6177],[201,-0.6223],[205,-0.6223],[206,-0.6146],[207,-0.6146],[208,-0.6069],[212,-0.6069],[214,-0.6223],[215,-0.6223],[217,-0.6368],[220,-0.6368],[223,-0.6131],[224,-0.6131],[225,-0.621],[232,-0.732],[234,-0.7494],[235,-0.7494],[236,-0.7582],[239,-0.7582],[240,-0.7494],[241,-0.7494],[243,-0.7317],[244,-0.7317],[248,-0.6965],[249,-0.6918],[250,-0.683],[254,-0.6653],[255,-0.6653],[256,-0.6609],[260,-0.6609],[261,-0.6653],[262,-0.6653],[263,-0.6698],[264,-0.6698],[265,-0.6748],[266,-0.6748],[267,-0.6843],[268,-0.6843],[269,-0.6938],[270,-0.6938],[271,-0.7033],[272,-0.7033],[273,-0.7128],[274,-0.7128],[275,-0.7223],[276,-0.7223],[277,-0.7317],[279,-0.7317],[280,-0.7412],[282,-0.7412],[283,-0.75],[285,-0.75],[286,-0.7587],[290,-0.7587],[291,-0.7675],[296,-0.7675],[297,0],[367,0]]
       }
       calf_r_fk_ctrl.Location.Y {
        type=MovieSceneScriptingFloatChannel
        default=0
        thinned_from=368
        keys [[0,0.1042],[16,0.1042],[17,0.1037],[23,0.1037],[24,0.1034],[31,0.1033],[32,0.103],[39,0.1029],[40,0.1023],[41,0.1023],[43,0.1012],[44,0.1012],[45,0.1001],[46,0.0995],[53,0.0916],[54,0.0911],[55,0.09],[56,0.0895],[57,0.0884],[59,0.0873],[60,0.0873],[61,0.0867],[64,0.0867],[65,0.0862],[71,0.0862],[72,0.0858],[79,0.0857],[80,0.0853],[101,0.0854],[102,0.086],[103,0.086],[104,0.0878],[106,0.0878],[107,0.0896],[109,0.0896],[110,0.0914],[111,0.0914],[112,0.0932],[113,0.0932],[114,0.095],[115,0.095],[116,0.0968],[117,0.0968],[118,0.0986],[119,0.099],[121,0.101],[122,0.101],[123,0.1021],[124,0.1021],[125,0.1031],[126,0.1031],[127,0.1041],[129,0.1041],[130,0.1051],[133,0.1051],[134,0.1062],[137,0.106],[138,0.107],[143,0.107],[144,0.108],[150,0.1081],[151,0.1091],[152,0.1091],[153,0.1087],[154,0.1092],[159,0.1092],[160,0.1098],[168,0.1098],[174,0.11],[175,0.1108],[177,0.1142],[180,0.1244],[181,0.1286],[182,0.132],[183,0.1345],[184,0.1362],[185,0.1364],[186,0.1356],[187,0.1339],[188,0.1314],[189,0.128],[190,0.1238],[191,0.1179],[195,0.0908],[196,0.0866],[197,0.0849],[199,0.0832],[200,0.0841],[201,0.0847],[205,0.0847],[206,0.0837],[207,0.0836],[208,0.0826],[212,0.0826],[214,0.0847],[215,0.0847],[217,0.0867],[220,0.0867],[223,0.0834],[224,0.0834],[225,0.0845],[232,0.0996],[234,0.102],[235,0.102],[236,0.1032],[239,0.1032],[240,0.102],[241,0.102],[243,0.0996],[244,0.0996],[248,0.0948],[249,0.0942],[250,0.093],[254,0.0906],[255,0.0906],[256,0.09],[260,0.09],[261,0.0906],[262,0.0906],[263,0.0912],[264,0.0912],[265,0.0918],[266,0.0919],[267,0.0931],[268,0.0931],[269,0.0944],[270,0.0944],[271,0.0957],[272,0.0957],[273,0.097],[274,0.097],[275,0.0983],[276,0.0983],[277,0.0996],[279,0.0996],[280,0.1009],[282,0.1009],[283,0.1021],[285,0.1021],[286,0.1033],[290,0.1033],[291,0.1045],[296,0.1045],[297,0],[367,0]]
       }
       calf_r_fk_ctrl.Location.Z {type=MovieSceneScriptingFloatChannel, default=0, thinned_from=368, keys [[0,0.0008],[296,0.0008],[297,0],[367,0]]}
       calf_r_fk_ctrl.Rotation.X {type=MovieSceneScriptingFloatChannel, default=0.0752, thinned_from=368, keys [[0,0.008],[296,0.008],[297,0],[367,0]]}
       calf_r_fk_ctrl.Rotation.Y {type=MovieSceneScriptingFloatChannel, default=1.0359, thinned_from=368, keys [[0,-0.0078],[296,-0.0078],[297,0],[367,0]]}
       calf_r_fk_ctrl.Rotation.Z {type=MovieSceneScriptingFloatChannel, default=0.0678, thinned_from=368, keys [[0,-7.8239],[296,-7.8239],[297,0],[367,0]]}
       calf_r_fk_ctrl.Scale.X {type=MovieSceneScriptingFloatChannel, default=1, const [1,368]}
       calf_r_fk_ctrl.Scale.Y {type=MovieSceneScriptingFloatChannel, default=1, const [1,368]}
       calf_r_fk_ctrl.Scale.Z {type=MovieSceneScriptingFloatChannel, default=1, const [1,368]}
       foot_r_fk_ctrl.Location.X {
        type=MovieSceneScriptingFloatChannel
        default=0
        thinned_from=368
        keys [[0,-0.8336],[16,-0.8336],[17,-0.8304],[33,-0.8304],[34,-0.8272],[37,-0.8272],[38,-0.8174],[42,-0.8174],[43,-0.8076],[44,-0.8076],[45,-0.7978],[46,-0.7978],[50,-0.7586],[51,-0.7535],[55,-0.7152],[56,-0.7152],[57,-0.7057],[58,-0.7057],[59,-0.6961],[61,-0.6961],[62,-0.6866],[84,-0.6869],[85,-0.6803],[101,-0.6803],[104,-0.6915],[105,-0.6981],[106,-0.7015],[114,-0.7546],[115,-0.7645],[120,-0.7987],[123,-0.8157],[127,-0.8307],[130,-0.8363],[131,-0.8401],[135,-0.8476],[136,-0.8478],[137,-0.8515],[138,-0.8515],[139,-0.8552],[140,-0.8552],[141,-0.8588],[143,-0.8588],[144,-0.8625],[146,-0.8625],[147,-0.8662],[149,-0.8662],[150,-0.8699],[153,-0.8701],[154,-0.8747],[159,-0.8747],[160,-0.8793],[168,-0.8793],[174,-0.8804],[175,-0.8873],[177,-0.915],[180,-0.998],[181,-1.0326],[182,-1.0602],[183,-1.081],[184,-1.0948],[186,-1.0965],[187,-1.0825],[189,-1.0265],[191,-0.9425],[194,-0.7745],[196,-0.6906],[198,-0.6626],[200,-0.6626],[201,-0.6721],[202,-0.6758],[204,-0.6758],[209,-0.6574],[211,-0.6574],[213,-0.6648],[214,-0.6721],[215,-0.6758],[217,-0.6907],[220,-0.6907],[223,-0.6643],[224,-0.6643],[225,-0.6731],[232,-0.7963],[233,-0.8026],[234,-0.8118],[237,-0.8255],[238,-0.8255],[239,-0.8209],[240,-0.8209],[241,-0.8118],[242,-0.8072],[243,-0.7981],[244,-0.7935],[248,-0.7569],[249,-0.7518],[250,-0.7403],[251,-0.7403],[252,-0.7289],[253,-0.7289],[254,-0.7174],[261,-0.7174],[262,-0.7289],[264,-0.7289],[279,-0.8017],[280,-0.8017],[284,-0.8191],[285,-0.8191],[286,-0.8233],[287,-0.8233],[288,-0.8275],[289,-0.8275],[290,-0.8316],[295,-0.8316],[296,-0.8358],[297,0],[367,0]]
       }
       foot_r_fk_ctrl.Location.Y {
        type=MovieSceneScriptingFloatChannel
        default=-0.0543
        thinned_from=368
        keys [[0,-0.001],[283,-0.001],[284,-0.0007],[285,-0.001],[296,-0.001],[297,0],[367,0]]
       }
       foot_r_fk_ctrl.Location.Z {type=MovieSceneScriptingFloatChannel, default=-0.8609, thinned_from=368, keys [[0,0.0008],[296,0.0008],[297,0],[367,0]]}
       foot_r_fk_ctrl.Rotation.X {
        type=MovieSceneScriptingFloatChannel
        default=-0.9802
        thinned_from=368
        keys [[0,-0.4053],[5,-0.4054],[21,-0.4006],[26,-0.3983],[27,-0.3971],[31,-0.3952],[37,-0.3921],[38,-0.391],[39,-0.3909],[40,-0.3898],[41,-0.3897],[44,-0.3871],[45,-0.3869],[48,-0.3809],[50,-0.3758],[57,-0.356],[60,-0.3512],[62,-0.3496],[69,-0.3486],[86,-0.3459],[87,-0.3465],[89,-0.3454],[96,-0.3457],[99,-0.3453],[100,-0.3457],[101,-0.3451],[105,-0.3493],[108,-0.3543],[113,-0.3663],[121,-0.386],[128,-0.3976],[141,-0.4121],[152,-0.4213],[155,-0.4225],[160,-0.4266],[162,-0.4267],[164,-0.4284],[173,-0.4308],[174,-0.4296],[175,-0.4267],[177,-0.418],[178,-0.4123],[180,-0.3988],[183,-0.3732],[184,-0.3635],[186,-0.3415],[187,-0.3296],[189,-0.3037],[191,-0.2755],[192,-0.2608],[194,-0.2293],[196,-0.1966],[197,-0.1806],[198,-0.1654],[199,-0.152],[200,-0.1407],[201,-0.1316],[202,-0.1264],[203,-0.1254],[204,-0.1284],[205,-0.1352],[206,-0.145],[207,-0.1572],[208,-0.1713],[209,-0.1878],[210,-0.2049],[211,-0.2237],[212,-0.2431],[214,-0.2841],[215,-0.304],[216,-0.3247],[218,-0.3639],[220,-0.3993],[221,-0.4148],[222,-0.4292],[223,-0.4423],[224,-0.454],[225,-0.4641],[226,-0.4728],[227,-0.4804],[228,-0.4868],[229,-0.4905],[230,-0.4932],[232,-0.4944],[234,-0.4904],[236,-0.4828],[246,-0.4332],[250,-0.4171],[254,-0.4046],[256,-0.4005],[266,-0.3955],[280,-0.3989],[281,-0.3998],[292,-0.4048],[296,-0.4055],[297,0],[367,0]]
       }
       foot_r_fk_ctrl.Rotation.Y {
        type=MovieSceneScriptingFloatChannel
        default=0.1214
        thinned_from=368
        keys [[0,-3.3085],[24,-3.2605],[27,-3.2444],[45,-3.1591],[48,-3.1131],[56,-2.9344],[59,-2.8841],[61,-2.8653],[88,-2.8301],[101,-2.8249],[104,-2.845],[108,-2.898],[121,-3.1523],[126,-3.225],[152,-3.4364],[161,-3.4814],[173,-3.5086],[174,-3.4999],[175,-3.4779],[177,-3.411],[178,-3.3662],[180,-3.254],[181,-3.1911],[183,-3.0481],[185,-2.8831],[186,-2.7935],[187,-2.6975],[189,-2.4901],[191,-2.2654],[192,-2.1492],[194,-1.9018],[196,-1.6459],[197,-1.5216],[198,-1.404],[199,-1.2996],[200,-1.2116],[201,-1.1413],[202,-1.1019],[203,-1.0963],[204,-1.1212],[205,-1.1762],[206,-1.2525],[207,-1.3496],[208,-1.4618],[209,-1.5917],[210,-1.7276],[211,-1.8752],[212,-2.0285],[216,-2.6681],[218,-2.9779],[220,-3.2572],[221,-3.3816],[222,-3.4959],[223,-3.6001],[224,-3.6945],[225,-3.7745],[226,-3.8448],[227,-3.9054],[228,-3.9563],[229,-3.9879],[230,-4.0098],[232,-4.0198],[234,-3.9886],[236,-3.9284],[246,-3.531],[250,-3.402],[254,-3.3009],[256,-3.268],[268,-3.2272],[296,-3.3092],[297,0],[367,0]]
       }
       foot_r_fk_ctrl.Rotation.Z {
        type=MovieSceneScriptingFloatChannel
        default=0.0013
        thinned_from=368
        keys [[0,6.3245],[34,6.2704],[49,6.1623],[54,6.0467],[64,5.7143],[68,5.6243],[99,5.503],[101,5.5082],[105,5.569],[108,5.6506],[118,6.0129],[123,6.1455],[153,6.3763],[173,6.44],[174,6.4346],[176,6.3792],[178,6.2674],[180,6.1167],[186,5.5607],[188,5.3981],[189,5.329],[190,5.271],[191,5.2348],[192,5.2134],[194,5.202],[196,5.2217],[201,5.3216],[202,5.3631],[203,5.4358],[204,5.5327],[205,5.6465],[207,5.9077],[212,6.6174],[214,6.8774],[216,7.098],[218,7.2729],[220,7.3941],[221,7.4311],[222,7.4514],[225,7.4481],[228,7.3936],[231,7.2787],[234,7.1108],[237,6.8891],[243,6.4087],[246,6.2148],[251,5.9903],[254,5.9066],[256,5.8812],[266,5.9617],[291,6.3015],[296,6.3235],[297,0],[367,0]]
       }
       foot_r_fk_ctrl.Scale.X {type=MovieSceneScriptingFloatChannel, default=1, const [1,368]}
       foot_r_fk_ctrl.Scale.Y {type=MovieSceneScriptingFloatChannel, default=1, const [1,368]}
       foot_r_fk_ctrl.Scale.Z {type=MovieSceneScriptingFloatChannel, default=1, const [1,368]}
       ball_r_fk_ctrl.Location.X {type=MovieSceneScriptingFloatChannel, default=0, const [0,368]}
       ball_r_fk_ctrl.Location.Y {type=MovieSceneScriptingFloatChannel, default=0, const [0,368]}
       ball_r_fk_ctrl.Location.Z {type=MovieSceneScriptingFloatChannel, default=0, const [0,368]}
       ball_r_fk_ctrl.Rotation.X {type=MovieSceneScriptingFloatChannel, default=0, const [0,368]}
       ball_r_fk_ctrl.Rotation.Y {type=MovieSceneScriptingFloatChannel, default=0, const [0,368]}
       ball_r_fk_ctrl.Rotation.Z {type=MovieSceneScriptingFloatChannel, default=0, const [0,368]}
       ball_r_fk_ctrl.Scale.X {type=MovieSceneScriptingFloatChannel, default=1, const [1,368]}
       ball_r_fk_ctrl.Scale.Y {type=MovieSceneScriptingFloatChannel, default=1, const [1,368]}
       ball_r_fk_ctrl.Scale.Z {type=MovieSceneScriptingFloatChannel, default=1, const [1,368]}
       foot_r_ik_ctrl.Location.X {
        type=MovieSceneScriptingFloatChannel
        default=0
        thinned_from=368
        keys [[0,0.5202],[3,0.5195],[4,0.5182],[5,0.5187],[7,0.5145],[8,0.5192],[9,0.517],[10,0.5185],[11,0.5178],[12,0.5213],[13,0.5158],[16,0.5172],[17,0.5139],[19,0.5169],[21,0.5177],[22,0.5147],[23,0.5206],[24,0.517],[25,0.5151],[28,0.5204],[29,0.5193],[32,0.5187],[33,0.5174],[34,0.5203],[35,0.5173],[37,0.5195],[38,0.5168],[40,0.5177],[41,0.5166],[42,0.5182],[43,0.5147],[44,0.5189],[45,0.5162],[46,0.5207],[47,0.5159],[48,0.5198],[49,0.5184],[50,0.5148],[51,0.5181],[52,0.5189],[53,0.5175],[54,0.5198],[55,0.5184],[56,0.5154],[57,0.5181],[58,0.5153],[59,0.5183],[61,0.5193],[62,0.5143],[63,0.516],[64,0.5196],[66,0.5161],[68,0.5166],[69,0.521],[71,0.5142],[72,0.5167],[74,0.5162],[75,0.5171],[76,0.515],[77,0.5181],[78,0.5164],[79,0.5158],[81,0.5173],[82,0.5194],[83,0.5157],[84,0.5196],[85,0.5196],[86,0.5215],[87,0.5169],[89,0.5182],[91,0.5153],[92,0.5174],[93,0.5161],[94,0.5176],[95,0.5153],[96,0.5155],[97,0.5185],[98,0.5158],[99,0.5154],[100,0.5197],[101,0.5181],[105,0.5184],[106,0.5157],[107,0.5148],[108,0.5162],[110,0.5137],[111,0.5186],[113,0.5178],[114,0.5157],[115,0.52],[116,0.517],[117,0.5198],[118,0.5135],[119,0.5175],[120,0.5181],[121,0.5173],[122,0.5189],[123,0.5174],[127,0.5174],[128,0.5211],[129,0.518],[130,0.5175],[131,0.5185],[132,0.5161],[133,0.515],[134,0.5196],[135,0.5176],[136,0.5175],[137,0.5204],[138,0.5175],[140,0.519],[141,0.5176],[142,0.52],[143,0.5182],[144,0.5191],[145,0.515],[147,0.516],[149,0.5205],[150,0.5199],[152,0.516],[153,0.518],[154,0.5178],[155,0.5153],[156,0.5197],[159,0.5142],[160,0.5174],[161,0.5169],[162,0.5204],[163,0.5162],[164,0.5195],[165,0.5175],[167,0.519],[168,0.5168],[169,0.5181],[170,0.5147],[171,0.5195],[172,0.5166],[173,0.5185],[174,0.5167],[175,0.5183],[176,0.5149],[178,0.5207],[179,0.5191],[180,0.5187],[181,0.5154],[182,0.5205],[183,0.5189],[184,0.52],[185,0.5171],[187,0.5166],[188,0.5181],[189,0.5158],[190,0.5153],[192,0.5188],[197,0.5196],[199,0.517],[201,0.5172],[202,0.5157],[203,0.5181],[204,0.5162],[205,0.5168],[206,0.5163],[207,0.5188],[208,0.516],[209,0.518],[210,0.5167],[213,0.5176],[214,0.5187],[215,0.517],[216,0.5174],[217,0.5167],[219,0.5184],[221,0.5182],[222,0.5173],[223,0.5185],[226,0.519],[228,0.5156],[229,0.517],[230,0.5164],[232,0.518],[234,0.5177],[235,0.5148],[237,0.5197],[238,0.5165],[239,0.5195],[240,0.5212],[241,0.518],[242,0.5163],[243,0.518],[244,0.5225],[245,0.5154],[246,0.5208],[247,0.519],[248,0.5149],[249,0.5163],[250,0.5211],[251,0.5148],[252,0.5149],[253,0.5187],[254,0.5163],[255,0.5223],[257,0.5151],[258,0.5159],[259,0.519],[260,0.517],[261,0.5196],[262,0.5208],[264,0.5163],[265,0.5165],[266,0.5151],[267,0.5177],[268,0.5172],[269,0.52],[270,0.5169],[271,0.5157],[272,0.5224],[273,0.5191],[274,0.5197],[275,0.5171],[276,0.5161],[277,0.517],[278,0.5164],[279,0.5186],[280,0.5186],[281,0.5163],[282,0.517],[283,0.5165],[284,0.5178],[289,0.5187],[290,0.5174],[291,0.52],[292,0.5208],[293,0.5154],[294,0.5134],[296,0.5173],[297,0.0001],[367,0]]
       }
       foot_r_ik_ctrl.Location.Y {type=MovieSceneScriptingFloatChannel, default=0, thinned_from=368, keys [[0,-19.8336],[296,-19.8387],[297,-0.001],[298,0],[367,0]]}
       foot_r_ik_ctrl.Location.Z {
        type=MovieSceneScriptingFloatChannel
        default=0
        thinned_from=368
        keys [[0,0.0031],[2,0.0031],[3,0.003],[4,0.0032],[5,0.0026],[6,0.0024],[7,0.0025],[8,0.002],[9,0.0017],[11,0.0005],[13,-0.0002],[14,-0.0011],[15,-0.0015],[16,-0.0021],[17,0.0042],[19,0.0029],[21,0.0013],[22,0.0008],[23,-0.0001],[24,0.0021],[25,0.0012],[30,-0.002],[31,-0.0028],[32,-0.001],[33,-0.0015],[34,0.0021],[35,0.0014],[36,0.0009],[37,-0.0002],[38,0.008],[39,0.0053],[40,0.0065],[41,0.0022],[42,0.001],[43,0.0083],[44,0.0006],[45,0.0086],[46,0.0001],[47,0.0031],[48,0.005],[50,0.0052],[51,0.0009],[52,0.0009],[53,0.0023],[54,0.001],[55,0.0056],[56,-0.0019],[57,0.0051],[58,0.0005],[59,0.0063],[60,0.0005],[61,0.0002],[62,0.0075],[64,0.004],[65,0.0069],[68,0.0038],[69,0.0022],[70,0.0011],[71,0.0004],[72,0.0021],[73,0.0011],[74,0.0005],[75,-0.0008],[76,-0.001],[77,-0.0021],[78,-0.003],[79,-0.0031],[80,-0.0007],[82,-0.0018],[84,-0.0023],[85,0.0033],[86,0.0035],[87,0.0032],[88,0.0032],[89,0.0027],[91,0.0022],[92,0.0023],[93,0.002],[94,0.0024],[95,0.0022],[96,0.0024],[97,0.0022],[98,0.0023],[99,0.0029],[100,0.0049],[101,0.0082],[102,0.0042],[103,0.0068],[104,-0.0025],[105,-0.0002],[106,0.0057],[107,-0.0035],[108,0.0013],[109,0.0072],[110,-0.0002],[111,0.0063],[112,0.0004],[113,0.0078],[114,0.0023],[115,0.0064],[116,0.0007],[117,0.008],[118,0.0013],[119,0.0046],[120,0.0023],[121,0.0004],[122,0.0057],[123,0.0026],[124,0.0073],[125,0.004],[126,0.0066],[127,0.0003],[128,0.0034],[129,0.0055],[130,0.0006],[131,0.0007],[133,0.005],[134,-0.0006],[135,0.0006],[136,0.0055],[137,0.0053],[138,0.0005],[139,0.0001],[140,0.0032],[141,0.0024],[142,0.005],[143,0.008],[144,-0.0013],[145,0.0015],[146,0.0039],[147,0.0027],[148,0.0046],[149,0.0069],[150,0.0051],[151,-0.0003],[152,0.0015],[153,0.006],[154,-0.0007],[156,0.0021],[157,0.004],[158,0.0052],[159,0.0068],[160,-0.0007],[163,0.0026],[166,0.0054],[168,0.0069],[169,0.0051],[171,0.0059],[172,0.0061],[173,0.0057],[174,0.0082],[175,0.0062],[176,0.0016],[177,0.0062],[178,-0.0014],[179,-0.0003],[180,0.0081],[181,0.0018],[182,0.0028],[183,0.0009],[184,-0.0023],[185,0.0002],[186,-0.003],[187,-0.0004],[188,0.0071],[189,0.0041],[190,0.0057],[191,0.0042],[192,0.0094],[193,0.0084],[194,0.0068],[195,0.0001],[196,0.005],[197,0.0023],[198,0.0039],[199,0.0047],[200,0.0044],[201,-0.0017],[202,0.0031],[203,0.0071],[204,0.0061],[205,0.0042],[206,0.0077],[207,0.0028],[208,0.0065],[209,0.0051],[210,0.0033],[212,0.0074],[213,0.0045],[214,-0.0006],[215,0.0066],[216,0.0018],[217,-0.0044],[218,0.0011],[219,0.0013],[220,-0.0036],[221,0.0029],[222,0.0045],[223,0.0075],[224,0.0051],[225,0.007],[226,0.0039],[227,0.0031],[228,0.0049],[229,0.0076],[230,0.0087],[231,0.0057],[232,-0.0032],[233,0.0012],[234,-0.002],[235,0.0049],[236,-0.0009],[237,-0.0009],[238,-0.0002],[239,0.0018],[240,0.0042],[241,0.0041],[242,0.0048],[243,0.0078],[244,-0.0033],[245,-0.0017],[246,-0.0004],[247,0.0011],[248,0.0038],[249,-0.0008],[250,0.0063],[251,-0.0012],[252,0.0046],[253,0.0009],[254,0.0094],[255,0.0043],[256,0.0056],[257,0.004],[258,0.0036],[259,0.0049],[260,0.0068],[261,0.0057],[262,-0.001],[263,-0.0001],[264,0.0064],[265,0.0045],[266,0.0074],[267,0.0019],[268,0.0055],[269,0.0007],[270,0.0054],[271,0.0006],[272,0.0057],[273,0.0008],[274,0.0054],[275,0.0006],[276,0.0044],[277,-0.0011],[279,0.0046],[280,0.0032],[281,0.0051],[282,0.0074],[283,0.0006],[284,0.0021],[285,0.007],[286,-0.0012],[287,0.0031],[288,0.0026],[289,0.0059],[290,0.0047],[291,-0.0017],[292,0.0005],[293,0.0022],[295,0.0033],[296,-0.0005],[297,0.0001],[367,0]]
       }
       foot_r_ik_ctrl.Rotation.X {
        type=MovieSceneScriptingFloatChannel
        default=0
        thinned_from=368
        keys [[0,-0.0317],[1,-0.0317],[2,-0.031],[3,-0.0315],[4,-0.0304],[5,-0.0306],[6,-0.0297],[7,-0.0283],[8,-0.0304],[9,-0.0298],[10,-0.0298],[11,-0.0304],[12,-0.0319],[13,-0.0285],[14,-0.0285],[15,-0.0297],[16,-0.0289],[17,-0.0263],[19,-0.0299],[21,-0.0325],[22,-0.0288],[23,-0.0282],[24,-0.0288],[25,-0.0313],[26,-0.0281],[27,-0.0293],[28,-0.0307],[29,-0.0297],[32,-0.0305],[34,-0.0289],[35,-0.032],[36,-0.0298],[37,-0.0279],[38,-0.0309],[39,-0.029],[40,-0.0308],[41,-0.0273],[42,-0.0304],[43,-0.0301],[44,-0.0317],[45,-0.0289],[46,-0.0292],[47,-0.0277],[49,-0.0282],[50,-0.0294],[52,-0.0306],[53,-0.0309],[54,-0.031],[55,-0.0308],[56,-0.0276],[57,-0.0324],[58,-0.027],[59,-0.0311],[60,-0.0283],[61,-0.0304],[62,-0.0283],[63,-0.0288],[64,-0.0327],[65,-0.0277],[66,-0.0269],[67,-0.0297],[68,-0.0284],[69,-0.0301],[70,-0.0292],[71,-0.0279],[72,-0.0287],[73,-0.0309],[74,-0.0286],[76,-0.0283],[77,-0.0299],[78,-0.0312],[79,-0.0291],[80,-0.0292],[81,-0.0289],[82,-0.0296],[83,-0.0295],[84,-0.0315],[85,-0.0298],[86,-0.0319],[87,-0.0281],[88,-0.03],[89,-0.0301],[90,-0.0317],[91,-0.03],[92,-0.0277],[93,-0.0323],[94,-0.0294],[95,-0.028],[96,-0.0274],[97,-0.0316],[98,-0.0296],[99,-0.0309],[100,-0.0276],[101,-0.0308],[102,-0.029],[103,-0.0281],[104,-0.0299],[105,-0.0287],[107,-0.0276],[108,-0.0292],[109,-0.027],[111,-0.0322],[112,-0.0293],[113,-0.0286],[114,-0.0294],[115,-0.0316],[116,-0.0283],[117,-0.0291],[118,-0.0274],[119,-0.0297],[120,-0.0291],[121,-0.0301],[122,-0.0295],[123,-0.0291],[124,-0.0296],[125,-0.0309],[127,-0.0306],[128,-0.0299],[129,-0.0303],[130,-0.0286],[131,-0.0289],[132,-0.0301],[133,-0.0277],[134,-0.0293],[135,-0.0292],[137,-0.0309],[138,-0.0302],[139,-0.0315],[140,-0.03],[141,-0.0298],[142,-0.0315],[143,-0.0288],[144,-0.0319],[145,-0.0288],[146,-0.0294],[147,-0.0294],[148,-0.0316],[149,-0.0309],[150,-0.0308],[151,-0.0289],[152,-0.0297],[153,-0.0282],[154,-0.0317],[155,-0.0306],[156,-0.0311],[157,-0.0291],[158,-0.0289],[159,-0.0278],[160,-0.0276],[161,-0.0317],[162,-0.0316],[163,-0.0292],[164,-0.0293],[165,-0.0283],[166,-0.0306],[167,-0.0291],[169,-0.0278],[170,-0.0302],[171,-0.0269],[172,-0.0295],[173,-0.0259],[174,-0.0277],[175,-0.03],[176,-0.0282],[177,-0.0296],[178,-0.0302],[179,-0.0328],[180,-0.0302],[181,-0.0317],[182,-0.0283],[183,-0.0279],[184,-0.0269],[185,-0.0292],[186,-0.0294],[187,-0.0286],[188,-0.0294],[189,-0.0287],[190,-0.0274],[191,-0.0295],[192,-0.0309],[193,-0.0313],[194,-0.0307],[195,-0.0296],[196,-0.0312],[197,-0.0314],[199,-0.0301],[200,-0.0292],[201,-0.0304],[202,-0.0297],[203,-0.032],[204,-0.0303],[205,-0.0304],[206,-0.0279],[207,-0.0287],[208,-0.0282],[210,-0.0309],[211,-0.0302],[212,-0.0306],[213,-0.0312],[214,-0.0282],[215,-0.0299],[216,-0.0293],[217,-0.0289],[219,-0.0297],[220,-0.0292],[221,-0.0291],[222,-0.0297],[223,-0.03],[224,-0.0306],[225,-0.0301],[226,-0.0305],[227,-0.0291],[228,-0.0287],[230,-0.03],[231,-0.0301],[232,-0.0306],[233,-0.0295],[234,-0.0304],[235,-0.0291],[236,-0.0274],[237,-0.0303],[238,-0.029],[239,-0.0327],[240,-0.0314],[241,-0.0273],[242,-0.029],[243,-0.0316],[244,-0.0304],[245,-0.0287],[246,-0.0306],[247,-0.0315],[248,-0.0281],[249,-0.0279],[250,-0.0301],[251,-0.0312],[252,-0.0292],[253,-0.0305],[254,-0.0296],[255,-0.0312],[256,-0.0315],[257,-0.0295],[258,-0.0309],[259,-0.0299],[260,-0.0305],[261,-0.0304],[262,-0.0283],[263,-0.0291],[264,-0.0295],[265,-0.0294],[266,-0.0301],[267,-0.0286],[268,-0.0306],[269,-0.0291],[270,-0.0312],[271,-0.0308],[272,-0.0313],[273,-0.0292],[274,-0.0293],[275,-0.0302],[277,-0.0287],[278,-0.0281],[279,-0.0291],[280,-0.0305],[281,-0.027],[282,-0.0307],[283,-0.0272],[284,-0.0305],[285,-0.0283],[286,-0.0307],[287,-0.0308],[288,-0.0311],[289,-0.0308],[291,-0.0283],[292,-0.0311],[293,-0.0301],[294,-0.0295],[295,-0.0283],[296,-0.029],[297,0.0003],[298,0],[367,0]]
       }
       foot_r_ik_ctrl.Rotation.Y {
        type=MovieSceneScriptingFloatChannel
        default=0
        thinned_from=368
        keys [[0,0.0269],[2,0.0268],[3,0.0262],[4,0.0265],[5,0.0249],[6,0.0276],[7,0.0272],[8,0.0256],[9,0.0277],[10,0.0257],[11,0.0273],[12,0.025],[13,0.0274],[14,0.0277],[15,0.0253],[16,0.0263],[17,0.0278],[19,0.0272],[21,0.0253],[22,0.0273],[23,0.0284],[24,0.0258],[25,0.0267],[27,0.0293],[28,0.0255],[29,0.0271],[30,0.0283],[31,0.0288],[32,0.0244],[33,0.0256],[34,0.0255],[36,0.0262],[37,0.0261],[38,0.027],[39,0.0263],[40,0.0272],[41,0.0269],[42,0.027],[43,0.0273],[44,0.0275],[45,0.025],[46,0.029],[47,0.0284],[48,0.0272],[49,0.027],[50,0.0265],[51,0.0273],[52,0.0269],[54,0.0269],[55,0.0266],[56,0.0279],[57,0.027],[58,0.0275],[59,0.0267],[60,0.026],[61,0.0262],[62,0.0278],[63,0.0274],[64,0.0269],[66,0.0267],[67,0.0265],[68,0.0278],[69,0.0264],[70,0.0262],[71,0.0276],[72,0.0272],[73,0.0264],[74,0.0271],[75,0.0261],[76,0.0274],[77,0.0263],[78,0.0266],[79,0.028],[80,0.0268],[81,0.0268],[82,0.0263],[83,0.0272],[84,0.0267],[85,0.0254],[86,0.0261],[87,0.0256],[88,0.0253],[89,0.0293],[90,0.0282],[91,0.0283],[92,0.0281],[93,0.0272],[94,0.028],[95,0.0272],[96,0.0276],[97,0.0276],[98,0.0277],[99,0.0268],[100,0.0282],[101,0.0297],[102,0.0271],[103,0.0262],[104,0.0276],[105,0.0265],[106,0.0272],[107,0.0276],[108,0.0275],[110,0.0266],[111,0.0258],[112,0.0279],[113,0.0277],[114,0.0262],[115,0.0261],[116,0.0263],[117,0.0278],[118,0.0266],[119,0.0267],[120,0.0274],[121,0.026],[122,0.0266],[124,0.0264],[125,0.0283],[126,0.0271],[127,0.0271],[128,0.0274],[130,0.027],[131,0.0274],[132,0.0268],[133,0.0278],[135,0.0262],[136,0.027],[137,0.0272],[138,0.026],[139,0.0259],[140,0.0277],[141,0.0267],[142,0.0275],[143,0.0267],[144,0.0264],[145,0.0276],[147,0.0284],[148,0.0276],[149,0.0278],[150,0.0269],[151,0.0272],[152,0.0261],[153,0.0292],[154,0.0298],[156,0.0285],[157,0.0289],[158,0.0278],[159,0.027],[160,0.0256],[161,0.0249],[162,0.0283],[163,0.0268],[164,0.0246],[165,0.0279],[166,0.0261],[167,0.0284],[168,0.0259],[169,0.0275],[170,0.0252],[171,0.0271],[172,0.0284],[173,0.0268],[174,0.0246],[175,0.0268],[176,0.0283],[177,0.027],[178,0.0266],[179,0.0274],[180,0.0273],[181,0.0256],[182,0.0281],[183,0.0258],[185,0.0268],[186,0.0275],[187,0.0277],[188,0.0285],[189,0.0264],[190,0.0284],[191,0.026],[192,0.0257],[193,0.0274],[194,0.0275],[195,0.0284],[196,0.0275],[197,0.0259],[198,0.0268],[199,0.0274],[200,0.0274],[201,0.0282],[202,0.0275],[203,0.0266],[204,0.0274],[205,0.0254],[206,0.0266],[207,0.0267],[208,0.0283],[209,0.0256],[210,0.0282],[211,0.0278],[212,0.0278],[213,0.0272],[214,0.0255],[215,0.0283],[216,0.0267],[217,0.0273],[218,0.0247],[219,0.0292],[220,0.0263],[221,0.0289],[222,0.0287],[224,0.0244],[225,0.0273],[226,0.028],[227,0.0273],[228,0.0248],[229,0.0274],[230,0.0271],[231,0.029],[232,0.0251],[233,0.0267],[234,0.0262],[235,0.0268],[236,0.0264],[237,0.0271],[238,0.0266],[239,0.0272],[240,0.027],[241,0.0278],[242,0.0272],[243,0.0263],[244,0.0263],[245,0.0272],[246,0.0275],[247,0.0269],[248,0.027],[249,0.0282],[250,0.0288],[251,0.0259],[252,0.0252],[253,0.0269],[254,0.0272],[255,0.0271],[256,0.0286],[257,0.0268],[258,0.0276],[259,0.0262],[260,0.0273],[261,0.0263],[262,0.0264],[263,0.0261],[264,0.0273],[265,0.0269],[266,0.0273],[267,0.0281],[268,0.0261],[270,0.027],[271,0.0267],[272,0.0274],[273,0.0264],[274,0.0267],[275,0.0279],[276,0.0269],[277,0.0276],[278,0.0265],[279,0.0259],[280,0.028],[281,0.0265],[282,0.0276],[283,0.0271],[284,0.0269],[285,0.0257],[286,0.0276],[287,0.0271],[288,0.0259],[289,0.0274],[290,0.0263],[291,0.0267],[292,0.0282],[293,0.0286],[294,0.0263],[296,0.0279],[297,0.0001],[367,0]]
       }
       foot_r_ik_ctrl.Rotation.Z {type=MovieSceneScriptingFloatChannel, default=0, thinned_from=368, keys [[0,-59.7345],[296,-59.7342],[297,-0.0011],[298,0],[367,0]]}
       foot_r_ik_ctrl.Scale.X {type=MovieSceneScriptingFloatChannel, default=1, const [1,368]}
       foot_r_ik_ctrl.Scale.Y {type=MovieSceneScriptingFloatChannel, default=1, const [1,368]}
       foot_r_ik_ctrl.Scale.Z {type=MovieSceneScriptingFloatChannel, default=1, const [1,368]}
       ball_r_ik_ctrl.Location.X {type=MovieSceneScriptingFloatChannel, default=0, const [0,368]}
       ball_r_ik_ctrl.Location.Y {type=MovieSceneScriptingFloatChannel, default=0, const [0,368]}
       ball_r_ik_ctrl.Location.Z {type=MovieSceneScriptingFloatChannel, default=0, const [0,368]}
       ball_r_ik_ctrl.Rotation.X {type=MovieSceneScriptingFloatChannel, default=0, const [0,368]}
       ball_r_ik_ctrl.Rotation.Y {type=MovieSceneScriptingFloatChannel, default=0, const [0,368]}
       ball_r_ik_ctrl.Rotation.Z {type=MovieSceneScriptingFloatChannel, default=0, const [0,368]}
       ball_r_ik_ctrl.Scale.X {type=MovieSceneScriptingFloatChannel, default=1, const [1,368]}
       ball_r_ik_ctrl.Scale.Y {type=MovieSceneScriptingFloatChannel, default=1, const [1,368]}
       ball_r_ik_ctrl.Scale.Z {type=MovieSceneScriptingFloatChannel, default=1, const [1,368]}
       foot_bk1_r_ctrl.Location.X {type=MovieSceneScriptingFloatChannel, default=0, const [0,368]}
       foot_bk1_r_ctrl.Location.Y {type=MovieSceneScriptingFloatChannel, default=0, const [0,368]}
       foot_bk1_r_ctrl.Location.Z {type=MovieSceneScriptingFloatChannel, default=0, const [0,368]}
       foot_bk1_r_ctrl.Rotation.X {type=MovieSceneScriptingFloatChannel, default=0, const [0,368]}
       foot_bk1_r_ctrl.Rotation.Y {type=MovieSceneScriptingFloatChannel, default=0, const [0,368]}
       foot_bk1_r_ctrl.Rotation.Z {type=MovieSceneScriptingFloatChannel, default=0, const [0,368]}
       foot_bk1_r_ctrl.Scale.X {type=MovieSceneScriptingFloatChannel, default=1, const [1,368]}
       foot_bk1_r_ctrl.Scale.Y {type=MovieSceneScriptingFloatChannel, default=1, const [1,368]}
       foot_bk1_r_ctrl.Scale.Z {type=MovieSceneScriptingFloatChannel, default=1, const [1,368]}
       foot_roll_r_ctrl.Location.X {type=MovieSceneScriptingFloatChannel, default=0, const [0,368]}
       foot_roll_r_ctrl.Location.Y {type=MovieSceneScriptingFloatChannel, default=0, const [0,368]}
       foot_roll_r_ctrl.Location.Z {type=MovieSceneScriptingFloatChannel, default=0, const [0,368]}
       foot_roll_r_ctrl.Rotation.X {type=MovieSceneScriptingFloatChannel, default=0, const [0,368]}
       foot_roll_r_ctrl.Rotation.Y {type=MovieSceneScriptingFloatChannel, default=0, const [0,368]}
       foot_roll_r_ctrl.Rotation.Z {type=MovieSceneScriptingFloatChannel, default=0, const [0,368]}
       foot_roll_r_ctrl.Scale.X {type=MovieSceneScriptingFloatChannel, default=1, const [1,368]}
       foot_roll_r_ctrl.Scale.Y {type=MovieSceneScriptingFloatChannel, default=1, const [1,368]}
       foot_roll_r_ctrl.Scale.Z {type=MovieSceneScriptingFloatChannel, default=1, const [1,368]}
       roll_blend_r_ctrl {type=MovieSceneScriptingFloatChannel, default=1, const [1,368]}
       tip_r_ctrl.Location.X {type=MovieSceneScriptingFloatChannel, default=0, const [0,368]}
       tip_r_ctrl.Location.Y {type=MovieSceneScriptingFloatChannel, default=0, const [0,368]}
       tip_r_ctrl.Location.Z {type=MovieSceneScriptingFloatChannel, default=0, const [0,368]}
       tip_r_ctrl.Rotation.X {type=MovieSceneScriptingFloatChannel, default=0, const [0,368]}
       tip_r_ctrl.Rotation.Y {type=MovieSceneScriptingFloatChannel, default=0, const [0,368]}
       tip_r_ctrl.Rotation.Z {type=MovieSceneScriptingFloatChannel, default=0, const [0,368]}
       tip_r_ctrl.Scale.X {type=MovieSceneScriptingFloatChannel, default=1, const [1,368]}
       tip_r_ctrl.Scale.Y {type=MovieSceneScriptingFloatChannel, default=1, const [1,368]}
       tip_r_ctrl.Scale.Z {type=MovieSceneScriptingFloatChannel, default=1, const [1,368]}
       heel_r_ctrl.Location.X {type=MovieSceneScriptingFloatChannel, default=0, const [0,368]}
       heel_r_ctrl.Location.Y {type=MovieSceneScriptingFloatChannel, default=0, const [0,368]}
       heel_r_ctrl.Location.Z {type=MovieSceneScriptingFloatChannel, default=0, const [0,368]}
       heel_r_ctrl.Rotation.X {type=MovieSceneScriptingFloatChannel, default=0, const [0,368]}
       heel_r_ctrl.Rotation.Y {type=MovieSceneScriptingFloatChannel, default=0, const [0,368]}
       heel_r_ctrl.Rotation.Z {type=MovieSceneScriptingFloatChannel, default=0, const [0,368]}
       heel_r_ctrl.Scale.X {type=MovieSceneScriptingFloatChannel, default=1, const [1,368]}
       heel_r_ctrl.Scale.Y {type=MovieSceneScriptingFloatChannel, default=1, const [1,368]}
       heel_r_ctrl.Scale.Z {type=MovieSceneScriptingFloatChannel, default=1, const [1,368]}
       leg_r_softness {type=MovieSceneScriptingFloatChannel, default=0, const [0,368]}
       leg_r_pv_ik_ctrl.Location.X {
        type=MovieSceneScriptingFloatChannel
        default=0
        thinned_from=368
        keys [[0,0.0977],[11,0.0974],[14,0.0973],[16,0.0974],[17,0.0971],[21,0.0979],[22,0.0978],[24,0.0982],[26,0.0988],[27,0.0988],[33,0.1015],[34,0.1018],[35,0.1027],[37,0.1035],[38,0.104],[39,0.1048],[42,0.1066],[43,0.1075],[44,0.1081],[45,0.1088],[47,0.109],[48,0.1096],[49,0.1106],[51,0.1135],[53,0.1175],[55,0.1223],[58,0.1306],[59,0.1337],[60,0.1365],[63,0.146],[64,0.1488],[66,0.1535],[67,0.1555],[68,0.1571],[69,0.1583],[74,0.1618],[75,0.1621],[76,0.1627],[77,0.163],[78,0.1636],[88,0.1665],[89,0.1663],[90,0.1669],[92,0.1669],[93,0.1675],[100,0.1675],[102,0.1666],[103,0.1657],[104,0.1643],[106,0.1606],[108,0.1556],[111,0.1469],[113,0.1403],[114,0.1373],[117,0.1278],[120,0.1199],[122,0.1155],[124,0.112],[125,0.1105],[127,0.1083],[133,0.1033],[140,0.0993],[148,0.096],[151,0.095],[152,0.0948],[153,0.0942],[155,0.0936],[156,0.0936],[159,0.0925],[160,0.0924],[161,0.0922],[162,0.0917],[163,0.0917],[165,0.0907],[166,0.0909],[169,0.0896],[170,0.0897],[171,0.0891],[172,0.0891],[173,0.0889],[174,0.0893],[175,0.0912],[176,0.094],[177,0.0986],[178,0.104],[179,0.1104],[182,0.1326],[183,0.1406],[184,0.1482],[185,0.1554],[186,0.1618],[187,0.1673],[188,0.1718],[189,0.1752],[190,0.1772],[191,0.1772],[192,0.1757],[193,0.173],[194,0.1694],[195,0.165],[196,0.1603],[198,0.1501],[199,0.1456],[200,0.1416],[201,0.1381],[202,0.1341],[203,0.1287],[204,0.1226],[205,0.1158],[206,0.1087],[207,0.101],[208,0.0927],[209,0.0841],[211,0.0658],[212,0.0562],[213,0.047],[214,0.0373],[215,0.0282],[216,0.0196],[217,0.0113],[218,0.004],[219,-0.0024],[220,-0.0076],[221,-0.0115],[222,-0.0138],[223,-0.0145],[224,-0.0146],[225,-0.0141],[226,-0.0131],[227,-0.0114],[228,-0.0087],[229,-0.0055],[230,-0.0014],[231,0.0033],[232,0.0088],[233,0.0147],[234,0.0214],[235,0.0287],[236,0.0365],[237,0.0449],[239,0.0621],[240,0.0705],[241,0.0785],[242,0.0861],[243,0.0933],[244,0.1],[245,0.1062],[246,0.1118],[247,0.1168],[248,0.1212],[250,0.1286],[251,0.132],[252,0.1349],[253,0.1374],[254,0.1391],[255,0.1404],[256,0.141],[257,0.141],[260,0.1389],[262,0.137],[264,0.1348],[268,0.1298],[269,0.1283],[270,0.1272],[273,0.123],[274,0.1214],[278,0.1159],[283,0.1087],[284,0.1077],[285,0.1063],[286,0.1052],[287,0.1039],[291,0.0999],[292,0.0995],[293,0.0986],[294,0.0982],[296,0.098],[297,0],[367,0]]
       }
       leg_r_pv_ik_ctrl.Location.Y {type=MovieSceneScriptingFloatChannel, default=0, thinned_from=368, keys [[0,-3.5877],[296,-3.5877],[297,0],[367,0]]}
       leg_r_pv_ik_ctrl.Location.Z {type=MovieSceneScriptingFloatChannel, default=0, thinned_from=368, keys [[0,-3.0173],[296,-3.0173],[297,0],[367,0]]}
       leg_r_pv_ik_ctrl.Rotation.X {type=MovieSceneScriptingFloatChannel, default=0, thinned_from=368, keys [[0,4.0994],[296,4.0994],[297,0],[367,0]]}
       leg_r_pv_ik_ctrl.Rotation.Y {
        type=MovieSceneScriptingFloatChannel
        default=0
        thinned_from=368
        keys [[0,-0.0026],[49,-0.0013],[68,0.0031],[102,0.004],[125,-0.0013],[174,-0.0034],[176,-0.0029],[179,-0.0013],[186,0.0036],[188,0.0045],[190,0.0051],[192,0.0049],[202,0.0009],[206,-0.0015],[217,-0.0109],[219,-0.0122],[221,-0.0131],[224,-0.0134],[227,-0.013],[231,-0.0116],[235,-0.0092],[243,-0.003],[247,-0.0007],[253,0.0012],[257,0.0016],[295,-0.0025],[296,-0.0025],[297,0],[367,0]]
       }
       leg_r_pv_ik_ctrl.Rotation.Z {
        type=MovieSceneScriptingFloatChannel
        default=0
        thinned_from=368
        keys [[0,-0.142],[11,-0.1415],[16,-0.1416],[17,-0.1412],[21,-0.1423],[22,-0.1421],[24,-0.1426],[26,-0.1434],[27,-0.1434],[33,-0.1469],[34,-0.1473],[35,-0.1485],[38,-0.1502],[42,-0.1536],[43,-0.1547],[45,-0.1564],[47,-0.1566],[48,-0.1574],[49,-0.1587],[51,-0.1624],[53,-0.1676],[55,-0.1739],[58,-0.1847],[59,-0.1887],[60,-0.1923],[63,-0.2046],[64,-0.2083],[66,-0.2143],[67,-0.217],[68,-0.219],[69,-0.2205],[74,-0.2251],[84,-0.23],[86,-0.2304],[88,-0.2312],[89,-0.231],[90,-0.2318],[92,-0.2318],[93,-0.2326],[100,-0.2326],[102,-0.2314],[103,-0.2301],[104,-0.2283],[106,-0.2236],[108,-0.2171],[111,-0.2057],[113,-0.1973],[114,-0.1934],[117,-0.181],[120,-0.1708],[122,-0.1651],[124,-0.1605],[125,-0.1586],[127,-0.1558],[133,-0.1492],[140,-0.1441],[148,-0.1398],[151,-0.1385],[152,-0.1382],[153,-0.1375],[155,-0.1367],[156,-0.1367],[159,-0.1352],[160,-0.1352],[161,-0.1349],[162,-0.1342],[163,-0.1342],[165,-0.133],[166,-0.1331],[169,-0.1315],[170,-0.1317],[171,-0.1309],[172,-0.1309],[173,-0.1306],[174,-0.1311],[175,-0.1335],[176,-0.1372],[177,-0.1431],[178,-0.1502],[179,-0.1585],[182,-0.1873],[183,-0.1976],[184,-0.2075],[185,-0.2168],[186,-0.2251],[187,-0.2322],[188,-0.2381],[189,-0.2425],[190,-0.2451],[191,-0.2451],[192,-0.2432],[193,-0.2396],[194,-0.235],[195,-0.2293],[198,-0.21],[199,-0.2041],[200,-0.1989],[201,-0.1944],[202,-0.1892],[203,-0.1823],[204,-0.1742],[205,-0.1655],[206,-0.1562],[207,-0.1462],[208,-0.1355],[209,-0.1244],[210,-0.1126],[211,-0.1006],[212,-0.0882],[213,-0.0761],[214,-0.0637],[215,-0.0518],[216,-0.0407],[217,-0.0299],[218,-0.0205],[219,-0.0121],[220,-0.0054],[221,-0.0003],[222,0.0026],[223,0.0036],[224,0.0037],[225,0.0031],[226,0.0018],[227,-0.0005],[228,-0.0039],[229,-0.0082],[230,-0.0134],[231,-0.0196],[232,-0.0266],[233,-0.0344],[234,-0.0431],[235,-0.0525],[236,-0.0626],[237,-0.0735],[239,-0.0958],[240,-0.1067],[241,-0.1171],[242,-0.127],[243,-0.1363],[244,-0.145],[245,-0.153],[246,-0.1603],[247,-0.1668],[248,-0.1724],[250,-0.182],[251,-0.1864],[252,-0.1902],[253,-0.1935],[254,-0.1957],[255,-0.1973],[256,-0.1982],[257,-0.1982],[260,-0.1954],[262,-0.1929],[264,-0.1901],[268,-0.1836],[269,-0.1817],[270,-0.1803],[273,-0.1748],[278,-0.1656],[283,-0.1563],[284,-0.155],[285,-0.1531],[286,-0.1517],[287,-0.15],[291,-0.1449],[292,-0.1443],[293,-0.1432],[294,-0.1426],[296,-0.1423],[297,0],[367,0]]
       }
       leg_r_pv_ik_ctrl.Scale.X {type=MovieSceneScriptingFloatChannel, default=1, const [1,368]}
       leg_r_pv_ik_ctrl.Scale.Y {type=MovieSceneScriptingFloatChannel, default=1, const [1,368]}
       leg_r_pv_ik_ctrl.Scale.Z {type=MovieSceneScriptingFloatChannel, default=1, const [1,368]}
       upperarm_l_fk_ctrl.Location.X {type=MovieSceneScriptingFloatChannel, default=0, thinned_from=371, keys [[-109,0],[367,0]]}
       upperarm_l_fk_ctrl.Location.Y {type=MovieSceneScriptingFloatChannel, default=0, thinned_from=371, keys [[-109,0],[367,0]]}
       upperarm_l_fk_ctrl.Location.Z {type=MovieSceneScriptingFloatChannel, default=0, thinned_from=371, keys [[-109,0],[367,0]]}
       upperarm_l_fk_ctrl.Rotation.X {
        type=MovieSceneScriptingFloatChannel
        default=0
        thinned_from=371
        keys [[-109,10.0488],[-56,10.1156],[-16,10.0355],[0,9.3055],[4,9.2554],[11,9.0168],[25,8.34],[31,8.1562],[41,8.0781],[44,8.105],[46,8.1648],[47,8.2251],[48,8.3181],[49,8.4468],[50,8.6026],[51,8.7795],[52,8.9833],[53,9.2088],[54,9.4554],[62,11.6653],[64,12.1603],[66,12.5854],[68,12.9433],[81,14.164],[88,14.6095],[92,14.7459],[98,14.7424],[100,14.6411],[102,14.4348],[104,14.1352],[107,13.5521],[110,12.8596],[117,11.0686],[119,10.5941],[121,10.1694],[123,9.8089],[124,9.6587],[125,9.5317],[127,9.3492],[131,9.0766],[138,8.7394],[149,8.4089],[166,8.1681],[171,8.127],[173,8.0896],[174,8.106],[175,8.1969],[176,8.36],[177,8.5933],[178,8.8994],[179,9.274],[180,9.7161],[181,10.2163],[182,10.7638],[187,13.7362],[188,14.2952],[189,14.8249],[190,15.3232],[191,15.7836],[192,16.2066],[193,16.5929],[194,16.9408],[195,17.245],[196,17.4985],[197,17.6995],[198,17.8465],[200,18.0134],[201,18.0311],[202,17.9929],[203,17.8951],[204,17.7372],[205,17.5286],[206,17.2727],[207,16.969],[208,16.6246],[210,15.8361],[218,12.2054],[220,11.3579],[221,10.9613],[223,10.2347],[224,9.9068],[225,9.6012],[227,9.056],[228,8.7988],[229,8.5713],[230,8.3715],[231,8.2069],[232,8.0799],[233,8.0025],[234,7.9826],[235,8.029],[236,8.1319],[239,8.6015],[240,8.7362],[241,8.847],[242,8.9387],[243,9.0083],[245,9.1027],[249,9.1739],[256,9.1917],[268,9.367],[296,9.3118],[297,0],[367,0]]
       }
       upperarm_l_fk_ctrl.Rotation.Y {
        type=MovieSceneScriptingFloatChannel
        default=0
        thinned_from=371
        keys [[-109,-2.5018],[-56,-2.4748],[-16,-2.5073],[0,-1.9635],[1,-1.9635],[8,-2.0721],[13,-2.1279],[22,-2.1874],[25,-2.2007],[42,-2.1889],[43,-2.1858],[44,-2.1755],[46,-2.1449],[47,-2.1342],[48,-2.1356],[49,-2.1601],[50,-2.2085],[51,-2.2729],[53,-2.4296],[54,-2.4966],[55,-2.5519],[56,-2.5956],[57,-2.6325],[58,-2.6581],[59,-2.6732],[60,-2.6814],[62,-2.6806],[64,-2.6637],[65,-2.6506],[66,-2.6318],[67,-2.6034],[69,-2.5323],[70,-2.4839],[71,-2.4246],[72,-2.3513],[73,-2.2653],[74,-2.1682],[75,-2.0644],[76,-1.947],[77,-1.822],[78,-1.6802],[79,-1.5341],[80,-1.3798],[81,-1.2214],[84,-0.7343],[86,-0.422],[87,-0.273],[88,-0.1327],[89,-0.0002],[90,0.1272],[91,0.25],[92,0.3641],[93,0.4727],[94,0.5732],[95,0.667],[96,0.7519],[97,0.8271],[98,0.8977],[99,0.954],[100,0.9956],[101,1.0225],[102,1.0331],[103,1.0313],[104,1.0155],[105,0.9915],[106,0.9553],[107,0.9157],[108,0.8619],[109,0.7913],[110,0.7047],[111,0.6021],[112,0.4835],[113,0.3498],[114,0.2075],[115,0.058],[117,-0.2562],[118,-0.4155],[119,-0.5693],[120,-0.72],[121,-0.8624],[122,-0.9931],[123,-1.1142],[124,-1.2227],[125,-1.3174],[126,-1.3971],[127,-1.4645],[128,-1.5243],[129,-1.5769],[130,-1.6247],[131,-1.6649],[133,-1.7342],[135,-1.7946],[137,-1.8495],[140,-1.9186],[144,-1.9974],[146,-2.03],[147,-2.0502],[152,-2.1152],[158,-2.1749],[164,-2.2129],[167,-2.2224],[168,-2.2213],[173,-2.2505],[174,-2.2701],[175,-2.3212],[176,-2.4049],[178,-2.6123],[179,-2.6902],[180,-2.7208],[181,-2.6925],[182,-2.6041],[183,-2.462],[184,-2.2775],[185,-2.0606],[186,-1.8222],[187,-1.5687],[188,-1.2913],[189,-0.9811],[190,-0.6315],[191,-0.247],[192,0.167],[193,0.5961],[194,1.0292],[195,1.4356],[196,1.8041],[197,2.1174],[198,2.3832],[199,2.6043],[200,2.7914],[201,2.9455],[202,3.0603],[203,3.1374],[204,3.1722],[205,3.1687],[206,3.1252],[207,3.041],[208,2.9189],[209,2.7613],[210,2.5668],[211,2.3441],[212,2.0898],[213,1.8141],[214,1.517],[215,1.2035],[216,0.8796],[217,0.5459],[218,0.2042],[219,-0.1348],[220,-0.4645],[221,-0.788],[222,-1.0986],[223,-1.3896],[224,-1.6646],[225,-1.9196],[226,-2.1551],[227,-2.3863],[228,-2.6236],[229,-2.8362],[230,-3.0219],[231,-3.1794],[232,-3.3],[233,-3.3829],[234,-3.424],[235,-3.418],[236,-3.3798],[237,-3.3134],[239,-3.1593],[240,-3.0915],[241,-3.036],[242,-2.9918],[243,-2.9581],[245,-2.9138],[247,-2.9014],[249,-2.9055],[252,-2.9314],[253,-2.9284],[254,-2.9112],[255,-2.8869],[256,-2.8484],[258,-2.7545],[264,-2.3873],[266,-2.2872],[268,-2.1989],[269,-2.165],[270,-2.1363],[272,-2.0957],[276,-2.0345],[279,-2.0018],[284,-1.9682],[294,-1.9523],[296,-1.955],[297,0],[367,0]]
       }
       upperarm_l_fk_ctrl.Rotation.Z {
        type=MovieSceneScriptingFloatChannel
        default=0
        thinned_from=371
        keys [[-109,13.2747],[-16,13.2744],[1,13.3746],[4,13.4834],[8,13.7572],[12,14.1614],[22,15.4926],[27,15.9909],[31,16.2311],[40,16.3747],[42,16.3639],[43,16.3939],[44,16.5171],[45,16.7563],[46,17.1068],[49,18.4642],[50,18.8824],[51,19.248],[52,19.5601],[53,19.8187],[54,20.019],[56,20.2943],[62,20.7152],[73,20.9197],[77,20.8432],[81,20.54],[89,19.5868],[92,19.1094],[95,18.5025],[99,17.5371],[109,14.8011],[112,14.0816],[114,13.6827],[116,13.363],[121,12.8404],[125,12.5328],[134,12.1131],[154,11.7755],[168,11.7212],[169,11.6835],[170,11.5912],[171,11.4185],[172,11.1804],[175,10.2875],[176,10.04],[177,9.8528],[178,9.7348],[179,9.6871],[180,9.7072],[181,9.7881],[182,9.9058],[188,10.8209],[190,11.0797],[192,11.2882],[196,11.6473],[198,11.836],[199,11.8899],[200,11.9006],[201,11.8711],[202,11.8091],[207,11.2816],[209,11.132],[211,11.0627],[213,11.0691],[216,11.176],[224,11.67],[232,12.0525],[234,12.0928],[236,12.0486],[239,11.9007],[241,11.8511],[244,11.8732],[247,11.993],[252,12.358],[261,13.1604],[264,13.3415],[268,13.4542],[296,13.3647],[297,0],[367,0]]
       }
       upperarm_l_fk_ctrl.Scale.X {type=MovieSceneScriptingFloatChannel, default=1, const [1,371]}
       upperarm_l_fk_ctrl.Scale.Y {type=MovieSceneScriptingFloatChannel, default=1, const [1,371]}
       upperarm_l_fk_ctrl.Scale.Z {type=MovieSceneScriptingFloatChannel, default=1, const [1,371]}
       lowerarm_l_fk_ctrl.Location.X {type=MovieSceneScriptingFloatChannel, default=0, thinned_from=371, keys [[-109,0],[296,0],[297,-0.0001],[367,0]]}
       lowerarm_l_fk_ctrl.Location.Y {type=MovieSceneScriptingFloatChannel, default=0, thinned_from=371, keys [[-109,0],[367,0]]}
       lowerarm_l_fk_ctrl.Location.Z {type=MovieSceneScriptingFloatChannel, default=0, thinned_from=371, keys [[-109,0],[367,0]]}
       lowerarm_l_fk_ctrl.Rotation.X {
        type=MovieSceneScriptingFloatChannel
        default=0
        thinned_from=371
        keys [[-109,0.2549],[-16,0.2549],[0,-0.2878],[1,-0.2878],[7,-0.2779],[14,-0.2622],[21,-0.2449],[25,-0.2375],[28,-0.2336],[32,-0.2308],[42,-0.2295],[43,-0.2283],[44,-0.2243],[45,-0.2168],[46,-0.206],[47,-0.193],[48,-0.1796],[49,-0.167],[50,-0.1558],[51,-0.1463],[52,-0.1385],[53,-0.1321],[54,-0.1271],[55,-0.123],[56,-0.1194],[62,-0.1016],[64,-0.0963],[67,-0.0891],[70,-0.0825],[72,-0.0787],[73,-0.077],[74,-0.0756],[75,-0.0744],[76,-0.0736],[77,-0.0732],[78,-0.0734],[79,-0.074],[80,-0.0751],[81,-0.0766],[82,-0.0783],[83,-0.0803],[84,-0.0826],[85,-0.0851],[87,-0.0908],[89,-0.0973],[90,-0.101],[91,-0.1052],[92,-0.1096],[93,-0.1143],[94,-0.1194],[95,-0.1248],[96,-0.1306],[97,-0.1368],[98,-0.1434],[99,-0.1502],[101,-0.1648],[106,-0.2044],[111,-0.2456],[113,-0.2609],[115,-0.2745],[116,-0.2806],[119,-0.2953],[122,-0.3078],[125,-0.318],[129,-0.328],[135,-0.3379],[148,-0.3487],[168,-0.3527],[169,-0.3538],[170,-0.3568],[171,-0.3626],[172,-0.3706],[175,-0.4008],[176,-0.4086],[177,-0.4137],[178,-0.4153],[179,-0.4131],[180,-0.4074],[181,-0.3983],[182,-0.3873],[186,-0.3386],[193,-0.2603],[196,-0.2286],[197,-0.2192],[198,-0.2123],[199,-0.2087],[200,-0.2088],[201,-0.2125],[202,-0.2197],[203,-0.2301],[204,-0.2432],[205,-0.2582],[206,-0.2744],[209,-0.3262],[211,-0.3592],[213,-0.3894],[215,-0.4172],[217,-0.442],[219,-0.4637],[221,-0.4828],[223,-0.4994],[225,-0.5136],[226,-0.5198],[227,-0.5243],[229,-0.526],[230,-0.5251],[232,-0.5202],[238,-0.5006],[240,-0.4918],[242,-0.4796],[244,-0.465],[247,-0.4396],[257,-0.3447],[259,-0.3275],[260,-0.3196],[261,-0.3125],[262,-0.3062],[263,-0.3007],[264,-0.2961],[265,-0.2925],[267,-0.2874],[268,-0.2856],[271,-0.2833],[277,-0.2825],[294,-0.2884],[296,-0.2884],[297,-0.0001],[367,0]]
       }
       lowerarm_l_fk_ctrl.Rotation.Y {
        type=MovieSceneScriptingFloatChannel
        default=0
        thinned_from=371
        keys [[-109,0.3826],[-16,0.3826],[0,-0.7203],[1,-0.7203],[9,-0.6946],[22,-0.6354],[26,-0.622],[30,-0.6138],[42,-0.6091],[43,-0.6067],[44,-0.5987],[45,-0.5836],[46,-0.5613],[47,-0.5342],[49,-0.4777],[50,-0.4522],[51,-0.4305],[52,-0.412],[53,-0.3969],[54,-0.3847],[55,-0.3747],[56,-0.3658],[59,-0.3435],[62,-0.3205],[64,-0.3065],[68,-0.2811],[71,-0.2639],[73,-0.2541],[74,-0.2502],[75,-0.247],[76,-0.2444],[77,-0.2432],[78,-0.2438],[79,-0.2458],[80,-0.2488],[81,-0.2529],[82,-0.2578],[83,-0.2634],[84,-0.2697],[86,-0.284],[88,-0.3004],[89,-0.3093],[90,-0.319],[92,-0.3411],[93,-0.353],[95,-0.3791],[97,-0.4081],[99,-0.4395],[109,-0.6088],[112,-0.6558],[114,-0.6835],[116,-0.707],[119,-0.7339],[123,-0.7628],[127,-0.7839],[134,-0.8065],[162,-0.8336],[168,-0.8339],[169,-0.8359],[170,-0.8409],[171,-0.8504],[172,-0.8637],[175,-0.9126],[176,-0.925],[177,-0.933],[178,-0.9355],[179,-0.9321],[180,-0.923],[181,-0.9087],[182,-0.891],[191,-0.7109],[196,-0.6073],[197,-0.5884],[198,-0.5744],[199,-0.567],[200,-0.5672],[201,-0.5749],[202,-0.5895],[203,-0.6103],[204,-0.636],[205,-0.665],[209,-0.7887],[211,-0.8447],[213,-0.8944],[215,-0.9385],[217,-0.9768],[219,-1.0096],[221,-1.0378],[224,-1.0726],[226,-1.0911],[227,-1.0975],[229,-1.0998],[231,-1.0958],[239,-1.058],[241,-1.0426],[244,-1.0114],[248,-0.9591],[258,-0.8053],[260,-0.7772],[262,-0.7534],[263,-0.7436],[264,-0.7353],[266,-0.7239],[268,-0.7163],[274,-0.7104],[296,-0.7214],[297,-0.0001],[367,0]]
       }
       lowerarm_l_fk_ctrl.Rotation.Z {
        type=MovieSceneScriptingFloatChannel
        default=0
        thinned_from=371
        keys [[-109,20.7936],[-16,20.7905],[1,20.8321],[9,20.1289],[22,18.502],[27,18.0622],[32,17.8484],[42,17.7735],[43,17.7104],[44,17.4849],[45,17.0705],[46,16.45],[47,15.6964],[49,14.1102],[50,13.3935],[51,12.7764],[52,12.2512],[53,11.8226],[54,11.4734],[55,11.1868],[56,10.9337],[63,9.423],[68,8.4865],[71,7.9828],[73,7.6959],[74,7.5821],[75,7.4866],[76,7.4115],[77,7.3774],[78,7.3933],[79,7.4525],[80,7.5412],[81,7.6618],[82,7.8029],[83,7.9691],[84,8.1514],[86,8.5712],[88,9.0471],[89,9.3038],[90,9.5864],[92,10.2238],[94,10.9312],[95,11.3127],[97,12.1404],[100,13.4985],[109,17.7681],[112,19.0636],[114,19.8248],[116,20.4669],[120,21.4205],[124,22.1556],[129,22.7824],[141,23.5343],[168,23.9233],[169,23.9761],[170,24.1108],[171,24.3719],[172,24.7309],[175,26.051],[176,26.3866],[177,26.6009],[178,26.6705],[179,26.5777],[180,26.3345],[181,25.947],[182,25.4679],[190,21.1416],[196,17.7256],[197,17.203],[198,16.8144],[199,16.6095],[200,16.6131],[201,16.8281],[202,17.2348],[203,17.8101],[204,18.5196],[205,19.318],[209,22.6966],[210,23.482],[212,24.9135],[214,26.1746],[216,27.2837],[218,28.2371],[220,29.0589],[223,30.0718],[226,30.8561],[227,31.0243],[229,31.0896],[231,30.9797],[239,29.9645],[241,29.5522],[244,28.7137],[248,27.3044],[259,22.7549],[261,22.0435],[263,21.4673],[264,21.2434],[266,20.9295],[268,20.7236],[274,20.559],[296,20.8617],[297,0.0005],[298,0],[367,0]]
       }
       lowerarm_l_fk_ctrl.Scale.X {type=MovieSceneScriptingFloatChannel, default=1, const [1,371]}
       lowerarm_l_fk_ctrl.Scale.Y {type=MovieSceneScriptingFloatChannel, default=1, const [1,371]}
       lowerarm_l_fk_ctrl.Scale.Z {type=MovieSceneScriptingFloatChannel, default=1, const [1,371]}
       hand_l_fk_ctrl.Location.X {type=MovieSceneScriptingFloatChannel, default=0, thinned_from=371, keys [[-109,0],[296,0],[297,-0.0002],[298,0],[367,0]]}
       hand_l_fk_ctrl.Location.Y {type=MovieSceneScriptingFloatChannel, default=0, thinned_from=371, keys [[-109,-0.1788],[-16,-0.1788],[0,0],[367,0]]}
       hand_l_fk_ctrl.Location.Z {type=MovieSceneScriptingFloatChannel, default=0, thinned_from=371, keys [[-109,0.0249],[-16,0.0249],[0,0],[296,0],[297,-0.0001],[367,0]]}
       hand_l_fk_ctrl.Rotation.X {
        type=MovieSceneScriptingFloatChannel
        default=0
        thinned_from=371
        keys [[-109,27.8847],[-56,27.8245],[-16,27.8974],[0,28.4413],[47,29.7475],[53,30.1805],[55,30.1734],[58,29.9351],[73,27.7438],[78,26.7121],[87,24.6057],[93,23.556],[98,22.9217],[100,22.7939],[102,22.7899],[105,22.9808],[108,23.3408],[110,23.6975],[113,24.4079],[121,26.6419],[124,27.2776],[127,27.6831],[172,29.0263],[174,29.0394],[176,29.1916],[179,29.5637],[180,29.5975],[181,29.5443],[182,29.3995],[183,29.1675],[184,28.8616],[186,28.0899],[188,27.1822],[190,26.0744],[192,24.7729],[195,22.7093],[196,22.1067],[197,21.5868],[198,21.1424],[199,20.7657],[200,20.4465],[201,20.1887],[202,20.0075],[203,19.8997],[204,19.8682],[205,19.91],[206,20.0226],[207,20.1992],[208,20.4263],[209,20.7025],[211,21.3743],[220,25.172],[222,25.9298],[224,26.5826],[230,28.1854],[232,28.5627],[233,28.6775],[234,28.7314],[236,28.6561],[240,28.2676],[242,28.2281],[245,28.3769],[254,29.4326],[258,29.629],[263,29.5186],[289,28.4586],[296,28.4218],[297,0.0009],[298,0],[367,0]]
       }
       hand_l_fk_ctrl.Rotation.Y {
        type=MovieSceneScriptingFloatChannel
        default=0
        thinned_from=371
        keys [[-109,-10.1178],[-56,-10.091],[-16,-10.1213],[0,-10.4205],[1,-10.4205],[8,-10.0602],[23,-9.0121],[27,-8.8097],[31,-8.6787],[42,-8.5613],[43,-8.5319],[44,-8.4374],[45,-8.2675],[46,-8.0231],[47,-7.7272],[49,-7.1168],[50,-6.8534],[51,-6.6348],[52,-6.4584],[53,-6.3134],[54,-6.1884],[56,-5.9681],[57,-5.8776],[63,-5.2187],[68,-4.7503],[70,-4.588],[77,-3.8982],[78,-3.8141],[79,-3.7381],[80,-3.6713],[81,-3.612],[83,-3.5161],[84,-3.4806],[85,-3.4542],[86,-3.4359],[87,-3.4267],[88,-3.4289],[89,-3.4417],[90,-3.4718],[91,-3.5184],[92,-3.577],[93,-3.6497],[94,-3.7368],[95,-3.838],[96,-3.9541],[97,-4.0838],[98,-4.2262],[99,-4.3868],[100,-4.5664],[101,-4.7633],[102,-4.974],[104,-5.4318],[106,-5.9215],[108,-6.4496],[115,-8.4211],[117,-8.934],[119,-9.3982],[121,-9.8203],[123,-10.1839],[125,-10.4809],[127,-10.707],[130,-10.9615],[138,-11.3817],[155,-11.7955],[169,-11.8919],[170,-11.9387],[171,-12.0291],[174,-12.4447],[175,-12.5372],[176,-12.5654],[177,-12.5078],[178,-12.3455],[179,-12.0584],[180,-11.6445],[181,-11.113],[182,-10.4926],[183,-9.8202],[185,-8.4204],[186,-7.7334],[189,-5.7628],[190,-5.0934],[192,-3.7312],[193,-3.0598],[194,-2.4157],[195,-1.8018],[196,-1.2258],[197,-0.7319],[198,-0.385],[199,-0.202],[200,-0.1966],[201,-0.3743],[202,-0.7259],[203,-1.2341],[204,-1.876],[205,-2.6175],[206,-3.4233],[207,-4.2654],[209,-5.9916],[210,-6.8401],[211,-7.6673],[212,-8.4704],[213,-9.2462],[214,-10.0033],[215,-10.7339],[216,-11.4387],[217,-12.1147],[218,-12.7588],[219,-13.374],[220,-13.9599],[221,-14.5142],[222,-15.0345],[223,-15.5208],[225,-16.3996],[226,-16.7858],[227,-17.1083],[228,-17.3279],[229,-17.4946],[230,-17.6014],[231,-17.6473],[232,-17.6317],[233,-17.5643],[234,-17.4413],[236,-17.0651],[247,-14.3623],[260,-11.4048],[262,-11.0112],[263,-10.8459],[264,-10.7072],[265,-10.5962],[267,-10.4306],[269,-10.3236],[274,-10.2457],[296,-10.4389],[297,-0.0001],[367,0]]
       }
       hand_l_fk_ctrl.Rotation.Z {
        type=MovieSceneScriptingFloatChannel
        default=0
        thinned_from=371
        keys [[-109,24.3765],[-16,24.3804],[43,23.8211],[52,24.1312],[68,23.817],[80,23.6758],[101,22.7209],[107,22.734],[125,23.8483],[171,24.2414],[180,24.0074],[183,23.7847],[189,22.9288],[196,21.7654],[203,20.8966],[205,20.7344],[207,20.6666],[209,20.714],[211,20.8624],[214,21.2398],[227,23.5134],[231,23.9187],[234,24.0919],[296,24.0958],[297,0.0003],[298,0],[367,0]]
       }
       hand_l_fk_ctrl.Scale.X {type=MovieSceneScriptingFloatChannel, default=1, const [1,371]}
       hand_l_fk_ctrl.Scale.Y {type=MovieSceneScriptingFloatChannel, default=1, const [1,371]}
       hand_l_fk_ctrl.Scale.Z {type=MovieSceneScriptingFloatChannel, default=1, const [1,371]}
       thumb_01_l_ctrl.Location.X {
        type=MovieSceneScriptingFloatChannel
        default=0
        thinned_from=368
        keys [[0,0.0961],[33,0.0961],[34,0.0959],[41,0.0959],[43,0.1037],[44,0.112],[45,0.1237],[46,0.1394],[47,0.1584],[48,0.1746],[49,0.1941],[50,0.2171],[51,0.2354],[52,0.2579],[54,0.2985],[56,0.3339],[58,0.3596],[59,0.3694],[60,0.3745],[61,0.3773],[62,0.3821],[66,0.3922],[67,0.3925],[69,0.3956],[70,0.3958],[71,0.3976],[84,0.3978],[85,0.3995],[95,0.3995],[98,0.3868],[100,0.3774],[101,0.3685],[102,0.3623],[115,0.2323],[116,0.2235],[117,0.2223],[118,0.2123],[119,0.2075],[120,0.1966],[121,0.1912],[122,0.181],[123,0.1749],[128,0.1476],[129,0.1429],[130,0.1422],[131,0.1368],[132,0.1368],[133,0.1313],[134,0.1313],[135,0.1259],[136,0.1259],[137,0.1189],[138,0.1179],[140,0.1179],[141,0.1109],[142,0.1099],[147,0.1099],[148,0.1029],[149,0.1019],[152,0.1019],[157,0.102],[158,0.0936],[159,0.0923],[169,0.0923],[170,0.0969],[171,0.102],[172,0.1158],[173,0.1382],[174,0.1658],[175,0.1979],[176,0.2261],[177,0.2577],[178,0.2852],[180,0.3312],[183,0.3725],[184,0.3817],[186,0.3913],[188,0.3995],[200,0.3995],[201,0.3985],[248,0.3985],[249,0.3995],[252,0.3995],[254,0.3874],[256,0.3631],[260,0.2902],[261,0.2659],[264,0.2113],[265,0.1979],[266,0.1786],[267,0.1651],[268,0.1576],[269,0.1441],[273,0.1172],[274,0.1113],[275,0.1105],[276,0.1046],[277,0.1038],[278,0.1038],[279,0.0979],[281,0.0963],[282,0.0963],[283,0.0959],[296,0.0959],[297,0],[367,0]]
       }
       thumb_01_l_ctrl.Location.Y {
        type=MovieSceneScriptingFloatChannel
        default=0
        thinned_from=368
        keys [[0,-0.1394],[33,-0.1394],[41,-0.1392],[43,-0.1506],[44,-0.162],[45,-0.179],[46,-0.2017],[47,-0.23],[48,-0.2528],[49,-0.2811],[50,-0.315],[51,-0.3412],[52,-0.3742],[54,-0.433],[56,-0.4845],[58,-0.5213],[59,-0.5359],[60,-0.5433],[61,-0.547],[62,-0.5543],[66,-0.569],[67,-0.569],[69,-0.5738],[70,-0.5739],[71,-0.5767],[84,-0.5768],[85,-0.5796],[95,-0.5796],[100,-0.5473],[101,-0.5344],[102,-0.5257],[116,-0.3228],[117,-0.3226],[118,-0.3081],[119,-0.3011],[120,-0.2853],[121,-0.2774],[122,-0.2618],[129,-0.2066],[130,-0.2064],[131,-0.1985],[132,-0.1985],[133,-0.1906],[134,-0.1906],[135,-0.1827],[136,-0.1827],[137,-0.1714],[140,-0.1711],[141,-0.1597],[147,-0.1595],[148,-0.1481],[157,-0.148],[158,-0.1343],[159,-0.134],[169,-0.134],[171,-0.1474],[172,-0.1674],[173,-0.2006],[174,-0.2406],[175,-0.2872],[176,-0.3272],[177,-0.3737],[178,-0.4137],[180,-0.4803],[183,-0.5402],[184,-0.5535],[188,-0.5796],[200,-0.5796],[201,-0.5779],[248,-0.5779],[249,-0.5796],[252,-0.5796],[254,-0.562],[256,-0.5267],[260,-0.421],[261,-0.3857],[264,-0.3064],[265,-0.2872],[266,-0.2581],[267,-0.2386],[268,-0.2287],[269,-0.2092],[274,-0.1606],[275,-0.1605],[276,-0.1509],[278,-0.1507],[279,-0.1411],[280,-0.141],[281,-0.1393],[296,-0.1392],[297,0],[367,0]]
       }
       thumb_01_l_ctrl.Location.Z {
        type=MovieSceneScriptingFloatChannel
        default=0
        thinned_from=368
        keys [[0,0.0751],[41,0.0752],[43,0.0813],[44,0.0868],[45,0.096],[46,0.1082],[47,0.1241],[48,0.1357],[49,0.151],[50,0.1699],[51,0.1836],[52,0.2018],[54,0.2334],[56,0.2612],[58,0.2805],[59,0.2889],[60,0.2928],[61,0.2944],[62,0.2988],[66,0.3067],[67,0.3063],[69,0.3092],[70,0.3089],[71,0.3108],[84,0.3105],[85,0.3124],[95,0.3124],[96,0.3089],[97,0.3047],[98,0.3012],[99,0.2984],[100,0.2949],[101,0.2878],[102,0.2833],[115,0.1818],[116,0.1724],[117,0.174],[118,0.1661],[119,0.1624],[120,0.1539],[121,0.1496],[122,0.1403],[123,0.1369],[128,0.1156],[129,0.1105],[130,0.1114],[131,0.1071],[132,0.1071],[133,0.1029],[134,0.1029],[135,0.0986],[136,0.0986],[137,0.0911],[138,0.0924],[140,0.0924],[141,0.0849],[142,0.0861],[147,0.0861],[148,0.0786],[149,0.0799],[152,0.0799],[157,0.0797],[158,0.0707],[159,0.0724],[169,0.0724],[170,0.076],[171,0.0788],[172,0.0896],[173,0.1082],[174,0.1297],[175,0.1548],[176,0.1756],[177,0.2014],[178,0.2229],[180,0.2587],[183,0.291],[184,0.2982],[185,0.3019],[186,0.3048],[187,0.3089],[188,0.3124],[200,0.3124],[201,0.3113],[248,0.3113],[249,0.3124],[252,0.3124],[254,0.3029],[256,0.2839],[260,0.2268],[261,0.2078],[264,0.165],[265,0.1549],[266,0.1381],[267,0.1276],[268,0.1234],[269,0.1129],[273,0.0919],[274,0.0856],[275,0.0866],[276,0.0803],[277,0.0814],[278,0.0814],[279,0.0751],[280,0.0761],[281,0.0746],[282,0.0746],[283,0.0752],[296,0.0752],[297,0],[367,0]]
       }
       thumb_01_l_ctrl.Rotation.X {
        type=MovieSceneScriptingFloatChannel
        default=0
        thinned_from=368
        keys [[0,2.907],[41,2.9058],[42,2.9145],[43,2.941],[44,3.0158],[45,3.0407],[46,3.1455],[47,3.2324],[48,3.2925],[49,3.3955],[50,3.5128],[51,3.611],[52,3.6999],[53,3.8425],[54,3.9393],[56,4.1179],[57,4.1474],[58,4.2343],[59,4.3085],[60,4.3367],[61,4.3141],[62,4.3795],[65,4.4228],[66,4.3862],[67,4.4008],[68,4.4342],[69,4.4112],[70,4.4216],[72,4.4216],[73,4.432],[85,4.432],[96,4.3879],[97,4.3683],[100,4.2693],[101,4.2472],[104,4.1028],[105,4.0987],[106,4.0047],[107,3.9569],[109,3.8969],[110,3.8066],[114,3.6326],[115,3.6353],[116,3.5601],[119,3.4674],[123,3.3018],[124,3.3154],[125,3.2355],[127,3.1835],[128,3.1707],[129,3.1453],[134,3.0827],[135,3.0395],[136,3.0562],[137,3.0292],[138,3.0214],[139,3.025],[140,3.0172],[141,2.9766],[142,3.0018],[144,2.9864],[146,2.946],[147,2.9712],[148,2.9523],[149,2.9636],[150,2.9447],[151,2.956],[152,2.9372],[153,2.9371],[154,2.9179],[155,2.9314],[156,2.9122],[157,2.9365],[158,2.9065],[169,2.8882],[170,2.8968],[171,2.9495],[172,3.0174],[173,3.1137],[174,3.2633],[175,3.4034],[176,3.5808],[178,3.8716],[179,3.9834],[181,4.1747],[182,4.2356],[183,4.3075],[184,4.3421],[185,4.3549],[186,4.3893],[187,4.4043],[188,4.432],[251,4.432],[252,4.4232],[253,4.3972],[254,4.3841],[255,4.3247],[257,4.1797],[258,4.0959],[259,3.9676],[260,3.8943],[261,3.7956],[262,3.6785],[263,3.5939],[264,3.49],[265,3.4174],[266,3.3202],[269,3.1559],[271,3.079],[272,3.0488],[273,3.0079],[274,3.0043],[275,2.9577],[276,2.9751],[277,2.9606],[278,2.9215],[279,2.9391],[280,2.9319],[281,2.9137],[296,2.9137],[297,0],[367,0]]
       }
       thumb_01_l_ctrl.Rotation.Y {
        type=MovieSceneScriptingFloatChannel
        default=0
        thinned_from=368
        keys [[0,16.7792],[42,16.7714],[45,16.6844],[56,15.9965],[58,15.9427],[59,15.8793],[95,15.8121],[100,15.8879],[107,16.0922],[108,16.0947],[127,16.5952],[134,16.6617],[135,16.7104],[136,16.6772],[138,16.713],[144,16.7259],[145,16.7509],[147,16.7367],[148,16.7617],[149,16.7421],[150,16.7671],[151,16.7475],[153,16.7778],[170,16.7845],[172,16.7204],[178,16.1553],[181,15.9491],[185,15.8245],[187,15.8184],[253,15.8077],[256,15.9054],[265,16.4539],[273,16.7284],[296,16.7748],[297,0],[367,0]]
       }
       thumb_01_l_ctrl.Rotation.Z {
        type=MovieSceneScriptingFloatChannel
        default=0
        thinned_from=368
        keys [[0,-3.1709],[33,-3.1709],[34,-3.1874],[41,-3.1874],[42,-3.2391],[43,-3.3944],[44,-3.6103],[45,-3.9636],[46,-4.3347],[47,-4.8003],[49,-5.8259],[50,-6.3949],[54,-8.5855],[56,-9.5358],[57,-9.9318],[58,-10.2405],[59,-10.4864],[61,-10.7467],[65,-11.0297],[67,-11.1542],[68,-11.1616],[70,-11.2402],[72,-11.2402],[73,-11.2832],[85,-11.2832],[86,-11.2492],[88,-11.2492],[89,-11.2153],[90,-11.2153],[91,-11.1813],[92,-11.1813],[95,-11.0794],[97,-10.8942],[98,-10.7739],[101,-10.2492],[102,-10.0551],[105,-9.3432],[107,-8.8519],[109,-8.3207],[114,-7.1079],[115,-6.8774],[119,-6.1373],[123,-5.2913],[124,-5.092],[125,-4.9388],[127,-4.6568],[128,-4.5862],[129,-4.4452],[134,-4.0926],[135,-3.9784],[137,-3.845],[138,-3.8017],[139,-3.7421],[140,-3.6988],[141,-3.648],[142,-3.6122],[144,-3.5256],[146,-3.4748],[147,-3.439],[148,-3.4119],[149,-3.3957],[150,-3.3686],[151,-3.3524],[152,-3.3253],[153,-3.31],[156,-3.2461],[157,-3.2115],[158,-3.2046],[159,-3.1632],[160,-3.1632],[161,-3.1218],[163,-3.1218],[164,-3.0804],[168,-3.0804],[169,-3.0695],[170,-3.1306],[171,-3.3494],[172,-3.7325],[173,-4.301],[174,-5.0248],[175,-5.8489],[176,-6.7085],[177,-7.5158],[178,-8.2483],[179,-8.8841],[180,-9.4285],[181,-9.9003],[182,-10.2615],[183,-10.567],[184,-10.8061],[185,-11.0045],[186,-11.1196],[188,-11.2832],[251,-11.2832],[252,-11.2437],[253,-11.125],[254,-10.934],[255,-10.6573],[256,-10.3016],[257,-9.8885],[258,-9.3993],[259,-8.8792],[261,-7.7798],[263,-6.7465],[264,-6.2661],[265,-5.8476],[266,-5.4422],[267,-5.0836],[269,-4.4893],[271,-4.0648],[272,-3.895],[273,-3.7412],[274,-3.6402],[275,-3.548],[276,-3.4704],[277,-3.3855],[278,-3.3357],[279,-3.2581],[280,-3.2156],[281,-3.2062],[296,-3.2062],[297,0],[367,0]]
       }
       thumb_01_l_ctrl.Scale.X {type=MovieSceneScriptingFloatChannel, default=1, const [1,368]}
       thumb_01_l_ctrl.Scale.Y {type=MovieSceneScriptingFloatChannel, default=1, const [1,368]}
       thumb_01_l_ctrl.Scale.Z {type=MovieSceneScriptingFloatChannel, default=1, const [1,368]}
       thumb_02_l_ctrl.Location.X {type=MovieSceneScriptingFloatChannel, default=0, thinned_from=368, keys [[0,0],[367,0]]}
       thumb_02_l_ctrl.Location.Y {type=MovieSceneScriptingFloatChannel, default=0, thinned_from=368, keys [[0,0],[367,0]]}
       thumb_02_l_ctrl.Location.Z {type=MovieSceneScriptingFloatChannel, default=0, thinned_from=368, keys [[0,0],[367,0]]}
       thumb_02_l_ctrl.Rotation.X {
        type=MovieSceneScriptingFloatChannel
        default=0
        thinned_from=368
        keys [[0,-0.7914],[33,-0.7914],[34,-0.7806],[41,-0.7806],[43,-0.8471],[44,-0.9388],[45,-1.0369],[47,-1.2929],[51,-1.9337],[52,-2.0996],[54,-2.4439],[55,-2.6076],[57,-2.8546],[58,-2.9463],[59,-3.0285],[62,-3.1555],[63,-3.1601],[64,-3.1962],[65,-3.2007],[66,-3.2368],[69,-3.2406],[70,-3.2745],[71,-3.281],[87,-3.2781],[88,-3.26],[91,-3.257],[92,-3.239],[94,-3.236],[96,-3.1938],[100,-3.0269],[102,-2.9025],[108,-2.4609],[109,-2.3696],[113,-2.0737],[115,-1.9297],[117,-1.8171],[119,-1.7069],[120,-1.6181],[121,-1.5678],[122,-1.5245],[123,-1.433],[127,-1.2513],[128,-1.2111],[129,-1.2064],[130,-1.1596],[131,-1.1191],[132,-1.1135],[133,-1.0729],[134,-1.0729],[135,-1.0266],[136,-1.0254],[137,-1],[138,-0.9717],[139,-0.9688],[140,-0.9433],[141,-0.9404],[142,-0.9148],[143,-0.9148],[144,-0.9122],[145,-0.8863],[147,-0.8865],[148,-0.8578],[151,-0.8579],[152,-0.832],[153,-0.8389],[154,-0.8273],[155,-0.8264],[157,-0.8022],[158,-0.8012],[159,-0.7896],[160,-0.7887],[161,-0.776],[163,-0.776],[164,-0.7644],[168,-0.7634],[169,-0.7691],[170,-0.7894],[171,-0.8441],[172,-0.9589],[173,-1.1434],[174,-1.3508],[176,-1.8773],[177,-2.1229],[178,-2.3472],[179,-2.5534],[180,-2.7081],[181,-2.8566],[182,-2.9719],[183,-3.0658],[184,-3.1399],[185,-3.1823],[186,-3.2436],[187,-3.2482],[188,-3.281],[252,-3.281],[253,-3.2335],[254,-3.1753],[256,-2.9832],[257,-2.8375],[258,-2.68],[260,-2.3853],[261,-2.2335],[262,-2.0365],[263,-1.8815],[264,-1.7355],[265,-1.6127],[267,-1.4006],[269,-1.1821],[273,-0.9618],[274,-0.9123],[275,-0.9061],[276,-0.9066],[277,-0.8565],[278,-0.8503],[279,-0.8004],[280,-0.8004],[281,-0.8023],[296,-0.8023],[297,0],[367,0]]
       }
       thumb_02_l_ctrl.Rotation.Y {
        type=MovieSceneScriptingFloatChannel
        default=0
        thinned_from=368
        keys [[0,2.4877],[33,2.4877],[34,2.4534],[41,2.4534],[43,2.6542],[45,3.2596],[47,4.0667],[50,5.5812],[51,6.0772],[55,8.1946],[57,8.9922],[59,9.5221],[62,9.9206],[63,9.9196],[64,10.054],[65,10.053],[66,10.1874],[69,10.1866],[70,10.3129],[87,10.312],[88,10.2447],[91,10.2453],[92,10.178],[94,10.1787],[96,10.0453],[100,9.512],[102,9.1256],[108,7.7375],[109,7.4486],[115,6.0639],[118,5.5457],[119,5.3637],[120,5.0779],[122,4.7937],[123,4.5088],[128,3.7952],[129,3.7967],[131,3.5108],[132,3.5124],[133,3.3687],[134,3.3687],[135,3.2267],[136,3.227],[138,3.0495],[139,3.0503],[140,2.9612],[141,2.962],[142,2.8729],[144,2.8738],[145,2.7846],[147,2.7847],[148,2.6964],[151,2.6965],[152,2.6073],[153,2.6392],[154,2.5991],[155,2.5994],[157,2.5195],[158,2.5198],[159,2.4797],[160,2.48],[161,2.4402],[163,2.4403],[164,2.4002],[168,2.4005],[169,2.4198],[170,2.4783],[171,2.6555],[172,3.0084],[173,3.599],[174,4.2476],[176,5.9015],[177,6.6702],[178,7.3812],[179,8.0335],[181,8.9828],[182,9.3388],[183,9.6362],[184,9.8746],[185,9.9988],[186,10.2082],[187,10.2072],[188,10.3114],[252,10.3114],[254,9.9985],[256,9.3682],[261,7.0188],[262,6.3931],[264,5.457],[265,5.0862],[269,3.7134],[274,2.8566],[276,2.8585],[277,2.6858],[278,2.6876],[279,2.5152],[280,2.5152],[281,2.5218],[296,2.5218],[297,0],[367,0]]
       }
       thumb_02_l_ctrl.Rotation.Z {
        type=MovieSceneScriptingFloatChannel
        default=0
        thinned_from=368
        keys [[0,2.291],[33,2.291],[34,2.2781],[41,2.2781],[43,2.371],[44,2.4611],[45,2.6481],[46,2.782],[47,2.9634],[50,3.6311],[51,3.8861],[54,4.6095],[55,4.7782],[56,5.0227],[57,5.1975],[59,5.4278],[61,5.5426],[62,5.5312],[63,5.5988],[64,5.5882],[65,5.6559],[66,5.6452],[67,5.7128],[68,5.6975],[70,5.6873],[71,5.7637],[84,5.7637],[85,5.728],[90,5.7335],[91,5.6978],[93,5.7032],[94,5.6676],[95,5.673],[96,5.607],[97,5.582],[100,5.3991],[102,5.2154],[108,4.6105],[109,4.4825],[114,3.9634],[115,3.8876],[116,3.779],[118,3.6199],[119,3.5668],[120,3.4439],[121,3.3826],[122,3.2526],[128,2.8828],[129,2.8149],[130,2.8199],[131,2.7563],[132,2.7566],[133,2.6928],[134,2.6928],[135,2.6291],[136,2.6291],[138,2.5501],[139,2.5503],[140,2.5106],[141,2.5107],[142,2.471],[143,2.471],[144,2.429],[146,2.4313],[147,2.3892],[150,2.3916],[151,2.3495],[153,2.351],[154,2.3365],[155,2.3365],[157,2.3074],[158,2.3075],[160,2.2774],[161,2.2784],[162,2.2629],[164,2.2638],[165,2.2483],[169,2.2479],[170,2.2757],[171,2.3593],[172,2.5253],[173,2.7707],[174,3.0691],[175,3.4436],[176,3.7829],[177,4.1491],[178,4.4559],[179,4.7334],[180,4.9618],[181,5.1597],[182,5.3066],[184,5.5545],[185,5.6208],[186,5.6601],[188,5.7637],[200,5.7637],[201,5.7483],[248,5.7483],[249,5.7637],[252,5.7637],[254,5.6311],[256,5.3616],[259,4.7455],[260,4.4559],[262,4.0411],[264,3.6088],[265,3.4565],[266,3.2239],[267,3.0729],[268,3.0036],[269,2.8508],[274,2.4639],[275,2.4642],[276,2.3815],[278,2.3862],[279,2.3076],[296,2.3061],[297,0],[367,0]]
       }
       thumb_02_l_ctrl.Scale.X {type=MovieSceneScriptingFloatChannel, default=1, const [1,368]}
       thumb_02_l_ctrl.Scale.Y {type=MovieSceneScriptingFloatChannel, default=1, const [1,368]}
       thumb_02_l_ctrl.Scale.Z {type=MovieSceneScriptingFloatChannel, default=1, const [1,368]}
       thumb_03_l_ctrl.Location.X {type=MovieSceneScriptingFloatChannel, default=0, thinned_from=368, keys [[0,0],[367,0]]}
       thumb_03_l_ctrl.Location.Y {type=MovieSceneScriptingFloatChannel, default=0, thinned_from=368, keys [[0,0],[367,0]]}
       thumb_03_l_ctrl.Location.Z {type=MovieSceneScriptingFloatChannel, default=0, thinned_from=368, keys [[0,0],[367,0]]}
       thumb_03_l_ctrl.Rotation.X {
        type=MovieSceneScriptingFloatChannel
        default=0
        thinned_from=368
        keys [[0,-1.4073],[33,-1.4073],[34,-1.3887],[41,-1.3887],[43,-1.508],[44,-1.6267],[45,-1.8695],[46,-2.0374],[47,-2.2604],[48,-2.5544],[49,-2.8361],[50,-3.1721],[51,-3.4535],[52,-3.7607],[53,-4.0582],[54,-4.3651],[55,-4.6171],[56,-4.8831],[57,-5.1014],[58,-5.2841],[59,-5.3819],[60,-5.4984],[61,-5.5339],[62,-5.6074],[65,-5.7131],[66,-5.7058],[68,-5.7805],[71,-5.8149],[72,-5.8116],[73,-5.8265],[75,-5.8258],[76,-5.8382],[81,-5.8382],[82,-5.8258],[86,-5.8358],[87,-5.8123],[89,-5.8173],[90,-5.7938],[91,-5.7987],[95,-5.7195],[96,-5.6823],[97,-5.6264],[99,-5.4777],[100,-5.3836],[101,-5.2702],[102,-5.1892],[107,-4.539],[108,-4.3264],[114,-3.52],[115,-3.464],[116,-3.3261],[117,-3.2599],[118,-3.1207],[119,-3.0099],[123,-2.5444],[124,-2.4729],[127,-2.2385],[128,-2.1558],[130,-2.0821],[131,-1.9977],[135,-1.8423],[136,-1.8195],[137,-1.772],[142,-1.6556],[145,-1.581],[146,-1.5837],[148,-1.533],[149,-1.5336],[150,-1.509],[151,-1.5095],[152,-1.4849],[154,-1.4851],[155,-1.4413],[157,-1.4413],[158,-1.4452],[159,-1.3973],[162,-1.3973],[163,-1.3521],[169,-1.3529],[170,-1.3889],[171,-1.5151],[172,-1.7078],[173,-2.0257],[174,-2.4244],[175,-2.8667],[176,-3.3352],[177,-3.7828],[178,-4.1889],[179,-4.5255],[180,-4.8328],[181,-5.0679],[182,-5.2699],[183,-5.4477],[184,-5.5703],[185,-5.671],[188,-5.8505],[200,-5.8505],[201,-5.8332],[248,-5.8332],[249,-5.8505],[251,-5.8505],[253,-5.7652],[255,-5.5144],[256,-5.2943],[257,-5.0917],[258,-4.8357],[261,-3.9065],[263,-3.3768],[264,-3.1028],[265,-2.863],[267,-2.4551],[269,-2.1339],[270,-1.9923],[272,-1.8083],[273,-1.7113],[276,-1.5721],[277,-1.5225],[278,-1.5013],[279,-1.451],[280,-1.4272],[296,-1.4266],[297,0],[367,0]]
       }
       thumb_03_l_ctrl.Rotation.Y {
        type=MovieSceneScriptingFloatChannel
        default=0
        thinned_from=368
        keys [[0,1.1195],[33,1.1195],[34,1.136],[41,1.136],[42,1.1432],[43,1.2214],[44,1.3075],[45,1.4416],[46,1.6156],[47,1.8375],[48,2.0264],[49,2.2617],[50,2.5153],[51,2.7597],[52,2.9919],[53,3.2355],[55,3.6826],[57,4.0478],[58,4.1842],[59,4.3011],[60,4.371],[61,4.4024],[62,4.4487],[65,4.544],[66,4.5515],[68,4.5932],[72,4.6308],[81,4.6425],[85,4.6397],[87,4.6247],[88,4.6247],[92,4.5947],[95,4.5499],[96,4.5202],[97,4.4757],[99,4.3511],[100,4.2783],[102,4.1131],[103,3.9991],[104,3.9186],[107,3.5867],[108,3.4679],[114,2.8433],[115,2.7513],[116,2.652],[117,2.5703],[118,2.4727],[119,2.4012],[122,2.1212],[123,2.0512],[124,1.9618],[127,1.779],[128,1.7228],[129,1.6889],[130,1.6357],[132,1.5701],[135,1.4819],[137,1.409],[140,1.3522],[141,1.3191],[142,1.301],[143,1.2998],[146,1.2468],[147,1.2457],[148,1.2277],[149,1.2128],[150,1.2098],[151,1.1949],[152,1.1918],[153,1.1863],[154,1.1863],[155,1.1464],[157,1.1464],[158,1.1425],[159,1.1066],[162,1.1066],[163,1.1011],[164,1.1011],[165,1.067],[168,1.067],[169,1.0759],[170,1.1069],[171,1.1936],[172,1.3653],[173,1.6089],[174,1.9205],[175,2.2795],[176,2.65],[177,3.0071],[178,3.3332],[179,3.6036],[180,3.8373],[181,4.032],[182,4.2003],[183,4.3254],[184,4.4304],[185,4.5048],[188,4.6477],[200,4.6477],[248,4.6404],[251,4.6477],[252,4.6294],[253,4.591],[254,4.4969],[255,4.3836],[256,4.2258],[257,4.034],[259,3.599],[261,3.123],[262,2.906],[263,2.6671],[264,2.4593],[265,2.2734],[267,1.9577],[268,1.8121],[269,1.697],[270,1.5895],[271,1.5003],[272,1.4284],[273,1.3735],[274,1.3215],[276,1.2509],[278,1.1838],[280,1.1479],[281,1.1349],[296,1.1349],[297,0],[367,0]]
       }
       thumb_03_l_ctrl.Rotation.Z {
        type=MovieSceneScriptingFloatChannel
        default=0
        thinned_from=368
        keys [[0,4.9489],[33,4.9489],[41,4.9426],[43,5.3453],[44,5.7478],[45,6.3515],[46,7.156],[47,8.1616],[48,8.9654],[49,9.9701],[50,11.1768],[51,12.1633],[52,13.1977],[53,14.3632],[54,15.268],[55,16.3048],[57,17.8592],[58,18.5084],[59,19.0285],[60,19.287],[61,19.4168],[62,19.6773],[66,20.1988],[67,20.1968],[68,20.3095],[72,20.4873],[75,20.4867],[76,20.5313],[81,20.5313],[82,20.4867],[85,20.5095],[86,20.4428],[88,20.4434],[89,20.3767],[90,20.3773],[91,20.3106],[92,20.3112],[95,20.1131],[96,19.981],[97,19.7828],[99,19.255],[100,18.9251],[101,18.5292],[102,18.2199],[116,11.6332],[117,11.3994],[118,10.9295],[119,10.5468],[120,10.2774],[124,8.6527],[127,7.8406],[128,7.706],[129,7.4345],[131,7.1644],[132,6.8929],[135,6.4865],[136,6.4047],[137,6.2401],[139,6.0754],[140,5.9104],[142,5.7461],[143,5.746],[146,5.499],[147,5.4989],[148,5.4166],[149,5.4169],[150,5.3342],[151,5.3346],[152,5.2518],[154,5.252],[155,5.0873],[157,5.0873],[158,4.922],[162,4.9227],[163,4.7572],[169,4.7578],[170,4.8745],[171,5.2833],[172,6.0419],[173,7.092],[174,8.4918],[175,10.066],[176,11.7574],[177,13.2733],[178,14.7325],[179,15.9593],[180,16.9521],[181,17.8297],[182,18.5909],[183,19.1173],[184,19.5862],[185,19.8925],[188,20.5759],[200,20.5759],[201,20.5135],[248,20.5135],[249,20.5759],[251,20.5759],[253,20.2658],[255,19.3381],[256,18.7218],[259,15.9457],[261,13.7899],[262,12.8673],[263,11.7902],[264,10.8678],[267,8.6261],[269,7.4743],[271,6.6517],[272,6.3224],[274,5.8283],[276,5.4989],[277,5.4168],[278,5.2521],[280,5.0872],[281,5.0169],[296,5.0169],[297,0],[367,0]]
       }
       thumb_03_l_ctrl.Scale.X {type=MovieSceneScriptingFloatChannel, default=1, const [1,368]}
       thumb_03_l_ctrl.Scale.Y {type=MovieSceneScriptingFloatChannel, default=1, const [1,368]}
       thumb_03_l_ctrl.Scale.Z {type=MovieSceneScriptingFloatChannel, default=1, const [1,368]}
       index_metacarpal_l_ctrl.Location.X {type=MovieSceneScriptingFloatChannel, default=0, thinned_from=368, keys [[0,0],[367,0]]}
       index_metacarpal_l_ctrl.Location.Y {type=MovieSceneScriptingFloatChannel, default=0, thinned_from=368, keys [[0,0],[367,0]]}
       index_metacarpal_l_ctrl.Location.Z {type=MovieSceneScriptingFloatChannel, default=0, thinned_from=368, keys [[0,0],[367,0]]}
       index_metacarpal_l_ctrl.Rotation.X {
        type=MovieSceneScriptingFloatChannel
        default=0
        thinned_from=368
        keys [[0,-1.8326],[43,-1.8318],[46,-1.85],[47,-1.8678],[48,-1.8799],[49,-1.8982],[50,-1.9223],[51,-1.9419],[52,-1.9665],[53,-1.9846],[54,-2.0099],[55,-2.0285],[56,-2.0538],[57,-2.0727],[60,-2.1107],[62,-2.1235],[63,-2.1235],[64,-2.1299],[66,-2.1298],[67,-2.1363],[70,-2.136],[71,-2.1426],[84,-2.1426],[91,-2.1395],[96,-2.1301],[98,-2.1207],[104,-2.0684],[105,-2.0535],[109,-2.0144],[110,-1.9996],[112,-1.9802],[113,-1.9755],[115,-1.956],[117,-1.9466],[118,-1.937],[119,-1.9321],[120,-1.9223],[122,-1.9133],[123,-1.9035],[127,-1.8849],[128,-1.8852],[130,-1.8758],[131,-1.8758],[133,-1.8661],[134,-1.8665],[136,-1.8582],[143,-1.8481],[144,-1.8444],[146,-1.8447],[147,-1.841],[150,-1.8412],[151,-1.8376],[155,-1.8381],[156,-1.8341],[159,-1.8341],[165,-1.8306],[166,-1.8268],[174,-1.827],[176,-1.8343],[177,-1.8414],[179,-1.8733],[184,-1.9823],[187,-2.0355],[190,-2.0761],[195,-2.1217],[198,-2.1355],[199,-2.1355],[200,-2.1401],[252,-2.1426],[254,-2.13],[256,-2.1048],[261,-2.012],[262,-1.987],[264,-1.9504],[268,-1.8929],[270,-1.8733],[273,-1.8538],[278,-1.8372],[279,-1.8374],[296,-1.8339],[297,0],[367,0]]
       }
       index_metacarpal_l_ctrl.Rotation.Y {
        type=MovieSceneScriptingFloatChannel
        default=0
        thinned_from=368
        keys [[0,1.5734],[33,1.5734],[34,1.5575],[43,1.5575],[46,1.707],[47,1.8569],[48,1.9567],[49,2.1061],[50,2.3059],[51,2.4575],[53,2.87],[54,3.0244],[55,3.2305],[57,3.5397],[60,3.8489],[62,3.9519],[63,3.9519],[64,4.0034],[65,4.0034],[66,4.0549],[69,4.0549],[70,4.1056],[87,4.1055],[88,4.0804],[92,4.0805],[93,4.0556],[94,4.0556],[97,3.9808],[98,3.931],[99,3.856],[100,3.8063],[102,3.6607],[105,3.42],[106,3.2997],[114,2.6578],[115,2.6177],[116,2.5375],[117,2.4974],[118,2.4171],[120,2.3383],[121,2.2614],[124,2.1462],[125,2.0692],[126,2.0693],[130,1.9155],[131,1.9155],[132,1.8771],[133,1.8772],[135,1.8002],[136,1.8002],[138,1.7453],[139,1.7453],[140,1.7179],[141,1.7179],[142,1.6904],[144,1.6904],[145,1.663],[147,1.6629],[148,1.6355],[151,1.6355],[152,1.6081],[153,1.6182],[154,1.6182],[155,1.5844],[159,1.5845],[160,1.5508],[168,1.5507],[169,1.5374],[173,1.5374],[174,1.5171],[175,1.5374],[176,1.5782],[177,1.6597],[178,1.7617],[179,1.9044],[180,2.0675],[184,2.8013],[185,2.97],[186,3.0815],[187,3.2302],[190,3.5648],[191,3.6392],[192,3.7507],[193,3.825],[194,3.8622],[195,3.9366],[199,4.0852],[200,4.0853],[201,4.0943],[216,4.0943],[217,4.1055],[252,4.1055],[254,4.0044],[256,3.8022],[260,3.1957],[261,2.9933],[264,2.5384],[265,2.4171],[266,2.2817],[268,2.0651],[270,1.9026],[273,1.7401],[277,1.6319],[278,1.6319],[279,1.6048],[280,1.6048],[281,1.5844],[296,1.5844],[297,0],[367,0]]
       }
       index_metacarpal_l_ctrl.Rotation.Z {
        type=MovieSceneScriptingFloatChannel
        default=0
        thinned_from=368
        keys [[0,-0.5295],[33,-0.5295],[34,-0.5266],[43,-0.5266],[45,-0.544],[46,-0.5597],[47,-0.5787],[48,-0.5961],[49,-0.6293],[51,-0.6849],[52,-0.7204],[53,-0.7488],[54,-0.7827],[55,-0.8183],[57,-0.8722],[58,-0.8902],[59,-0.9153],[60,-0.9262],[62,-0.9443],[63,-0.9443],[64,-0.9533],[65,-0.9533],[66,-0.9622],[70,-0.9655],[71,-0.9713],[88,-0.9704],[89,-0.9669],[92,-0.9668],[93,-0.9624],[94,-0.9624],[97,-0.9491],[99,-0.9306],[101,-0.9051],[102,-0.8941],[105,-0.8516],[106,-0.8305],[114,-0.7183],[115,-0.7113],[116,-0.6974],[117,-0.6904],[118,-0.6765],[119,-0.6709],[120,-0.6586],[122,-0.6438],[123,-0.6316],[125,-0.6169],[126,-0.6114],[127,-0.6101],[130,-0.59],[131,-0.59],[132,-0.5832],[133,-0.5777],[134,-0.5765],[135,-0.5698],[136,-0.5698],[138,-0.5599],[139,-0.5599],[140,-0.555],[141,-0.555],[142,-0.5501],[144,-0.5501],[145,-0.5452],[147,-0.5451],[148,-0.5402],[151,-0.5402],[152,-0.5353],[153,-0.537],[155,-0.536],[156,-0.5313],[159,-0.5313],[160,-0.5255],[168,-0.5255],[169,-0.5232],[173,-0.5232],[174,-0.5197],[175,-0.5232],[176,-0.5302],[177,-0.5443],[178,-0.562],[179,-0.5868],[180,-0.6151],[184,-0.7433],[187,-0.819],[190,-0.877],[191,-0.8899],[192,-0.9092],[193,-0.9222],[194,-0.9288],[195,-0.9417],[199,-0.9676],[216,-0.9693],[252,-0.9713],[254,-0.9535],[256,-0.9179],[257,-0.8912],[258,-0.8717],[259,-0.8381],[260,-0.8117],[261,-0.7836],[262,-0.7501],[264,-0.6975],[265,-0.6765],[266,-0.6528],[268,-0.6149],[269,-0.6007],[270,-0.5904],[271,-0.5771],[272,-0.5677],[273,-0.5621],[274,-0.5536],[277,-0.5395],[278,-0.5395],[279,-0.5348],[280,-0.5348],[281,-0.5314],[296,-0.5314],[297,0],[367,0]]
       }
       index_metacarpal_l_ctrl.Scale.X {type=MovieSceneScriptingFloatChannel, default=1, const [1,368]}
       index_metacarpal_l_ctrl.Scale.Y {type=MovieSceneScriptingFloatChannel, default=1, const [1,368]}
       index_metacarpal_l_ctrl.Scale.Z {type=MovieSceneScriptingFloatChannel, default=1, const [1,368]}
       index_01_l_ctrl.Location.X {type=MovieSceneScriptingFloatChannel, default=0, thinned_from=368, keys [[0,0],[367,0]]}
       index_01_l_ctrl.Location.Y {type=MovieSceneScriptingFloatChannel, default=0, thinned_from=368, keys [[0,0],[367,0]]}
       index_01_l_ctrl.Location.Z {type=MovieSceneScriptingFloatChannel, default=0, thinned_from=368, keys [[0,0],[367,0]]}
       index_01_l_ctrl.Rotation.X {
        type=MovieSceneScriptingFloatChannel
        default=0
        thinned_from=368
        keys [[0,-0.4203],[33,-0.4203],[34,-0.4249],[43,-0.4249],[44,-0.3257],[45,-0.0897],[46,0.2974],[47,0.8204],[48,1.4331],[49,2.1372],[50,2.9063],[51,3.7094],[52,4.5498],[53,5.4017],[54,6.241],[55,7.0582],[56,7.8188],[57,8.5128],[58,9.1286],[59,9.6296],[60,10.009],[61,10.2489],[62,10.4078],[64,10.6532],[66,10.8505],[68,10.9818],[70,11.0755],[73,11.1476],[84,11.1542],[87,11.1299],[91,11.045],[94,10.9322],[95,10.8526],[96,10.7354],[97,10.5515],[98,10.3447],[99,10.0794],[100,9.7835],[102,9.1127],[105,7.9672],[106,7.549],[110,5.9367],[112,5.2013],[114,4.5662],[115,4.2914],[116,4.0424],[118,3.512],[119,3.2554],[120,2.9919],[121,2.7677],[122,2.5152],[123,2.3022],[124,2.1287],[125,1.9212],[126,1.7911],[127,1.6232],[128,1.4976],[129,1.3341],[130,1.2429],[131,1.1262],[132,1.0018],[133,0.881],[134,0.7953],[135,0.6749],[136,0.6037],[138,0.4254],[139,0.3661],[140,0.2773],[142,0.1592],[143,0.1297],[144,0.0707],[145,0.0137],[146,-0.0175],[148,-0.0763],[149,-0.1303],[150,-0.1625],[151,-0.1918],[152,-0.2182],[153,-0.2305],[157,-0.3462],[158,-0.3766],[159,-0.4298],[161,-0.4875],[162,-0.492],[163,-0.5178],[164,-0.5452],[165,-0.5466],[166,-0.5753],[167,-0.5753],[168,-0.5997],[169,-0.6057],[170,-0.6077],[171,-0.6077],[172,-0.6291],[174,-0.6291],[175,-0.5663],[176,-0.3951],[177,-0.0573],[178,0.4169],[179,1.0452],[180,1.7953],[183,4.2985],[184,5.1013],[185,5.8436],[186,6.518],[187,7.1354],[188,7.6967],[189,8.2101],[190,8.6741],[191,9.0884],[192,9.469],[193,9.8048],[194,10.1129],[195,10.3713],[196,10.5906],[197,10.7587],[198,10.8963],[199,10.994],[200,11.0657],[203,11.1673],[251,11.1641],[252,11.1033],[253,10.9629],[254,10.6842],[255,10.2942],[256,9.7765],[257,9.1724],[258,8.4848],[259,7.7403],[261,6.1745],[262,5.4181],[263,4.6962],[264,4.0191],[265,3.394],[266,2.8298],[267,2.3205],[268,1.8768],[269,1.474],[270,1.143],[271,0.8477],[272,0.5997],[273,0.3952],[274,0.2353],[275,0.1066],[276,0.0047],[277,-0.0974],[278,-0.1965],[279,-0.2702],[280,-0.3266],[281,-0.3749],[282,-0.3928],[283,-0.4084],[284,-0.4084],[285,-0.4108],[286,-0.425],[296,-0.425],[297,0],[367,0]]
       }
       index_01_l_ctrl.Rotation.Y {
        type=MovieSceneScriptingFloatChannel
        default=0
        thinned_from=368
        keys [[0,-0.7668],[33,-0.7668],[34,-0.7582],[43,-0.7582],[44,-0.812],[45,-0.9767],[46,-1.228],[47,-1.5413],[48,-1.9374],[49,-2.3912],[50,-2.8609],[51,-3.3874],[54,-4.9941],[55,-5.5055],[56,-5.9907],[57,-6.4308],[58,-6.8071],[59,-7.136],[60,-7.3701],[61,-7.521],[62,-7.6225],[64,-7.7833],[66,-7.9074],[68,-7.9923],[70,-8.0537],[71,-8.0596],[74,-8.1042],[85,-8.1012],[86,-8.0745],[87,-8.0798],[88,-8.057],[89,-8.0663],[93,-7.9963],[94,-7.9571],[96,-7.8333],[97,-7.7302],[99,-7.4223],[100,-7.2466],[102,-6.809],[104,-6.3353],[108,-5.2934],[110,-4.7876],[112,-4.3289],[113,-4.1246],[114,-3.9346],[115,-3.7548],[118,-3.2518],[119,-3.1071],[120,-2.9224],[121,-2.764],[122,-2.6407],[123,-2.5092],[124,-2.3703],[125,-2.269],[126,-2.1552],[127,-2.0463],[128,-1.9641],[129,-1.8865],[130,-1.7997],[131,-1.7473],[132,-1.6632],[133,-1.6114],[134,-1.5219],[136,-1.4181],[138,-1.305],[139,-1.2671],[140,-1.21],[142,-1.1335],[143,-1.1143],[144,-1.0758],[145,-1.059],[146,-1.0179],[148,-0.9791],[152,-0.9033],[153,-0.8824],[157,-0.817],[158,-0.782],[159,-0.7675],[161,-0.7345],[162,-0.7161],[163,-0.6993],[164,-0.7013],[165,-0.6828],[166,-0.6661],[167,-0.6661],[168,-0.668],[169,-0.6488],[171,-0.6488],[172,-0.6326],[174,-0.6326],[175,-0.6636],[176,-0.7927],[177,-0.9956],[178,-1.3056],[179,-1.7158],[180,-2.1865],[181,-2.7173],[182,-3.2629],[183,-3.7894],[184,-4.2871],[185,-4.7482],[186,-5.1816],[187,-5.5783],[188,-5.9254],[189,-6.2392],[191,-6.7982],[193,-7.2436],[195,-7.6136],[197,-7.8448],[198,-7.9323],[201,-8.0897],[202,-8.0779],[203,-8.1071],[248,-8.1071],[251,-8.0952],[252,-8.0655],[253,-7.9611],[254,-7.7999],[255,-7.5499],[256,-7.2304],[257,-6.8585],[259,-5.9481],[260,-5.4363],[261,-4.9422],[263,-4.0132],[264,-3.58],[265,-3.1774],[266,-2.8197],[267,-2.5022],[268,-2.2139],[269,-1.9712],[270,-1.7502],[271,-1.5674],[272,-1.4113],[273,-1.2836],[274,-1.1848],[275,-1.1052],[276,-1.0343],[278,-0.9141],[279,-0.8629],[280,-0.8216],[281,-0.8025],[283,-0.7708],[296,-0.772],[297,0],[367,0]]
       }
       index_01_l_ctrl.Rotation.Z {
        type=MovieSceneScriptingFloatChannel
        default=0
        thinned_from=368
        keys [[0,1.7463],[33,1.7463],[43,1.7495],[44,1.6141],[45,1.1579],[46,0.5195],[47,-0.3969],[48,-1.4565],[49,-2.6619],[50,-3.969],[51,-5.3772],[52,-6.8041],[54,-9.7241],[55,-11.1212],[56,-12.4342],[57,-13.6348],[58,-14.6938],[59,-15.5603],[60,-16.225],[61,-16.6369],[63,-17.1286],[66,-17.6497],[69,-17.9719],[72,-18.1329],[82,-18.2008],[86,-18.1519],[87,-18.108],[88,-18.104],[93,-17.8686],[95,-17.6739],[96,-17.4394],[97,-17.1537],[98,-16.7741],[99,-16.3481],[100,-15.8328],[102,-14.6802],[103,-14.0416],[105,-12.6966],[110,-9.2028],[111,-8.5538],[112,-7.9444],[113,-7.3761],[114,-6.8478],[115,-6.3585],[116,-5.9452],[118,-5.0486],[119,-4.5897],[122,-3.2863],[123,-2.9261],[124,-2.6377],[125,-2.2796],[126,-2.064],[127,-1.7778],[128,-1.5634],[129,-1.2785],[130,-1.0636],[131,-0.9223],[133,-0.4961],[134,-0.3538],[135,-0.1413],[136,-0.0192],[138,0.2878],[139,0.3901],[140,0.5433],[142,0.7475],[143,0.7985],[145,1.0024],[149,1.2051],[150,1.3079],[151,1.3588],[152,1.3578],[153,1.4161],[162,1.9119],[163,1.9108],[164,1.9658],[165,1.9658],[166,2.0208],[168,2.0197],[169,2.0354],[170,2.0735],[171,2.0735],[172,2.1108],[174,2.1108],[175,1.9989],[176,1.7005],[177,1.1403],[178,0.3149],[179,-0.7021],[180,-1.9538],[181,-3.3302],[182,-4.7579],[183,-6.198],[184,-7.5726],[185,-8.8716],[186,-10.0461],[187,-11.1574],[188,-12.1329],[189,-13.0415],[190,-13.882],[191,-14.6204],[192,-15.2516],[193,-15.8482],[194,-16.3728],[195,-16.8266],[196,-17.202],[197,-17.5054],[198,-17.7341],[199,-17.9227],[200,-18.0387],[203,-18.2019],[251,-18.2002],[252,-18.0991],[253,-17.8433],[254,-17.3864],[255,-16.6811],[256,-15.8265],[257,-14.7828],[258,-13.5972],[260,-10.9761],[261,-9.5911],[263,-7.0465],[264,-5.8829],[265,-4.8221],[266,-3.855],[267,-2.9714],[268,-2.1951],[269,-1.5259],[270,-0.9352],[271,-0.4234],[272,-0.0153],[273,0.3414],[274,0.6212],[275,0.824],[276,1.0017],[277,1.205],[279,1.509],[280,1.5844],[281,1.6719],[282,1.7151],[284,1.7144],[285,1.7582],[296,1.7576],[297,0],[367,0]]
       }
       index_01_l_ctrl.Scale.X {type=MovieSceneScriptingFloatChannel, default=1, const [1,368]}
       index_01_l_ctrl.Scale.Y {type=MovieSceneScriptingFloatChannel, default=1, const [1,368]}
       index_01_l_ctrl.Scale.Z {type=MovieSceneScriptingFloatChannel, default=1, const [1,368]}
       index_02_l_ctrl.Location.X {type=MovieSceneScriptingFloatChannel, default=0, thinned_from=368, keys [[0,0],[367,0]]}
       index_02_l_ctrl.Location.Y {type=MovieSceneScriptingFloatChannel, default=0, thinned_from=368, keys [[0,0],[367,0]]}
       index_02_l_ctrl.Location.Z {type=MovieSceneScriptingFloatChannel, default=0, thinned_from=368, keys [[0,0],[367,0]]}
       index_02_l_ctrl.Rotation.X {
        type=MovieSceneScriptingFloatChannel
        default=0
        thinned_from=368
        keys [[0,-4.349],[44,-4.3482],[49,-4.2964],[58,-4.1486],[62,-4.1205],[97,-4.1173],[136,-4.3283],[176,-4.3478],[180,-4.3021],[189,-4.166],[199,-4.108],[253,-4.1095],[259,-4.177],[267,-4.2914],[296,-4.348],[297,0],[367,0]]
       }
       index_02_l_ctrl.Rotation.Y {
        type=MovieSceneScriptingFloatChannel
        default=0
        thinned_from=368
        keys [[0,-1.0457],[43,-1.0457],[44,-1.0498],[45,-1.0684],[46,-1.0947],[48,-1.1705],[49,-1.22],[50,-1.2729],[53,-1.4461],[55,-1.556],[56,-1.6079],[57,-1.6553],[58,-1.6994],[59,-1.7354],[60,-1.7591],[61,-1.7759],[65,-1.8121],[67,-1.8244],[71,-1.8358],[89,-1.8348],[94,-1.8232],[96,-1.8094],[97,-1.7978],[100,-1.7444],[104,-1.6482],[110,-1.4803],[112,-1.4307],[114,-1.388],[118,-1.3142],[122,-1.2471],[125,-1.2071],[128,-1.1757],[136,-1.1143],[137,-1.1096],[138,-1.1015],[139,-1.1008],[140,-1.0927],[141,-1.092],[142,-1.0839],[143,-1.0832],[145,-1.0751],[146,-1.0744],[148,-1.0663],[150,-1.0656],[152,-1.0574],[153,-1.0589],[154,-1.0547],[155,-1.0547],[163,-1.0381],[164,-1.0381],[165,-1.036],[167,-1.036],[169,-1.0327],[173,-1.0327],[174,-1.0298],[175,-1.0356],[176,-1.0483],[177,-1.0703],[178,-1.105],[179,-1.1489],[180,-1.1997],[181,-1.2568],[182,-1.3173],[183,-1.3719],[184,-1.4294],[186,-1.5236],[188,-1.6051],[190,-1.6703],[191,-1.6988],[194,-1.7672],[196,-1.7996],[199,-1.828],[201,-1.8356],[216,-1.8356],[251,-1.8387],[252,-1.8348],[253,-1.8248],[254,-1.8066],[255,-1.779],[256,-1.7436],[257,-1.7026],[258,-1.6555],[261,-1.4967],[263,-1.3966],[264,-1.3491],[265,-1.3055],[266,-1.2672],[267,-1.2335],[269,-1.1744],[271,-1.1321],[272,-1.1152],[274,-1.0898],[276,-1.0736],[277,-1.069],[278,-1.0605],[281,-1.0491],[296,-1.0491],[297,0],[367,0]]
       }
       index_02_l_ctrl.Rotation.Z {
        type=MovieSceneScriptingFloatChannel
        default=0
        thinned_from=368
        keys [[0,-26.3932],[33,-26.3932],[44,-26.4233],[46,-27.0273],[48,-28.0386],[52,-30.9227],[53,-31.7585],[55,-33.2313],[58,-35.1453],[59,-35.5738],[60,-35.896],[62,-36.3268],[65,-36.6506],[66,-36.6505],[70,-36.919],[72,-36.9189],[85,-36.9772],[94,-36.7742],[97,-36.4195],[100,-35.7128],[104,-34.4404],[105,-34.0244],[106,-33.6926],[107,-33.2786],[108,-32.9483],[109,-32.5363],[113,-31.2249],[120,-29.5088],[125,-28.5397],[130,-27.8967],[135,-27.4066],[136,-27.3502],[138,-27.1147],[139,-27.1147],[140,-26.9972],[141,-26.9972],[142,-26.8797],[144,-26.8799],[145,-26.7623],[147,-26.7625],[148,-26.645],[151,-26.6452],[152,-26.5278],[153,-26.5689],[154,-26.5102],[156,-26.5103],[157,-26.4517],[158,-26.4517],[159,-26.3931],[160,-26.3931],[161,-26.3345],[162,-26.3346],[163,-26.276],[167,-26.2761],[168,-26.2175],[173,-26.2423],[174,-26.159],[175,-26.2422],[176,-26.4087],[177,-26.7422],[178,-27.1598],[179,-27.7462],[180,-28.4188],[184,-31.4786],[186,-32.7502],[188,-33.8128],[190,-34.7295],[191,-35.113],[194,-36.0374],[195,-36.2694],[198,-36.7344],[201,-36.9308],[216,-36.9308],[251,-36.9772],[252,-36.9245],[254,-36.5563],[255,-36.1889],[256,-35.7179],[258,-34.5209],[264,-30.4266],[266,-29.3164],[267,-28.8653],[268,-28.5279],[269,-28.0787],[270,-27.8547],[271,-27.5189],[272,-27.2955],[273,-27.1841],[274,-26.961],[276,-26.7383],[277,-26.7384],[279,-26.5159],[280,-26.516],[281,-26.4393],[296,-26.4393],[297,0],[367,0]]
       }
       index_02_l_ctrl.Scale.X {type=MovieSceneScriptingFloatChannel, default=1, const [1,368]}
       index_02_l_ctrl.Scale.Y {type=MovieSceneScriptingFloatChannel, default=1, const [1,368]}
       index_02_l_ctrl.Scale.Z {type=MovieSceneScriptingFloatChannel, default=1, const [1,368]}
       index_03_l_ctrl.Location.X {type=MovieSceneScriptingFloatChannel, default=0, thinned_from=368, keys [[0,0.0001],[296,0.0001],[367,0]]}
       index_03_l_ctrl.Location.Y {type=MovieSceneScriptingFloatChannel, default=0, thinned_from=368, keys [[0,-0.0003],[296,-0.0003],[297,0],[367,0]]}
       index_03_l_ctrl.Location.Z {type=MovieSceneScriptingFloatChannel, default=0, thinned_from=368, keys [[0,0],[367,0]]}
       index_03_l_ctrl.Rotation.X {
        type=MovieSceneScriptingFloatChannel
        default=0
        thinned_from=368
        keys [[0,0],[33,0],[34,-0.0002],[43,-0.0002],[44,0.0001],[48,0.0001],[49,-0.0002],[50,-0.0002],[51,0.0001],[60,0.0001],[61,-0.0002],[62,0],[70,0],[89,-0.0002],[90,0.0008],[91,-0.0001],[96,-0.0001],[97,0.0008],[98,-0.0004],[99,-0.0004],[100,0],[101,0.0001],[103,-0.0004],[104,-0.0003],[105,0.0006],[106,-0.0001],[107,-0.0001],[108,-0.0003],[109,-0.0003],[110,0.0007],[112,0.0007],[113,-0.0003],[114,-0.0001],[115,-0.0003],[116,-0.0001],[118,-0.0002],[119,0.0002],[121,0.0002],[122,-0.0003],[123,0.0002],[124,-0.0002],[126,-0.0002],[127,0.0002],[128,0.0002],[129,-0.0002],[130,0.0002],[131,0.0002],[132,-0.0002],[133,-0.0003],[134,0.0002],[135,0.0002],[136,0.0004],[137,-0.0002],[138,-0.0002],[139,0.0004],[140,-0.0002],[141,0.0004],[142,-0.0002],[143,-0.0002],[144,0.0004],[145,-0.0002],[146,-0.0002],[147,0.0004],[148,-0.0002],[150,-0.0002],[151,0.0004],[152,-0.0002],[153,0],[154,0],[155,0.0006],[156,-0.0001],[165,-0.0001],[166,0.0005],[168,0.0005],[169,-0.0002],[174,-0.0002],[175,0.0006],[176,0.0007],[177,-0.0002],[178,-0.0003],[180,0],[181,-0.0002],[182,0],[184,0],[185,-0.0002],[189,-0.0001],[190,-0.0003],[191,-0.0001],[192,0.0006],[193,0.0005],[194,-0.0001],[195,-0.0004],[196,-0.0001],[198,-0.0002],[199,0.0005],[200,-0.0002],[201,0],[202,-0.0004],[216,-0.0002],[248,0],[251,-0.0002],[252,0.0003],[253,-0.0001],[254,0.0003],[255,-0.0001],[256,-0.0002],[257,0.0005],[258,0.0004],[259,-0.0001],[264,0],[265,-0.0002],[266,0.0009],[267,-0.0004],[268,-0.0002],[273,-0.0002],[274,0.0009],[275,-0.0002],[276,0.0009],[277,-0.0002],[278,-0.0002],[279,0.0009],[280,-0.0002],[281,0],[367,0]]
       }
       index_03_l_ctrl.Rotation.Y {
        type=MovieSceneScriptingFloatChannel
        default=0
        thinned_from=368
        keys [[0,0],[33,0],[43,-0.0001],[48,0.0001],[49,-0.0002],[53,0.0001],[55,0.0002],[61,-0.0001],[62,0.0001],[65,0],[66,0.0001],[67,0],[70,0.0001],[71,-0.0002],[89,-0.0002],[90,0.0011],[91,-0.0002],[96,-0.0002],[97,0.0011],[98,0.0005],[99,0.0005],[100,-0.0001],[102,-0.0003],[103,0.0002],[104,0.0001],[105,0.0006],[106,-0.0004],[107,-0.0004],[108,0],[109,0.0001],[110,0.0005],[112,0.0005],[114,-0.0003],[115,0.0002],[116,-0.0003],[118,-0.0002],[119,0.0002],[121,0.0002],[122,0],[123,0.0002],[124,-0.0002],[126,-0.0002],[127,0.0002],[128,0.0002],[129,-0.0002],[130,0.0002],[131,0.0002],[132,-0.0002],[134,0.0002],[136,0.0004],[137,-0.0002],[138,-0.0002],[139,0.0004],[140,-0.0002],[141,0.0004],[142,-0.0002],[143,-0.0002],[144,0.0004],[145,-0.0002],[146,-0.0002],[147,0.0004],[148,-0.0002],[150,-0.0002],[151,0.0004],[152,-0.0002],[153,0],[154,0],[155,0.0006],[156,-0.0001],[165,-0.0001],[166,0.0005],[168,0.0005],[169,-0.0002],[174,-0.0002],[175,0.0006],[176,0.0006],[177,-0.0003],[178,0.0002],[179,-0.0003],[180,-0.0003],[181,0.0002],[182,-0.0002],[184,0.0001],[189,-0.0004],[190,0.0001],[191,-0.0004],[192,0.0005],[193,0.0005],[194,-0.0004],[195,0.0001],[196,-0.0003],[198,-0.0003],[199,0.0006],[200,-0.0003],[202,0.0001],[203,-0.0002],[216,-0.0002],[217,0],[248,0],[249,-0.0002],[251,-0.0002],[252,0.0004],[253,-0.0003],[254,0.0003],[255,-0.0004],[258,0.0004],[259,-0.0002],[262,0],[263,-0.0001],[264,0],[265,-0.0002],[266,0.001],[268,-0.0003],[273,-0.0002],[274,0.001],[275,-0.0002],[276,0.001],[277,-0.0002],[278,-0.0002],[279,0.001],[280,-0.0002],[367,0]]
       }
       index_03_l_ctrl.Rotation.Z {
        type=MovieSceneScriptingFloatChannel
        default=0
        thinned_from=368
        keys [[0,-13.5969],[33,-13.5969],[43,-13.5773],[44,-13.6762],[45,-13.9237],[46,-14.2705],[47,-14.7171],[49,-15.9115],[50,-16.6106],[54,-19.7242],[55,-20.4564],[56,-21.1381],[57,-21.7692],[58,-22.3491],[59,-22.8247],[60,-23.1423],[61,-23.3542],[62,-23.5133],[66,-23.9381],[68,-24.0346],[70,-24.0346],[71,-24.2078],[90,-24.2078],[91,-23.9789],[95,-23.9789],[96,-23.7503],[97,-23.7503],[99,-23.2937],[100,-22.838],[102,-22.3525],[105,-21.3317],[106,-20.8229],[114,-18.1268],[115,-17.9592],[116,-17.6244],[117,-17.4572],[118,-17.1231],[119,-16.9968],[122,-16.2895],[123,-16.1326],[124,-15.8974],[126,-15.5841],[127,-15.5058],[129,-15.1929],[131,-15.0365],[132,-14.8803],[136,-14.5682],[138,-14.33],[139,-14.33],[140,-14.2111],[141,-14.211],[142,-14.0921],[144,-14.0921],[145,-13.9732],[147,-13.9732],[148,-13.8544],[151,-13.8544],[152,-13.7356],[153,-13.7772],[155,-13.7772],[156,-13.6386],[159,-13.6386],[160,-13.5002],[168,-13.5002],[169,-13.3617],[174,-13.3617],[177,-13.8724],[179,-14.8965],[180,-15.5813],[181,-16.4398],[182,-17.1285],[183,-17.9919],[184,-18.6847],[187,-20.5186],[190,-21.9218],[191,-22.2347],[192,-22.7049],[193,-23.0189],[194,-23.176],[195,-23.4906],[199,-24.1212],[201,-24.1212],[202,-24.2078],[252,-24.2078],[254,-23.7846],[256,-22.9406],[257,-22.4146],[258,-21.785],[261,-19.5942],[262,-18.9718],[263,-18.2476],[264,-17.6285],[265,-17.1231],[269,-15.2444],[273,-14.3098],[274,-14.3098],[275,-14.0767],[276,-14.0766],[277,-13.8437],[279,-13.8437],[280,-13.6109],[281,-13.6431],[296,-13.6431],[297,0],[367,0]]
       }
       index_03_l_ctrl.Scale.X {type=MovieSceneScriptingFloatChannel, default=1, const [1,368]}
       index_03_l_ctrl.Scale.Y {type=MovieSceneScriptingFloatChannel, default=1, const [1,368]}
       index_03_l_ctrl.Scale.Z {type=MovieSceneScriptingFloatChannel, default=1, const [1,368]}
       middle_metacarpal_l_ctrl.Location.X {type=MovieSceneScriptingFloatChannel, default=0, thinned_from=368, keys [[0,0],[367,0]]}
       middle_metacarpal_l_ctrl.Location.Y {type=MovieSceneScriptingFloatChannel, default=0, thinned_from=368, keys [[0,0],[367,0]]}
       middle_metacarpal_l_ctrl.Location.Z {type=MovieSceneScriptingFloatChannel, default=0, thinned_from=368, keys [[0,0],[367,0]]}
       middle_metacarpal_l_ctrl.Rotation.X {type=MovieSceneScriptingFloatChannel, default=0, thinned_from=368, keys [[0,0],[367,0]]}
       middle_metacarpal_l_ctrl.Rotation.Y {type=MovieSceneScriptingFloatChannel, default=0, thinned_from=368, keys [[0,-1.1678],[296,-1.1678],[297,0],[367,0]]}
       middle_metacarpal_l_ctrl.Rotation.Z {type=MovieSceneScriptingFloatChannel, default=0, thinned_from=368, keys [[0,0],[367,0]]}
       middle_metacarpal_l_ctrl.Scale.X {type=MovieSceneScriptingFloatChannel, default=1, const [1,368]}
       middle_metacarpal_l_ctrl.Scale.Y {type=MovieSceneScriptingFloatChannel, default=1, const [1,368]}
       middle_metacarpal_l_ctrl.Scale.Z {type=MovieSceneScriptingFloatChannel, default=1, const [1,368]}
       middle_01_l_ctrl.Location.X {type=MovieSceneScriptingFloatChannel, default=0, thinned_from=368, keys [[0,0],[367,0]]}
       middle_01_l_ctrl.Location.Y {type=MovieSceneScriptingFloatChannel, default=0, thinned_from=368, keys [[0,0],[367,0]]}
       middle_01_l_ctrl.Location.Z {type=MovieSceneScriptingFloatChannel, default=0, thinned_from=368, keys [[0,0],[367,0]]}
       middle_01_l_ctrl.Rotation.X {
        type=MovieSceneScriptingFloatChannel
        default=0
        thinned_from=368
        keys [[0,0],[89,0],[96,-0.0001],[195,0],[196,-0.0001],[201,-0.0001],[202,0.0001],[367,0]]
       }
       middle_01_l_ctrl.Rotation.Y {
        type=MovieSceneScriptingFloatChannel
        default=0
        thinned_from=368
        keys [[0,0],[44,-0.0001],[45,0.0001],[46,-0.0001],[79,0],[80,0.0001],[83,-0.0001],[88,0],[89,-0.0001],[90,0.0001],[95,0],[96,0.0001],[100,0],[101,0.0001],[102,0],[137,0],[138,-0.0001],[139,0.0001],[143,-0.0001],[145,0.0001],[148,-0.0001],[174,0],[185,-0.0001],[186,0.0001],[189,-0.0001],[195,-0.0001],[196,0.0001],[199,-0.0001],[201,0.0001],[202,-0.0001],[203,0],[255,0.0001],[265,-0.0001],[269,0.0001],[270,-0.0001],[271,0.0001],[273,-0.0001],[280,0.0001],[367,0]]
       }
       middle_01_l_ctrl.Rotation.Z {
        type=MovieSceneScriptingFloatChannel
        default=0
        thinned_from=368
        keys [[0,-11.5867],[42,-11.5868],[43,-11.8529],[44,-12.6202],[45,-13.8272],[46,-15.4067],[47,-17.3111],[48,-19.4805],[49,-21.8483],[50,-24.3603],[53,-32.1737],[54,-34.6706],[55,-37.0193],[56,-39.158],[57,-41.0302],[58,-42.5728],[59,-43.7296],[60,-44.4421],[62,-45.2804],[65,-46.1979],[69,-46.906],[88,-47.1254],[94,-46.7039],[95,-46.5305],[96,-46.2337],[97,-45.8139],[98,-45.2965],[100,-43.9796],[102,-42.3642],[105,-39.5739],[110,-34.6384],[112,-32.8086],[116,-29.6146],[118,-27.8673],[123,-23.2341],[125,-21.53],[126,-20.7495],[127,-20.0173],[128,-19.3384],[129,-18.71],[130,-18.1317],[131,-17.6007],[132,-17.1167],[133,-16.6714],[134,-16.2675],[135,-15.8993],[136,-15.566],[138,-14.7615],[140,-14.0995],[142,-13.5571],[143,-13.3269],[145,-12.941],[146,-12.7704],[148,-12.4961],[151,-12.1848],[160,-11.5137],[164,-11.334],[171,-11.2065],[173,-11.1983],[174,-11.3414],[175,-11.7913],[176,-12.5938],[177,-13.7627],[178,-15.2922],[179,-17.1361],[180,-19.2113],[183,-25.8746],[184,-28.0091],[185,-30.0326],[186,-31.9429],[187,-33.7265],[188,-35.398],[189,-36.9474],[190,-38.3978],[191,-39.7395],[192,-40.9961],[193,-42.1664],[194,-43.2154],[195,-44.1074],[196,-44.8662],[197,-45.4906],[198,-45.9965],[199,-46.4004],[201,-46.9415],[203,-47.1931],[252,-47.2578],[253,-47.1012],[254,-46.6118],[255,-45.7769],[256,-44.5803],[257,-43.0429],[258,-41.1955],[259,-39.089],[260,-36.8064],[262,-32.0366],[263,-29.7017],[264,-27.4729],[265,-25.3904],[266,-23.4771],[267,-21.7349],[268,-20.168],[269,-18.7807],[270,-17.5511],[271,-16.4912],[272,-15.5732],[273,-14.7961],[274,-14.1528],[275,-13.6294],[276,-13.2256],[278,-12.5114],[279,-12.1876],[280,-11.9367],[281,-11.7763],[282,-11.6894],[283,-11.6406],[296,-11.5877],[297,0],[367,0]]
       }
       middle_01_l_ctrl.Scale.X {type=MovieSceneScriptingFloatChannel, default=1, const [1,368]}
       middle_01_l_ctrl.Scale.Y {type=MovieSceneScriptingFloatChannel, default=1, const [1,368]}
       middle_01_l_ctrl.Scale.Z {type=MovieSceneScriptingFloatChannel, default=1, const [1,368]}
       middle_02_l_ctrl.Location.X {type=MovieSceneScriptingFloatChannel, default=0, thinned_from=368, keys [[0,0],[367,0]]}
       middle_02_l_ctrl.Location.Y {type=MovieSceneScriptingFloatChannel, default=0, thinned_from=368, keys [[0,0],[367,0]]}
       middle_02_l_ctrl.Location.Z {type=MovieSceneScriptingFloatChannel, default=0, thinned_from=368, keys [[0,0],[367,0]]}
       middle_02_l_ctrl.Rotation.X {
        type=MovieSceneScriptingFloatChannel
        default=0
        thinned_from=368
        keys [[0,1.2087],[33,1.2087],[34,1.2063],[42,1.2063],[43,1.2308],[44,1.3292],[45,1.4767],[46,1.6587],[47,1.8903],[48,2.1504],[49,2.4185],[50,2.7166],[51,3.0396],[52,3.3529],[53,3.6472],[54,3.9509],[55,4.229],[56,4.4912],[57,4.7092],[58,4.8914],[59,5.0295],[60,5.1039],[62,5.2118],[66,5.3516],[69,5.4098],[70,5.4128],[71,5.4344],[84,5.4373],[90,5.4257],[91,5.3972],[94,5.3925],[96,5.326],[100,5.0603],[103,4.7588],[105,4.5373],[106,4.4049],[109,4.0722],[110,3.94],[111,3.8291],[112,3.7392],[113,3.6282],[114,3.5362],[115,3.4547],[116,3.342],[117,3.2516],[118,3.1406],[120,2.9299],[121,2.81],[124,2.4914],[125,2.3927],[126,2.2999],[128,2.1296],[129,2.0524],[130,1.9796],[131,1.9157],[132,1.874],[133,1.8101],[134,1.7675],[135,1.7159],[136,1.6906],[137,1.6307],[139,1.5409],[141,1.481],[142,1.4374],[143,1.421],[144,1.3911],[145,1.3784],[146,1.3484],[147,1.3311],[148,1.3184],[149,1.3012],[150,1.2884],[151,1.2884],[152,1.2575],[153,1.2576],[154,1.249],[155,1.249],[156,1.232],[157,1.232],[158,1.2151],[160,1.1982],[161,1.1982],[162,1.1813],[165,1.1813],[167,1.1644],[173,1.1634],[174,1.1799],[175,1.2295],[176,1.3257],[177,1.4743],[178,1.6697],[179,1.9249],[182,2.7421],[183,3.0016],[184,3.2446],[185,3.4768],[186,3.6881],[188,4.0651],[189,4.2453],[192,4.7039],[193,4.8414],[195,5.0743],[196,5.1673],[197,5.2394],[198,5.2977],[200,5.3851],[203,5.4417],[248,5.446],[249,5.459],[252,5.459],[253,5.4281],[255,5.2784],[256,5.1388],[257,4.9618],[258,4.7334],[259,4.4812],[260,4.2182],[261,3.9176],[262,3.6435],[263,3.3602],[264,3.088],[265,2.8431],[266,2.6231],[267,2.4163],[268,2.2216],[270,1.9127],[271,1.7839],[272,1.6808],[273,1.5922],[274,1.5149],[275,1.4529],[276,1.4012],[277,1.3602],[278,1.3238],[279,1.2722],[280,1.2464],[281,1.2332],[283,1.2148],[284,1.2148],[285,1.2041],[296,1.2041],[297,0],[367,0]]
       }
       middle_02_l_ctrl.Rotation.Y {
        type=MovieSceneScriptingFloatChannel
        default=0
        thinned_from=368
        keys [[0,-0.0328],[33,-0.0328],[34,-0.0251],[42,-0.0251],[43,-0.0271],[44,-0.0348],[45,-0.0445],[46,-0.0375],[47,-0.0342],[48,-0.0516],[49,-0.0384],[50,-0.0484],[52,-0.0615],[53,-0.056],[54,-0.0665],[55,-0.0668],[56,-0.0723],[57,-0.0716],[58,-0.0753],[59,-0.0773],[60,-0.071],[61,-0.0659],[62,-0.0745],[63,-0.0707],[64,-0.0731],[65,-0.0768],[66,-0.0789],[67,-0.0768],[68,-0.0723],[69,-0.0895],[70,-0.0729],[71,-0.0901],[72,-0.08],[73,-0.0735],[84,-0.0735],[85,-0.068],[88,-0.068],[89,-0.0907],[90,-0.0907],[91,-0.068],[92,-0.068],[93,-0.0906],[95,-0.0904],[96,-0.0902],[97,-0.0669],[98,-0.0656],[99,-0.0639],[100,-0.0835],[101,-0.0791],[102,-0.0682],[103,-0.0673],[104,-0.0747],[105,-0.0806],[106,-0.0663],[107,-0.0693],[108,-0.071],[109,-0.0713],[110,-0.051],[111,-0.0485],[112,-0.0643],[113,-0.0593],[114,-0.0639],[115,-0.0674],[116,-0.0504],[117,-0.061],[118,-0.0503],[119,-0.0525],[120,-0.0586],[121,-0.0495],[122,-0.0531],[123,-0.0555],[124,-0.0567],[125,-0.0425],[126,-0.0559],[127,-0.0544],[128,-0.0521],[129,-0.0353],[130,-0.0407],[131,-0.0376],[132,-0.0407],[133,-0.0369],[134,-0.0341],[135,-0.0391],[136,-0.0364],[137,-0.0348],[138,-0.0457],[139,-0.0318],[141,-0.0294],[142,-0.0398],[143,-0.0266],[144,-0.0251],[145,-0.0437],[146,-0.042],[147,-0.0218],[148,-0.0403],[149,-0.0201],[150,-0.0384],[151,-0.0384],[152,-0.0298],[153,-0.0363],[154,-0.0295],[155,-0.0295],[156,-0.0297],[158,-0.0299],[159,-0.0369],[160,-0.0301],[165,-0.0302],[166,-0.0412],[167,-0.0303],[168,-0.0303],[169,-0.0315],[173,-0.0315],[174,-0.0333],[175,-0.0384],[176,-0.0267],[177,-0.0392],[178,-0.035],[179,-0.0442],[180,-0.0433],[181,-0.048],[182,-0.0496],[183,-0.0571],[184,-0.059],[185,-0.0636],[186,-0.0641],[187,-0.0584],[188,-0.0586],[189,-0.0754],[190,-0.0677],[191,-0.0677],[192,-0.076],[193,-0.0715],[194,-0.0707],[195,-0.0795],[196,-0.0792],[197,-0.078],[198,-0.0841],[199,-0.0797],[200,-0.0802],[201,-0.0733],[203,-0.0769],[216,-0.0769],[217,-0.0803],[248,-0.0803],[249,-0.0906],[252,-0.0906],[253,-0.0661],[254,-0.0746],[255,-0.0823],[257,-0.0837],[258,-0.0799],[259,-0.0733],[260,-0.0715],[261,-0.0603],[262,-0.0546],[263,-0.0648],[265,-0.045],[266,-0.0484],[267,-0.0485],[268,-0.0417],[269,-0.0414],[270,-0.0386],[271,-0.0345],[272,-0.0299],[273,-0.0356],[274,-0.0309],[275,-0.0433],[276,-0.0395],[277,-0.0197],[278,-0.0332],[279,-0.0287],[280,-0.0264],[281,-0.0401],[282,-0.031],[283,-0.0418],[284,-0.0418],[285,-0.0291],[296,-0.0291],[297,0],[367,0]]
       }
       middle_02_l_ctrl.Rotation.Z {
        type=MovieSceneScriptingFloatChannel
        default=0
        thinned_from=368
        keys [[0,-39.1872],[33,-39.1872],[42,-39.2295],[43,-39.3558],[44,-39.8623],[45,-40.6258],[46,-41.6502],[47,-42.813],[48,-44.1197],[51,-48.8936],[53,-52.197],[55,-55.3153],[57,-57.8241],[59,-59.57],[61,-60.2975],[64,-60.9279],[68,-61.4841],[69,-61.4867],[70,-61.6182],[71,-61.6208],[72,-61.7529],[88,-61.7516],[89,-61.5806],[92,-61.5771],[93,-61.4064],[94,-61.4064],[97,-60.8824],[99,-60.1931],[102,-58.6829],[113,-52.0589],[114,-51.6235],[115,-51.0461],[116,-50.613],[124,-45.9634],[125,-45.506],[126,-44.9388],[128,-44.0354],[133,-42.3595],[136,-41.7235],[137,-41.415],[138,-41.2617],[139,-40.9536],[146,-39.8842],[147,-39.8834],[148,-39.732],[149,-39.7312],[150,-39.5801],[152,-39.58],[154,-39.4164],[155,-39.4164],[156,-39.3236],[157,-39.3236],[158,-39.2309],[159,-39.2313],[160,-39.1382],[161,-39.1382],[162,-39.0456],[165,-39.0456],[166,-38.9535],[173,-38.9435],[174,-39.0264],[175,-39.2756],[176,-39.8578],[177,-40.6115],[178,-41.6225],[179,-42.8133],[180,-44.1884],[183,-48.5063],[184,-49.8623],[185,-51.0918],[188,-54.3659],[191,-57.001],[194,-59.1979],[195,-59.7996],[197,-60.665],[200,-61.4515],[203,-61.753],[253,-61.7513],[255,-60.9002],[256,-60.0562],[258,-57.9779],[260,-55.1464],[261,-53.5646],[262,-52.2012],[264,-49.1533],[266,-46.647],[267,-45.6638],[268,-44.55],[270,-42.8994],[271,-42.2182],[272,-41.6759],[275,-40.4657],[276,-40.198],[277,-40.0636],[279,-39.5311],[280,-39.3982],[283,-39.185],[296,-39.1845],[297,0],[367,0]]
       }
       middle_02_l_ctrl.Scale.X {type=MovieSceneScriptingFloatChannel, default=1, const [1,368]}
       middle_02_l_ctrl.Scale.Y {type=MovieSceneScriptingFloatChannel, default=1, const [1,368]}
       middle_02_l_ctrl.Scale.Z {type=MovieSceneScriptingFloatChannel, default=1, const [1,368]}
       middle_03_l_ctrl.Location.X {type=MovieSceneScriptingFloatChannel, default=0, thinned_from=368, keys [[0,0],[367,0]]}
       middle_03_l_ctrl.Location.Y {type=MovieSceneScriptingFloatChannel, default=0, thinned_from=368, keys [[0,0],[367,0]]}
       middle_03_l_ctrl.Location.Z {type=MovieSceneScriptingFloatChannel, default=0, thinned_from=368, keys [[0,0],[367,0]]}
       middle_03_l_ctrl.Rotation.X {
        type=MovieSceneScriptingFloatChannel
        default=0
        thinned_from=368
        keys [[0,0],[33,0],[34,0.0004],[42,0.0004],[43,-0.0012],[44,0.0004],[48,0.0003],[50,0.0009],[51,-0.0021],[52,-0.0024],[53,0.0025],[54,-0.0026],[55,-0.0029],[56,-0.0025],[57,-0.0024],[58,0.0033],[59,-0.0019],[60,0.0034],[61,-0.0019],[62,-0.0018],[63,-0.0015],[64,-0.0017],[65,-0.0015],[66,0.0039],[67,-0.0014],[68,-0.0002],[69,-0.0005],[70,0.0011],[71,-0.0008],[72,-0.0008],[73,0.0007],[83,0.0009],[84,0.0007],[85,0],[90,0.0001],[91,-0.0005],[92,0.0001],[95,0.0001],[96,-0.0004],[100,0.0002],[101,-0.0002],[102,0.0004],[111,0.0002],[112,0.0003],[113,-0.0016],[114,0.0005],[115,0.0006],[116,-0.0013],[117,0.0008],[118,0.0009],[119,-0.001],[120,-0.001],[121,0.0017],[122,-0.0013],[123,0.0015],[124,-0.0014],[125,-0.0015],[126,0.0014],[128,0.0014],[129,-0.0015],[130,0.0014],[131,-0.0015],[132,0.0014],[135,0.0015],[136,-0.0014],[138,-0.0014],[139,0.0027],[140,-0.0014],[141,0.0027],[142,-0.0014],[143,0.0027],[144,-0.0014],[145,-0.0014],[146,0.0027],[147,-0.0014],[149,-0.0015],[150,0.0027],[151,0.0027],[152,-0.0014],[153,-0.0003],[154,-0.0003],[155,0.002],[156,-0.0006],[158,-0.0006],[159,0.0017],[160,-0.0008],[163,-0.0008],[164,0.0015],[168,0.0015],[169,0.0011],[175,0.001],[176,-0.0012],[177,0.0009],[180,0.001],[181,-0.0011],[184,0.0001],[185,-0.0021],[186,0.0023],[187,-0.0023],[188,-0.0022],[189,0.0026],[190,-0.0019],[191,0.0029],[192,-0.0016],[193,-0.0012],[196,-0.0008],[197,0.004],[198,-0.0004],[201,-0.0001],[202,0.0034],[203,-0.0012],[216,-0.0012],[217,0],[248,0],[249,-0.0012],[252,-0.0012],[253,0.0015],[254,0.0015],[255,-0.0013],[256,0.0012],[258,0.0011],[259,-0.0015],[260,0.0012],[261,-0.0012],[262,-0.0008],[264,0.0003],[265,-0.0002],[268,0],[269,-0.0016],[270,0.0003],[273,0.0009],[274,-0.0008],[275,0.0011],[276,0.0012],[277,-0.0005],[280,-0.0003],[281,0],[367,0]]
       }
       middle_03_l_ctrl.Rotation.Y {
        type=MovieSceneScriptingFloatChannel
        default=0
        thinned_from=368
        keys [[0,0],[33,0],[34,0.0002],[42,0.0002],[43,-0.0008],[44,0.0002],[48,0.0002],[50,0.0005],[51,-0.0011],[52,-0.0013],[53,0.0015],[54,-0.0015],[55,-0.001],[56,-0.0014],[57,-0.0011],[58,0.0012],[59,-0.001],[60,0.0017],[61,-0.0006],[62,-0.0004],[63,-0.001],[64,-0.0001],[65,-0.0007],[66,0.0018],[67,-0.0005],[68,-0.0001],[69,-0.0002],[70,0.0005],[71,-0.0003],[72,-0.0003],[73,0.0004],[74,0.0004],[75,0.0001],[83,0.0001],[84,0.0004],[85,0],[90,0],[91,-0.0002],[95,0.0001],[96,-0.0002],[100,0.0001],[101,-0.0001],[102,0.0002],[104,0],[106,0.0001],[108,-0.0001],[111,0.0002],[112,0.0001],[113,-0.0007],[114,0.0001],[115,0.0002],[116,-0.0006],[117,0.0004],[118,0.0005],[119,-0.0004],[120,-0.0006],[121,0.001],[122,-0.0006],[123,0.0009],[124,-0.0008],[125,-0.0009],[126,0.0007],[128,0.0008],[129,-0.0009],[130,0.0007],[131,-0.001],[132,0.0007],[134,0.0009],[135,0.0008],[136,-0.0008],[138,-0.0009],[139,0.0016],[140,-0.0009],[141,0.0016],[142,-0.0009],[143,0.0016],[144,-0.0009],[145,-0.0008],[146,0.0017],[147,-0.0009],[149,-0.0008],[150,0.0017],[151,0.0017],[152,-0.0009],[153,-0.0002],[154,-0.0002],[155,0.0013],[156,-0.0004],[158,-0.0004],[159,0.0011],[160,-0.0006],[163,-0.0005],[164,0.0009],[168,0.0009],[169,0.0007],[175,0.0006],[176,-0.0008],[177,0.0005],[180,0.0006],[181,-0.0006],[183,-0.0002],[184,0.0002],[185,-0.001],[186,0.0009],[187,-0.0007],[188,-0.0008],[189,0.0009],[190,-0.0008],[191,0.0009],[192,-0.0005],[193,-0.0008],[196,-0.0002],[197,0.0015],[198,-0.0004],[200,0],[201,-0.0001],[202,0.0015],[203,-0.0005],[216,-0.0005],[217,0],[248,0],[249,-0.0005],[252,-0.0005],[253,0.0006],[254,0.0004],[255,-0.0007],[256,0.0005],[258,0.0004],[259,-0.0007],[260,0.0006],[261,-0.0005],[263,-0.0002],[264,0.0001],[266,-0.0001],[268,0],[269,-0.0009],[270,0.0002],[273,0.0005],[274,-0.0005],[275,0.0006],[276,0.0007],[277,-0.0003],[367,0]]
       }
       middle_03_l_ctrl.Rotation.Z {
        type=MovieSceneScriptingFloatChannel
        default=0
        thinned_from=368
        keys [[0,-45.519],[42,-45.5609],[43,-45.3723],[44,-45.0896],[45,-44.5256],[46,-43.7763],[47,-42.937],[48,-42.0088],[50,-39.799],[55,-33.9696],[57,-32.2501],[59,-31.1107],[62,-30.2596],[63,-30.2596],[64,-29.9766],[66,-29.9765],[67,-29.6939],[68,-29.7551],[69,-29.6531],[70,-29.653],[71,-29.551],[91,-29.6297],[94,-29.7503],[96,-29.9614],[97,-30.1425],[99,-30.6563],[101,-31.3229],[106,-33.3723],[107,-33.8961],[112,-36.0039],[113,-36.3218],[114,-36.7464],[116,-37.3849],[118,-38.2393],[119,-38.5576],[121,-39.5215],[122,-39.8439],[123,-40.3283],[125,-40.9762],[126,-41.4635],[128,-42.1151],[129,-42.2784],[130,-42.6053],[131,-42.769],[132,-43.0967],[135,-43.5895],[136,-43.6549],[139,-44.3567],[140,-44.3568],[141,-44.5913],[142,-44.5914],[143,-44.8261],[145,-44.8262],[146,-45.0613],[149,-45.0614],[150,-45.2968],[154,-45.2968],[155,-45.4288],[158,-45.4288],[159,-45.5609],[163,-45.561],[164,-45.6931],[174,-45.6931],[175,-45.4393],[176,-45.0593],[177,-44.5538],[178,-43.7981],[179,-42.9204],[180,-41.9223],[181,-40.8057],[184,-37.858],[185,-37.0344],[186,-36.3133],[187,-35.3559],[188,-34.6405],[189,-34.1647],[190,-33.4532],[191,-32.9799],[192,-32.272],[195,-30.8624],[200,-29.6939],[202,-29.6938],[203,-29.449],[216,-29.449],[217,-29.5102],[248,-29.5102],[249,-29.449],[252,-29.449],[253,-29.5889],[255,-30.1493],[256,-30.711],[257,-31.4148],[259,-33.1128],[260,-34.2516],[262,-36.2588],[263,-37.4141],[264,-38.4302],[266,-40.1717],[267,-40.956],[268,-41.6449],[270,-42.8316],[271,-43.3282],[273,-44.1256],[274,-44.3255],[275,-44.6257],[276,-44.8262],[277,-44.9265],[278,-45.1273],[281,-45.4337],[296,-45.4337],[297,0],[367,0]]
       }
       middle_03_l_ctrl.Scale.X {type=MovieSceneScriptingFloatChannel, default=1, const [1,368]}
       middle_03_l_ctrl.Scale.Y {type=MovieSceneScriptingFloatChannel, default=1, const [1,368]}
       middle_03_l_ctrl.Scale.Z {type=MovieSceneScriptingFloatChannel, default=1, const [1,368]}
       ring_metacarpal_l_ctrl.Location.X {type=MovieSceneScriptingFloatChannel, default=0, thinned_from=368, keys [[0,0],[367,0]]}
       ring_metacarpal_l_ctrl.Location.Y {type=MovieSceneScriptingFloatChannel, default=0, thinned_from=368, keys [[0,0],[367,0]]}
       ring_metacarpal_l_ctrl.Location.Z {type=MovieSceneScriptingFloatChannel, default=0, thinned_from=368, keys [[0,0],[367,0]]}
       ring_metacarpal_l_ctrl.Rotation.X {type=MovieSceneScriptingFloatChannel, default=0, thinned_from=368, keys [[0,0],[367,0]]}
       ring_metacarpal_l_ctrl.Rotation.Y {type=MovieSceneScriptingFloatChannel, default=0, thinned_from=368, keys [[0,-5.131],[296,-5.131],[297,0],[367,0]]}
       ring_metacarpal_l_ctrl.Rotation.Z {type=MovieSceneScriptingFloatChannel, default=0, thinned_from=368, keys [[0,0],[367,0]]}
       ring_metacarpal_l_ctrl.Scale.X {type=MovieSceneScriptingFloatChannel, default=1, const [1,368]}
       ring_metacarpal_l_ctrl.Scale.Y {type=MovieSceneScriptingFloatChannel, default=1, const [1,368]}
       ring_metacarpal_l_ctrl.Scale.Z {type=MovieSceneScriptingFloatChannel, default=1, const [1,368]}
       ring_01_l_ctrl.Location.X {type=MovieSceneScriptingFloatChannel, default=0, thinned_from=368, keys [[0,0],[367,0]]}
       ring_01_l_ctrl.Location.Y {type=MovieSceneScriptingFloatChannel, default=0, thinned_from=368, keys [[0,0],[367,0]]}
       ring_01_l_ctrl.Location.Z {type=MovieSceneScriptingFloatChannel, default=0, thinned_from=368, keys [[0,0],[367,0]]}
       ring_01_l_ctrl.Rotation.X {
        type=MovieSceneScriptingFloatChannel
        default=0
        thinned_from=368
        keys [[0,0],[33,0],[34,-0.0006],[40,-0.0006],[41,0.0005],[42,-0.0004],[43,0.0012],[44,-0.0015],[45,0.001],[46,-0.0024],[48,-0.002],[49,0.0011],[50,-0.001],[51,-0.0003],[52,-0.0001],[53,0.0005],[54,-0.0002],[55,0.0004],[56,0.0001],[57,0.0001],[58,-0.0018],[59,0.0009],[60,0.0024],[61,-0.0013],[62,0.0002],[63,-0.0002],[64,-0.0018],[65,-0.0002],[67,-0.0008],[68,-0.0008],[69,0.0011],[71,0.0013],[72,-0.0005],[73,0.0013],[74,-0.0005],[84,-0.0005],[85,-0.0008],[86,0.0004],[87,0.0004],[88,-0.001],[89,0.0002],[90,-0.0012],[91,0],[95,-0.0007],[97,-0.0016],[98,0.0014],[99,0.0006],[100,-0.0005],[101,0.002],[102,0.0016],[104,-0.0022],[105,-0.0014],[106,0.0011],[107,-0.0027],[108,-0.002],[109,0.0002],[110,-0.001],[111,0.0008],[112,-0.0034],[113,0.001],[114,-0.0008],[115,0.0003],[116,-0.0003],[117,0.0017],[118,-0.0011],[119,0.0023],[120,-0.0026],[121,0.0031],[122,0.0012],[123,0.0009],[124,0],[125,0.0019],[126,0.0019],[127,0],[128,-0.0011],[129,0.0018],[130,-0.0002],[131,-0.001],[134,-0.0008],[135,0.0017],[136,-0.0012],[137,-0.0007],[138,0.0004],[140,-0.0002],[141,0.0006],[143,-0.0001],[144,0.0007],[145,0.0008],[146,0],[149,-0.0007],[150,0.0003],[151,-0.001],[152,0],[153,-0.0009],[155,-0.001],[156,0.001],[158,0.001],[159,0.0003],[160,-0.0018],[161,-0.0005],[162,-0.0005],[163,0.0001],[164,-0.0006],[165,0.0014],[166,-0.0006],[167,-0.0006],[168,0.0014],[169,-0.0008],[171,-0.0008],[172,0.0006],[173,-0.0001],[174,0.0002],[175,0.0027],[176,0.0028],[177,0.0013],[178,0.0022],[179,-0.0001],[180,0.0009],[181,-0.0001],[182,-0.0015],[183,0],[184,-0.001],[185,0.0019],[186,-0.0002],[187,0.001],[188,0.0005],[189,-0.0008],[190,0.002],[191,0.0016],[192,-0.0003],[193,0.0005],[194,-0.0017],[196,-0.0002],[197,0.0007],[198,-0.0012],[199,-0.0001],[200,-0.0006],[202,-0.0005],[203,-0.0029],[204,0],[248,0],[249,0.0003],[253,0.0003],[254,0.0008],[255,0],[256,-0.0023],[257,-0.0001],[258,-0.0017],[260,-0.003],[261,0.0006],[262,-0.0004],[263,-0.0001],[264,-0.0005],[265,0.0011],[266,-0.0016],[267,0.0009],[268,0.0018],[270,-0.001],[271,0],[272,0.0006],[273,-0.0009],[274,-0.0015],[276,-0.0005],[278,0.0011],[279,-0.0001],[280,0.0007],[282,0.0013],[283,-0.0019],[284,0.0014],[285,0.0014],[286,0.0004],[287,-0.0018],[296,-0.0018],[297,0],[367,0]]
       }
       ring_01_l_ctrl.Rotation.Y {
        type=MovieSceneScriptingFloatChannel
        default=0
        thinned_from=368
        keys [[0,0],[33,0],[34,0.001],[40,0.001],[41,0.0017],[42,-0.0011],[43,0.0006],[44,0.0007],[45,-0.0008],[46,0.0013],[47,0.0008],[48,-0.0009],[49,0.0012],[50,-0.0004],[51,0],[52,-0.0007],[53,0],[54,0.0016],[55,0.0009],[56,-0.0008],[57,0.0005],[58,0.0007],[59,-0.0005],[60,0],[61,-0.0006],[62,0],[63,-0.0007],[65,0.0011],[67,0.0002],[68,-0.0015],[69,-0.0013],[70,0.0015],[71,0.0004],[72,0.0018],[73,-0.0008],[74,0.0007],[84,0.0007],[85,0.0012],[86,-0.0006],[87,-0.0006],[88,0.0013],[89,-0.0005],[90,0.0014],[91,-0.0005],[96,0],[100,0.0008],[101,0.0009],[102,0.0018],[104,-0.0002],[105,0.0008],[106,-0.0007],[107,0.0005],[108,0.0012],[109,-0.0007],[110,0.0021],[111,-0.0001],[112,0.0006],[113,-0.0002],[114,0.0019],[115,-0.0012],[116,-0.0019],[117,-0.0014],[118,-0.0003],[119,-0.0014],[120,0.0006],[121,-0.0011],[122,-0.0009],[123,0],[124,0.0003],[125,-0.0014],[126,-0.0013],[127,0.0009],[128,0],[129,0.0008],[130,-0.0007],[131,0.0013],[134,-0.0009],[135,-0.0007],[136,0.0007],[137,-0.0007],[138,-0.0001],[140,0],[141,0.0003],[143,0.0001],[144,0.0004],[145,-0.0013],[146,0.0002],[149,-0.0001],[150,0.0002],[151,-0.0003],[152,0.0001],[153,0.0014],[155,0.0002],[156,0.0003],[158,-0.0008],[159,0.0007],[160,-0.0005],[161,0.0017],[162,0.0011],[163,-0.0016],[164,0.0005],[165,0.0006],[166,0],[168,0],[169,0.0011],[171,0.0011],[172,-0.0019],[173,-0.0013],[174,0.0003],[175,-0.0004],[176,-0.0016],[177,0.0002],[178,-0.0013],[179,-0.001],[180,-0.0002],[181,-0.001],[182,0.0009],[183,-0.0001],[184,0.0012],[185,-0.0002],[186,0.0003],[187,0],[188,0.0004],[191,-0.0008],[192,-0.0001],[193,-0.0001],[194,0.0002],[196,0.0001],[197,0.0003],[199,0.0013],[200,0.0007],[201,-0.002],[202,0.0023],[203,-0.0021],[204,0],[248,0],[249,-0.0004],[253,-0.0004],[254,0.0001],[255,0.0005],[256,0.0001],[257,0.0003],[258,0.001],[259,-0.0008],[260,-0.0003],[261,-0.0009],[262,0.0006],[263,-0.0011],[264,-0.0012],[266,0.0005],[267,0.0002],[268,-0.0015],[269,-0.0019],[270,0.0004],[271,0.0005],[272,0.0003],[273,-0.0009],[274,0.0012],[276,0.0005],[278,0.0002],[279,-0.0006],[280,-0.0008],[281,-0.0006],[282,-0.0009],[283,0.0014],[284,-0.001],[285,-0.001],[286,0.002],[287,0.0013],[296,0.0013],[297,0],[367,0]]
       }
       ring_01_l_ctrl.Rotation.Z {
        type=MovieSceneScriptingFloatChannel
        default=0
        thinned_from=368
        keys [[0,-14.7374],[40,-14.7303],[41,-15.0099],[42,-15.7802],[43,-16.9739],[44,-18.5959],[45,-20.5112],[46,-22.7253],[47,-25.1019],[49,-30.2169],[50,-32.8907],[52,-37.9799],[53,-40.2951],[54,-42.4663],[55,-44.3149],[56,-45.8314],[57,-46.948],[58,-47.6571],[62,-49.025],[66,-49.8632],[72,-50.3816],[85,-50.4482],[95,-49.9933],[97,-49.669],[100,-48.8293],[104,-47.2252],[108,-45.0896],[112,-42.5771],[116,-39.7818],[118,-38.1238],[120,-36.0402],[122,-33.5282],[124,-30.9072],[125,-29.5398],[126,-28.2479],[127,-27.0301],[129,-24.746],[130,-23.7445],[132,-22.0172],[133,-21.2896],[134,-20.6298],[135,-20.0373],[137,-19.129],[138,-18.7605],[139,-18.2086],[140,-17.731],[141,-17.2908],[142,-16.9245],[143,-16.5951],[144,-16.3026],[145,-16.047],[147,-15.6819],[148,-15.536],[150,-15.3173],[163,-14.5032],[164,-14.5031],[165,-14.4416],[167,-14.4416],[168,-14.3801],[169,-14.4172],[171,-14.4172],[172,-14.4976],[173,-14.9789],[174,-15.7827],[175,-16.9922],[176,-18.6119],[177,-20.5666],[178,-22.699],[179,-25.0136],[180,-27.2651],[181,-29.4524],[182,-31.5745],[183,-33.5447],[184,-35.359],[186,-38.6433],[188,-41.4486],[189,-42.7376],[191,-44.9768],[192,-45.9752],[193,-46.8736],[194,-47.617],[195,-48.257],[196,-48.7923],[198,-49.5983],[200,-50.084],[203,-50.4482],[253,-50.4483],[254,-50.3035],[255,-49.7977],[256,-49.0061],[257,-47.7906],[258,-46.3011],[259,-44.4751],[260,-42.3914],[262,-37.7532],[263,-35.3413],[264,-33.0225],[265,-30.7954],[266,-28.7511],[267,-26.7874],[268,-25.0266],[269,-23.4644],[270,-22.0967],[271,-20.8592],[272,-19.75],[273,-18.8287],[274,-18.0323],[275,-17.3602],[276,-16.8113],[278,-15.9593],[280,-15.2308],[281,-14.9734],[282,-14.8225],[283,-14.8223],[284,-14.7471],[296,-14.7468],[297,0],[367,0]]
       }
       ring_01_l_ctrl.Scale.X {type=MovieSceneScriptingFloatChannel, default=1, const [1,368]}
       ring_01_l_ctrl.Scale.Y {type=MovieSceneScriptingFloatChannel, default=1, const [1,368]}
       ring_01_l_ctrl.Scale.Z {type=MovieSceneScriptingFloatChannel, default=1, const [1,368]}
       ring_02_l_ctrl.Location.X {type=MovieSceneScriptingFloatChannel, default=0, thinned_from=368, keys [[0,0],[367,0]]}
       ring_02_l_ctrl.Location.Y {type=MovieSceneScriptingFloatChannel, default=0, thinned_from=368, keys [[0,0],[367,0]]}
       ring_02_l_ctrl.Location.Z {type=MovieSceneScriptingFloatChannel, default=0, thinned_from=368, keys [[0,0],[367,0]]}
       ring_02_l_ctrl.Rotation.X {
        type=MovieSceneScriptingFloatChannel
        default=0
        thinned_from=368
        keys [[0,-2.6638],[40,-2.6634],[41,-2.7253],[42,-2.9113],[43,-3.1995],[44,-3.5835],[45,-4.0354],[46,-4.5512],[47,-5.1143],[48,-5.715],[51,-7.5871],[52,-8.1697],[53,-8.7255],[54,-9.232],[55,-9.678],[56,-10.036],[57,-10.3011],[58,-10.4553],[60,-10.6381],[62,-10.7933],[66,-10.9929],[70,-11.0902],[71,-11.0902],[72,-11.1294],[85,-11.1324],[91,-11.0864],[93,-11.0649],[94,-11.0311],[95,-11.0311],[98,-10.896],[101,-10.672],[105,-10.2553],[110,-9.576],[113,-9.1138],[114,-8.9204],[116,-8.5975],[118,-8.1979],[119,-7.9763],[122,-7.1272],[123,-6.8221],[124,-6.4846],[126,-5.8716],[128,-5.2905],[130,-4.802],[131,-4.59],[132,-4.3967],[133,-4.2165],[135,-3.9213],[136,-3.8026],[137,-3.6979],[138,-3.6106],[140,-3.3661],[142,-3.1739],[144,-3.034],[146,-2.9293],[147,-2.8825],[148,-2.8476],[149,-2.8302],[150,-2.7951],[152,-2.7601],[154,-2.7409],[155,-2.7219],[156,-2.712],[157,-2.693],[158,-2.6932],[160,-2.6351],[161,-2.6351],[162,-2.6163],[163,-2.6062],[165,-2.6062],[166,-2.5873],[168,-2.5873],[169,-2.5822],[171,-2.5822],[172,-2.6145],[173,-2.7263],[174,-2.9222],[175,-3.2176],[176,-3.6364],[177,-4.1421],[178,-4.6828],[179,-5.2427],[180,-5.787],[181,-6.2964],[182,-6.778],[183,-7.2165],[184,-7.6339],[186,-8.3646],[187,-8.7005],[188,-9.0172],[190,-9.5873],[192,-10.0656],[193,-10.2765],[195,-10.6097],[198,-10.9344],[200,-11.0471],[202,-11.0988],[203,-11.1446],[216,-11.1446],[253,-11.1315],[254,-11.0988],[255,-10.9797],[256,-10.7756],[257,-10.4929],[258,-10.1456],[259,-9.7119],[261,-8.6774],[262,-8.1071],[264,-7.0114],[265,-6.484],[266,-5.9755],[267,-5.521],[268,-5.1134],[269,-4.729],[270,-4.4043],[271,-4.1202],[272,-3.8539],[273,-3.6283],[274,-3.4438],[275,-3.2958],[276,-3.1589],[277,-3.0516],[278,-2.9628],[279,-2.8552],[280,-2.7774],[281,-2.7257],[283,-2.6754],[284,-2.6567],[296,-2.6567],[297,0],[367,0]]
       }
       ring_02_l_ctrl.Rotation.Y {
        type=MovieSceneScriptingFloatChannel
        default=0
        thinned_from=368
        keys [[0,-0.2818],[33,-0.2818],[40,-0.2814],[41,-0.2888],[42,-0.3109],[43,-0.3365],[44,-0.3769],[45,-0.4299],[46,-0.4843],[48,-0.6001],[49,-0.6731],[50,-0.7371],[52,-0.8574],[53,-0.9328],[54,-0.9884],[55,-1.0138],[56,-1.0673],[57,-1.0941],[58,-1.1128],[59,-1.1135],[60,-1.1303],[61,-1.1352],[62,-1.144],[63,-1.1561],[64,-1.156],[65,-1.1593],[66,-1.1596],[67,-1.166],[68,-1.161],[69,-1.1785],[70,-1.1691],[71,-1.1691],[72,-1.1773],[84,-1.1773],[85,-1.1808],[86,-1.1668],[87,-1.1668],[88,-1.1756],[90,-1.1756],[91,-1.1615],[92,-1.1704],[93,-1.1704],[94,-1.1653],[95,-1.1653],[96,-1.1603],[97,-1.1645],[98,-1.1458],[99,-1.1365],[100,-1.1371],[101,-1.1286],[102,-1.0997],[103,-1.1069],[104,-1.0859],[105,-1.0827],[106,-1.0857],[107,-1.051],[108,-1.0452],[109,-1.0127],[110,-1.0098],[112,-0.9755],[113,-0.9609],[114,-0.9545],[115,-0.9337],[116,-0.9082],[117,-0.8986],[118,-0.8669],[119,-0.844],[120,-0.8077],[121,-0.7778],[122,-0.7537],[124,-0.6866],[125,-0.6649],[126,-0.6107],[127,-0.5831],[128,-0.5609],[129,-0.5278],[131,-0.4781],[132,-0.4706],[133,-0.4552],[134,-0.4376],[136,-0.4048],[137,-0.3917],[138,-0.3814],[139,-0.3677],[140,-0.3549],[141,-0.353],[142,-0.3366],[145,-0.3191],[146,-0.3152],[147,-0.3044],[149,-0.2993],[152,-0.2961],[153,-0.284],[154,-0.2962],[155,-0.283],[156,-0.2953],[157,-0.2821],[158,-0.2797],[159,-0.2812],[160,-0.2804],[161,-0.2804],[162,-0.265],[163,-0.2796],[165,-0.2796],[166,-0.2642],[168,-0.2642],[169,-0.2722],[171,-0.2722],[172,-0.2799],[173,-0.2856],[174,-0.3107],[175,-0.3351],[176,-0.3854],[177,-0.4385],[178,-0.4898],[179,-0.5611],[180,-0.6189],[181,-0.6622],[182,-0.7205],[183,-0.7729],[184,-0.8044],[185,-0.8509],[186,-0.8873],[187,-0.9258],[188,-0.9453],[189,-0.9904],[190,-1.0137],[191,-1.0432],[192,-1.0664],[193,-1.0824],[194,-1.1111],[195,-1.1236],[196,-1.1468],[197,-1.1417],[198,-1.1538],[199,-1.1723],[200,-1.1691],[201,-1.1502],[202,-1.1422],[203,-1.1949],[216,-1.1949],[217,-1.1773],[248,-1.1773],[249,-1.1798],[253,-1.1798],[255,-1.1645],[256,-1.1504],[257,-1.1105],[258,-1.0539],[259,-1.0243],[260,-0.9756],[261,-0.9141],[262,-0.8642],[263,-0.8081],[264,-0.7388],[266,-0.6351],[267,-0.574],[268,-0.5442],[269,-0.503],[270,-0.4678],[271,-0.4299],[272,-0.414],[273,-0.3911],[274,-0.3533],[275,-0.35],[276,-0.3323],[277,-0.3158],[278,-0.3158],[279,-0.3031],[280,-0.2882],[281,-0.3045],[282,-0.2838],[283,-0.2662],[284,-0.2904],[296,-0.2904],[297,0],[367,0]]
       }
       ring_02_l_ctrl.Rotation.Z {
        type=MovieSceneScriptingFloatChannel
        default=0
        thinned_from=368
        keys [[0,-36.7912],[40,-36.7775],[41,-36.9965],[42,-37.6005],[43,-38.5942],[44,-39.8738],[45,-41.3907],[46,-43.1531],[47,-45.0555],[52,-55.3792],[53,-57.2779],[54,-58.9367],[55,-60.4365],[56,-61.6696],[57,-62.5352],[58,-63.1146],[61,-63.9934],[64,-64.5847],[68,-65.0823],[87,-65.3423],[91,-65.2294],[96,-64.8915],[99,-64.3309],[103,-63.1271],[107,-61.5173],[108,-60.9878],[109,-60.5925],[114,-57.9925],[115,-57.3566],[116,-56.8481],[117,-56.2159],[120,-53.8327],[123,-50.7779],[124,-49.6665],[125,-48.6746],[126,-47.5848],[128,-45.6438],[129,-44.7915],[132,-42.5789],[135,-41.0205],[137,-40.2945],[138,-39.999],[140,-39.1746],[142,-38.53],[145,-37.8303],[149,-37.3077],[153,-37.1193],[154,-37.0179],[155,-37.018],[156,-36.9167],[157,-36.9167],[158,-36.8159],[159,-36.8155],[160,-36.7143],[161,-36.7143],[162,-36.6137],[165,-36.6133],[166,-36.5127],[171,-36.5365],[172,-36.5992],[173,-36.9752],[174,-37.6666],[175,-38.6153],[176,-39.9536],[177,-41.5639],[178,-43.3246],[181,-48.754],[183,-51.948],[185,-54.6815],[186,-55.9519],[189,-59.1987],[192,-61.7947],[194,-63.077],[196,-64.029],[199,-64.9031],[200,-65.0812],[201,-65.0844],[202,-65.3422],[253,-65.3423],[254,-65.2239],[255,-64.8657],[256,-64.1572],[257,-63.2219],[259,-60.5852],[260,-58.9034],[264,-51.4142],[265,-49.6274],[266,-47.9682],[267,-46.4344],[268,-45.019],[270,-42.6287],[271,-41.6456],[272,-40.766],[274,-39.4097],[276,-38.4482],[277,-38.0657],[280,-37.2079],[282,-36.8831],[283,-36.8832],[284,-36.748],[296,-36.748],[297,0],[367,0]]
       }
       ring_02_l_ctrl.Scale.X {type=MovieSceneScriptingFloatChannel, default=1, const [1,368]}
       ring_02_l_ctrl.Scale.Y {type=MovieSceneScriptingFloatChannel, default=1, const [1,368]}
       ring_02_l_ctrl.Scale.Z {type=MovieSceneScriptingFloatChannel, default=1, const [1,368]}
       ring_03_l_ctrl.Location.X {type=MovieSceneScriptingFloatChannel, default=0, thinned_from=368, keys [[0,0],[367,0]]}
       ring_03_l_ctrl.Location.Y {type=MovieSceneScriptingFloatChannel, default=0, thinned_from=368, keys [[0,0],[367,0]]}
       ring_03_l_ctrl.Location.Z {type=MovieSceneScriptingFloatChannel, default=0, thinned_from=368, keys [[0,0],[367,0]]}
       ring_03_l_ctrl.Rotation.X {type=MovieSceneScriptingFloatChannel, default=0, thinned_from=368, keys [[0,0],[367,0]]}
       ring_03_l_ctrl.Rotation.Y {type=MovieSceneScriptingFloatChannel, default=0, thinned_from=368, keys [[0,0],[367,0]]}
       ring_03_l_ctrl.Rotation.Z {
        type=MovieSceneScriptingFloatChannel
        default=0
        thinned_from=368
        keys [[0,-57.4033],[40,-57.4269],[41,-57.2931],[42,-56.8923],[43,-56.3596],[44,-55.696],[46,-53.852],[47,-52.8072],[51,-48.1246],[54,-44.8917],[55,-44.0383],[56,-43.4005],[58,-42.5532],[59,-42.5532],[62,-41.9199],[63,-41.9199],[64,-41.7092],[66,-41.7092],[67,-41.4987],[68,-41.5596],[84,-41.3971],[93,-41.4551],[94,-41.5945],[96,-41.5945],[101,-42.2925],[104,-42.7924],[105,-43.0854],[106,-43.232],[110,-44.4087],[111,-44.5563],[112,-44.8517],[113,-45.2957],[117,-46.4844],[121,-48.3406],[125,-50.7978],[126,-51.2932],[127,-51.9144],[129,-52.9129],[130,-53.2888],[131,-53.7912],[132,-54.169],[133,-54.4213],[134,-54.8006],[135,-55.0538],[136,-55.1193],[138,-55.719],[139,-55.719],[141,-56.3208],[142,-56.3208],[143,-56.6226],[144,-56.6226],[145,-56.9249],[148,-56.9249],[149,-57.2278],[154,-57.2278],[158,-57.3384],[163,-57.449],[171,-57.5597],[173,-57.3275],[174,-56.9414],[175,-56.3254],[176,-55.5586],[177,-54.6431],[181,-50.5826],[183,-48.807],[184,-47.9987],[186,-46.5408],[190,-44.2573],[191,-43.8792],[192,-43.3133],[194,-42.5611],[199,-41.6246],[200,-41.6246],[201,-41.4987],[202,-41.4987],[203,-41.3159],[216,-41.3159],[248,-41.3768],[253,-41.3159],[254,-41.4387],[255,-41.6845],[256,-42.0538],[257,-42.5471],[259,-44.034],[262,-47.0403],[263,-48.1795],[265,-50.1976],[267,-51.9848],[268,-52.7724],[270,-54.1308],[271,-54.7],[273,-55.6147],[274,-55.959],[278,-56.8808],[279,-56.9964],[280,-57.2278],[281,-57.1615],[282,-57.4269],[296,-57.4269],[297,0],[367,0]]
       }
       ring_03_l_ctrl.Scale.X {type=MovieSceneScriptingFloatChannel, default=1, const [1,368]}
       ring_03_l_ctrl.Scale.Y {type=MovieSceneScriptingFloatChannel, default=1, const [1,368]}
       ring_03_l_ctrl.Scale.Z {type=MovieSceneScriptingFloatChannel, default=1, const [1,368]}
       pinky_metacarpal_l_ctrl.Location.X {type=MovieSceneScriptingFloatChannel, default=0, thinned_from=368, keys [[0,0],[367,0]]}
       pinky_metacarpal_l_ctrl.Location.Y {type=MovieSceneScriptingFloatChannel, default=0, thinned_from=368, keys [[0,0],[367,0]]}
       pinky_metacarpal_l_ctrl.Location.Z {type=MovieSceneScriptingFloatChannel, default=0, thinned_from=368, keys [[0,0],[367,0]]}
       pinky_metacarpal_l_ctrl.Rotation.X {type=MovieSceneScriptingFloatChannel, default=0, thinned_from=368, keys [[0,0],[367,0]]}
       pinky_metacarpal_l_ctrl.Rotation.Y {type=MovieSceneScriptingFloatChannel, default=0, thinned_from=368, keys [[0,-9.0736],[296,-9.0736],[297,0],[367,0]]}
       pinky_metacarpal_l_ctrl.Rotation.Z {type=MovieSceneScriptingFloatChannel, default=0, thinned_from=368, keys [[0,0],[367,0]]}
       pinky_metacarpal_l_ctrl.Scale.X {type=MovieSceneScriptingFloatChannel, default=1, const [1,368]}
       pinky_metacarpal_l_ctrl.Scale.Y {type=MovieSceneScriptingFloatChannel, default=1, const [1,368]}
       pinky_metacarpal_l_ctrl.Scale.Z {type=MovieSceneScriptingFloatChannel, default=1, const [1,368]}
       pinky_01_l_ctrl.Location.X {type=MovieSceneScriptingFloatChannel, default=0, thinned_from=368, keys [[0,0],[367,0]]}
       pinky_01_l_ctrl.Location.Y {type=MovieSceneScriptingFloatChannel, default=0, thinned_from=368, keys [[0,0],[367,0]]}
       pinky_01_l_ctrl.Location.Z {type=MovieSceneScriptingFloatChannel, default=0, thinned_from=368, keys [[0,0],[367,0]]}
       pinky_01_l_ctrl.Rotation.X {
        type=MovieSceneScriptingFloatChannel
        default=0
        thinned_from=368
        keys [[0,0],[33,0],[34,-0.001],[39,-0.0009],[40,0.0016],[41,-0.0007],[42,0.0003],[43,0.001],[44,0.0006],[45,-0.0023],[46,0.0006],[47,0],[48,0.002],[49,-0.002],[50,-0.0005],[51,-0.0014],[52,0.0021],[54,0.0017],[55,0.0009],[56,0.0016],[59,0.0002],[60,0.0012],[61,-0.0028],[62,0.0003],[65,-0.0008],[66,0.0003],[67,0.0002],[68,0.0003],[69,0.0015],[70,-0.0012],[71,-0.0012],[72,0.0028],[73,0],[84,0],[85,-0.0008],[86,-0.0008],[87,0.0011],[88,0.0011],[89,-0.0013],[90,0.0005],[91,0.0005],[92,0],[93,0],[94,-0.0005],[95,0.0014],[97,0.0003],[98,0.0005],[99,-0.0013],[100,-0.0024],[101,-0.0003],[102,-0.0013],[104,-0.001],[105,-0.0027],[106,0.0005],[107,0.0025],[108,0.0006],[111,0.0004],[113,-0.0001],[115,-0.0012],[116,-0.0019],[117,-0.0029],[118,0.0024],[119,-0.0022],[120,0.0007],[121,0.0002],[122,0.0021],[123,0.0013],[124,0.002],[125,-0.0002],[128,-0.0005],[129,-0.0014],[130,0.0011],[131,0.0003],[132,0.0004],[133,-0.0006],[136,-0.0002],[139,-0.0007],[140,0.0013],[141,0.0008],[142,-0.0013],[143,0.0006],[144,0.0001],[145,0.0004],[148,-0.0006],[149,-0.0001],[150,0.0019],[151,0.0003],[152,0.0001],[153,-0.0006],[157,-0.0004],[158,0.0003],[159,-0.0005],[161,-0.0004],[162,0.0002],[163,0.0002],[164,-0.0006],[165,0.0002],[166,0.0002],[167,-0.0006],[168,-0.0006],[169,0.0014],[171,0.0009],[172,-0.0014],[173,-0.0013],[174,0.0004],[175,-0.0003],[176,0.0003],[177,0.0003],[178,-0.0004],[179,0.0003],[180,-0.0009],[181,-0.0014],[182,0.0005],[183,0.0006],[184,0.0025],[185,0.0007],[186,-0.004],[187,-0.0029],[188,-0.0052],[189,0.0016],[190,0.0012],[192,-0.0001],[194,-0.0017],[195,0.0004],[197,-0.0009],[198,0.0014],[200,0.0007],[201,0.0011],[202,-0.004],[216,-0.004],[217,0],[248,0],[249,0.0007],[254,0.0007],[255,0.001],[256,0.0004],[257,0.0004],[258,0.0008],[259,0.0008],[260,-0.0019],[261,0.0029],[262,0.0006],[263,0],[264,0.0017],[265,0.0003],[267,0.0012],[268,0.0014],[269,-0.0007],[270,0.0008],[271,0.0007],[272,0],[273,0.0007],[274,0.0003],[275,-0.0007],[276,-0.001],[277,0.001],[278,0.0011],[279,0.0005],[280,0.0004],[283,-0.0012],[296,-0.0012],[297,0],[367,0]]
       }
       pinky_01_l_ctrl.Rotation.Y {
        type=MovieSceneScriptingFloatChannel
        default=0
        thinned_from=368
        keys [[0,0],[33,0],[34,0.0018],[38,0.0018],[39,0.0017],[40,-0.003],[41,0.0005],[43,-0.0002],[44,-0.0008],[45,0.0031],[46,-0.0004],[47,0.0004],[48,-0.0023],[49,0.0031],[50,0.0011],[51,0.0013],[52,-0.0012],[53,-0.0012],[54,-0.001],[55,-0.001],[56,-0.0003],[59,-0.0007],[60,0.0001],[61,0.0022],[62,-0.0002],[65,-0.0006],[66,0.0003],[67,0.0002],[68,-0.0003],[69,-0.0022],[70,0.0019],[71,0.0019],[72,-0.0041],[73,0],[84,0],[85,0.0012],[86,0.0012],[87,-0.0017],[88,-0.0017],[89,0.0019],[90,-0.001],[91,-0.001],[92,-0.0003],[93,-0.0003],[94,0.0004],[95,-0.0025],[97,-0.0012],[98,0.0011],[99,0.0009],[100,0.0023],[101,0.0016],[104,0.0027],[105,0.0018],[106,-0.0024],[107,-0.0008],[108,-0.0016],[117,0.002],[118,-0.0001],[119,0.0032],[120,-0.0007],[121,-0.0017],[122,-0.0013],[123,-0.0024],[124,-0.0025],[125,0.0009],[126,0.0007],[127,0.0009],[128,0.0012],[129,0.0014],[130,-0.0015],[131,-0.0011],[132,-0.0004],[133,-0.0001],[135,0.0015],[136,-0.0003],[139,0.0009],[140,-0.0024],[141,-0.0021],[142,0.0024],[143,-0.0011],[144,-0.0009],[145,-0.0004],[148,0],[149,0.0004],[150,-0.0034],[151,0.0009],[152,0.0009],[153,0.0014],[157,-0.0002],[158,-0.0009],[159,0.0009],[161,0.0001],[163,-0.0005],[164,0.0017],[165,-0.001],[166,-0.001],[167,0.0013],[168,0.0013],[169,-0.0017],[171,-0.0023],[172,0.0015],[173,0.0008],[174,0.0007],[175,-0.0003],[178,-0.0006],[179,0.0004],[180,0.0009],[181,0.002],[182,-0.0011],[183,0.0007],[184,-0.0022],[185,-0.0018],[186,0.0044],[187,0.005],[188,0.0037],[189,-0.001],[192,-0.0012],[194,-0.0016],[195,0],[197,-0.0005],[198,0.0012],[200,0.001],[201,0.0005],[202,0.006],[216,0.006],[217,0],[248,0],[249,-0.001],[254,-0.001],[255,-0.0008],[256,-0.0012],[257,-0.001],[259,0.0002],[260,-0.0005],[261,-0.003],[262,-0.0022],[263,0],[264,-0.0026],[265,-0.0008],[266,-0.0015],[267,-0.0018],[268,-0.0018],[269,0.0014],[270,-0.0016],[272,-0.001],[273,-0.0001],[274,0.0003],[275,0.0004],[276,0.0008],[277,-0.0019],[278,-0.0015],[279,-0.0014],[280,-0.001],[281,0.0016],[296,0.0017],[297,0],[367,0]]
       }
       pinky_01_l_ctrl.Rotation.Z {
        type=MovieSceneScriptingFloatChannel
        default=0
        thinned_from=368
        keys [[0,-13.3153],[33,-13.3153],[38,-13.3341],[39,-13.5531],[40,-14.1384],[41,-15.1644],[42,-16.489],[43,-18.0419],[44,-19.8272],[46,-23.8882],[49,-30.3438],[50,-32.3811],[51,-34.2984],[52,-35.9945],[53,-37.5146],[54,-38.7268],[55,-39.6253],[56,-40.1407],[59,-40.9816],[61,-41.4358],[65,-42.0221],[70,-42.3829],[72,-42.3834],[86,-42.4656],[97,-42.0613],[100,-41.7726],[105,-40.9257],[109,-39.8296],[113,-38.3779],[116,-37.0274],[117,-36.4899],[119,-35.2985],[121,-33.5546],[123,-31.2751],[125,-28.6802],[127,-25.9813],[128,-24.7101],[129,-23.512],[130,-22.386],[131,-21.3963],[132,-20.476],[133,-19.6897],[134,-18.9709],[135,-18.319],[136,-17.7923],[137,-17.3139],[139,-16.6319],[140,-16.1558],[141,-15.748],[143,-15.0695],[145,-14.5278],[146,-14.325],[150,-13.7849],[152,-13.6496],[157,-13.4495],[158,-13.3624],[163,-13.1447],[164,-13.1446],[165,-13.1013],[168,-13.1011],[169,-13.0476],[170,-13.2012],[171,-13.5855],[172,-14.3551],[173,-15.359],[174,-16.599],[175,-18.1564],[176,-19.8796],[177,-21.6928],[178,-23.5985],[179,-25.4385],[180,-27.212],[181,-28.9177],[182,-30.4727],[183,-31.8734],[185,-34.4818],[186,-35.6202],[187,-36.6382],[188,-37.5336],[190,-39.0792],[191,-39.7269],[193,-40.7681],[194,-41.1602],[197,-41.9469],[200,-42.3417],[201,-42.3417],[202,-42.4651],[254,-42.4658],[255,-42.3662],[256,-41.9686],[257,-41.2751],[258,-40.2891],[259,-39.1132],[260,-37.654],[261,-35.9184],[262,-34.1028],[263,-32.2094],[264,-30.2404],[266,-26.5608],[267,-24.8639],[268,-23.2946],[269,-21.9077],[270,-20.586],[271,-19.4417],[272,-18.416],[273,-17.5074],[274,-16.7147],[275,-16.0933],[276,-15.5296],[277,-15.0234],[278,-14.63],[279,-14.2932],[280,-13.9008],[281,-13.6236],[283,-13.3332],[296,-13.3332],[297,0],[367,0]]
       }
       pinky_01_l_ctrl.Scale.X {type=MovieSceneScriptingFloatChannel, default=1, const [1,368]}
       pinky_01_l_ctrl.Scale.Y {type=MovieSceneScriptingFloatChannel, default=1, const [1,368]}
       pinky_01_l_ctrl.Scale.Z {type=MovieSceneScriptingFloatChannel, default=1, const [1,368]}
       pinky_02_l_ctrl.Location.X {type=MovieSceneScriptingFloatChannel, default=0, thinned_from=368, keys [[0,0],[367,0]]}
       pinky_02_l_ctrl.Location.Y {type=MovieSceneScriptingFloatChannel, default=0, thinned_from=368, keys [[0,0],[367,0]]}
       pinky_02_l_ctrl.Location.Z {type=MovieSceneScriptingFloatChannel, default=0, thinned_from=368, keys [[0,0],[367,0]]}
       pinky_02_l_ctrl.Rotation.X {
        type=MovieSceneScriptingFloatChannel
        default=0
        thinned_from=368
        keys [[0,0],[49,0],[50,0.0001],[51,0.0001],[52,-0.0002],[53,0],[56,0.0001],[57,-0.0002],[60,-0.0002],[61,0.0001],[62,0.0001],[63,-0.0002],[64,0.0001],[65,-0.0002],[66,0.0001],[69,0.0001],[70,0.0003],[71,-0.0001],[84,-0.0001],[104,0],[122,0],[123,0.0002],[131,0],[132,-0.0002],[134,0],[137,-0.0001],[138,0],[139,-0.0001],[141,0],[142,-0.0001],[144,0],[145,-0.0001],[149,0.0001],[150,-0.0001],[151,0.0001],[152,-0.0001],[176,0.0001],[178,-0.0001],[179,0.0001],[180,-0.0001],[181,0.0001],[186,-0.0001],[187,0],[188,-0.0001],[189,0],[191,-0.0001],[192,0.0001],[193,-0.0001],[194,0.0002],[195,0],[248,0],[254,-0.0001],[255,0.0003],[256,0.0003],[257,-0.0002],[258,0.0002],[262,0.0002],[263,-0.0001],[266,0.0001],[273,-0.0001],[275,0.0001],[367,0]]
       }
       pinky_02_l_ctrl.Rotation.Y {
        type=MovieSceneScriptingFloatChannel
        default=0
        thinned_from=368
        keys [[0,0],[49,0],[51,0.0001],[52,-0.0002],[56,0],[57,-0.0002],[60,-0.0002],[61,0.0001],[62,0.0001],[63,-0.0001],[64,0.0001],[65,-0.0001],[66,0.0001],[69,0.0001],[70,0.0003],[71,-0.0001],[122,0],[123,-0.0001],[131,0],[132,0.0001],[137,-0.0001],[176,0.0001],[178,-0.0001],[179,0.0001],[180,-0.0001],[181,0.0001],[186,-0.0001],[187,0],[188,-0.0001],[189,0],[191,-0.0001],[192,0.0001],[254,-0.0001],[255,0.0003],[256,0.0002],[257,-0.0001],[258,0.0002],[262,0.0002],[263,-0.0001],[268,0.0001],[273,-0.0001],[367,0]]
       }
       pinky_02_l_ctrl.Rotation.Z {
        type=MovieSceneScriptingFloatChannel
        default=0
        thinned_from=368
        keys [[0,-32.7685],[38,-32.7799],[39,-32.8599],[40,-33.2606],[41,-33.7427],[42,-34.4684],[43,-35.3596],[44,-36.3372],[45,-37.4028],[48,-40.895],[49,-41.9922],[50,-43.1829],[51,-44.2197],[53,-45.9212],[54,-46.6364],[55,-47.0672],[56,-47.3553],[62,-48.223],[63,-48.223],[64,-48.3682],[65,-48.3682],[66,-48.5135],[69,-48.5135],[70,-48.6462],[99,-48.3212],[104,-47.9151],[109,-47.213],[111,-46.8633],[116,-45.6711],[118,-45.0788],[120,-44.3302],[123,-42.5939],[126,-40.3593],[129,-38.3027],[131,-37.1394],[132,-36.633],[133,-36.2002],[135,-35.4815],[137,-34.9229],[138,-34.7707],[142,-33.8602],[145,-33.4067],[150,-33.0298],[151,-33.0298],[154,-32.8878],[168,-32.6316],[170,-32.7497],[171,-32.9192],[172,-33.3437],[173,-33.8545],[174,-34.6235],[175,-35.4819],[177,-37.3863],[179,-39.4897],[180,-40.3745],[181,-41.3538],[183,-42.9701],[184,-43.6942],[187,-45.4425],[189,-46.4109],[192,-47.456],[194,-47.9465],[197,-48.3684],[255,-48.6462],[256,-48.4259],[258,-47.5483],[259,-46.8939],[261,-45.1638],[265,-40.9668],[267,-39.0488],[269,-37.4052],[271,-36.0898],[273,-35.0326],[275,-34.2903],[277,-33.6741],[281,-32.939],[282,-32.8365],[296,-32.7682],[297,0],[367,0]]
       }
       pinky_02_l_ctrl.Scale.X {type=MovieSceneScriptingFloatChannel, default=1, const [1,368]}
       pinky_02_l_ctrl.Scale.Y {type=MovieSceneScriptingFloatChannel, default=1, const [1,368]}
       pinky_02_l_ctrl.Scale.Z {type=MovieSceneScriptingFloatChannel, default=1, const [1,368]}
       pinky_03_l_ctrl.Location.X {type=MovieSceneScriptingFloatChannel, default=0, thinned_from=368, keys [[0,0],[367,0]]}
       pinky_03_l_ctrl.Location.Y {type=MovieSceneScriptingFloatChannel, default=0, thinned_from=368, keys [[0,0],[367,0]]}
       pinky_03_l_ctrl.Location.Z {type=MovieSceneScriptingFloatChannel, default=0, thinned_from=368, keys [[0,0],[367,0]]}
       pinky_03_l_ctrl.Rotation.X {
        type=MovieSceneScriptingFloatChannel
        default=0
        thinned_from=368
        keys [[0,-0.2527],[33,-0.2527],[34,-0.2383],[38,-0.2383],[39,-0.3375],[40,-0.2594],[41,-0.3128],[42,-0.36],[43,-0.4271],[44,-0.3876],[45,-0.4458],[46,-0.5926],[47,-0.5636],[48,-0.7066],[49,-0.6552],[50,-0.7859],[51,-0.8369],[52,-0.7171],[53,-0.9452],[55,-1.0082],[56,-1.0082],[57,-0.982],[58,-1.0369],[62,-1.0369],[63,-1.0098],[64,-1.0636],[67,-1.0636],[68,-1.0555],[84,-1.0555],[85,-1.0583],[88,-1.0583],[89,-1.0613],[90,-1.078],[91,-1.0544],[92,-1.0544],[93,-1.0573],[94,-1.0573],[95,-1.0504],[96,-1.0533],[97,-1.0464],[98,-1.0492],[101,-1.037],[102,-1.0399],[103,-1.0306],[104,-1.0306],[105,-1.0212],[106,-1.0212],[108,-1.0016],[109,-0.9831],[110,-0.9916],[113,-0.9601],[114,-0.885],[115,-0.9269],[117,-0.9037],[118,-0.8797],[119,-0.8576],[120,-0.8289],[121,-0.8305],[122,-0.8102],[123,-0.6985],[124,-0.7425],[125,-0.7255],[126,-0.6072],[127,-0.6523],[128,-0.5294],[129,-0.4954],[130,-0.4598],[131,-0.4226],[132,-0.4752],[133,-0.3638],[134,-0.4089],[135,-0.3932],[136,-0.4141],[137,-0.3631],[139,-0.3476],[140,-0.3317],[141,-0.3129],[142,-0.3069],[144,-0.2899],[145,-0.279],[146,-0.2812],[147,-0.2724],[148,-0.3079],[149,-0.2635],[150,-0.2635],[151,-0.299],[152,-0.2524],[153,-0.2731],[157,-0.2731],[158,-0.1761],[159,-0.2427],[160,-0.2469],[170,-0.2469],[172,-0.2991],[173,-0.2413],[174,-0.2894],[175,-0.4426],[176,-0.3998],[177,-0.4593],[178,-0.5135],[179,-0.5625],[180,-0.6978],[181,-0.6323],[182,-0.7609],[183,-0.7818],[184,-0.8003],[185,-0.8363],[187,-0.902],[188,-0.9381],[189,-0.9547],[190,-0.9816],[191,-0.9896],[193,-1.0159],[194,-1.0345],[195,-1.0346],[196,-1.0406],[197,-1.0527],[198,-1.0586],[199,-1.0524],[200,-1.0644],[201,-1.0597],[216,-1.0597],[248,-1.0615],[249,-1.0652],[254,-1.0652],[255,-1.0481],[256,-1.0404],[257,-1.0418],[258,-0.959],[259,-0.932],[260,-0.8928],[261,-0.8501],[262,-0.8587],[263,-0.7953],[264,-0.7266],[265,-0.6891],[266,-0.6495],[267,-0.6046],[268,-0.5801],[269,-0.5273],[270,-0.4989],[271,-0.4693],[272,-0.4384],[273,-0.4062],[274,-0.3728],[275,-0.3381],[276,-0.3381],[277,-0.3022],[278,-0.3022],[279,-0.141],[280,-0.2651],[281,-0.2606],[296,-0.2606],[297,0],[367,0]]
       }
       pinky_03_l_ctrl.Rotation.Y {
        type=MovieSceneScriptingFloatChannel
        default=0
        thinned_from=368
        keys [[0,-2.5496],[33,-2.5496],[34,-2.5319],[38,-2.5319],[39,-2.6149],[40,-2.7042],[42,-3.4],[43,-3.9219],[44,-4.3586],[45,-4.8813],[50,-7.944],[52,-8.8825],[53,-9.3814],[55,-10.028],[57,-10.0277],[58,-10.3518],[63,-10.3516],[64,-10.6759],[67,-10.6759],[68,-10.648],[84,-10.648],[85,-10.7014],[89,-10.7014],[91,-10.6617],[94,-10.6617],[95,-10.622],[96,-10.622],[97,-10.5823],[98,-10.5824],[101,-10.4634],[102,-10.4634],[103,-10.3557],[104,-10.3557],[105,-10.248],[106,-10.248],[109,-9.9252],[110,-9.9253],[113,-9.6028],[115,-9.2807],[117,-9.0661],[118,-8.8517],[119,-8.7377],[120,-8.4261],[121,-8.2708],[122,-7.9598],[124,-7.183],[125,-6.8732],[128,-5.7115],[131,-4.7856],[134,-4.0916],[135,-3.9385],[136,-3.7992],[137,-3.6801],[139,-3.5181],[141,-3.1935],[145,-2.8697],[146,-2.8703],[147,-2.7894],[148,-2.7459],[149,-2.7085],[150,-2.7085],[151,-2.665],[152,-2.627],[153,-2.6432],[157,-2.6432],[159,-2.4826],[170,-2.4839],[172,-2.8484],[173,-3.1245],[174,-3.4894],[176,-4.4031],[179,-6.0526],[182,-7.4325],[184,-8.1704],[185,-8.4784],[188,-9.3399],[190,-9.7715],[194,-10.3478],[198,-10.6363],[199,-10.6363],[200,-10.7085],[216,-10.6903],[248,-10.7085],[249,-10.741],[254,-10.741],[257,-10.4032],[258,-10.117],[259,-9.7799],[261,-8.8824],[262,-8.3827],[264,-7.2653],[265,-6.8138],[267,-5.7107],[268,-5.4354],[269,-4.8852],[275,-3.2388],[276,-3.2388],[277,-2.965],[278,-2.965],[279,-2.8328],[280,-2.6914],[281,-2.6289],[296,-2.6289],[297,0],[367,0]]
       }
       pinky_03_l_ctrl.Rotation.Z {
        type=MovieSceneScriptingFloatChannel
        default=0
        thinned_from=368
        keys [[0,-61.7194],[38,-61.6857],[41,-62.1706],[42,-62.4956],[45,-63.9772],[46,-64.6404],[48,-65.6651],[49,-66.353],[51,-67.3197],[52,-67.9324],[55,-68.8664],[56,-68.8664],[57,-69.1584],[62,-69.1811],[63,-69.4741],[102,-69.2988],[114,-68.3736],[120,-67.4752],[124,-66.2517],[125,-65.8066],[126,-65.5063],[127,-65.0593],[132,-63.5908],[134,-63.2963],[135,-63.0124],[144,-62.0858],[159,-61.7987],[160,-61.6275],[170,-61.6275],[172,-61.9689],[175,-63.0032],[179,-65.12],[180,-65.4803],[181,-66.0237],[187,-67.994],[193,-69.1035],[255,-69.5557],[258,-69.0192],[261,-67.8407],[267,-64.7015],[268,-64.4355],[269,-63.9066],[275,-62.3448],[276,-62.3448],[277,-62.0879],[279,-62.0866],[280,-61.8319],[296,-61.7957],[297,0],[367,0]]
       }
       pinky_03_l_ctrl.Scale.X {type=MovieSceneScriptingFloatChannel, default=1, const [1,368]}
       pinky_03_l_ctrl.Scale.Y {type=MovieSceneScriptingFloatChannel, default=1, const [1,368]}
       pinky_03_l_ctrl.Scale.Z {type=MovieSceneScriptingFloatChannel, default=1, const [1,368]}
       hand_l_ik_ctrl.Location.X {
        type=MovieSceneScriptingFloatChannel
        default=0
        thinned_from=373
        keys [[-109,-17.1926],[-73.0013,-17.1926],[-73,-31.5424],[-56,-31.542],[-16,-34.4298],[0,-31.542],[50,-31.7758],[68,-30.6868],[99,-29.7912],[109,-30.2496],[124,-31.4873],[174,-31.9083],[178,-31.6401],[182,-31.0404],[190,-29.2627],[196,-27.4554],[198,-27.014],[200,-26.747],[203,-26.6258],[205,-26.6573],[208,-26.9144],[229,-30.6899],[233,-31.1744],[256,-32.343],[296,-31.5454],[297,12.9005],[367,12.9011]]
       }
       hand_l_ik_ctrl.Location.Y {
        type=MovieSceneScriptingFloatChannel
        default=0
        thinned_from=373
        keys [[-109,27.6769],[-73.0013,27.6769],[-73,-7.4392],[-56,-7.4394],[-16,-15.6568],[0,-7.4394],[40,-7.3708],[48,-7.4685],[56,-7.6193],[61,-7.5762],[68,-7.4501],[72,-7.3596],[76,-7.3439],[79,-7.4027],[82,-7.5062],[92,-7.9745],[99,-8.1818],[103,-8.1861],[107,-8.1158],[122,-7.6325],[156,-7.4539],[175,-7.4602],[181,-7.3335],[183,-7.3326],[184,-7.3579],[185,-7.4001],[186,-7.4617],[188,-7.6424],[191,-7.934],[192,-8.0134],[194,-8.1196],[196,-8.2043],[197,-8.286],[198,-8.3879],[200,-8.6465],[201,-8.7993],[202,-8.9716],[207,-9.9947],[209,-10.3467],[211,-10.6151],[212,-10.7181],[214,-10.8584],[215,-10.8972],[217,-10.9136],[218,-10.8915],[220,-10.7792],[225,-10.2697],[236,-9.3138],[241,-8.9772],[261,-8.0452],[269,-7.7648],[278,-7.5487],[286,-7.461],[296,-7.4422],[297,-9.8495],[367,-9.8495]]
       }
       hand_l_ik_ctrl.Location.Z {
        type=MovieSceneScriptingFloatChannel
        default=0
        thinned_from=373
        keys [[-109,33.3577],[-73.0013,33.3577],[-73,6.8811],[-56,6.8809],[-16,6.2467],[0,6.8809],[33,6.7279],[40,6.6473],[45,6.5455],[47,6.4718],[49,6.3515],[53,6.0154],[54,5.9497],[55,5.9011],[56,5.8748],[57,5.8645],[58,5.8684],[59,5.8883],[61,5.9624],[67,6.2889],[68,6.3523],[70,6.464],[72,6.6046],[75,6.8615],[84,7.7027],[87,7.9299],[95,8.3359],[98,8.4386],[100,8.4667],[103,8.4425],[107,8.3451],[109,8.2675],[111,8.1522],[114,7.9146],[120,7.3536],[122,7.1915],[124,7.0625],[127,6.9403],[129,6.8866],[140,6.8006],[173,6.8427],[174,6.8181],[175,6.7517],[176,6.6449],[177,6.5049],[179,6.1892],[180,6.0566],[181,5.9563],[182,5.8779],[183,5.8383],[184,5.8118],[187,5.7991],[188,5.8094],[189,5.8412],[190,5.9064],[191,6.0133],[193,6.3109],[195,6.6284],[196,6.7654],[197,6.8732],[198,6.9555],[199,7.0072],[200,7.0438],[202,7.0877],[203,7.1293],[211,7.6539],[214,7.7934],[216,7.8452],[218,7.8561],[220,7.8291],[222,7.7445],[227,7.4052],[230,7.157],[233,6.8876],[234,6.8088],[237,6.6438],[239,6.5594],[241,6.4154],[247,5.8378],[250,5.576],[252,5.4341],[253,5.3793],[255,5.3315],[257,5.3757],[259,5.4654],[262,5.6515],[267,5.9917],[269,6.1097],[273,6.2893],[274,6.3264],[275,6.3832],[277,6.4507],[278,6.5017],[279,6.5321],[280,6.5764],[286,6.7581],[287,6.7718],[288,6.8001],[293,6.8708],[296,6.8778],[297,20.7838],[367,20.7839]]
       }
       hand_l_ik_ctrl.Rotation.X {
        type=MovieSceneScriptingFloatChannel
        default=0
        thinned_from=373
        keys [[-109,89.4928],[-73.0013,89.4928],[-73,147.8746],[110,145.0512],[180,148.2307],[192,144.9865],[198,142.3705],[202,141.8311],[227,150.0163],[296,147.8881],[297,54.733],[367,54.7302]]
       }
       hand_l_ik_ctrl.Rotation.Y {
        type=MovieSceneScriptingFloatChannel
        default=0
        thinned_from=373
        keys [[-109,17.3705],[-73.0013,17.3705],[-73,-11.3213],[40,-11.3],[45,-11.1803],[47,-11.0874],[51,-10.7809],[53,-10.5906],[55,-10.3442],[63,-9.0794],[66,-8.6783],[68,-8.441],[72,-8.0882],[84,-6.7925],[86,-6.6064],[88,-6.4503],[93,-6.1411],[96,-5.9897],[98,-5.918],[99,-5.9009],[100,-5.9159],[101,-5.9638],[102,-6.0387],[103,-6.1361],[104,-6.2603],[105,-6.3984],[106,-6.5587],[107,-6.7352],[108,-6.9277],[109,-7.1391],[111,-7.6102],[115,-8.6852],[118,-9.5052],[120,-10.0048],[122,-10.441],[123,-10.6269],[124,-10.7897],[125,-10.928],[126,-11.0419],[129,-11.2924],[137,-11.7112],[155,-12.1802],[174,-12.3313],[176,-12.4049],[178,-12.5355],[179,-12.5781],[180,-12.5762],[181,-12.5096],[182,-12.3641],[183,-12.138],[184,-11.8357],[185,-11.4622],[186,-11.0308],[187,-10.5492],[188,-10.0188],[189,-9.4447],[190,-8.8328],[192,-7.5584],[193,-6.9357],[194,-6.3428],[195,-5.8024],[196,-5.3194],[197,-4.8992],[198,-4.5268],[199,-4.2115],[200,-3.9424],[201,-3.723],[202,-3.5586],[203,-3.4441],[204,-3.3807],[205,-3.377],[206,-3.4285],[207,-3.5419],[208,-3.7136],[209,-3.9466],[210,-4.2282],[211,-4.5532],[212,-4.9245],[213,-5.3197],[214,-5.7434],[217,-7.0823],[219,-7.9881],[221,-8.8534],[222,-9.2678],[223,-9.6628],[224,-10.0381],[225,-10.3923],[226,-10.723],[227,-11.0292],[228,-11.3078],[229,-11.5632],[230,-11.7938],[231,-11.9994],[232,-12.1617],[233,-12.2847],[234,-12.3621],[235,-12.3836],[236,-12.3576],[237,-12.2897],[242,-11.8398],[245,-11.6304],[248,-11.4848],[256,-11.2905],[264,-11.0133],[268,-10.974],[277,-11.1415],[278,-11.1431],[279,-11.1759],[290,-11.3122],[296,-11.3275],[297,41.5692],[367,41.5696]]
       }
       hand_l_ik_ctrl.Rotation.Z {
        type=MovieSceneScriptingFloatChannel
        default=0
        thinned_from=373
        keys [[-109,69.5729],[-73.0013,69.5729],[-73,123.8818],[296,123.8858],[297,73.0535],[367,73.052]]
       }
       hand_l_ik_ctrl.Scale.X {type=MovieSceneScriptingFloatChannel, default=1, const [1,373]}
       hand_l_ik_ctrl.Scale.Y {type=MovieSceneScriptingFloatChannel, default=1, const [1,373]}
       hand_l_ik_ctrl.Scale.Z {type=MovieSceneScriptingFloatChannel, default=1, const [1,373]}
       Softness {type=MovieSceneScriptingFloatChannel, default=0, const [0,368]}
       hand_l_ikorient_ctrl.Rotation.Roll {type=MovieSceneScriptingFloatChannel, default=0, const [0,368]}
       hand_l_ikorient_ctrl.Rotation.Pitch {type=MovieSceneScriptingFloatChannel, default=0, const [0,368]}
       hand_l_ikorient_ctrl.Rotation.Yaw {type=MovieSceneScriptingFloatChannel, default=0, const [0,368]}
       arm_l_pv_ik_ctrl.Location.X {
        type=MovieSceneScriptingFloatChannel
        default=0
        thinned_from=371
        keys [[-109,-4.4839],[-16,-4.4894],[1,-4.4603],[7,-4.5694],[27,-5.1478],[34,-5.2758],[42,-5.3866],[43,-5.4105],[44,-5.4688],[45,-5.557],[46,-5.6873],[47,-5.8398],[49,-6.171],[50,-6.3272],[51,-6.4636],[52,-6.5812],[53,-6.6702],[54,-6.7242],[55,-6.7529],[56,-6.7576],[58,-6.7228],[68,-6.3791],[69,-6.3707],[71,-6.3189],[72,-6.2818],[74,-6.1737],[76,-6.027],[77,-5.9385],[79,-5.7176],[82,-5.3299],[87,-4.6373],[96,-3.4781],[98,-3.2331],[99,-3.1207],[100,-3.0333],[101,-2.9613],[102,-2.9001],[103,-2.8484],[104,-2.8121],[105,-2.7844],[107,-2.7477],[109,-2.7536],[110,-2.7811],[111,-2.8176],[112,-2.8694],[113,-2.9359],[114,-3.0129],[116,-3.1992],[120,-3.6253],[122,-3.8144],[124,-3.961],[125,-4.0141],[126,-4.0553],[129,-4.141],[131,-4.1745],[140,-4.262],[158,-4.3329],[167,-4.3313],[169,-4.3173],[170,-4.2913],[171,-4.2537],[174,-4.0662],[175,-4.0303],[176,-4.0287],[177,-4.0618],[179,-4.1797],[180,-4.229],[181,-4.2583],[182,-4.2671],[183,-4.2393],[184,-4.1927],[186,-4.0378],[187,-3.9462],[188,-3.8422],[189,-3.7241],[190,-3.5872],[191,-3.4177],[194,-2.8363],[195,-2.6612],[196,-2.5123],[197,-2.3923],[199,-2.1644],[200,-2.0358],[201,-1.8923],[202,-1.7057],[203,-1.4712],[204,-1.1955],[205,-0.9096],[206,-0.6348],[207,-0.3666],[208,-0.1204],[209,0.1005],[210,0.285],[211,0.4397],[212,0.5615],[213,0.6469],[214,0.7042],[215,0.7273],[216,0.7228],[217,0.6873],[218,0.6188],[219,0.5241],[220,0.4115],[221,0.2631],[222,0.0949],[223,-0.0953],[225,-0.4808],[226,-0.6766],[227,-0.8814],[228,-1.1019],[229,-1.3253],[231,-1.7586],[232,-1.9581],[233,-2.1374],[234,-2.2852],[235,-2.3933],[236,-2.4741],[237,-2.5301],[238,-2.5757],[239,-2.6326],[240,-2.7215],[241,-2.8324],[242,-2.9585],[243,-3.102],[245,-3.4253],[249,-4.1159],[250,-4.2987],[252,-4.6372],[253,-4.7992],[255,-5.053],[256,-5.1393],[257,-5.2085],[259,-5.3068],[261,-5.3532],[262,-5.3658],[264,-5.3491],[268,-5.2421],[271,-5.1347],[274,-5.0388],[275,-4.9933],[286,-4.6102],[291,-4.4976],[294,-4.4557],[296,-4.4521],[297,0.0004],[298,0],[367,0]]
       }
       arm_l_pv_ik_ctrl.Location.Y {
        type=MovieSceneScriptingFloatChannel
        default=0
        thinned_from=371
        keys [[-109,31.5223],[-16,31.5166],[43,31.7257],[45,31.8431],[49,32.2911],[52,32.4004],[61,32.1624],[81,32.7865],[95,32.2637],[106,31.5827],[170,31.3489],[173,31.1219],[175,30.8149],[182,29.2772],[195,27.095],[201,26.4491],[204,26.3253],[206,26.406],[208,26.6406],[211,27.2303],[221,29.6729],[226,30.5524],[229,30.8626],[233,31.0364],[248,30.7328],[296,31.5523],[297,0.0004],[298,0],[367,0]]
       }
       arm_l_pv_ik_ctrl.Location.Z {
        type=MovieSceneScriptingFloatChannel
        default=0
        thinned_from=371
        keys [[-109,-6.0384],[-16,-6.0417],[1,-6.0223],[4,-5.9865],[27,-5.6071],[40,-5.4938],[44,-5.4109],[46,-5.3409],[48,-5.2488],[56,-4.8263],[60,-4.6604],[62,-4.5977],[68,-4.4879],[70,-4.4795],[72,-4.4578],[75,-4.3915],[76,-4.3619],[80,-4.204],[88,-3.8379],[91,-3.7254],[93,-3.6648],[96,-3.5976],[99,-3.5548],[100,-3.5641],[102,-3.6159],[103,-3.6597],[105,-3.78],[106,-3.8514],[107,-3.9333],[109,-4.1219],[111,-4.3384],[118,-5.1804],[120,-5.3973],[122,-5.5863],[124,-5.7357],[126,-5.8433],[131,-6.0055],[147,-6.2692],[168,-6.3849],[172,-6.3472],[174,-6.291],[176,-6.1921],[177,-6.1274],[178,-6.0453],[179,-5.9345],[180,-5.7933],[181,-5.6248],[182,-5.4209],[183,-5.1926],[184,-4.933],[185,-4.6544],[186,-4.3552],[187,-4.0354],[188,-3.6994],[190,-3.0043],[191,-2.6715],[192,-2.3603],[193,-2.0851],[194,-1.8443],[195,-1.6354],[196,-1.4442],[197,-1.2477],[198,-1.0719],[199,-0.919],[200,-0.7968],[201,-0.7036],[202,-0.6457],[203,-0.6226],[204,-0.6391],[205,-0.6943],[206,-0.7844],[207,-0.9166],[208,-1.0862],[209,-1.2927],[210,-1.5324],[211,-1.7992],[212,-2.0918],[213,-2.4055],[214,-2.7375],[215,-3.0828],[217,-3.7999],[218,-4.1648],[219,-4.5432],[221,-5.2798],[222,-5.6395],[223,-5.9807],[224,-6.3018],[225,-6.583],[226,-6.8357],[227,-7.0585],[228,-7.2278],[229,-7.3685],[230,-7.4828],[231,-7.5662],[232,-7.6234],[233,-7.6493],[234,-7.6418],[235,-7.6048],[236,-7.5393],[238,-7.3434],[245,-6.4667],[248,-6.1258],[250,-5.9388],[252,-5.781],[256,-5.5328],[259,-5.3887],[261,-5.3252],[262,-5.3021],[264,-5.2886],[266,-5.3151],[270,-5.4201],[282,-5.8328],[288,-5.9632],[295,-6.0254],[296,-6.0267],[297,0.0001],[367,0]]
       }
       arm_l_pv_ik_ctrl.Rotation.X {
        type=MovieSceneScriptingFloatChannel
        default=0
        thinned_from=371
        keys [[-109,-22.9497],[-16,-22.9562],[32,-24.1669],[43,-24.1191],[44,-24.2185],[45,-24.4151],[46,-24.7027],[49,-25.7548],[51,-26.3097],[53,-26.6559],[55,-26.7946],[68,-26.6239],[74,-26.6038],[77,-26.4484],[90,-24.677],[101,-22.5405],[104,-22.1075],[108,-21.7248],[111,-21.6045],[114,-21.6393],[123,-22.2397],[168,-22.5243],[169,-22.4939],[170,-22.4142],[171,-22.2617],[172,-22.0504],[176,-20.9036],[178,-20.4241],[183,-19.5811],[186,-18.9476],[188,-18.4448],[193,-17.0087],[194,-16.7557],[195,-16.5455],[196,-16.39],[198,-16.2011],[202,-15.9533],[204,-15.9291],[205,-15.9689],[206,-16.0548],[207,-16.1903],[208,-16.374],[209,-16.6051],[210,-16.8794],[212,-17.5339],[220,-20.5858],[222,-21.2661],[224,-21.8594],[227,-22.575],[229,-22.9652],[231,-23.251],[232,-23.3423],[233,-23.3863],[234,-23.3723],[235,-23.2978],[237,-22.9908],[240,-22.4675],[242,-22.2371],[245,-22.0703],[248,-22.0751],[263,-23.0424],[296,-23.2783],[297,0.0003],[298,0],[367,0]]
       }
       arm_l_pv_ik_ctrl.Rotation.Y {
        type=MovieSceneScriptingFloatChannel
        default=0
        thinned_from=371
        keys [[-109,-3.0207],[-56,-3.0207],[-16,-3.0297],[0,-2.7895],[1,-2.7895],[10,-2.8912],[36,-3.0067],[38,-3.0044],[40,-2.9816],[41,-2.9618],[42,-2.9346],[43,-2.885],[44,-2.7877],[45,-2.6394],[46,-2.4431],[47,-2.2061],[49,-1.7048],[50,-1.4754],[51,-1.2675],[52,-1.0804],[53,-0.9063],[54,-0.7412],[55,-0.5809],[56,-0.427],[57,-0.2802],[58,-0.1287],[59,0.0253],[61,0.334],[62,0.4873],[63,0.6395],[64,0.7848],[65,0.924],[68,1.3251],[69,1.4282],[70,1.5398],[71,1.6565],[72,1.78],[75,2.1726],[78,2.5819],[81,2.9744],[82,3.0998],[84,3.3359],[85,3.4431],[86,3.5431],[88,3.7186],[89,3.7926],[90,3.856],[91,3.9052],[92,3.9418],[93,3.965],[94,3.9748],[95,3.9743],[96,3.9578],[97,3.9305],[98,3.8938],[99,3.8337],[100,3.7421],[101,3.6246],[102,3.4799],[103,3.3113],[104,3.1226],[105,2.9151],[106,2.6939],[107,2.4627],[108,2.2149],[109,1.9518],[110,1.6762],[111,1.3842],[112,1.0806],[113,0.7706],[114,0.4564],[115,0.1435],[116,-0.169],[117,-0.4695],[118,-0.7636],[119,-1.0473],[120,-1.3204],[121,-1.5764],[122,-1.8172],[123,-2.0333],[124,-2.2249],[125,-2.392],[126,-2.5327],[127,-2.6478],[128,-2.7528],[129,-2.8506],[130,-2.9375],[131,-3.018],[133,-3.157],[135,-3.2782],[137,-3.3887],[140,-3.5286],[143,-3.648],[147,-3.7819],[152,-3.9077],[158,-4.0189],[164,-4.0884],[168,-4.1228],[169,-4.1405],[170,-4.1783],[171,-4.2463],[172,-4.3362],[173,-4.4422],[174,-4.5615],[175,-4.6982],[177,-4.9916],[178,-5.1118],[179,-5.1755],[180,-5.1536],[181,-5.0343],[182,-4.813],[183,-4.4967],[184,-4.0928],[185,-3.6135],[186,-3.0751],[187,-2.4846],[188,-1.8514],[189,-1.1803],[190,-0.478],[191,0.2311],[192,0.9282],[193,1.5953],[194,2.2208],[195,2.7808],[196,3.2706],[197,3.6835],[198,4.0291],[199,4.3015],[200,4.5018],[201,4.6394],[202,4.7156],[203,4.7384],[204,4.7082],[205,4.6335],[206,4.5169],[207,4.357],[208,4.1599],[209,3.9257],[210,3.6549],[211,3.3541],[212,3.0219],[213,2.666],[214,2.2897],[215,1.8963],[216,1.4918],[217,1.0776],[218,0.6554],[219,0.239],[220,-0.173],[221,-0.5757],[222,-0.9612],[223,-1.3326],[224,-1.6835],[225,-2.0155],[226,-2.3259],[228,-2.9372],[229,-3.2201],[230,-3.4758],[231,-3.6905],[232,-3.8497],[233,-3.9558],[234,-4.0066],[235,-3.9948],[236,-3.937],[237,-3.8389],[238,-3.7124],[241,-3.3058],[242,-3.176],[243,-3.0558],[245,-2.8385],[246,-2.7514],[247,-2.6707],[248,-2.6022],[249,-2.5474],[250,-2.506],[251,-2.4755],[253,-2.4324],[255,-2.377],[256,-2.3434],[258,-2.2678],[260,-2.1864],[262,-2.1205],[264,-2.0749],[266,-2.063],[267,-2.0692],[268,-2.0799],[269,-2.1046],[270,-2.1353],[276,-2.3705],[280,-2.5129],[282,-2.575],[286,-2.6725],[290,-2.7416],[293,-2.7713],[296,-2.7827],[297,-0.0001],[367,0]]
       }
       arm_l_pv_ik_ctrl.Rotation.Z {
        type=MovieSceneScriptingFloatChannel
        default=0
        thinned_from=371
        keys [[-109,57.4854],[-16,57.4767],[43,57.089],[46,57.4102],[52,58.4514],[78,58.5907],[110,55.8675],[170,56.1089],[179,55.1182],[205,53.2185],[210,53.4594],[229,56.1625],[234,56.4171],[248,56.0919],[296,56.6291],[297,0.0011],[298,0],[367,0]]
       }
       arm_l_pv_ik_ctrl.Scale.X {type=MovieSceneScriptingFloatChannel, default=1, const [1,371]}
       arm_l_pv_ik_ctrl.Scale.Y {type=MovieSceneScriptingFloatChannel, default=1, const [1,371]}
       arm_l_pv_ik_ctrl.Scale.Z {type=MovieSceneScriptingFloatChannel, default=1, const [1,371]}
       upperarm_r_fk_ctrl.Location.X {type=MovieSceneScriptingFloatChannel, default=0, thinned_from=368, keys [[0,0],[367,0]]}
       upperarm_r_fk_ctrl.Location.Y {type=MovieSceneScriptingFloatChannel, default=0, thinned_from=368, keys [[0,0],[367,0]]}
       upperarm_r_fk_ctrl.Location.Z {type=MovieSceneScriptingFloatChannel, default=0, thinned_from=368, keys [[0,0],[367,0]]}
       upperarm_r_fk_ctrl.Rotation.X {
        type=MovieSceneScriptingFloatChannel
        default=0
        thinned_from=368
        keys [[0,5.6558],[1,5.6558],[3,5.6998],[7,5.7197],[9,5.7089],[13,5.6326],[24,5.3082],[29,5.2052],[34,5.1631],[45,5.1691],[46,5.1559],[47,5.1093],[48,5.019],[49,4.8815],[50,4.7042],[51,4.495],[52,4.2461],[53,3.9755],[54,3.6856],[55,3.3833],[56,3.0691],[58,2.4282],[59,2.1199],[60,1.8218],[61,1.5381],[62,1.2813],[63,1.0422],[64,0.8273],[65,0.6321],[66,0.4577],[67,0.3116],[68,0.1856],[69,0.1219],[70,0.0726],[71,0.0466],[72,0.0325],[73,0.0311],[74,0.0391],[75,0.0539],[76,0.072],[77,0.1015],[78,0.1301],[79,0.1634],[80,0.2004],[82,0.265],[83,0.2931],[84,0.3135],[85,0.3295],[86,0.3414],[87,0.343],[88,0.3341],[89,0.3205],[90,0.3043],[91,0.2823],[92,0.2542],[93,0.2272],[94,0.1963],[95,0.1615],[96,0.1281],[97,0.0918],[98,0.0557],[99,0.0374],[100,0.0434],[101,0.0763],[102,0.1291],[103,0.2106],[104,0.3088],[105,0.4339],[106,0.5745],[107,0.734],[108,0.9138],[109,1.1046],[110,1.3118],[111,1.5323],[112,1.7643],[113,2.0025],[114,2.2469],[117,2.9917],[119,3.4748],[120,3.7066],[121,3.9281],[122,4.1368],[123,4.3324],[124,4.5126],[125,4.6711],[126,4.8076],[130,5.288],[134,5.7091],[138,6.0745],[143,6.4483],[146,6.6174],[149,6.7478],[152,6.8258],[156,6.8738],[161,6.864],[168,6.7385],[171,6.6493],[173,6.5706],[174,6.4952],[175,6.3629],[176,6.1837],[177,5.9628],[178,5.6927],[179,5.382],[180,5.0333],[181,4.6522],[182,4.2427],[183,3.8131],[186,2.4763],[187,2.0357],[188,1.6086],[189,1.1977],[190,0.804],[191,0.4325],[192,0.0755],[193,-0.2624],[194,-0.5765],[195,-0.8666],[196,-1.1415],[197,-1.4083],[198,-1.6786],[199,-1.9598],[200,-2.2331],[201,-2.4581],[202,-2.5976],[203,-2.6388],[204,-2.5843],[205,-2.4434],[206,-2.2206],[207,-1.926],[208,-1.5734],[209,-1.1711],[210,-0.7296],[211,-0.2531],[212,0.2417],[213,0.7551],[216,2.31],[217,2.8186],[218,3.3155],[219,3.8018],[220,4.2701],[221,4.7223],[222,5.1583],[223,5.5722],[224,5.963],[225,6.3325],[226,6.6776],[227,6.9912],[228,7.2719],[229,7.519],[230,7.7352],[231,7.9184],[232,8.06],[233,8.1489],[234,8.1803],[235,8.144],[236,8.0433],[237,7.8894],[241,7.1423],[244,6.6283],[247,6.1554],[250,5.7278],[252,5.4801],[254,5.2614],[256,5.0633],[258,4.8901],[260,4.7482],[262,4.6486],[264,4.587],[265,4.5705],[266,4.5637],[268,4.5732],[270,4.6006],[272,4.6452],[275,4.7427],[281,5.0056],[292,5.5141],[296,5.6599],[297,0],[367,0]]
       }
       upperarm_r_fk_ctrl.Rotation.Y {
        type=MovieSceneScriptingFloatChannel
        default=0
        thinned_from=368
        keys [[0,-19.0743],[39,-19.5072],[45,-19.319],[51,-18.8567],[70,-17.105],[80,-16.6452],[99,-16.591],[104,-16.7934],[122,-18.3506],[132,-18.7254],[174,-19.2618],[175,-19.313],[176,-19.4081],[178,-19.7059],[182,-20.4658],[184,-20.7415],[185,-20.8196],[186,-20.8498],[187,-20.8375],[189,-20.6825],[191,-20.4209],[195,-19.6948],[196,-19.4728],[197,-19.2009],[198,-18.8534],[201,-17.5491],[202,-17.1961],[203,-16.9373],[204,-16.7622],[205,-16.6618],[206,-16.6234],[208,-16.6518],[211,-16.8376],[226,-18.1694],[232,-18.5739],[237,-18.7179],[239,-18.6852],[242,-18.4783],[250,-17.6189],[253,-17.4191],[257,-17.3363],[261,-17.4074],[271,-17.9123],[284,-18.7348],[296,-19.0727],[297,0],[367,0]]
       }
       upperarm_r_fk_ctrl.Rotation.Z {
        type=MovieSceneScriptingFloatChannel
        default=0
        thinned_from=368
        keys [[0,-14.9417],[1,-14.9417],[4,-15.0771],[8,-15.1249],[12,-15.0447],[27,-14.3333],[33,-14.2287],[45,-14.3143],[46,-14.2966],[48,-14.1746],[50,-13.9764],[56,-13.1328],[60,-12.5243],[62,-12.2819],[65,-12.0115],[67,-11.8875],[69,-11.8247],[70,-11.8271],[71,-11.8762],[72,-11.9555],[74,-12.1934],[77,-12.684],[81,-13.4056],[83,-13.7125],[85,-13.9349],[86,-14.0016],[87,-14.0337],[89,-14.0131],[91,-13.9073],[93,-13.733],[97,-13.2424],[105,-12.0941],[107,-11.8694],[109,-11.7073],[111,-11.6199],[113,-11.6287],[115,-11.6988],[117,-11.8252],[119,-12.0014],[122,-12.3499],[126,-12.9348],[136,-14.5599],[141,-15.2518],[144,-15.5826],[147,-15.8233],[150,-15.9581],[155,-15.9487],[161,-15.6976],[168,-15.1642],[174,-14.5259],[175,-14.456],[176,-14.4432],[177,-14.4756],[178,-14.5511],[180,-14.8005],[187,-16.0164],[189,-16.2839],[190,-16.3769],[191,-16.4371],[192,-16.4479],[193,-16.4129],[194,-16.3409],[195,-16.2254],[196,-16.0515],[197,-15.8074],[198,-15.4792],[199,-15.0721],[201,-14.1954],[202,-13.8374],[203,-13.5716],[204,-13.3884],[205,-13.2863],[206,-13.2621],[207,-13.2926],[208,-13.3659],[209,-13.4738],[211,-13.7759],[213,-14.1782],[215,-14.6591],[226,-17.8384],[228,-18.3169],[230,-18.6958],[232,-18.9651],[234,-19.1118],[236,-19.1242],[238,-19.0071],[239,-18.8921],[240,-18.7378],[242,-18.3186],[244,-17.7815],[253,-14.7943],[255,-14.1877],[257,-13.6563],[259,-13.2089],[261,-12.8598],[262,-12.7263],[263,-12.619],[264,-12.5372],[265,-12.4828],[267,-12.4564],[270,-12.5222],[273,-12.6836],[279,-13.2159],[293,-14.6942],[296,-14.9476],[297,0],[367,0]]
       }
       upperarm_r_fk_ctrl.Scale.X {type=MovieSceneScriptingFloatChannel, default=1, const [1,368]}
       upperarm_r_fk_ctrl.Scale.Y {type=MovieSceneScriptingFloatChannel, default=1, const [1,368]}
       upperarm_r_fk_ctrl.Scale.Z {type=MovieSceneScriptingFloatChannel, default=1, const [1,368]}
       lowerarm_r_fk_ctrl.Location.X {type=MovieSceneScriptingFloatChannel, default=0, thinned_from=368, keys [[0,-0.0001],[367,0]]}
       lowerarm_r_fk_ctrl.Location.Y {type=MovieSceneScriptingFloatChannel, default=0, thinned_from=368, keys [[0,0],[367,0]]}
       lowerarm_r_fk_ctrl.Location.Z {type=MovieSceneScriptingFloatChannel, default=0, thinned_from=368, keys [[0,0],[367,0]]}
       lowerarm_r_fk_ctrl.Rotation.X {
        type=MovieSceneScriptingFloatChannel
        default=0
        thinned_from=368
        keys [[0,0.5421],[1,0.5421],[42,0.5008],[50,0.4987],[52,0.5001],[54,0.5038],[58,0.5165],[68,0.5567],[71,0.5651],[74,0.5695],[76,0.5694],[78,0.5663],[88,0.5327],[90,0.5282],[92,0.5265],[95,0.5289],[99,0.538],[110,0.5744],[114,0.583],[117,0.5852],[122,0.5817],[129,0.5674],[147,0.5147],[152,0.5056],[156,0.5031],[163,0.5074],[198,0.5797],[207,0.5884],[210,0.5884],[213,0.5838],[216,0.5745],[220,0.5567],[231,0.4968],[234,0.4837],[236,0.4772],[238,0.473],[240,0.4714],[242,0.4727],[244,0.4762],[247,0.4851],[262,0.5515],[266,0.5643],[269,0.5697],[274,0.5702],[296,0.5421],[297,0],[367,0]]
       }
       lowerarm_r_fk_ctrl.Rotation.Y {
        type=MovieSceneScriptingFloatChannel
        default=0
        thinned_from=368
        keys [[0,0.2547],[1,0.2547],[6,0.2665],[32,0.3118],[42,0.3231],[49,0.3264],[51,0.3259],[52,0.3243],[53,0.3217],[54,0.3183],[55,0.3141],[56,0.3092],[58,0.2977],[61,0.2775],[66,0.2421],[68,0.2294],[69,0.2239],[70,0.2188],[71,0.2146],[72,0.211],[73,0.2084],[74,0.2067],[75,0.2062],[76,0.2069],[77,0.2091],[78,0.2124],[79,0.2167],[80,0.222],[82,0.234],[86,0.2599],[87,0.2656],[88,0.2708],[89,0.2751],[90,0.2783],[91,0.2804],[92,0.2811],[93,0.2806],[94,0.2793],[95,0.2771],[96,0.2741],[97,0.2706],[98,0.2664],[100,0.2566],[103,0.2393],[107,0.2147],[109,0.2033],[110,0.198],[111,0.1932],[112,0.189],[113,0.1852],[114,0.1823],[115,0.1801],[116,0.1786],[117,0.1782],[118,0.1784],[119,0.1792],[120,0.1805],[121,0.1823],[123,0.1873],[125,0.1937],[127,0.2015],[130,0.2153],[142,0.2784],[145,0.2925],[148,0.3043],[150,0.3106],[152,0.3154],[154,0.3184],[156,0.3194],[158,0.3189],[161,0.3158],[163,0.3126],[167,0.3033],[172,0.2877],[180,0.2566],[190,0.2155],[194,0.2008],[197,0.1912],[200,0.1831],[202,0.1787],[204,0.1753],[206,0.1729],[207,0.1721],[208,0.1716],[209,0.1716],[210,0.1721],[211,0.174],[212,0.1769],[213,0.1808],[214,0.1857],[215,0.1913],[216,0.1979],[217,0.2049],[218,0.2126],[220,0.2295],[228,0.3043],[230,0.3216],[232,0.337],[234,0.3502],[236,0.3602],[237,0.3639],[238,0.3666],[239,0.3683],[240,0.369],[241,0.3685],[242,0.3671],[243,0.3647],[244,0.3617],[246,0.3532],[248,0.3422],[250,0.3293],[253,0.3073],[259,0.26],[261,0.2453],[263,0.2321],[264,0.2262],[265,0.2208],[266,0.216],[267,0.2119],[268,0.2087],[269,0.2064],[270,0.205],[271,0.2044],[274,0.2055],[277,0.2087],[279,0.2119],[283,0.2202],[296,0.2547],[297,0],[367,0]]
       }
       lowerarm_r_fk_ctrl.Rotation.Z {
        type=MovieSceneScriptingFloatChannel
        default=0
        thinned_from=368
        keys [[0,-10.2525],[1,-10.2525],[6,-10.547],[36,-11.822],[48,-12.0528],[51,-12.0452],[52,-12.0046],[54,-11.8514],[56,-11.6215],[59,-11.1683],[66,-9.9382],[68,-9.6246],[70,-9.3618],[72,-9.1688],[73,-9.1051],[74,-9.064],[75,-9.0497],[76,-9.0681],[77,-9.1215],[78,-9.2037],[79,-9.3105],[81,-9.5836],[86,-10.381],[88,-10.6567],[89,-10.7631],[90,-10.8445],[91,-10.8946],[92,-10.9134],[94,-10.8696],[96,-10.7381],[98,-10.5439],[100,-10.2996],[108,-9.1142],[111,-8.7302],[113,-8.5348],[114,-8.4619],[116,-8.3725],[118,-8.3692],[120,-8.4203],[122,-8.5189],[125,-8.744],[128,-9.043],[144,-11.0871],[148,-11.4982],[151,-11.7231],[154,-11.8559],[158,-11.8683],[163,-11.7076],[168,-11.4027],[177,-10.6058],[192,-9.0935],[196,-8.7548],[200,-8.4817],[203,-8.3305],[207,-8.2125],[209,-8.2018],[210,-8.2125],[211,-8.2608],[212,-8.3305],[213,-8.427],[214,-8.545],[215,-8.6844],[217,-9.0186],[219,-9.4122],[222,-10.0796],[229,-11.7207],[231,-12.1391],[233,-12.5054],[235,-12.8017],[236,-12.9224],[238,-13.0869],[239,-13.1308],[240,-13.1473],[241,-13.1363],[243,-13.0376],[244,-12.9608],[246,-12.7414],[248,-12.4615],[252,-11.7669],[260,-10.2],[262,-9.8502],[264,-9.544],[266,-9.2901],[267,-9.1908],[268,-9.1127],[269,-9.056],[270,-9.0205],[272,-9.0063],[277,-9.1127],[281,-9.2857],[296,-10.2521],[297,0],[367,0]]
       }
       lowerarm_r_fk_ctrl.Scale.X {type=MovieSceneScriptingFloatChannel, default=1, const [1,368]}
       lowerarm_r_fk_ctrl.Scale.Y {type=MovieSceneScriptingFloatChannel, default=1, const [1,368]}
       lowerarm_r_fk_ctrl.Scale.Z {type=MovieSceneScriptingFloatChannel, default=1, const [1,368]}
       hand_r_fk_ctrl.Location.X {type=MovieSceneScriptingFloatChannel, default=0, thinned_from=368, keys [[0,-0.0001],[367,0]]}
       hand_r_fk_ctrl.Location.Y {type=MovieSceneScriptingFloatChannel, default=0, thinned_from=368, keys [[0,0],[367,0]]}
       hand_r_fk_ctrl.Location.Z {type=MovieSceneScriptingFloatChannel, default=0, thinned_from=368, keys [[0,0],[367,0]]}
       hand_r_fk_ctrl.Rotation.X {type=MovieSceneScriptingFloatChannel, default=0, thinned_from=368, keys [[0,9.4292],[296,9.4292],[297,0],[367,0]]}
       hand_r_fk_ctrl.Rotation.Y {type=MovieSceneScriptingFloatChannel, default=0, thinned_from=368, keys [[0,3.4167],[296,3.4167],[297,0],[367,0]]}
       hand_r_fk_ctrl.Rotation.Z {type=MovieSceneScriptingFloatChannel, default=0, thinned_from=368, keys [[0,-0.1507],[296,-0.1507],[297,0],[367,0]]}
       hand_r_fk_ctrl.Scale.X {type=MovieSceneScriptingFloatChannel, default=1, const [1,368]}
       hand_r_fk_ctrl.Scale.Y {type=MovieSceneScriptingFloatChannel, default=1, const [1,368]}
       hand_r_fk_ctrl.Scale.Z {type=MovieSceneScriptingFloatChannel, default=1, const [1,368]}
       thumb_01_r_ctrl.Location.X {
        type=MovieSceneScriptingFloatChannel
        default=0
        thinned_from=368
        keys [[0,-0.0001],[33,-0.0001],[34,-0.0015],[35,0.0026],[36,0.0079],[37,0.0173],[38,0.0313],[39,0.0501],[42,0.1203],[43,0.1484],[47,0.2421],[49,0.2796],[50,0.2937],[54,0.3392],[59,0.3678],[60,0.3729],[61,0.3736],[62,0.3793],[63,0.3793],[64,0.3843],[65,0.385],[66,0.3901],[67,0.3908],[68,0.3906],[69,0.3932],[70,0.3932],[71,0.3959],[72,0.3959],[73,0.3982],[184,0.3984],[185,0.3966],[186,0.3918],[187,0.3829],[188,0.3728],[189,0.3585],[190,0.3394],[192,0.2924],[193,0.2679],[194,0.2399],[195,0.2155],[196,0.1875],[198,0.1399],[199,0.1202],[200,0.1011],[201,0.0827],[202,0.0659],[203,0.0546],[204,0.0427],[205,0.0322],[206,0.021],[209,0.0041],[210,0.0034],[211,-0.0015],[216,-0.0015],[217,-0.0001],[296,-0.0001],[367,0]]
       }
       thumb_01_r_ctrl.Location.Y {
        type=MovieSceneScriptingFloatChannel
        default=0
        thinned_from=368
        keys [[0,0.0001],[33,0.0001],[34,0.0023],[35,-0.0044],[36,-0.0113],[37,-0.0249],[38,-0.0453],[39,-0.0725],[42,-0.1744],[43,-0.2152],[47,-0.3512],[49,-0.4056],[50,-0.426],[54,-0.4917],[60,-0.5415],[61,-0.5416],[62,-0.5499],[63,-0.5499],[64,-0.5581],[65,-0.5582],[66,-0.5664],[68,-0.5665],[69,-0.5703],[70,-0.5703],[71,-0.5741],[72,-0.5741],[73,-0.5778],[184,-0.5778],[185,-0.5755],[186,-0.5686],[188,-0.541],[189,-0.5203],[190,-0.4926],[193,-0.389],[194,-0.3477],[195,-0.313],[196,-0.2717],[198,-0.2026],[200,-0.1472],[201,-0.12],[202,-0.0956],[204,-0.0628],[206,-0.0303],[209,-0.0058],[210,-0.0057],[211,0.0023],[216,0.0023],[217,0.0001],[367,0]]
       }
       thumb_01_r_ctrl.Location.Z {
        type=MovieSceneScriptingFloatChannel
        default=0
        thinned_from=368
        keys [[0,0.0001],[33,0.0001],[34,-0.0014],[35,0.003],[36,0.0059],[37,0.0133],[38,0.0243],[39,0.0389],[42,0.0939],[43,0.1159],[47,0.1892],[49,0.2186],[50,0.2296],[51,0.2379],[54,0.2647],[59,0.2871],[60,0.2925],[61,0.2916],[62,0.2961],[63,0.2961],[64,0.3014],[65,0.3005],[66,0.3059],[67,0.305],[68,0.3053],[69,0.3072],[70,0.3072],[71,0.3092],[72,0.3092],[73,0.3116],[184,0.3114],[185,0.3104],[186,0.3067],[187,0.2985],[188,0.2918],[189,0.2807],[190,0.2658],[191,0.2465],[192,0.2279],[193,0.2101],[194,0.187],[195,0.1692],[196,0.1461],[198,0.109],[199,0.0948],[200,0.08],[201,0.0648],[202,0.0515],[203,0.0427],[204,0.0348],[205,0.0251],[206,0.0163],[209,0.003],[210,0.0039],[211,-0.0014],[216,-0.0014],[217,0.0001],[367,0]]
       }
       thumb_01_r_ctrl.Rotation.X {
        type=MovieSceneScriptingFloatChannel
        default=0
        thinned_from=368
        keys [[0,2.4221],[33,2.4221],[34,2.4257],[35,2.4557],[36,2.4522],[37,2.5342],[38,2.5735],[39,2.6661],[40,2.7864],[41,2.8908],[42,3.0276],[43,3.1377],[44,3.2865],[45,3.4439],[46,3.5264],[47,3.6484],[49,3.8088],[50,3.93],[51,3.9685],[53,4.0787],[54,4.1459],[55,4.1697],[57,4.232],[58,4.243],[59,4.2815],[60,4.2758],[61,4.3174],[62,4.3119],[64,4.3411],[65,4.3758],[66,4.3632],[68,4.4333],[69,4.3959],[70,4.4512],[71,4.4139],[72,4.4139],[73,4.4502],[74,4.432],[184,4.432],[186,4.3748],[187,4.3638],[188,4.2899],[189,4.203],[190,4.1387],[191,4.0158],[192,3.9023],[193,3.7457],[194,3.6322],[196,3.3872],[197,3.2542],[198,3.1561],[199,3.026],[200,2.909],[201,2.8483],[202,2.7537],[204,2.6387],[205,2.5659],[206,2.549],[207,2.4903],[208,2.4656],[209,2.4473],[210,2.4351],[212,2.423],[296,2.4221],[297,0],[367,0]]
       }
       thumb_01_r_ctrl.Rotation.Y {
        type=MovieSceneScriptingFloatChannel
        default=0
        thinned_from=368
        keys [[0,17.0943],[35,17.1085],[38,17.0118],[39,16.9232],[41,16.796],[42,16.6833],[43,16.6264],[44,16.5128],[49,16.1704],[57,15.9129],[58,15.9164],[72,15.7969],[73,15.8177],[185,15.7961],[188,15.8746],[191,16.047],[192,16.1519],[193,16.2113],[196,16.4922],[201,16.8274],[207,17.0484],[296,17.0943],[297,0],[367,0]]
       }
       thumb_01_r_ctrl.Rotation.Z {
        type=MovieSceneScriptingFloatChannel
        default=0
        thinned_from=368
        keys [[0,-0.5914],[33,-0.5914],[34,-0.6118],[35,-0.646],[36,-0.7991],[37,-1.0595],[38,-1.4574],[39,-1.9331],[40,-2.5288],[41,-3.1524],[42,-3.8148],[43,-4.524],[45,-5.8437],[46,-6.4561],[48,-7.567],[49,-8.0312],[50,-8.4397],[51,-8.8107],[52,-9.142],[53,-9.4364],[54,-9.6633],[56,-10.0482],[57,-10.1955],[59,-10.4533],[60,-10.5577],[61,-10.6374],[62,-10.7419],[64,-10.8892],[65,-10.9321],[66,-10.9997],[69,-11.1433],[70,-11.1555],[71,-11.2132],[72,-11.2132],[74,-11.2832],[184,-11.2832],[185,-11.2377],[186,-11.111],[187,-10.8684],[188,-10.5517],[189,-10.1716],[190,-9.6758],[191,-9.106],[193,-7.7768],[194,-7.052],[196,-5.6597],[197,-4.9923],[198,-4.3593],[199,-3.779],[200,-3.262],[201,-2.788],[202,-2.3721],[203,-1.9962],[204,-1.6722],[205,-1.4138],[206,-1.1681],[207,-0.9816],[208,-0.8376],[209,-0.7295],[210,-0.6575],[212,-0.5854],[216,-0.5854],[217,-0.5914],[296,-0.5914],[297,0],[367,0]]
       }
       thumb_01_r_ctrl.Scale.X {type=MovieSceneScriptingFloatChannel, default=1, const [1,368]}
       thumb_01_r_ctrl.Scale.Y {type=MovieSceneScriptingFloatChannel, default=1, const [1,368]}
       thumb_01_r_ctrl.Scale.Z {type=MovieSceneScriptingFloatChannel, default=1, const [1,368]}
       thumb_02_r_ctrl.Location.X {type=MovieSceneScriptingFloatChannel, default=0, thinned_from=368, keys [[0,0],[367,0]]}
       thumb_02_r_ctrl.Location.Y {type=MovieSceneScriptingFloatChannel, default=0, thinned_from=368, keys [[0,0],[367,0]]}
       thumb_02_r_ctrl.Location.Z {type=MovieSceneScriptingFloatChannel, default=0, thinned_from=368, keys [[0,0],[367,0]]}
       thumb_02_r_ctrl.Rotation.X {
        type=MovieSceneScriptingFloatChannel
        default=0
        thinned_from=368
        keys [[0,0],[33,0],[34,0.0122],[35,-0.0244],[36,-0.063],[37,-0.1419],[38,-0.2573],[39,-0.4118],[40,-0.6012],[41,-0.7963],[42,-0.9887],[43,-1.2177],[46,-1.7991],[47,-1.9869],[48,-2.1449],[49,-2.2971],[51,-2.5265],[53,-2.7123],[55,-2.8502],[59,-3.0324],[60,-3.0526],[67,-3.2178],[68,-3.2152],[69,-3.2372],[70,-3.2372],[71,-3.2618],[72,-3.259],[73,-3.2774],[84,-3.2837],[185,-3.2804],[186,-3.2375],[187,-3.1582],[188,-3.049],[189,-2.9253],[191,-2.6105],[192,-2.4144],[193,-2.1841],[194,-1.9894],[195,-1.7592],[196,-1.5589],[197,-1.3626],[198,-1.1703],[199,-0.9748],[200,-0.8196],[201,-0.6792],[202,-0.5458],[203,-0.4332],[205,-0.2543],[207,-0.1209],[208,-0.0765],[209,-0.0325],[210,-0.0096],[211,-0.009],[216,-0.009],[217,0],[367,0]]
       }
       thumb_02_r_ctrl.Rotation.Y {
        type=MovieSceneScriptingFloatChannel
        default=0
        thinned_from=368
        keys [[0,0],[33,0],[34,-0.0405],[35,0.081],[36,0.2015],[37,0.4424],[38,0.8048],[39,1.288],[42,3.1016],[43,3.8284],[47,6.2533],[49,7.2247],[51,7.9452],[53,8.5206],[55,8.9528],[59,9.5301],[67,10.1073],[68,10.108],[69,10.1758],[70,10.1758],[71,10.2428],[72,10.2436],[73,10.3122],[185,10.3115],[186,10.189],[187,9.9423],[189,9.2018],[191,8.2171],[192,7.6024],[193,6.8643],[194,6.2505],[195,5.5137],[199,3.0672],[200,2.5782],[201,2.1346],[202,1.7133],[203,1.3626],[205,0.8016],[207,0.3806],[209,0.0997],[210,0.0297],[211,0.03],[216,0.03],[217,0],[367,0]]
       }
       thumb_02_r_ctrl.Rotation.Z {
        type=MovieSceneScriptingFloatChannel
        default=0
        thinned_from=368
        keys [[0,1.1917],[35,1.1914],[37,1.4197],[39,1.7559],[41,2.3072],[42,2.578],[43,2.8415],[44,3.1624],[47,3.9349],[50,4.5404],[51,4.6937],[54,5.0785],[55,5.1356],[56,5.2351],[58,5.3596],[59,5.3846],[60,5.4526],[66,5.637],[67,5.6306],[69,5.7011],[71,5.6949],[72,5.7345],[84,5.7283],[85,5.7483],[185,5.7483],[186,5.6983],[187,5.597],[189,5.2899],[190,5.0827],[192,4.5487],[193,4.2272],[194,3.9561],[195,3.6263],[197,3.0672],[198,2.7824],[199,2.556],[200,2.3232],[201,2.1357],[202,1.9563],[204,1.653],[205,1.531],[207,1.3465],[208,1.2847],[211,1.1916],[296,1.1917],[297,0],[367,0]]
       }
       thumb_02_r_ctrl.Scale.X {type=MovieSceneScriptingFloatChannel, default=1, const [1,368]}
       thumb_02_r_ctrl.Scale.Y {type=MovieSceneScriptingFloatChannel, default=1, const [1,368]}
       thumb_02_r_ctrl.Scale.Z {type=MovieSceneScriptingFloatChannel, default=1, const [1,368]}
       thumb_03_r_ctrl.Location.X {type=MovieSceneScriptingFloatChannel, default=0, thinned_from=368, keys [[0,0],[367,0]]}
       thumb_03_r_ctrl.Location.Y {type=MovieSceneScriptingFloatChannel, default=0, thinned_from=368, keys [[0,0],[367,0]]}
       thumb_03_r_ctrl.Location.Z {type=MovieSceneScriptingFloatChannel, default=0, thinned_from=368, keys [[0,0],[367,0]]}
       thumb_03_r_ctrl.Rotation.X {
        type=MovieSceneScriptingFloatChannel
        default=0
        thinned_from=368
        keys [[0,0],[34,0.0001],[35,-0.0002],[36,-0.1494],[37,-0.2981],[38,-0.4413],[39,-0.7332],[40,-1.0208],[41,-1.3749],[42,-1.7961],[43,-2.1374],[44,-2.5451],[45,-2.8766],[47,-3.5264],[49,-4.047],[50,-4.2977],[51,-4.4699],[52,-4.6768],[53,-4.7976],[54,-4.9634],[55,-5.0421],[56,-5.166],[58,-5.3343],[59,-5.3666],[60,-5.4569],[65,-5.6473],[66,-5.6947],[67,-5.6867],[69,-5.7721],[71,-5.7607],[72,-5.8136],[84,-5.8114],[85,-5.8332],[184,-5.8332],[185,-5.8428],[186,-5.7803],[187,-5.6347],[189,-5.2471],[190,-4.9826],[192,-4.2981],[193,-3.886],[194,-3.5438],[195,-3.1209],[197,-2.4015],[198,-2.0344],[199,-1.7451],[200,-1.443],[201,-1.1839],[202,-0.9523],[203,-0.7572],[204,-0.6022],[205,-0.4433],[206,-0.3233],[207,-0.2007],[208,-0.1219],[211,-0.0001],[216,0.0001],[367,0]]
       }
       thumb_03_r_ctrl.Rotation.Y {
        type=MovieSceneScriptingFloatChannel
        default=0
        thinned_from=368
        keys [[0,0],[33,0],[34,-0.0131],[35,0.0298],[36,0.0781],[37,0.2135],[38,0.3967],[39,0.5882],[40,0.8274],[41,1.1223],[42,1.3915],[43,1.7025],[44,1.9907],[45,2.2745],[46,2.5651],[47,2.82],[48,3.0204],[49,3.2146],[50,3.4233],[51,3.5635],[52,3.7201],[54,3.9439],[56,4.107],[58,4.2316],[60,4.3143],[61,4.3734],[63,4.4411],[64,4.4575],[66,4.5176],[67,4.5258],[69,4.5868],[70,4.5868],[71,4.609],[73,4.6182],[74,4.6311],[84,4.6311],[85,4.6404],[184,4.6404],[185,4.6307],[186,4.566],[187,4.4877],[188,4.333],[189,4.1448],[190,3.9342],[192,3.4221],[193,3.1146],[194,2.7999],[195,2.4762],[196,2.215],[197,1.925],[198,1.643],[199,1.3728],[200,1.1512],[201,0.9661],[202,0.7697],[203,0.6085],[204,0.4789],[205,0.3533],[206,0.2601],[207,0.1696],[208,0.1066],[209,0.0474],[210,0.0171],[211,0.0157],[212,-0.0131],[216,-0.0131],[217,0],[367,0]]
       }
       thumb_03_r_ctrl.Rotation.Z {
        type=MovieSceneScriptingFloatChannel
        default=0
        thinned_from=368
        keys [[0,0],[33,0],[34,0.0002],[35,-0.0003],[36,0.4822],[37,0.9634],[38,1.6842],[40,3.6069],[41,4.8065],[42,6.247],[43,7.4445],[44,8.8834],[47,12.4774],[48,13.4371],[49,14.1546],[50,15.1147],[51,15.7585],[52,16.4724],[53,16.9001],[54,17.4724],[55,17.7574],[56,18.187],[57,18.473],[58,18.614],[59,18.902],[60,19.0439],[61,19.3312],[65,19.9046],[66,19.9025],[67,20.0478],[68,20.1091],[69,20.2709],[70,20.2709],[71,20.4342],[84,20.4322],[85,20.5135],[184,20.5135],[185,20.3443],[187,19.8524],[189,18.3695],[191,16.3971],[192,15.1639],[193,13.6852],[194,12.4561],[195,10.9797],[199,6.0589],[200,5.074],[201,4.2698],[202,3.4446],[204,2.068],[205,1.5169],[206,1.1034],[207,0.8276],[208,0.4138],[210,0.1382],[211,-0.0002],[212,0.0002],[216,0.0002],[367,0]]
       }
       thumb_03_r_ctrl.Scale.X {type=MovieSceneScriptingFloatChannel, default=1, const [1,368]}
       thumb_03_r_ctrl.Scale.Y {type=MovieSceneScriptingFloatChannel, default=1, const [1,368]}
       thumb_03_r_ctrl.Scale.Z {type=MovieSceneScriptingFloatChannel, default=1, const [1,368]}
       index_metacarpal_r_ctrl.Location.X {type=MovieSceneScriptingFloatChannel, default=0, thinned_from=368, keys [[0,0],[367,0]]}
       index_metacarpal_r_ctrl.Location.Y {type=MovieSceneScriptingFloatChannel, default=0, thinned_from=368, keys [[0,0],[367,0]]}
       index_metacarpal_r_ctrl.Location.Z {type=MovieSceneScriptingFloatChannel, default=0, thinned_from=368, keys [[0,0],[367,0]]}
       index_metacarpal_r_ctrl.Rotation.X {
        type=MovieSceneScriptingFloatChannel
        default=0
        thinned_from=368
        keys [[0,-1.732],[35,-1.7318],[37,-1.7499],[39,-1.7767],[43,-1.8667],[44,-1.8945],[47,-1.9635],[50,-2.0192],[54,-2.0723],[60,-2.1127],[61,-2.1127],[62,-2.1194],[63,-2.1194],[64,-2.1262],[66,-2.1266],[67,-2.133],[68,-2.1313],[75,-2.1391],[184,-2.1413],[187,-2.13],[190,-2.0913],[191,-2.0699],[192,-2.0532],[193,-2.0319],[194,-2.0063],[195,-1.9853],[196,-1.9602],[200,-1.8775],[201,-1.8618],[202,-1.8406],[203,-1.8284],[204,-1.8118],[208,-1.762],[209,-1.7582],[210,-1.7497],[212,-1.7415],[213,-1.7419],[214,-1.7378],[215,-1.7378],[216,-1.7333],[296,-1.732],[297,0],[367,0]]
       }
       index_metacarpal_r_ctrl.Rotation.Y {
        type=MovieSceneScriptingFloatChannel
        default=0
        thinned_from=368
        keys [[0,0.7486],[33,0.7486],[34,0.7354],[36,0.8103],[37,0.8851],[39,1.1095],[40,1.2967],[41,1.4463],[42,1.6333],[43,1.8579],[44,2.0448],[45,2.2693],[46,2.4563],[47,2.6058],[48,2.7929],[50,3.0921],[54,3.5291],[60,3.8607],[61,3.8607],[62,3.916],[63,3.916],[64,3.9713],[66,3.971],[67,4.0265],[68,4.0265],[69,4.0528],[70,4.0528],[71,4.0791],[73,4.0792],[74,4.1055],[185,4.1055],[186,4.0712],[187,4.0027],[190,3.6944],[195,2.8379],[196,2.6323],[200,1.9469],[201,1.802],[204,1.3943],[207,1.0885],[209,0.9525],[214,0.7826],[216,0.7825],[217,0.758],[232,0.758],[233,0.7486],[296,0.7486],[297,0],[367,0]]
       }
       index_metacarpal_r_ctrl.Rotation.Z {
        type=MovieSceneScriptingFloatChannel
        default=0
        thinned_from=368
        keys [[0,-0.3856],[33,-0.3856],[34,-0.3833],[36,-0.3962],[37,-0.4091],[38,-0.4285],[39,-0.4533],[40,-0.4804],[41,-0.5117],[42,-0.5443],[43,-0.5782],[44,-0.6163],[46,-0.6832],[47,-0.7149],[50,-0.7956],[54,-0.8712],[60,-0.9285],[61,-0.9285],[62,-0.9381],[63,-0.9381],[64,-0.9476],[65,-0.9476],[66,-0.9552],[67,-0.9572],[68,-0.9572],[69,-0.9619],[70,-0.9619],[71,-0.9665],[73,-0.9666],[74,-0.9712],[185,-0.9712],[186,-0.9651],[188,-0.9397],[189,-0.9167],[190,-0.8986],[191,-0.8684],[192,-0.8431],[193,-0.8131],[194,-0.7782],[195,-0.7483],[196,-0.7125],[200,-0.5937],[204,-0.4986],[207,-0.445],[209,-0.4212],[214,-0.3915],[216,-0.3915],[217,-0.3872],[232,-0.3872],[233,-0.3856],[296,-0.3856],[297,0],[367,0]]
       }
       index_metacarpal_r_ctrl.Scale.X {type=MovieSceneScriptingFloatChannel, default=1, const [1,368]}
       index_metacarpal_r_ctrl.Scale.Y {type=MovieSceneScriptingFloatChannel, default=1, const [1,368]}
       index_metacarpal_r_ctrl.Scale.Z {type=MovieSceneScriptingFloatChannel, default=1, const [1,368]}
       index_01_r_ctrl.Location.X {type=MovieSceneScriptingFloatChannel, default=0, thinned_from=368, keys [[0,0],[367,0]]}
       index_01_r_ctrl.Location.Y {type=MovieSceneScriptingFloatChannel, default=0, thinned_from=368, keys [[0,0],[367,0]]}
       index_01_r_ctrl.Location.Z {type=MovieSceneScriptingFloatChannel, default=0, thinned_from=368, keys [[0,0],[367,0]]}
       index_01_r_ctrl.Rotation.X {
        type=MovieSceneScriptingFloatChannel
        default=0
        thinned_from=368
        keys [[0,-4.1948],[34,-4.1973],[35,-4.1238],[36,-3.9181],[37,-3.5713],[38,-3.0893],[39,-2.4774],[40,-1.7596],[41,-0.9424],[42,-0.067],[43,0.8385],[44,1.7642],[45,2.6708],[46,3.5471],[47,4.3772],[48,5.149],[49,5.8678],[50,6.5194],[51,7.1097],[52,7.6376],[53,8.095],[54,8.5044],[55,8.8474],[57,9.3903],[58,9.6059],[60,9.9759],[62,10.2574],[65,10.584],[67,10.7694],[68,10.823],[69,10.902],[72,11.0605],[76,11.1665],[184,11.1673],[185,11.1152],[186,10.9701],[187,10.6886],[188,10.2894],[189,9.8209],[190,9.2077],[191,8.5359],[192,7.7801],[193,6.9724],[194,6.1464],[195,5.277],[196,4.4237],[197,3.596],[198,2.7917],[199,2.0316],[200,1.2951],[201,0.6135],[202,-0.0306],[203,-0.6588],[204,-1.2028],[205,-1.7398],[206,-2.199],[207,-2.6481],[208,-2.9837],[209,-3.2432],[210,-3.4619],[212,-3.7569],[214,-3.9763],[216,-4.1219],[217,-4.1458],[218,-4.1876],[232,-4.1877],[296,-4.1948],[297,0],[367,0]]
       }
       index_01_r_ctrl.Rotation.Y {
        type=MovieSceneScriptingFloatChannel
        default=0
        thinned_from=368
        keys [[0,1.6242],[34,1.6239],[35,1.5821],[36,1.4476],[37,1.2323],[38,0.9249],[39,0.539],[40,0.0814],[41,-0.4335],[42,-0.9945],[43,-1.5688],[44,-2.1561],[45,-2.7299],[46,-3.2862],[47,-3.8075],[48,-4.3003],[49,-4.7507],[50,-5.1669],[51,-5.5477],[52,-5.875],[53,-6.1592],[54,-6.4264],[55,-6.6288],[56,-6.8176],[57,-6.9866],[59,-7.2419],[62,-7.5322],[65,-7.7527],[67,-7.8544],[72,-8.0389],[76,-8.104],[184,-8.1071],[185,-8.0554],[186,-7.974],[187,-7.8015],[188,-7.5606],[189,-7.2424],[190,-6.8607],[191,-6.4519],[193,-5.4652],[195,-4.3758],[196,-3.8426],[197,-3.3179],[198,-2.8041],[199,-2.3021],[200,-1.8492],[202,-1.0033],[203,-0.6191],[204,-0.2697],[205,0.0635],[206,0.3574],[207,0.6559],[208,0.8594],[209,1.0146],[210,1.1439],[211,1.2511],[212,1.355],[213,1.4116],[214,1.4898],[215,1.5181],[216,1.572],[218,1.6169],[219,1.6239],[296,1.6242],[297,0],[367,0]]
       }
       index_01_r_ctrl.Rotation.Z {
        type=MovieSceneScriptingFloatChannel
        default=0
        thinned_from=368
        keys [[0,8.2441],[34,8.2359],[35,8.1326],[36,7.7669],[37,7.1744],[38,6.3535],[39,5.301],[40,4.0672],[41,2.6847],[42,1.185],[43,-0.3809],[44,-1.9598],[45,-3.5165],[46,-5.0324],[47,-6.4688],[48,-7.8059],[49,-9.0592],[50,-10.1898],[51,-11.1999],[52,-12.1042],[53,-12.9123],[54,-13.6253],[55,-14.1882],[56,-14.7016],[58,-15.5271],[60,-16.1527],[61,-16.4127],[64,-17.0424],[67,-17.5177],[68,-17.6091],[69,-17.7456],[72,-18.0189],[76,-18.2015],[184,-18.2019],[185,-18.1264],[186,-17.844],[187,-17.3462],[188,-16.7053],[189,-15.859],[190,-14.81],[191,-13.7002],[192,-12.395],[194,-9.5602],[195,-8.0975],[196,-6.586],[197,-5.1532],[198,-3.7974],[199,-2.4523],[200,-1.1819],[201,-0.0463],[202,1.0918],[203,2.156],[205,4.0056],[206,4.8615],[207,5.5791],[208,6.167],[209,6.6235],[211,7.274],[212,7.4657],[213,7.7286],[216,8.1148],[218,8.2303],[232,8.2301],[296,8.2441],[297,0],[367,0]]
       }
       index_01_r_ctrl.Scale.X {type=MovieSceneScriptingFloatChannel, default=1, const [1,368]}
       index_01_r_ctrl.Scale.Y {type=MovieSceneScriptingFloatChannel, default=1, const [1,368]}
       index_01_r_ctrl.Scale.Z {type=MovieSceneScriptingFloatChannel, default=1, const [1,368]}
       index_02_r_ctrl.Location.X {type=MovieSceneScriptingFloatChannel, default=0, thinned_from=368, keys [[0,0],[367,0]]}
       index_02_r_ctrl.Location.Y {type=MovieSceneScriptingFloatChannel, default=0, thinned_from=368, keys [[0,0],[367,0]]}
       index_02_r_ctrl.Location.Z {type=MovieSceneScriptingFloatChannel, default=0, thinned_from=368, keys [[0,0],[367,0]]}
       index_02_r_ctrl.Rotation.X {
        type=MovieSceneScriptingFloatChannel
        default=0
        thinned_from=368
        keys [[0,-4.4284],[35,-4.4277],[39,-4.3933],[47,-4.2448],[55,-4.1532],[74,-4.1057],[186,-4.1096],[190,-4.145],[193,-4.1916],[203,-4.3548],[208,-4.4038],[296,-4.4284],[297,0],[367,0]]
       }
       index_02_r_ctrl.Rotation.Y {
        type=MovieSceneScriptingFloatChannel
        default=0
        thinned_from=368
        keys [[0,-0.7872],[33,-0.7872],[34,-0.7854],[35,-0.7908],[36,-0.8081],[37,-0.8309],[38,-0.8654],[39,-0.9063],[40,-0.958],[43,-1.1326],[45,-1.2626],[47,-1.3806],[50,-1.5256],[52,-1.5976],[54,-1.6566],[55,-1.6811],[57,-1.7184],[59,-1.7453],[65,-1.8006],[67,-1.811],[68,-1.8133],[69,-1.8227],[70,-1.8227],[72,-1.832],[73,-1.832],[76,-1.8413],[84,-1.8413],[184,-1.8391],[185,-1.8363],[186,-1.8242],[187,-1.8073],[188,-1.7798],[189,-1.7462],[190,-1.7078],[192,-1.6095],[193,-1.5543],[194,-1.4909],[196,-1.3828],[197,-1.3191],[200,-1.1603],[201,-1.1148],[203,-1.0284],[205,-0.9538],[206,-0.9218],[208,-0.8684],[210,-0.8369],[212,-0.8155],[213,-0.8106],[214,-0.7999],[215,-0.7999],[216,-0.7941],[217,-0.7902],[232,-0.7902],[233,-0.7872],[296,-0.7872],[297,0],[367,0]]
       }
       index_02_r_ctrl.Rotation.Z {
        type=MovieSceneScriptingFloatChannel
        default=0
        thinned_from=368
        keys [[0,-22.9455],[33,-22.9455],[34,-22.8925],[36,-23.1983],[37,-23.5046],[38,-23.9649],[40,-25.1978],[44,-28.317],[45,-29.2639],[47,-30.8535],[50,-32.7813],[52,-33.7719],[54,-34.5369],[56,-35.1406],[59,-35.7467],[65,-36.4661],[68,-36.6327],[69,-36.7474],[70,-36.7474],[71,-36.8624],[73,-36.8622],[74,-36.9773],[185,-36.9773],[186,-36.8295],[188,-36.24],[189,-35.7993],[191,-34.6304],[193,-33.1815],[194,-32.3185],[196,-30.8904],[197,-30.0394],[200,-27.9304],[204,-25.6225],[207,-24.3496],[209,-23.7867],[214,-23.0854],[216,-23.0856],[217,-22.9846],[232,-22.9846],[296,-22.9455],[297,0],[367,0]]
       }
       index_02_r_ctrl.Scale.X {type=MovieSceneScriptingFloatChannel, default=1, const [1,368]}
       index_02_r_ctrl.Scale.Y {type=MovieSceneScriptingFloatChannel, default=1, const [1,368]}
       index_02_r_ctrl.Scale.Z {type=MovieSceneScriptingFloatChannel, default=1, const [1,368]}
       index_03_r_ctrl.Location.X {type=MovieSceneScriptingFloatChannel, default=0, thinned_from=368, keys [[0,0],[367,0]]}
       index_03_r_ctrl.Location.Y {type=MovieSceneScriptingFloatChannel, default=0, thinned_from=368, keys [[0,0],[367,0]]}
       index_03_r_ctrl.Location.Z {type=MovieSceneScriptingFloatChannel, default=0, thinned_from=368, keys [[0,0],[367,0]]}
       index_03_r_ctrl.Rotation.X {
        type=MovieSceneScriptingFloatChannel
        default=0
        thinned_from=368
        keys [[0,0],[33,0],[34,-0.0003],[40,0],[41,0.0017],[42,-0.0002],[45,-0.0001],[46,0.0002],[47,-0.0002],[48,0.0001],[50,0],[51,0.0003],[53,0.0003],[54,-0.0003],[55,0.0003],[56,-0.0002],[58,-0.0002],[59,0.0003],[60,-0.0002],[61,-0.0002],[62,0.0003],[63,0.0003],[64,-0.0002],[66,-0.0002],[67,0.0002],[68,0],[70,0],[71,0.0014],[72,-0.0002],[84,-0.0002],[85,0],[184,0],[186,-0.0003],[187,0.0002],[189,0.0002],[190,-0.0002],[191,0.0002],[192,-0.0001],[193,0.0002],[194,0],[199,0],[200,-0.0003],[201,0.0005],[203,0.0005],[204,-0.0003],[205,0.0005],[206,0.0005],[207,-0.0003],[208,0.0005],[209,-0.0003],[210,0.0005],[212,0.0005],[213,-0.0002],[215,-0.0004],[216,0.0005],[217,0],[367,0]]
       }
       index_03_r_ctrl.Rotation.Y {
        type=MovieSceneScriptingFloatChannel
        default=0
        thinned_from=368
        keys [[0,0],[33,0],[34,-0.0003],[40,-0.0007],[41,0.0008],[42,0],[45,0.0001],[46,-0.0006],[47,0.0004],[48,-0.0003],[51,0.0003],[53,0.0002],[54,-0.0001],[55,0.0002],[56,-0.0004],[58,-0.0004],[59,0.0003],[60,-0.0004],[61,-0.0004],[62,0.0003],[63,0.0003],[64,-0.0003],[66,-0.0003],[67,0.0003],[68,0],[70,0],[71,0.0019],[72,-0.0003],[84,-0.0003],[85,0],[184,0],[185,-0.0003],[188,0.0001],[189,0],[190,-0.0003],[191,0.0001],[192,-0.0002],[193,0.0002],[194,-0.0002],[199,0],[200,-0.0002],[201,0.0005],[203,0.0003],[204,0],[205,0.0003],[206,0.0003],[207,-0.0001],[208,0.0003],[209,0],[210,0.0003],[212,0.0003],[213,-0.0003],[214,-0.0003],[216,0.0004],[217,0],[367,0]]
       }
       index_03_r_ctrl.Rotation.Z {
        type=MovieSceneScriptingFloatChannel
        default=0
        thinned_from=368
        keys [[0,-10.1405],[33,-10.1405],[34,-10.0867],[35,-10.0867],[38,-11.0304],[40,-12.2932],[41,-13.2438],[42,-13.8792],[43,-14.835],[44,-15.474],[45,-16.4353],[46,-17.0781],[47,-18.0452],[50,-19.9904],[51,-20.5466],[53,-21.4277],[55,-22.0906],[59,-22.9774],[61,-23.1996],[62,-23.422],[63,-23.5334],[64,-23.5334],[67,-23.8676],[70,-23.8676],[71,-24.2078],[184,-24.2078],[185,-24.135],[186,-23.9894],[187,-23.7712],[188,-23.4081],[189,-22.973],[191,-21.8169],[193,-20.3796],[197,-17.2464],[200,-15.1314],[201,-14.5847],[204,-12.8565],[207,-11.567],[209,-10.9956],[214,-10.2829],[216,-10.2829],[217,-10.1799],[232,-10.1799],[233,-10.1405],[296,-10.1405],[297,0],[367,0]]
       }
       index_03_r_ctrl.Scale.X {type=MovieSceneScriptingFloatChannel, default=1, const [1,368]}
       index_03_r_ctrl.Scale.Y {type=MovieSceneScriptingFloatChannel, default=1, const [1,368]}
       index_03_r_ctrl.Scale.Z {type=MovieSceneScriptingFloatChannel, default=1, const [1,368]}
       middle_metacarpal_r_ctrl.Location.X {type=MovieSceneScriptingFloatChannel, default=0, thinned_from=368, keys [[0,0],[367,0]]}
       middle_metacarpal_r_ctrl.Location.Y {type=MovieSceneScriptingFloatChannel, default=0, thinned_from=368, keys [[0,0],[367,0]]}
       middle_metacarpal_r_ctrl.Location.Z {type=MovieSceneScriptingFloatChannel, default=0, thinned_from=368, keys [[0,0],[367,0]]}
       middle_metacarpal_r_ctrl.Rotation.X {type=MovieSceneScriptingFloatChannel, default=0, thinned_from=368, keys [[0,0],[367,0]]}
       middle_metacarpal_r_ctrl.Rotation.Y {type=MovieSceneScriptingFloatChannel, default=0, thinned_from=368, keys [[0,-1.1678],[296,-1.1678],[297,0],[367,0]]}
       middle_metacarpal_r_ctrl.Rotation.Z {type=MovieSceneScriptingFloatChannel, default=0, thinned_from=368, keys [[0,0],[367,0]]}
       middle_metacarpal_r_ctrl.Scale.X {type=MovieSceneScriptingFloatChannel, default=1, const [1,368]}
       middle_metacarpal_r_ctrl.Scale.Y {type=MovieSceneScriptingFloatChannel, default=1, const [1,368]}
       middle_metacarpal_r_ctrl.Scale.Z {type=MovieSceneScriptingFloatChannel, default=1, const [1,368]}
       middle_01_r_ctrl.Location.X {type=MovieSceneScriptingFloatChannel, default=0, thinned_from=368, keys [[0,0],[367,0]]}
       middle_01_r_ctrl.Location.Y {type=MovieSceneScriptingFloatChannel, default=0, thinned_from=368, keys [[0,0],[367,0]]}
       middle_01_r_ctrl.Location.Z {type=MovieSceneScriptingFloatChannel, default=0, thinned_from=368, keys [[0,0],[367,0]]}
       middle_01_r_ctrl.Rotation.X {type=MovieSceneScriptingFloatChannel, default=0, thinned_from=368, keys [[0,0],[367,0]]}
       middle_01_r_ctrl.Rotation.Y {
        type=MovieSceneScriptingFloatChannel
        default=0
        thinned_from=368
        keys [[0,0],[67,-0.0001],[69,0],[70,-0.0001],[71,0.0001],[194,0.0001],[202,0.0001],[204,-0.0002],[205,0],[207,0],[208,-0.0001],[211,0.0002],[212,-0.0001],[215,0.0001],[367,0]]
       }
       middle_01_r_ctrl.Rotation.Z {
        type=MovieSceneScriptingFloatChannel
        default=0
        thinned_from=368
        keys [[0,-0.3096],[34,-0.3098],[35,-0.5672],[36,-1.3656],[37,-2.7233],[38,-4.6253],[39,-7.0193],[40,-9.8109],[41,-12.8715],[43,-19.2655],[44,-22.3627],[45,-25.2916],[46,-27.9957],[47,-30.4579],[48,-32.669],[49,-34.6242],[50,-36.3423],[51,-37.8277],[52,-39.0955],[53,-40.1613],[54,-41.0486],[55,-41.8167],[57,-43.0615],[59,-44.0481],[62,-45.1809],[66,-46.2304],[71,-47.0132],[75,-47.2578],[184,-47.2578],[185,-47.1355],[186,-46.7544],[187,-46.0874],[188,-45.1084],[189,-43.8076],[190,-42.1832],[191,-40.2764],[192,-38.135],[193,-35.8192],[196,-28.4939],[197,-26.1051],[198,-23.7852],[199,-21.5582],[200,-19.4348],[201,-17.4052],[202,-15.4835],[203,-13.6529],[204,-11.9202],[205,-10.2842],[206,-8.7273],[207,-7.2485],[208,-5.8471],[209,-4.5225],[210,-3.5386],[211,-2.7951],[212,-2.2007],[213,-1.7304],[214,-1.3427],[215,-1.0376],[216,-0.782],[217,-0.5941],[218,-0.4435],[219,-0.3515],[220,-0.3096],[296,-0.3096],[297,0],[367,0]]
       }
       middle_01_r_ctrl.Scale.X {type=MovieSceneScriptingFloatChannel, default=1, const [1,368]}
       middle_01_r_ctrl.Scale.Y {type=MovieSceneScriptingFloatChannel, default=1, const [1,368]}
       middle_01_r_ctrl.Scale.Z {type=MovieSceneScriptingFloatChannel, default=1, const [1,368]}
       middle_02_r_ctrl.Location.X {type=MovieSceneScriptingFloatChannel, default=0, thinned_from=368, keys [[0,-0.0002],[296,-0.0002],[367,0]]}
       middle_02_r_ctrl.Location.Y {type=MovieSceneScriptingFloatChannel, default=0, thinned_from=368, keys [[0,0.0001],[367,0]]}
       middle_02_r_ctrl.Location.Z {type=MovieSceneScriptingFloatChannel, default=0, thinned_from=368, keys [[0,0],[367,0]]}
       middle_02_r_ctrl.Rotation.X {
        type=MovieSceneScriptingFloatChannel
        default=0
        thinned_from=368
        keys [[0,-0.1309],[33,-0.1309],[34,-0.1324],[35,-0.0987],[36,-0.0078],[37,0.1579],[38,0.3833],[39,0.6836],[40,1.0244],[41,1.4001],[43,2.1724],[44,2.5344],[45,2.8764],[46,3.1915],[47,3.4838],[48,3.7359],[49,3.9615],[50,4.1524],[52,4.4932],[55,4.8046],[59,5.07],[60,5.1095],[61,5.1717],[65,5.3072],[66,5.3325],[68,5.3569],[69,5.3891],[70,5.3934],[71,5.4256],[84,5.4298],[85,5.446],[185,5.4408],[186,5.3876],[187,5.3157],[188,5.196],[189,5.0206],[190,4.8153],[191,4.5517],[193,3.9774],[194,3.6539],[195,3.3589],[196,3.0761],[197,2.8002],[199,2.2848],[200,2.0531],[201,1.8308],[202,1.618],[203,1.4167],[204,1.219],[205,1.033],[206,0.8582],[207,0.6908],[208,0.5271],[209,0.367],[210,0.2526],[211,0.1648],[212,0.0923],[213,0.0387],[214,-0.0071],[215,-0.0454],[216,-0.076],[217,-0.0959],[219,-0.1309],[296,-0.1309],[297,0],[367,0]]
       }
       middle_02_r_ctrl.Rotation.Y {
        type=MovieSceneScriptingFloatChannel
        default=0
        thinned_from=368
        keys [[0,-0.0178],[33,-0.0178],[34,-0.0197],[35,-0.0269],[36,-0.0179],[37,-0.0122],[38,-0.0243],[39,-0.035],[40,-0.0309],[41,-0.0354],[42,-0.0369],[43,-0.0458],[44,-0.0424],[45,-0.056],[46,-0.0541],[47,-0.057],[48,-0.0641],[49,-0.0639],[50,-0.0592],[51,-0.0656],[52,-0.0859],[53,-0.0663],[54,-0.0683],[55,-0.0691],[56,-0.069],[57,-0.0684],[58,-0.0672],[59,-0.0897],[60,-0.0645],[61,-0.0864],[63,-0.0836],[65,-0.0802],[66,-0.0889],[67,-0.0783],[68,-0.0615],[69,-0.0872],[70,-0.0646],[71,-0.0902],[72,-0.0902],[73,-0.0676],[84,-0.0676],[85,-0.0803],[184,-0.0803],[185,-0.0762],[186,-0.0853],[187,-0.0885],[188,-0.0778],[189,-0.0614],[190,-0.0761],[191,-0.0693],[192,-0.0671],[193,-0.0658],[195,-0.0445],[196,-0.061],[197,-0.0518],[198,-0.047],[199,-0.0425],[200,-0.0292],[201,-0.0388],[202,-0.0371],[203,-0.0351],[204,-0.0319],[205,-0.0313],[206,-0.0304],[207,-0.0266],[208,-0.0249],[209,-0.0225],[210,-0.0241],[211,-0.0204],[212,-0.0188],[213,-0.0218],[214,-0.0173],[215,-0.0214],[216,-0.0182],[217,-0.0191],[219,-0.0178],[296,-0.0178],[297,0],[367,0]]
       }
       middle_02_r_ctrl.Rotation.Z {
        type=MovieSceneScriptingFloatChannel
        default=0
        thinned_from=368
        keys [[0,-32.0531],[34,-32.0325],[35,-32.1981],[36,-32.6952],[37,-33.6114],[38,-34.7865],[39,-36.3119],[40,-38.113],[44,-46.1389],[45,-48.0002],[46,-49.7018],[47,-51.2381],[48,-52.6016],[50,-54.8814],[51,-55.7915],[53,-57.2403],[55,-58.3419],[58,-59.4566],[59,-59.6469],[60,-60.019],[66,-61.1584],[67,-61.1577],[68,-61.3907],[69,-61.3946],[70,-61.5709],[72,-61.5748],[73,-61.7515],[185,-61.7529],[186,-61.4435],[187,-60.9808],[188,-60.3651],[189,-59.6005],[191,-57.2056],[192,-55.7394],[195,-51.0524],[196,-49.4025],[197,-47.9103],[198,-46.5714],[199,-45.1169],[200,-43.9398],[201,-42.6426],[203,-40.3839],[205,-38.2896],[208,-35.5345],[209,-34.723],[210,-34.077],[212,-33.2335],[214,-32.714],[216,-32.3554],[219,-32.0531],[296,-32.0531],[297,0],[367,0]]
       }
       middle_02_r_ctrl.Scale.X {type=MovieSceneScriptingFloatChannel, default=1, const [1,368]}
       middle_02_r_ctrl.Scale.Y {type=MovieSceneScriptingFloatChannel, default=1, const [1,368]}
       middle_02_r_ctrl.Scale.Z {type=MovieSceneScriptingFloatChannel, default=1, const [1,368]}
       middle_03_r_ctrl.Location.X {type=MovieSceneScriptingFloatChannel, default=0, thinned_from=368, keys [[0,-0.0005],[296,-0.0005],[297,0],[367,0]]}
       middle_03_r_ctrl.Location.Y {type=MovieSceneScriptingFloatChannel, default=0, thinned_from=368, keys [[0,0],[367,0]]}
       middle_03_r_ctrl.Location.Z {type=MovieSceneScriptingFloatChannel, default=0, thinned_from=368, keys [[0,-0.0061],[296,-0.0061],[297,0],[367,0]]}
       middle_03_r_ctrl.Rotation.X {
        type=MovieSceneScriptingFloatChannel
        default=0
        thinned_from=368
        keys [[0,0],[33,0],[34,-0.0014],[35,0.0029],[36,-0.0017],[37,0.0023],[40,0.001],[41,0.0008],[43,0.0009],[44,0.0013],[45,-0.0029],[46,0.0024],[47,-0.0016],[48,0.0038],[49,-0.0002],[50,0.0005],[51,-0.0008],[52,-0.001],[53,0.0014],[55,0.001],[58,0.0009],[59,-0.0017],[60,0.0008],[61,0.0007],[62,-0.0017],[64,-0.0018],[65,0.0007],[67,0.0005],[68,-0.0003],[69,-0.0007],[70,0.0013],[71,-0.0011],[72,-0.0011],[73,0.0009],[84,0.0011],[85,0],[184,0],[185,0.0023],[186,-0.0017],[188,-0.0019],[190,-0.0023],[191,0.0012],[192,0.0012],[193,-0.0025],[194,0.0013],[195,-0.0021],[196,0.0019],[197,0.0023],[198,-0.0006],[199,-0.0002],[201,0.0014],[204,0.0005],[209,0.0002],[210,-0.0018],[211,0.0003],[212,-0.0017],[213,0.0004],[216,0.0004],[217,-0.0002],[218,0.0027],[219,-0.0015],[232,-0.0015],[233,0],[367,0]]
       }
       middle_03_r_ctrl.Rotation.Y {
        type=MovieSceneScriptingFloatChannel
        default=0
        thinned_from=368
        keys [[0,0],[33,0],[34,-0.0009],[35,0.0022],[36,-0.0012],[37,0.0016],[42,0.0003],[44,0.0006],[45,-0.0014],[46,0.0012],[47,-0.0006],[48,0.0018],[49,-0.0001],[50,0.0003],[51,-0.0001],[52,-0.0003],[53,0.0004],[54,0.0003],[55,0.0006],[56,0.0002],[58,0.0002],[59,-0.0006],[60,0.0003],[61,0.0003],[62,-0.0008],[64,-0.0007],[65,0.0001],[66,0.0001],[67,0.0004],[68,-0.0002],[69,-0.0003],[70,0.0006],[71,-0.0004],[72,-0.0004],[73,0.0005],[74,0.0002],[84,0.0002],[184,0],[185,0.0007],[186,-0.0003],[187,-0.0008],[188,-0.0008],[189,-0.0007],[190,-0.001],[191,0.0006],[192,0.0005],[193,-0.0013],[194,0.0006],[195,-0.0011],[196,0.0011],[197,0.0014],[198,-0.0004],[199,0.0001],[201,0.0008],[208,0.0001],[209,0.0001],[210,-0.0013],[211,0.0002],[212,-0.0012],[213,0.0002],[216,0.0003],[217,-0.0001],[218,0.0021],[219,-0.0009],[232,-0.0009],[233,0],[367,0]]
       }
       middle_03_r_ctrl.Rotation.Z {
        type=MovieSceneScriptingFloatChannel
        default=0
        thinned_from=368
        keys [[0,-50.5799],[35,-50.5798],[37,-49.5057],[38,-48.7046],[39,-47.6424],[40,-46.3238],[41,-44.7544],[42,-43.4568],[43,-41.9114],[44,-40.6331],[45,-39.11],[46,-38.1008],[47,-36.8463],[48,-36.097],[49,-35.1022],[51,-33.6656],[52,-33.1393],[55,-31.9594],[59,-30.9155],[61,-30.6552],[62,-30.3952],[64,-30.1355],[65,-30.1354],[67,-29.876],[68,-29.8252],[69,-29.6939],[70,-29.6939],[71,-29.5627],[84,-29.5626],[184,-29.5102],[185,-29.6288],[186,-29.6289],[188,-30.42],[189,-31.015],[190,-31.8106],[193,-34.8184],[194,-36.0328],[195,-37.0502],[196,-38.2773],[198,-40.3391],[199,-41.1698],[200,-42.2131],[203,-44.7087],[207,-47.5333],[209,-48.7239],[211,-49.4423],[214,-50.1638],[216,-50.405],[217,-50.405],[218,-50.5798],[296,-50.5799],[297,0],[367,0]]
       }
       middle_03_r_ctrl.Scale.X {type=MovieSceneScriptingFloatChannel, default=1, const [1,368]}
       middle_03_r_ctrl.Scale.Y {type=MovieSceneScriptingFloatChannel, default=1, const [1,368]}
       middle_03_r_ctrl.Scale.Z {type=MovieSceneScriptingFloatChannel, default=1, const [1,368]}
       ring_metacarpal_r_ctrl.Location.X {type=MovieSceneScriptingFloatChannel, default=0, thinned_from=368, keys [[0,0],[367,0]]}
       ring_metacarpal_r_ctrl.Location.Y {type=MovieSceneScriptingFloatChannel, default=0, thinned_from=368, keys [[0,0],[367,0]]}
       ring_metacarpal_r_ctrl.Location.Z {type=MovieSceneScriptingFloatChannel, default=0, thinned_from=368, keys [[0,0],[367,0]]}
       ring_metacarpal_r_ctrl.Rotation.X {type=MovieSceneScriptingFloatChannel, default=0, thinned_from=368, keys [[0,0],[367,0]]}
       ring_metacarpal_r_ctrl.Rotation.Y {type=MovieSceneScriptingFloatChannel, default=0, thinned_from=368, keys [[0,-5.131],[296,-5.131],[297,0],[367,0]]}
       ring_metacarpal_r_ctrl.Rotation.Z {type=MovieSceneScriptingFloatChannel, default=0, thinned_from=368, keys [[0,0],[367,0]]}
       ring_metacarpal_r_ctrl.Scale.X {type=MovieSceneScriptingFloatChannel, default=1, const [1,368]}
       ring_metacarpal_r_ctrl.Scale.Y {type=MovieSceneScriptingFloatChannel, default=1, const [1,368]}
       ring_metacarpal_r_ctrl.Scale.Z {type=MovieSceneScriptingFloatChannel, default=1, const [1,368]}
       ring_01_r_ctrl.Location.X {type=MovieSceneScriptingFloatChannel, default=0, thinned_from=368, keys [[0,0],[367,0]]}
       ring_01_r_ctrl.Location.Y {type=MovieSceneScriptingFloatChannel, default=0, thinned_from=368, keys [[0,0],[367,0]]}
       ring_01_r_ctrl.Location.Z {type=MovieSceneScriptingFloatChannel, default=0, thinned_from=368, keys [[0,0],[367,0]]}
       ring_01_r_ctrl.Rotation.X {
        type=MovieSceneScriptingFloatChannel
        default=0
        thinned_from=368
        keys [[0,0],[33,0],[34,0.0022],[35,0.0003],[36,-0.0009],[37,-0.0009],[38,0.0002],[39,0.0021],[40,0.0015],[42,-0.0011],[43,-0.0001],[44,-0.0018],[45,-0.001],[46,0.0002],[47,0.0018],[48,0.0014],[49,-0.0003],[50,-0.0017],[52,0.0008],[53,-0.0011],[54,-0.0012],[55,0.0016],[56,-0.0018],[57,0.0008],[58,-0.0007],[59,0],[60,0.0015],[61,-0.0008],[62,0.0007],[64,-0.0001],[65,-0.0016],[66,-0.0007],[67,-0.0022],[68,0.0031],[69,0],[70,-0.0022],[71,0.0022],[72,0],[184,0],[185,0.0006],[186,-0.0017],[187,-0.0029],[188,0.0022],[189,0.0014],[191,-0.0035],[192,-0.0011],[193,0.0001],[194,0.0029],[195,0.002],[196,-0.0024],[197,-0.0002],[198,0.0001],[199,-0.0025],[200,0.0017],[201,-0.0005],[202,0.0008],[203,-0.0008],[204,-0.0001],[205,0.0002],[206,-0.0006],[207,-0.0011],[208,-0.0015],[209,-0.0005],[210,0.0021],[211,0.0008],[212,0.0003],[213,0.0022],[214,-0.0001],[215,-0.0006],[216,-0.0003],[217,0.0005],[220,-0.0017],[221,0.0009],[232,0.0009],[233,0],[367,0]]
       }
       ring_01_r_ctrl.Rotation.Y {
        type=MovieSceneScriptingFloatChannel
        default=0
        thinned_from=368
        keys [[0,0],[33,0],[34,-0.0008],[35,-0.0001],[36,0.0028],[37,0.0012],[38,0.0012],[39,-0.0018],[40,0.0003],[41,0],[42,0.001],[43,0.0007],[45,0.0019],[46,-0.0014],[47,-0.0011],[48,-0.0019],[49,0.0005],[50,-0.0006],[51,0.001],[53,-0.0001],[54,0.0002],[55,0.0001],[56,0.0003],[57,0],[58,0.0014],[59,-0.0004],[60,0.0007],[61,-0.001],[62,0.0001],[64,-0.0005],[65,0.0008],[66,-0.001],[67,0.0003],[68,0.004],[69,-0.002],[70,0.0021],[71,-0.0041],[72,0],[184,0],[185,0.0004],[186,-0.0007],[187,-0.0002],[188,0.001],[189,-0.0016],[190,-0.0005],[191,0.001],[192,0.001],[193,-0.0001],[194,-0.001],[195,-0.0007],[196,0.001],[197,0.0013],[198,-0.0001],[199,0.0005],[200,-0.0004],[201,0.0015],[202,0.0001],[203,0.0008],[204,-0.0018],[205,-0.0012],[206,-0.0015],[207,0.0016],[209,0.0002],[210,-0.0002],[211,0.0016],[212,0.0002],[213,-0.0006],[214,-0.0022],[215,0.0004],[216,-0.0004],[218,-0.0001],[220,0.0019],[221,0.0008],[232,0.0008],[233,0],[367,0]]
       }
       ring_01_r_ctrl.Rotation.Z {
        type=MovieSceneScriptingFloatChannel
        default=0
        thinned_from=368
        keys [[0,-3.4999],[33,-3.4999],[34,-3.4737],[35,-3.8343],[36,-4.918],[37,-6.73],[38,-9.3522],[39,-12.5803],[40,-16.2829],[41,-20.1029],[42,-23.8195],[43,-27.278],[44,-30.3927],[45,-33.1524],[46,-35.5459],[47,-37.6424],[48,-39.4337],[49,-40.9939],[50,-42.318],[51,-43.3726],[52,-44.2463],[53,-45.0142],[54,-45.6752],[56,-46.7825],[58,-47.6737],[62,-48.9634],[67,-49.9797],[70,-50.3107],[71,-50.3112],[72,-50.4482],[184,-50.4482],[185,-50.3456],[186,-50.1405],[187,-49.6288],[188,-48.8139],[189,-47.8008],[190,-46.4927],[191,-44.8966],[192,-43.0203],[193,-40.969],[197,-32.2262],[198,-30.0953],[200,-26.0708],[201,-24.1871],[202,-22.366],[204,-18.914],[205,-17.2801],[206,-15.7318],[209,-11.3565],[210,-9.9856],[211,-8.6949],[212,-7.6351],[213,-6.7288],[214,-6.0502],[216,-4.9965],[217,-4.6047],[218,-4.2399],[220,-3.6932],[221,-3.5113],[232,-3.5113],[233,-3.4999],[296,-3.4999],[297,0],[367,0]]
       }
       ring_01_r_ctrl.Scale.X {type=MovieSceneScriptingFloatChannel, default=1, const [1,368]}
       ring_01_r_ctrl.Scale.Y {type=MovieSceneScriptingFloatChannel, default=1, const [1,368]}
       ring_01_r_ctrl.Scale.Z {type=MovieSceneScriptingFloatChannel, default=1, const [1,368]}
       ring_02_r_ctrl.Location.X {type=MovieSceneScriptingFloatChannel, default=0, thinned_from=368, keys [[0,0],[367,0]]}
       ring_02_r_ctrl.Location.Y {type=MovieSceneScriptingFloatChannel, default=0, thinned_from=368, keys [[0,0],[367,0]]}
       ring_02_r_ctrl.Location.Z {type=MovieSceneScriptingFloatChannel, default=0, thinned_from=368, keys [[0,0],[367,0]]}
       ring_02_r_ctrl.Rotation.X {
        type=MovieSceneScriptingFloatChannel
        default=0
        thinned_from=368
        keys [[0,0],[33,0],[34,0.0033],[35,-0.0925],[36,-0.3321],[37,-0.7744],[38,-1.4301],[39,-2.2268],[40,-3.1179],[41,-4.0332],[42,-4.9152],[43,-5.7316],[44,-6.4331],[45,-7.0793],[46,-7.6553],[47,-8.1231],[48,-8.5377],[49,-8.8844],[50,-9.1969],[51,-9.4469],[52,-9.6574],[53,-9.8417],[54,-9.9998],[56,-10.2634],[58,-10.4743],[62,-10.7809],[66,-10.9831],[71,-11.1141],[185,-11.1307],[186,-11.0577],[187,-10.9368],[188,-10.7694],[189,-10.5092],[190,-10.1764],[191,-9.7495],[192,-9.257],[193,-8.7168],[194,-8.2045],[195,-7.6457],[197,-6.6263],[199,-5.6566],[200,-5.2188],[201,-4.7977],[202,-4.3866],[203,-3.9988],[204,-3.5983],[205,-3.2212],[206,-2.8669],[207,-2.536],[208,-2.1913],[209,-1.8708],[210,-1.5354],[211,-1.2247],[212,-0.9852],[213,-0.781],[214,-0.6125],[215,-0.4796],[216,-0.3566],[217,-0.2392],[218,-0.1653],[219,-0.1121],[220,-0.0381],[221,0.0043],[232,0.0043],[233,0],[367,0]]
       }
       ring_02_r_ctrl.Rotation.Y {
        type=MovieSceneScriptingFloatChannel
        default=0
        thinned_from=368
        keys [[0,0],[33,0],[34,-0.0057],[35,-0.0078],[36,-0.0425],[37,-0.0719],[38,-0.1465],[39,-0.2392],[40,-0.3354],[41,-0.4396],[42,-0.5413],[43,-0.6211],[44,-0.6908],[45,-0.7509],[46,-0.8081],[47,-0.8687],[48,-0.8927],[49,-0.9465],[50,-0.9831],[51,-1.006],[54,-1.0522],[55,-1.074],[56,-1.0817],[57,-1.1027],[58,-1.1087],[59,-1.1278],[60,-1.1393],[61,-1.132],[62,-1.1443],[64,-1.1614],[65,-1.1591],[66,-1.1597],[67,-1.1728],[68,-1.1606],[69,-1.1638],[70,-1.1753],[72,-1.1704],[73,-1.1808],[84,-1.1808],[85,-1.1773],[184,-1.1773],[185,-1.1788],[187,-1.1681],[188,-1.1511],[189,-1.1102],[191,-1.0367],[192,-0.9688],[193,-0.9219],[194,-0.8545],[195,-0.8008],[196,-0.7601],[197,-0.698],[198,-0.6637],[199,-0.6049],[200,-0.5505],[201,-0.5137],[202,-0.4535],[203,-0.422],[204,-0.3668],[205,-0.3426],[206,-0.2919],[207,-0.2722],[208,-0.2236],[209,-0.201],[210,-0.1518],[211,-0.125],[212,-0.1035],[213,-0.0873],[214,-0.0721],[215,-0.0579],[216,-0.028],[217,-0.0277],[218,-0.0083],[219,-0.0261],[220,-0.0074],[232,-0.0074],[233,0],[367,0]]
       }
       ring_02_r_ctrl.Rotation.Z {
        type=MovieSceneScriptingFloatChannel
        default=0
        thinned_from=368
        keys [[0,-27.8067],[34,-27.7879],[35,-28.1194],[36,-28.8952],[37,-30.345],[38,-32.4857],[39,-35.1145],[42,-44.1426],[43,-46.8726],[44,-49.4059],[45,-51.6046],[46,-53.4497],[47,-55.1869],[48,-56.5475],[49,-57.7837],[51,-59.6501],[53,-60.9578],[55,-62.0138],[57,-62.8153],[60,-63.7138],[65,-64.715],[72,-65.3423],[185,-65.3423],[187,-64.6582],[188,-64.1471],[189,-63.3047],[190,-62.1394],[191,-60.8209],[193,-57.6069],[194,-55.8866],[196,-52.2269],[197,-50.5893],[198,-48.8302],[200,-45.6709],[204,-40.0749],[205,-38.8399],[206,-37.4951],[207,-36.4024],[208,-35.1983],[211,-31.9905],[212,-31.052],[214,-29.8857],[216,-28.9584],[219,-28.1206],[221,-27.8251],[296,-27.8067],[297,0],[367,0]]
       }
       ring_02_r_ctrl.Scale.X {type=MovieSceneScriptingFloatChannel, default=1, const [1,368]}
       ring_02_r_ctrl.Scale.Y {type=MovieSceneScriptingFloatChannel, default=1, const [1,368]}
       ring_02_r_ctrl.Scale.Z {type=MovieSceneScriptingFloatChannel, default=1, const [1,368]}
       ring_03_r_ctrl.Location.X {type=MovieSceneScriptingFloatChannel, default=0, thinned_from=368, keys [[0,0],[367,0]]}
       ring_03_r_ctrl.Location.Y {type=MovieSceneScriptingFloatChannel, default=0, thinned_from=368, keys [[0,0],[367,0]]}
       ring_03_r_ctrl.Location.Z {type=MovieSceneScriptingFloatChannel, default=0, thinned_from=368, keys [[0,0],[367,0]]}
       ring_03_r_ctrl.Rotation.X {type=MovieSceneScriptingFloatChannel, default=0, thinned_from=368, keys [[0,0],[367,0]]}
       ring_03_r_ctrl.Rotation.Y {type=MovieSceneScriptingFloatChannel, default=0, thinned_from=368, keys [[0,0],[367,0]]}
       ring_03_r_ctrl.Rotation.Z {
        type=MovieSceneScriptingFloatChannel
        default=0
        thinned_from=368
        keys [[0,-62.4464],[34,-62.4464],[35,-62.3015],[36,-61.8677],[37,-61.0038],[38,-59.7172],[39,-58.3002],[40,-56.4768],[42,-53.1627],[44,-50.1794],[45,-48.9729],[47,-47.1113],[48,-46.3189],[50,-45.0052],[53,-43.7624],[61,-42.0819],[62,-42.0819],[63,-41.8727],[64,-41.8727],[65,-41.6638],[69,-41.6116],[70,-41.2986],[84,-41.2986],[184,-41.3768],[185,-41.4724],[186,-41.4724],[187,-41.8205],[188,-41.9947],[189,-42.5182],[191,-43.9207],[192,-44.802],[193,-45.8646],[194,-46.7544],[196,-48.9067],[200,-52.55],[201,-53.2137],[204,-55.6686],[205,-56.2191],[206,-57.0483],[208,-58.1605],[209,-58.9998],[212,-60.6922],[216,-61.8311],[220,-62.4464],[296,-62.4464],[297,0],[367,0]]
       }
       ring_03_r_ctrl.Scale.X {type=MovieSceneScriptingFloatChannel, default=1, const [1,368]}
       ring_03_r_ctrl.Scale.Y {type=MovieSceneScriptingFloatChannel, default=1, const [1,368]}
       ring_03_r_ctrl.Scale.Z {type=MovieSceneScriptingFloatChannel, default=1, const [1,368]}
       pinky_metacarpal_r_ctrl.Location.X {type=MovieSceneScriptingFloatChannel, default=0, thinned_from=368, keys [[0,0],[367,0]]}
       pinky_metacarpal_r_ctrl.Location.Y {type=MovieSceneScriptingFloatChannel, default=0, thinned_from=368, keys [[0,0],[367,0]]}
       pinky_metacarpal_r_ctrl.Location.Z {type=MovieSceneScriptingFloatChannel, default=0, thinned_from=368, keys [[0,0],[367,0]]}
       pinky_metacarpal_r_ctrl.Rotation.X {type=MovieSceneScriptingFloatChannel, default=0, thinned_from=368, keys [[0,0],[367,0]]}
       pinky_metacarpal_r_ctrl.Rotation.Y {type=MovieSceneScriptingFloatChannel, default=0, thinned_from=368, keys [[0,-9.0736],[296,-9.0736],[297,0],[367,0]]}
       pinky_metacarpal_r_ctrl.Rotation.Z {type=MovieSceneScriptingFloatChannel, default=0, thinned_from=368, keys [[0,0],[367,0]]}
       pinky_metacarpal_r_ctrl.Scale.X {type=MovieSceneScriptingFloatChannel, default=1, const [1,368]}
       pinky_metacarpal_r_ctrl.Scale.Y {type=MovieSceneScriptingFloatChannel, default=1, const [1,368]}
       pinky_metacarpal_r_ctrl.Scale.Z {type=MovieSceneScriptingFloatChannel, default=1, const [1,368]}
       pinky_01_r_ctrl.Location.X {type=MovieSceneScriptingFloatChannel, default=0, thinned_from=368, keys [[0,0],[367,0]]}
       pinky_01_r_ctrl.Location.Y {type=MovieSceneScriptingFloatChannel, default=0, thinned_from=368, keys [[0,0],[367,0]]}
       pinky_01_r_ctrl.Location.Z {type=MovieSceneScriptingFloatChannel, default=0, thinned_from=368, keys [[0,0],[367,0]]}
       pinky_01_r_ctrl.Rotation.X {
        type=MovieSceneScriptingFloatChannel
        default=0
        thinned_from=368
        keys [[0,0],[33,0],[34,0.0008],[35,0.0008],[36,0],[37,-0.0012],[38,-0.0016],[39,-0.0003],[40,0.0015],[41,-0.0008],[42,-0.0005],[43,-0.0009],[44,-0.0001],[45,0.001],[46,-0.0015],[47,0.0002],[48,-0.0016],[49,0.0005],[50,0.0006],[51,-0.0001],[53,0.0006],[57,0.0005],[58,0.0003],[59,0.0017],[61,0.0013],[62,-0.0005],[63,0.0009],[64,0.0007],[65,-0.002],[66,0.0004],[67,0.0018],[68,-0.0019],[69,-0.0001],[70,-0.0018],[84,-0.0018],[85,0],[184,0],[185,0.0009],[186,-0.0021],[187,-0.0016],[188,-0.0019],[189,0.0009],[190,0.0015],[191,0.001],[192,-0.0009],[193,0.0011],[195,0.0008],[196,0.0004],[197,-0.0003],[198,0.0004],[199,-0.0008],[201,0.0002],[202,-0.0008],[203,-0.0005],[204,-0.0019],[205,0.0004],[206,-0.001],[207,0.0012],[208,0.0008],[209,0.0011],[210,0.0006],[211,-0.0011],[212,-0.0013],[213,-0.0002],[214,-0.0001],[215,-0.0002],[216,0.0014],[217,0.001],[218,-0.0008],[220,0.0004],[221,0.0002],[222,0.0008],[232,0.0008],[233,0],[367,0]]
       }
       pinky_01_r_ctrl.Rotation.Y {
        type=MovieSceneScriptingFloatChannel
        default=0
        thinned_from=368
        keys [[0,0],[33,0],[34,-0.0009],[35,-0.0015],[36,0.0009],[38,0.0032],[39,0.0016],[40,-0.0026],[41,0.0019],[42,-0.0004],[43,0.002],[44,0.0006],[45,-0.0006],[46,0.0007],[47,-0.0006],[48,0.0007],[49,-0.001],[50,0.0008],[51,0],[58,-0.0009],[59,0],[61,-0.0003],[62,-0.0016],[63,-0.0007],[64,-0.0008],[65,0.0026],[66,-0.001],[67,-0.0001],[68,0.0022],[69,-0.0002],[70,0.0026],[84,0.0026],[85,0],[184,0],[185,-0.0006],[186,0.0025],[187,0.0026],[188,0.0022],[189,-0.0014],[190,-0.0013],[191,-0.0018],[192,0.002],[193,-0.0015],[194,-0.0011],[195,-0.0004],[197,0.0018],[198,-0.0006],[199,0.0009],[200,-0.0011],[201,-0.0015],[202,0.0014],[203,0.0002],[204,0.0034],[205,-0.0012],[206,0.0025],[207,-0.0018],[208,-0.0021],[210,-0.0019],[211,0.0025],[212,0.0029],[213,-0.0008],[215,0.0002],[216,-0.0035],[217,-0.0031],[218,0.0016],[220,-0.0007],[221,-0.0016],[222,-0.0023],[232,-0.0023],[233,0],[367,0]]
       }
       pinky_01_r_ctrl.Rotation.Z {
        type=MovieSceneScriptingFloatChannel
        default=0
        thinned_from=368
        keys [[0,-4.2057],[33,-4.2057],[34,-4.189],[35,-4.5606],[36,-5.8013],[37,-8.0445],[38,-11.2458],[39,-14.9878],[40,-18.7761],[41,-22.2884],[42,-25.3193],[43,-27.9189],[44,-30.0748],[45,-31.9107],[46,-33.4874],[47,-34.7302],[48,-35.8413],[49,-36.6792],[51,-38.054],[53,-39.0528],[56,-40.201],[58,-40.778],[62,-41.647],[65,-42.0828],[70,-42.4654],[184,-42.4657],[186,-42.2493],[187,-41.8897],[188,-41.3875],[189,-40.6729],[190,-39.8189],[191,-38.7571],[192,-37.5609],[194,-34.708],[197,-29.862],[198,-28.182],[201,-23.4261],[203,-20.4414],[205,-17.6299],[207,-14.9842],[209,-12.4987],[210,-11.3315],[213,-8.057],[214,-7.1752],[215,-6.4982],[216,-5.9578],[217,-5.5107],[220,-4.5808],[222,-4.1684],[232,-4.1684],[233,-4.2057],[296,-4.2057],[297,0],[367,0]]
       }
       pinky_01_r_ctrl.Scale.X {type=MovieSceneScriptingFloatChannel, default=1, const [1,368]}
       pinky_01_r_ctrl.Scale.Y {type=MovieSceneScriptingFloatChannel, default=1, const [1,368]}
       pinky_01_r_ctrl.Scale.Z {type=MovieSceneScriptingFloatChannel, default=1, const [1,368]}
       pinky_02_r_ctrl.Location.X {type=MovieSceneScriptingFloatChannel, default=0, thinned_from=368, keys [[0,0],[367,0]]}
       pinky_02_r_ctrl.Location.Y {type=MovieSceneScriptingFloatChannel, default=0, thinned_from=368, keys [[0,0],[367,0]]}
       pinky_02_r_ctrl.Location.Z {type=MovieSceneScriptingFloatChannel, default=0, thinned_from=368, keys [[0,0],[367,0]]}
       pinky_02_r_ctrl.Rotation.X {
        type=MovieSceneScriptingFloatChannel
        default=0
        thinned_from=368
        keys [[0,0],[39,0],[41,-0.0001],[43,0.0002],[45,0],[46,0.0002],[47,-0.0001],[48,0.0001],[49,-0.0001],[50,0.0001],[56,0],[57,0.0003],[58,0],[60,0.0001],[184,0],[185,0.0001],[186,-0.0002],[187,0.0001],[188,-0.0002],[189,0],[191,-0.0001],[192,0.0002],[193,0.0002],[194,-0.0001],[195,0.0002],[196,0.0002],[197,0],[198,0.0001],[199,0],[200,0.0001],[201,0.0001],[202,0.0002],[209,0],[210,-0.0002],[216,0.0001],[217,-0.0001],[367,0]]
       }
       pinky_02_r_ctrl.Rotation.Y {
        type=MovieSceneScriptingFloatChannel
        default=0
        thinned_from=368
        keys [[0,0],[42,0],[45,0],[46,0.0002],[47,-0.0001],[48,0],[49,-0.0001],[59,0],[60,0.0001],[184,0],[185,0.0001],[186,-0.0001],[187,0.0001],[188,-0.0002],[189,0],[191,0],[192,0.0002],[193,0.0001],[194,-0.0001],[195,0.0002],[196,0.0002],[197,0],[209,0],[367,0]]
       }
       pinky_02_r_ctrl.Rotation.Z {
        type=MovieSceneScriptingFloatChannel
        default=0
        thinned_from=368
        keys [[0,-27.8067],[33,-27.8067],[34,-27.866],[35,-28.0024],[36,-28.6856],[37,-29.9215],[38,-31.7214],[40,-35.946],[41,-37.81],[42,-39.4043],[43,-40.8679],[44,-42.049],[46,-43.8385],[48,-45.0439],[50,-45.9548],[55,-47.2196],[57,-47.612],[67,-48.5585],[84,-48.5786],[185,-48.6462],[187,-48.3292],[189,-47.6974],[190,-47.2256],[192,-45.9752],[193,-45.1995],[194,-44.2742],[195,-43.5076],[200,-38.9899],[207,-33.575],[209,-32.3698],[210,-31.6207],[213,-29.8354],[214,-29.3917],[217,-28.5207],[220,-28.0119],[222,-27.7862],[296,-27.8067],[297,0],[367,0]]
       }
       pinky_02_r_ctrl.Scale.X {type=MovieSceneScriptingFloatChannel, default=1, const [1,368]}
       pinky_02_r_ctrl.Scale.Y {type=MovieSceneScriptingFloatChannel, default=1, const [1,368]}
       pinky_02_r_ctrl.Scale.Z {type=MovieSceneScriptingFloatChannel, default=1, const [1,368]}
       pinky_03_r_ctrl.Location.X {type=MovieSceneScriptingFloatChannel, default=0, thinned_from=368, keys [[0,0],[367,0]]}
       pinky_03_r_ctrl.Location.Y {type=MovieSceneScriptingFloatChannel, default=0, thinned_from=368, keys [[0,0],[367,0]]}
       pinky_03_r_ctrl.Location.Z {type=MovieSceneScriptingFloatChannel, default=0, thinned_from=368, keys [[0,0],[367,0]]}
       pinky_03_r_ctrl.Rotation.X {
        type=MovieSceneScriptingFloatChannel
        default=0
        thinned_from=368
        keys [[0,0],[33,0],[34,-0.0001],[35,-0.0261],[36,-0.077],[37,-0.1108],[38,-0.1796],[39,-0.3186],[40,-0.37],[41,-0.543],[42,-0.532],[43,-0.658],[44,-0.6931],[45,-0.7989],[46,-0.817],[47,-0.827],[48,-0.9185],[49,-0.842],[50,-0.9275],[51,-0.9344],[53,-0.9747],[54,-0.9605],[55,-0.994],[56,-0.9796],[57,-1.0127],[58,-0.998],[59,-1.0309],[60,-1.0309],[61,-1.0159],[62,-1.0484],[64,-1.0484],[65,-1.0331],[66,-1.0654],[67,-1.0654],[68,-1.0581],[84,-1.0581],[85,-1.0615],[184,-1.0615],[185,-1.0542],[186,-1.0612],[187,-1.0101],[188,-1.0475],[189,-0.9911],[190,-0.9755],[192,-0.929],[193,-0.8961],[194,-0.8528],[195,-0.7662],[196,-0.7631],[197,-0.7085],[198,-0.6501],[199,-0.6358],[200,-0.5699],[201,-0.5396],[202,-0.3354],[203,-0.473],[204,-0.4471],[205,-0.4193],[206,-0.3897],[207,-0.3582],[208,-0.3249],[209,-0.2899],[210,-0.253],[211,-0.2144],[212,-0.174],[213,-0.1319],[214,-0.088],[215,-0.088],[216,-0.0424],[217,-0.0521],[218,-0.0351],[219,-0.0171],[220,-0.0178],[221,0.0004],[222,-0.0001],[232,-0.0001],[367,0]]
       }
       pinky_03_r_ctrl.Rotation.Y {
        type=MovieSceneScriptingFloatChannel
        default=0
        thinned_from=368
        keys [[0,0],[33,0],[34,-0.0003],[35,-0.1468],[36,-0.4398],[37,-1.1039],[38,-2.0622],[39,-3.0927],[40,-4.2017],[41,-5.1612],[42,-5.9775],[43,-6.6445],[44,-7.2391],[45,-7.7599],[46,-8.207],[47,-8.5053],[48,-8.8789],[50,-9.3272],[51,-9.4543],[53,-9.8121],[54,-9.8119],[55,-9.9912],[56,-9.991],[57,-10.1703],[58,-10.1702],[59,-10.3496],[61,-10.3495],[62,-10.529],[65,-10.5289],[66,-10.7085],[67,-10.7085],[68,-10.6737],[84,-10.6737],[85,-10.7085],[184,-10.7085],[185,-10.6702],[186,-10.6703],[188,-10.4282],[189,-10.2236],[190,-9.982],[192,-9.3383],[193,-8.9366],[194,-8.455],[196,-7.5739],[198,-6.6151],[199,-6.1784],[200,-5.7004],[201,-5.365],[202,-4.8803],[203,-4.3799],[213,-1.1097],[214,-0.7839],[215,-0.7839],[216,-0.4583],[217,-0.3948],[218,-0.2633],[219,-0.1306],[220,-0.1318],[221,0.0009],[222,-0.0003],[232,-0.0003],[233,0],[367,0]]
       }
       pinky_03_r_ctrl.Rotation.Z {
        type=MovieSceneScriptingFloatChannel
        default=0
        thinned_from=368
        keys [[0,-59.2687],[34,-59.2316],[35,-59.3657],[36,-59.6345],[37,-60.3094],[39,-62.2336],[40,-63.3566],[42,-65.0784],[44,-66.2534],[46,-67.1502],[49,-68.0612],[54,-68.8101],[55,-68.8221],[56,-68.9837],[57,-68.9959],[58,-69.1579],[60,-69.1703],[61,-69.3326],[64,-69.3453],[65,-69.5079],[185,-69.5588],[189,-69.088],[192,-68.2251],[201,-64.3917],[202,-64.0667],[203,-63.4239],[214,-59.9872],[215,-59.9872],[216,-59.6824],[220,-59.369],[221,-59.368],[222,-59.2316],[296,-59.2687],[297,0],[367,0]]
       }
       pinky_03_r_ctrl.Scale.X {type=MovieSceneScriptingFloatChannel, default=1, const [1,368]}
       pinky_03_r_ctrl.Scale.Y {type=MovieSceneScriptingFloatChannel, default=1, const [1,368]}
       pinky_03_r_ctrl.Scale.Z {type=MovieSceneScriptingFloatChannel, default=1, const [1,368]}
       hand_r_ik_ctrl.Location.X {
        type=MovieSceneScriptingFloatChannel
        default=0
        thinned_from=368
        keys [[0,-12.7773],[1,-12.7773],[27,-14.0575],[38,-14.2745],[43,-14.2509],[45,-14.1322],[46,-14.0336],[47,-13.9065],[51,-13.2123],[61,-11.1866],[64,-10.6383],[66,-10.3311],[68,-10.083],[69,-9.9838],[70,-9.9068],[72,-9.8083],[74,-9.7816],[76,-9.8117],[80,-9.9954],[84,-10.2158],[87,-10.3073],[91,-10.2796],[98,-10.0887],[101,-10.0436],[104,-10.0806],[108,-10.2483],[111,-10.4625],[139,-13.2089],[145,-13.6512],[150,-13.8708],[155,-13.92],[163,-13.7443],[170,-13.451],[172,-13.4371],[177,-13.5008],[180,-13.4576],[182,-13.3501],[184,-13.169],[186,-12.9211],[189,-12.4649],[194,-11.5952],[195,-11.4379],[196,-11.3051],[198,-11.1251],[200,-10.9934],[201,-10.9561],[202,-10.9634],[203,-11.0183],[204,-11.1111],[205,-11.228],[207,-11.5291],[209,-11.8971],[218,-13.8619],[220,-14.2327],[227,-15.1827],[230,-15.6224],[231,-15.7399],[233,-15.8852],[234,-15.9183],[236,-15.886],[237,-15.831],[239,-15.626],[241,-15.3041],[243,-14.8994],[252,-12.7383],[254,-12.3092],[256,-11.9418],[258,-11.6437],[260,-11.41],[262,-11.243],[264,-11.1526],[266,-11.136],[269,-11.1949],[274,-11.4248],[293,-12.6422],[296,-12.7716],[297,-0.0004],[298,0],[367,0]]
       }
       hand_r_ik_ctrl.Location.Y {
        type=MovieSceneScriptingFloatChannel
        default=0
        thinned_from=368
        keys [[0,12.1988],[1,12.1988],[9,12.5645],[33,13.2136],[40,13.248],[43,13.193],[44,13.1337],[45,13.0446],[46,12.9143],[47,12.7487],[49,12.3108],[51,11.7917],[55,10.5825],[60,9.0014],[62,8.4168],[64,7.8808],[65,7.632],[66,7.4028],[67,7.1893],[68,6.9981],[69,6.8379],[70,6.7106],[71,6.6271],[72,6.5738],[73,6.5502],[74,6.5472],[75,6.5705],[76,6.614],[77,6.6749],[79,6.8403],[83,7.2289],[85,7.3914],[87,7.4903],[89,7.5236],[91,7.5021],[93,7.4313],[95,7.3341],[100,7.0359],[102,6.9666],[103,6.9538],[104,6.9562],[105,6.9746],[106,7.0113],[108,7.1409],[110,7.3522],[111,7.4797],[113,7.7874],[115,8.1371],[125,10.1134],[133,11.528],[139,12.4488],[143,12.9594],[146,13.2658],[149,13.4928],[153,13.6618],[157,13.6966],[161,13.6293],[170,13.2544],[176,13.2648],[178,13.1999],[179,13.1314],[180,13.0345],[181,12.9103],[183,12.603],[185,12.2476],[188,11.6317],[192,10.711],[194,10.1824],[195,9.8831],[196,9.5343],[197,9.1099],[198,8.614],[199,8.0435],[200,7.4456],[201,6.8967],[202,6.4975],[203,6.2824],[204,6.2395],[205,6.3383],[206,6.5522],[207,6.8598],[208,7.228],[209,7.6442],[210,8.0933],[211,8.5785],[213,9.6125],[217,11.7671],[219,12.8054],[221,13.7713],[222,14.2183],[224,15.0202],[226,15.711],[227,16.0081],[228,16.2583],[229,16.4667],[230,16.6399],[231,16.7692],[232,16.8537],[233,16.895],[234,16.89],[235,16.8397],[236,16.7448],[238,16.4341],[239,16.2244],[241,15.684],[243,15.0406],[250,12.4697],[253,11.4488],[255,10.869],[256,10.6224],[257,10.4009],[259,10.0263],[260,9.8764],[261,9.7464],[262,9.6381],[264,9.4956],[265,9.4502],[266,9.4302],[268,9.4473],[270,9.5306],[272,9.6613],[274,9.8353],[279,10.3803],[289,11.5686],[294,12.0521],[296,12.2017],[297,-0.0002],[298,0],[367,0]]
       }
       hand_r_ik_ctrl.Location.Z {
        type=MovieSceneScriptingFloatChannel
        default=0
        thinned_from=368
        keys [[0,-24.2316],[45,-25.3867],[52,-25.1953],[57,-24.7809],[66,-23.6868],[69,-23.4854],[71,-23.4434],[74,-23.5432],[85,-24.6871],[88,-24.8368],[92,-24.8023],[102,-24.0912],[111,-23.3148],[117,-23.1453],[123,-23.2926],[145,-24.9721],[152,-25.1553],[171,-24.3993],[180,-24.6753],[183,-24.926],[185,-25.2253],[187,-25.6467],[195,-27.8003],[197,-28.1952],[199,-28.447],[201,-28.5074],[203,-28.3505],[206,-27.805],[215,-25.6397],[218,-25.0466],[221,-24.6162],[226,-24.2554],[231,-24.149],[235,-24.3518],[244,-25.0725],[248,-25.1967],[253,-25.0632],[267,-23.9741],[279,-23.7421],[296,-24.236],[297,-0.0009],[298,0],[367,0]]
       }
       hand_r_ik_ctrl.Rotation.X {
        type=MovieSceneScriptingFloatChannel
        default=0
        thinned_from=368
        keys [[0,-34.6158],[46,-35.7022],[52,-35.3855],[68,-33.3442],[73,-33.0591],[90,-33.7243],[111,-32.9168],[126,-33.5386],[150,-35.2563],[178,-34.673],[191,-35.1344],[194,-34.9954],[196,-34.7721],[198,-34.3934],[201,-33.5786],[203,-33.1933],[205,-33.0243],[209,-33.1018],[215,-33.7106],[231,-36.4122],[235,-36.8209],[239,-36.9016],[243,-36.6013],[258,-34.2118],[265,-33.5991],[279,-33.7572],[296,-34.612],[297,-0.0007],[298,0],[367,0]]
       }
       hand_r_ik_ctrl.Rotation.Y {
        type=MovieSceneScriptingFloatChannel
        default=0
        thinned_from=368
        keys [[0,-21.3294],[1,-21.3294],[35,-23.454],[45,-23.7142],[48,-23.6515],[51,-23.4227],[53,-23.1409],[56,-22.5439],[64,-20.5715],[67,-19.9495],[69,-19.649],[70,-19.5409],[72,-19.445],[74,-19.5026],[76,-19.7054],[78,-20.0414],[84,-21.3251],[86,-21.6644],[88,-21.8771],[90,-21.949],[92,-21.8836],[94,-21.6868],[97,-21.2192],[107,-19.1479],[110,-18.6773],[112,-18.4762],[114,-18.375],[116,-18.3688],[119,-18.5271],[123,-18.9678],[128,-19.7674],[141,-22.229],[146,-22.9599],[150,-23.3273],[154,-23.4504],[159,-23.3079],[166,-22.7292],[192,-19.1154],[199,-18.4247],[203,-18.1596],[205,-18.1051],[207,-18.152],[209,-18.3069],[210,-18.4216],[212,-18.763],[214,-19.219],[217,-20.0693],[226,-23.0394],[229,-23.9056],[231,-24.3832],[233,-24.7581],[235,-25.0096],[237,-25.1224],[239,-25.1065],[241,-24.9649],[244,-24.541],[247,-23.9203],[259,-20.6962],[262,-20.0367],[264,-19.7022],[266,-19.4763],[268,-19.3495],[270,-19.3087],[275,-19.4726],[296,-21.3296],[297,-0.0008],[298,0],[367,0]]
       }
       hand_r_ik_ctrl.Rotation.Z {
        type=MovieSceneScriptingFloatChannel
        default=0
        thinned_from=368
        keys [[0,3.8437],[1,3.8437],[3,4.0588],[6,4.3545],[9,4.6162],[22,5.6191],[28,6.0193],[32,6.249],[37,6.4806],[41,6.6125],[44,6.6726],[45,6.6776],[46,6.6569],[47,6.6042],[48,6.5168],[49,6.4019],[50,6.2539],[51,6.0735],[52,5.8459],[53,5.5814],[54,5.2901],[55,4.9761],[56,4.6418],[57,4.2898],[59,3.56],[61,2.8148],[63,2.0883],[64,1.7373],[65,1.4031],[66,1.0835],[67,0.7869],[68,0.5131],[69,0.2733],[70,0.0741],[71,-0.0616],[72,-0.1564],[73,-0.2078],[74,-0.2184],[75,-0.1922],[76,-0.1282],[77,-0.0242],[78,0.1067],[79,0.2624],[80,0.4369],[81,0.622],[83,0.9986],[84,1.1739],[85,1.3362],[86,1.4777],[87,1.5872],[88,1.6709],[89,1.7224],[90,1.7447],[91,1.7378],[92,1.696],[93,1.6244],[94,1.5326],[95,1.4217],[96,1.2892],[97,1.1478],[98,0.9938],[99,0.8316],[100,0.6671],[101,0.4976],[103,0.1625],[104,-0.0015],[105,-0.1519],[106,-0.2928],[107,-0.4199],[108,-0.5334],[109,-0.625],[110,-0.6955],[111,-0.7479],[112,-0.7682],[113,-0.7692],[114,-0.7453],[115,-0.6961],[116,-0.6287],[117,-0.5348],[118,-0.4201],[119,-0.2919],[120,-0.1457],[121,0.0123],[122,0.1805],[123,0.3615],[124,0.5574],[125,0.7564],[126,0.9629],[127,1.1779],[129,1.6224],[134,2.7604],[136,3.208],[139,3.8556],[141,4.2587],[142,4.4505],[143,4.6324],[144,4.8031],[145,4.9627],[146,5.1098],[147,5.2425],[148,5.3632],[149,5.4709],[151,5.6298],[153,5.7191],[154,5.7418],[155,5.7528],[157,5.7244],[159,5.652],[161,5.5384],[163,5.3913],[165,5.2135],[167,5.0065],[169,4.7764],[172,4.3967],[174,4.1255],[175,4.0151],[176,3.9539],[177,3.9301],[178,3.9351],[179,3.9665],[180,4.0176],[181,4.0814],[183,4.244],[187,4.6019],[188,4.6738],[189,4.7282],[190,4.7563],[191,4.745],[192,4.6701],[193,4.528],[194,4.3374],[195,4.0841],[196,3.7557],[197,3.337],[198,2.8062],[199,2.1615],[200,1.4567],[201,0.7877],[202,0.2333],[203,-0.1686],[204,-0.4348],[205,-0.5785],[206,-0.6137],[207,-0.5737],[208,-0.476],[209,-0.3326],[210,-0.1492],[211,0.0891],[212,0.3648],[213,0.681],[214,1.0275],[215,1.4024],[216,1.8081],[217,2.2293],[218,2.6695],[219,3.1267],[220,3.5916],[224,5.4987],[226,6.4271],[227,6.8692],[228,7.2951],[229,7.6953],[230,8.071],[231,8.4168],[232,8.7245],[233,8.995],[234,9.2201],[235,9.3903],[236,9.5168],[237,9.5871],[238,9.6051],[239,9.5634],[240,9.4602],[241,9.3012],[242,9.0883],[243,8.8269],[244,8.5256],[245,8.1832],[246,7.8103],[247,7.4084],[249,6.5535],[253,4.7684],[255,3.9209],[256,3.5175],[257,3.1379],[258,2.7793],[259,2.4478],[260,2.1452],[261,1.8698],[262,1.6309],[263,1.4225],[264,1.2442],[265,1.0959],[266,0.9818],[267,0.9013],[268,0.8497],[269,0.8207],[270,0.8199],[271,0.8436],[272,0.8869],[273,0.9505],[274,1.0234],[275,1.1076],[276,1.2039],[277,1.3065],[278,1.4204],[279,1.5401],[281,1.8],[283,2.0758],[287,2.6516],[291,3.2115],[294,3.6025],[295,3.7278],[296,3.8445],[297,-0.0004],[298,0],[367,0]]
       }
       hand_r_ik_ctrl.Scale.X {type=MovieSceneScriptingFloatChannel, default=1, const [1,368]}
       hand_r_ik_ctrl.Scale.Y {type=MovieSceneScriptingFloatChannel, default=1, const [1,368]}
       hand_r_ik_ctrl.Scale.Z {type=MovieSceneScriptingFloatChannel, default=1, const [1,368]}
       Softness_2 {type=MovieSceneScriptingFloatChannel, default=0, const [0,368]}
       hand_r_ikorient_ctrl.Rotation.Roll {type=MovieSceneScriptingFloatChannel, default=0, const [0,368]}
       hand_r_ikorient_ctrl.Rotation.Pitch {type=MovieSceneScriptingFloatChannel, default=0, const [0,368]}
       hand_r_ikorient_ctrl.Rotation.Yaw {type=MovieSceneScriptingFloatChannel, default=0, const [0,368]}
       arm_r_pv_ik_ctrl.Location.X {
        type=MovieSceneScriptingFloatChannel
        default=0
        thinned_from=368
        keys [[0,-28.0832],[38,-28.9083],[43,-28.7841],[45,-28.5927],[47,-28.2481],[61,-24.37],[65,-23.487],[68,-22.9851],[70,-22.7664],[72,-22.6676],[76,-22.7067],[85,-23.1284],[100,-22.7932],[103,-22.8697],[106,-23.1062],[108,-23.3503],[112,-24.0396],[124,-26.431],[137,-28.2235],[147,-29.0773],[159,-29.2864],[171,-29.0374],[176,-29.1819],[179,-29.0339],[181,-28.7709],[184,-28.1406],[187,-27.307],[196,-24.3466],[198,-23.5647],[200,-22.6783],[201,-22.2876],[202,-22.0407],[203,-21.9607],[204,-22.0371],[205,-22.2386],[206,-22.5412],[207,-22.9289],[209,-23.8597],[217,-28.2479],[220,-29.705],[222,-30.5054],[225,-31.4565],[228,-32.155],[230,-32.4487],[232,-32.567],[234,-32.4958],[236,-32.2378],[239,-31.5713],[243,-30.2711],[250,-27.7551],[253,-26.8374],[255,-26.3417],[257,-25.9709],[260,-25.5945],[263,-25.4207],[266,-25.437],[273,-25.9122],[291,-27.769],[296,-28.0813],[297,-0.0004],[298,0],[367,0]]
       }
       arm_r_pv_ik_ctrl.Location.Y {
        type=MovieSceneScriptingFloatChannel
        default=0
        thinned_from=368
        keys [[0,-13.8402],[43,-14.6364],[45,-14.5892],[55,-13.8108],[64,-12.8563],[67,-12.6202],[69,-12.5332],[71,-12.5069],[74,-12.573],[83,-13.0557],[86,-13.1386],[105,-12.8926],[111,-12.9576],[126,-13.4974],[142,-14.2663],[149,-14.4033],[159,-14.2759],[169,-13.9625],[170,-13.9488],[172,-13.9998],[174,-14.1316],[182,-14.879],[184,-15.1226],[186,-15.4474],[194,-17.1603],[196,-17.5309],[198,-17.8099],[199,-17.9065],[200,-17.9575],[201,-17.971],[203,-17.8944],[206,-17.572],[210,-16.9306],[220,-15.058],[223,-14.5943],[226,-14.2358],[229,-14.0054],[231,-13.9177],[233,-13.892],[236,-13.9537],[243,-14.1786],[248,-14.2085],[253,-14.0748],[265,-13.6005],[276,-13.5019],[296,-13.8427],[297,-0.0008],[298,0],[367,0]]
       }
       arm_r_pv_ik_ctrl.Location.Z {
        type=MovieSceneScriptingFloatChannel
        default=0
        thinned_from=368
        keys [[0,-3.2945],[1,-3.2945],[4,-3.172],[9,-2.9349],[16,-2.625],[22,-2.3813],[25,-2.2786],[27,-2.2165],[30,-2.1429],[32,-2.104],[34,-2.0725],[36,-2.0504],[37,-2.0362],[38,-2.0284],[41,-2.0227],[43,-2.0298],[45,-2.0576],[46,-2.084],[47,-2.1273],[48,-2.1897],[49,-2.2607],[50,-2.3457],[51,-2.4356],[53,-2.6406],[55,-2.8623],[56,-2.9656],[61,-3.5072],[64,-3.8158],[66,-3.994],[67,-4.0723],[69,-4.2068],[70,-4.2568],[72,-4.3185],[73,-4.3287],[75,-4.3198],[76,-4.304],[80,-4.1832],[83,-4.0842],[86,-4.0376],[87,-4.0369],[89,-4.0631],[92,-4.1364],[98,-4.3457],[101,-4.4343],[104,-4.4939],[107,-4.5217],[110,-4.5169],[112,-4.4783],[115,-4.3831],[119,-4.2119],[129,-3.6704],[137,-3.2222],[139,-3.119],[143,-2.9431],[146,-2.844],[147,-2.817],[149,-2.7793],[151,-2.7617],[153,-2.7606],[156,-2.7792],[159,-2.8215],[162,-2.8863],[166,-2.9987],[169,-3.106],[171,-3.1907],[174,-3.3449],[175,-3.3879],[176,-3.4203],[177,-3.4408],[178,-3.4474],[179,-3.4374],[182,-3.3929],[183,-3.4001],[184,-3.4193],[185,-3.463],[186,-3.5227],[187,-3.6005],[188,-3.6927],[189,-3.7999],[190,-3.9293],[191,-4.0761],[192,-4.2396],[194,-4.604],[195,-4.7761],[196,-4.9222],[197,-5.0195],[198,-5.1053],[199,-5.1729],[200,-5.2223],[201,-5.2453],[202,-5.2373],[203,-5.2025],[204,-5.1466],[205,-5.0751],[206,-4.9857],[207,-4.8836],[208,-4.7655],[209,-4.6335],[210,-4.4886],[211,-4.3328],[213,-3.9939],[218,-3.1082],[220,-2.7974],[221,-2.6642],[223,-2.4184],[224,-2.3047],[226,-2.0523],[227,-1.923],[228,-1.7625],[229,-1.6072],[230,-1.4617],[231,-1.3287],[232,-1.2248],[233,-1.1426],[234,-1.0844],[235,-1.0554],[236,-1.0524],[237,-1.0678],[238,-1.111],[239,-1.1725],[240,-1.2572],[241,-1.3651],[242,-1.4899],[243,-1.6273],[244,-1.7767],[245,-1.9413],[246,-2.1128],[248,-2.464],[250,-2.8262],[252,-3.173],[254,-3.4964],[255,-3.6453],[256,-3.784],[257,-3.9131],[258,-4.0326],[259,-4.1392],[260,-4.2311],[261,-4.3088],[262,-4.3695],[263,-4.414],[264,-4.4425],[265,-4.451],[266,-4.4454],[268,-4.413],[271,-4.3296],[274,-4.216],[287,-3.6071],[291,-3.4493],[294,-3.3538],[296,-3.3012],[297,0.0002],[367,0]]
       }
       arm_r_pv_ik_ctrl.Rotation.X {
        type=MovieSceneScriptingFloatChannel
        default=0
        thinned_from=368
        keys [[0,-1.2558],[1,-1.2558],[2,-1.1887],[4,-1.0618],[9,-0.7586],[12,-0.5937],[13,-0.5418],[14,-0.4917],[15,-0.4367],[16,-0.383],[17,-0.332],[19,-0.2337],[21,-0.1382],[22,-0.0917],[23,-0.0463],[24,-0.0026],[25,0.043],[27,0.125],[28,0.1639],[29,0.1991],[30,0.2356],[31,0.271],[32,0.3039],[33,0.3311],[34,0.3617],[35,0.3896],[36,0.4139],[37,0.4373],[38,0.4579],[39,0.4753],[40,0.4902],[41,0.5034],[43,0.5231],[44,0.5269],[45,0.5233],[46,0.4978],[47,0.4369],[48,0.3422],[49,0.2194],[50,0.0676],[51,-0.108],[52,-0.3087],[53,-0.5289],[54,-0.7621],[55,-1.0037],[56,-1.2546],[57,-1.5114],[58,-1.7717],[61,-2.5381],[63,-3.0103],[64,-3.2317],[65,-3.4356],[66,-3.6266],[67,-3.7952],[68,-3.9415],[69,-4.0584],[70,-4.1424],[71,-4.1755],[72,-4.1717],[73,-4.1351],[74,-4.0694],[75,-3.9812],[76,-3.8748],[78,-3.6212],[81,-3.2117],[82,-3.0866],[83,-2.9724],[84,-2.8808],[85,-2.8071],[86,-2.7597],[87,-2.7453],[88,-2.7579],[89,-2.7926],[90,-2.8445],[91,-2.9078],[92,-2.9893],[93,-3.0838],[94,-3.1874],[95,-3.2992],[101,-4.0528],[103,-4.2872],[104,-4.3954],[105,-4.4895],[106,-4.5723],[107,-4.6417],[108,-4.6947],[109,-4.7309],[110,-4.7475],[111,-4.7456],[112,-4.7167],[113,-4.6698],[114,-4.6098],[115,-4.536],[117,-4.3534],[118,-4.2445],[119,-4.1271],[121,-3.866],[123,-3.5786],[126,-3.109],[129,-2.619],[130,-2.4509],[133,-1.9667],[135,-1.659],[136,-1.5076],[138,-1.2233],[139,-1.083],[141,-0.8321],[142,-0.7168],[143,-0.6082],[144,-0.5105],[145,-0.4231],[146,-0.344],[147,-0.2756],[148,-0.2199],[149,-0.1723],[150,-0.1443],[151,-0.1234],[152,-0.1196],[153,-0.1207],[154,-0.1309],[155,-0.1464],[156,-0.1687],[157,-0.198],[158,-0.2372],[159,-0.2781],[160,-0.3285],[161,-0.3835],[162,-0.4446],[163,-0.5082],[165,-0.6505],[166,-0.7268],[167,-0.8072],[168,-0.8926],[170,-1.0702],[171,-1.1623],[174,-1.4509],[175,-1.5291],[176,-1.5732],[177,-1.5927],[178,-1.5905],[179,-1.5695],[180,-1.5342],[181,-1.4894],[182,-1.4333],[183,-1.3729],[186,-1.1731],[187,-1.1101],[188,-1.0528],[189,-1.0066],[190,-0.9809],[191,-0.9808],[192,-1.025],[193,-1.1139],[194,-1.2387],[195,-1.403],[196,-1.6213],[197,-1.9024],[198,-2.2608],[199,-2.6963],[200,-3.1695],[201,-3.6185],[202,-3.9846],[203,-4.2469],[204,-4.4135],[205,-4.4888],[206,-4.478],[207,-4.4052],[208,-4.2848],[209,-4.1257],[210,-3.934],[211,-3.714],[212,-3.4686],[213,-3.2016],[214,-2.9183],[215,-2.6196],[216,-2.3083],[217,-1.9883],[218,-1.6608],[221,-0.664],[222,-0.3346],[223,-0.011],[224,0.303],[225,0.6072],[226,0.8981],[227,1.1725],[228,1.4303],[229,1.6653],[230,1.877],[231,2.0633],[232,2.2187],[233,2.3459],[234,2.4374],[235,2.4899],[236,2.5054],[237,2.4869],[238,2.4338],[239,2.3441],[240,2.2201],[241,2.0622],[242,1.8737],[243,1.6599],[244,1.4183],[245,1.1571],[246,0.8775],[247,0.583],[248,0.28],[249,-0.0294],[250,-0.3418],[251,-0.6532],[252,-0.9609],[253,-1.2615],[254,-1.5506],[255,-1.8296],[256,-2.0924],[257,-2.3367],[258,-2.5644],[259,-2.7694],[260,-2.9511],[261,-3.1096],[262,-3.242],[263,-3.3469],[264,-3.4293],[265,-3.4837],[266,-3.5102],[267,-3.5156],[268,-3.5109],[269,-3.4937],[270,-3.4676],[271,-3.4307],[272,-3.3856],[273,-3.3298],[274,-3.267],[276,-3.1183],[277,-3.035],[279,-2.8539],[282,-2.5594],[288,-1.9518],[291,-1.6672],[293,-1.4915],[295,-1.3295],[296,-1.2536],[297,0.0003],[298,0],[367,0]]
       }
       arm_r_pv_ik_ctrl.Rotation.Y {
        type=MovieSceneScriptingFloatChannel
        default=0
        thinned_from=368
        keys [[0,-13.6454],[45,-13.8746],[49,-13.7884],[56,-13.3803],[70,-12.2769],[77,-12.0045],[88,-11.8681],[101,-12.0204],[132,-13.152],[175,-13.6157],[177,-13.7547],[179,-13.9751],[188,-15.3166],[190,-15.5505],[191,-15.628],[192,-15.6601],[193,-15.6419],[194,-15.5911],[195,-15.4903],[196,-15.3352],[197,-15.1133],[198,-14.812],[199,-14.4267],[201,-13.5776],[202,-13.2347],[203,-12.9868],[204,-12.8237],[205,-12.733],[206,-12.7015],[208,-12.7344],[211,-12.8968],[228,-14.2854],[234,-14.605],[238,-14.6903],[241,-14.6333],[245,-14.3879],[256,-13.345],[260,-13.0982],[266,-12.9403],[273,-12.9769],[296,-13.6444],[297,0.0001],[367,0]]
       }
       arm_r_pv_ik_ctrl.Rotation.Z {
        type=MovieSceneScriptingFloatChannel
        default=0
        thinned_from=368
        keys [[0,-43.8947],[47,-44.6498],[68,-42.9352],[75,-43.0235],[86,-43.682],[111,-42.5675],[150,-44.4908],[196,-43.2768],[203,-42.6227],[208,-42.6751],[231,-45.2547],[238,-45.3693],[262,-43.1038],[296,-43.8917],[297,-0.0011],[298,0],[367,0]]
       }
       arm_r_pv_ik_ctrl.Scale.X {type=MovieSceneScriptingFloatChannel, default=1, const [1,368]}
       arm_r_pv_ik_ctrl.Scale.Y {type=MovieSceneScriptingFloatChannel, default=1, const [1,368]}
       arm_r_pv_ik_ctrl.Scale.Z {type=MovieSceneScriptingFloatChannel, default=1, const [1,368]}
       body_ctrl.Location.X {
        type=MovieSceneScriptingFloatChannel
        default=0
        thinned_from=368
        keys [[0,-4.104],[7,-4.1141],[22,-4.2253],[42,-4.4895],[47,-4.5995],[50,-4.6967],[59,-5.0452],[64,-5.1798],[69,-5.2578],[100,-5.3618],[103,-5.3328],[105,-5.2824],[108,-5.1636],[118,-4.6309],[121,-4.5026],[125,-4.3743],[133,-4.2203],[148,-4.032],[170,-3.8648],[173,-3.8585],[174,-3.8653],[175,-3.8915],[176,-3.9293],[177,-3.9861],[179,-4.1282],[181,-4.3121],[183,-4.5442],[185,-4.8146],[187,-5.1178],[190,-5.595],[191,-5.7257],[193,-5.9482],[195,-6.1354],[198,-6.3576],[200,-6.47],[201,-6.5104],[202,-6.5047],[203,-6.4355],[204,-6.315],[205,-6.1545],[206,-5.9623],[207,-5.7467],[208,-5.5069],[209,-5.2548],[213,-4.1849],[215,-3.6726],[216,-3.4343],[217,-3.212],[218,-3.0098],[219,-2.8295],[220,-2.6728],[221,-2.545],[222,-2.4466],[223,-2.3789],[224,-2.3235],[225,-2.2802],[226,-2.2513],[227,-2.237],[228,-2.2392],[229,-2.2623],[230,-2.2993],[231,-2.3547],[232,-2.4242],[233,-2.5088],[234,-2.6057],[235,-2.7112],[237,-2.9561],[239,-3.2227],[244,-3.9279],[246,-4.1783],[248,-4.3904],[250,-4.5769],[252,-4.7297],[253,-4.7965],[255,-4.8916],[256,-4.9124],[259,-4.9131],[262,-4.8786],[269,-4.7245],[274,-4.5825],[286,-4.2433],[293,-4.1186],[295,-4.1043],[296,-4.1041],[297,-0.0001],[367,0]]
       }
       body_ctrl.Location.Y {
        type=MovieSceneScriptingFloatChannel
        default=0
        thinned_from=368
        keys [[0,-8.3815],[43,-8.3561],[46,-8.3994],[50,-8.5492],[57,-8.8749],[61,-8.9841],[99,-9.109],[106,-8.9998],[125,-8.4171],[167,-8.0341],[173,-8.0323],[176,-8.0923],[179,-8.2059],[181,-8.3147],[183,-8.478],[184,-8.5803],[186,-8.8433],[187,-8.9985],[189,-9.3627],[191,-9.78],[196,-10.9341],[198,-11.3449],[199,-11.5212],[200,-11.6679],[201,-11.7862],[202,-11.8763],[203,-11.9385],[204,-11.9713],[205,-11.9741],[206,-11.9506],[208,-11.8264],[210,-11.613],[212,-11.3312],[215,-10.8162],[222,-9.4873],[224,-9.1432],[226,-8.8385],[228,-8.5847],[229,-8.4752],[231,-8.3077],[232,-8.2438],[234,-8.1618],[236,-8.1281],[296,-8.3815],[297,-0.0002],[367,0]]
       }
       body_ctrl.Location.Z {
        type=MovieSceneScriptingFloatChannel
        default=0
        thinned_from=368
        keys [[0,0.4356],[5,0.4372],[14,0.4473],[28,0.4665],[48,0.4757],[51,0.4734],[54,0.467],[64,0.4298],[67,0.4212],[72,0.4117],[83,0.4008],[98,0.3976],[103,0.3993],[108,0.4057],[122,0.4311],[127,0.4359],[173,0.4426],[174,0.4446],[175,0.4498],[176,0.4602],[177,0.4737],[178,0.496],[179,0.5235],[181,0.5893],[182,0.6207],[183,0.6435],[184,0.6615],[185,0.673],[186,0.6805],[187,0.6851],[188,0.6842],[189,0.6742],[190,0.6532],[191,0.6184],[192,0.5693],[193,0.5108],[194,0.4479],[195,0.3893],[196,0.357],[197,0.3618],[198,0.3781],[199,0.4069],[200,0.4459],[201,0.4869],[202,0.5293],[203,0.5687],[204,0.6028],[205,0.6313],[206,0.6544],[207,0.6723],[208,0.6853],[209,0.6945],[210,0.7],[211,0.7023],[212,0.7016],[213,0.6975],[214,0.6895],[215,0.6775],[216,0.6609],[217,0.6391],[218,0.6118],[219,0.5778],[220,0.5377],[221,0.4919],[222,0.4423],[223,0.3948],[224,0.3595],[225,0.3459],[226,0.3463],[227,0.3513],[228,0.3611],[230,0.3875],[231,0.3997],[232,0.41],[233,0.4179],[234,0.4239],[235,0.4284],[236,0.4315],[249,0.4483],[261,0.451],[296,0.4355],[297,0],[367,0]]
       }
       body_ctrl.Rotation.X {
        type=MovieSceneScriptingFloatChannel
        default=0
        thinned_from=368
        keys [[0,-14.7222],[3,-14.686],[8,-14.4507],[13,-14.0507],[24,-12.957],[29,-12.5867],[34,-12.3747],[41,-12.2818],[50,-12.4084],[59,-12.774],[62,-12.9763],[71,-13.7039],[80,-14.0851],[170,-14.8127],[189,-14.4029],[204,-14.5655],[226,-15.4428],[232,-15.38],[243,-14.9031],[296,-14.7262],[297,0.0003],[298,0],[367,0]]
       }
       body_ctrl.Rotation.Y {
        type=MovieSceneScriptingFloatChannel
        default=0
        thinned_from=368
        keys [[0,-6.5801],[38,-6.6302],[41,-6.5838],[45,-6.441],[48,-6.285],[53,-5.9884],[55,-5.9036],[58,-5.8178],[94,-5.6808],[99,-5.6248],[102,-5.6345],[105,-5.6939],[110,-5.8721],[120,-6.3585],[123,-6.4752],[126,-6.547],[170,-6.7891],[174,-6.7828],[175,-6.8011],[176,-6.8348],[177,-6.8872],[181,-7.1634],[182,-7.2031],[183,-7.2113],[184,-7.1783],[185,-7.1],[186,-6.9783],[187,-6.811],[188,-6.6092],[191,-5.9215],[192,-5.7184],[193,-5.5472],[194,-5.4117],[195,-5.3118],[196,-5.2438],[198,-5.1505],[199,-5.119],[208,-5.0346],[210,-5.0389],[212,-5.0659],[217,-5.1998],[231,-5.7174],[233,-5.7487],[235,-5.7451],[237,-5.708],[239,-5.623],[241,-5.4963],[245,-5.1978],[247,-5.0734],[248,-5.0248],[249,-4.988],[250,-4.962],[251,-4.9475],[252,-4.9473],[254,-4.9784],[257,-5.0876],[275,-6.0945],[280,-6.3145],[286,-6.4838],[293,-6.5732],[296,-6.5803],[297,-0.0001],[367,0]]
       }
       body_ctrl.Rotation.Z {
        type=MovieSceneScriptingFloatChannel
        default=0
        thinned_from=368
        keys [[0,42.2812],[47,42.2266],[50,41.9359],[63,38.9321],[67,38.2953],[99,37.6621],[103,37.9942],[122,41.8332],[173,43.335],[175,43.1793],[177,42.7256],[179,42.0054],[182,40.5513],[188,37.2533],[190,36.3113],[192,35.5304],[194,34.9317],[196,34.5238],[198,34.3001],[201,34.2255],[203,34.3779],[205,34.7197],[207,35.2496],[210,36.3384],[218,39.6912],[223,41.3679],[227,42.3376],[231,42.9478],[233,43.0847],[235,43.038],[242,42.1923],[296,42.2788],[297,0.0011],[298,0],[367,0]]
       }
       body_ctrl.Scale.X {type=MovieSceneScriptingFloatChannel, default=1, const [1,368]}
       body_ctrl.Scale.Y {type=MovieSceneScriptingFloatChannel, default=1, const [1,368]}
       body_ctrl.Scale.Z {type=MovieSceneScriptingFloatChannel, default=1, const [1,368]}
       hips_ctrl.Location.X {type=MovieSceneScriptingFloatChannel, default=0, thinned_from=368, keys [[0,0],[367,0]]}
       hips_ctrl.Location.Y {type=MovieSceneScriptingFloatChannel, default=0, thinned_from=368, keys [[0,0],[367,0]]}
       hips_ctrl.Location.Z {type=MovieSceneScriptingFloatChannel, default=0, thinned_from=368, keys [[0,0],[367,0]]}
       hips_ctrl.Rotation.X {
        type=MovieSceneScriptingFloatChannel
        default=0
        thinned_from=368
        keys [[0,21.4145],[7,21.2432],[28,19.711],[38,19.4158],[49,19.5085],[59,20.023],[73,21.0452],[174,21.3343],[188,21.055],[296,21.4187],[297,0],[367,0]]
       }
       hips_ctrl.Rotation.Y {
        type=MovieSceneScriptingFloatChannel
        default=0
        thinned_from=368
        keys [[0,1.0203],[12,1.0278],[42,1.0898],[43,1.0938],[44,1.1004],[45,1.115],[46,1.1336],[47,1.1563],[49,1.2054],[50,1.2279],[51,1.2478],[52,1.2649],[53,1.2792],[54,1.2907],[56,1.306],[86,1.3176],[89,1.3193],[92,1.3096],[94,1.2961],[98,1.2537],[103,1.185],[110,1.0847],[113,1.0508],[115,1.0323],[117,1.02],[121,1.0014],[122,1.0015],[125,0.9883],[126,0.9885],[128,0.9796],[129,0.9798],[130,0.9754],[131,0.9754],[132,0.9711],[133,0.9711],[134,0.9666],[139,0.9599],[148,0.9505],[150,0.9505],[151,0.9486],[165,0.9472],[166,0.9514],[167,0.9514],[168,0.9554],[170,0.9544],[172,0.9502],[173,0.946],[174,0.931],[175,0.8944],[176,0.8342],[177,0.7523],[178,0.6489],[179,0.5282],[180,0.3924],[182,0.1164],[183,-0.0131],[184,-0.1318],[185,-0.2372],[186,-0.3263],[187,-0.4015],[188,-0.4615],[189,-0.5074],[190,-0.5422],[191,-0.5658],[192,-0.5783],[193,-0.5824],[194,-0.5781],[195,-0.5668],[196,-0.5501],[197,-0.5222],[198,-0.4734],[199,-0.4025],[200,-0.3078],[201,-0.1898],[202,-0.0477],[203,0.1147],[204,0.2926],[205,0.4821],[206,0.6744],[207,0.862],[208,1.0359],[209,1.1944],[210,1.3336],[211,1.4534],[212,1.554],[213,1.6352],[214,1.698],[215,1.7444],[216,1.7753],[217,1.7921],[218,1.7988],[219,1.7922],[220,1.7791],[221,1.7571],[223,1.6934],[224,1.6538],[227,1.5173],[229,1.4208],[231,1.3134],[232,1.252],[233,1.1872],[234,1.1182],[235,1.0446],[236,0.9687],[237,0.895],[238,0.8236],[239,0.7603],[240,0.7085],[241,0.6682],[242,0.6394],[243,0.6222],[244,0.6141],[245,0.6152],[246,0.6244],[247,0.6405],[248,0.6635],[249,0.6914],[250,0.723],[251,0.7585],[252,0.798],[253,0.8396],[256,0.9699],[257,1.0094],[259,1.0784],[260,1.1081],[261,1.1338],[262,1.1535],[263,1.1693],[264,1.1772],[265,1.1813],[268,1.1765],[272,1.1571],[276,1.1272],[284,1.0621],[289,1.0343],[293,1.0231],[296,1.0205],[297,0],[367,0]]
       }
       hips_ctrl.Rotation.Z {
        type=MovieSceneScriptingFloatChannel
        default=0
        thinned_from=368
        keys [[0,0.4001],[30,0.3819],[42,0.3839],[43,0.3852],[44,0.3879],[45,0.3933],[46,0.4002],[47,0.4086],[50,0.4358],[52,0.4507],[54,0.4622],[56,0.4707],[73,0.5052],[85,0.5165],[89,0.5182],[92,0.5149],[94,0.5095],[96,0.5021],[98,0.493],[110,0.4259],[113,0.4123],[115,0.4052],[117,0.4002],[121,0.3931],[122,0.3922],[124,0.3888],[126,0.3871],[127,0.3854],[129,0.3837],[130,0.3819],[131,0.3819],[132,0.3802],[133,0.3802],[139,0.3759],[154,0.3716],[155,0.3708],[164,0.3708],[166,0.3724],[167,0.3724],[168,0.3739],[170,0.3741],[172,0.3718],[173,0.3695],[174,0.3635],[175,0.349],[176,0.3251],[177,0.2929],[178,0.2523],[179,0.205],[180,0.1523],[181,0.0983],[182,0.0449],[183,-0.0051],[184,-0.0509],[185,-0.0913],[186,-0.1258],[187,-0.1546],[188,-0.1775],[189,-0.1956],[190,-0.2088],[191,-0.2181],[192,-0.2233],[193,-0.2251],[194,-0.2239],[195,-0.22],[196,-0.2137],[197,-0.2029],[198,-0.1843],[199,-0.1567],[200,-0.1202],[201,-0.074],[202,-0.0186],[203,0.0447],[204,0.1142],[205,0.1881],[206,0.2634],[207,0.3366],[208,0.4049],[209,0.4668],[210,0.5213],[211,0.5682],[212,0.6076],[213,0.6392],[214,0.6638],[215,0.6819],[216,0.694],[217,0.7009],[218,0.703],[219,0.7008],[220,0.6953],[221,0.6868],[223,0.6618],[225,0.6302],[227,0.5943],[229,0.5561],[231,0.5133],[232,0.4889],[233,0.4634],[234,0.4362],[235,0.4074],[237,0.3488],[238,0.3212],[239,0.2966],[240,0.2764],[241,0.2606],[242,0.2496],[243,0.2428],[244,0.2397],[245,0.2402],[246,0.2439],[247,0.2505],[248,0.2594],[249,0.2706],[250,0.283],[251,0.2973],[253,0.3287],[255,0.3635],[256,0.3802],[257,0.3957],[258,0.41],[259,0.4232],[260,0.4348],[261,0.4449],[262,0.4526],[263,0.4588],[264,0.4623],[265,0.4637],[268,0.4617],[272,0.4541],[284,0.4167],[289,0.4057],[294,0.4007],[296,0.4003],[297,0],[367,0]]
       }
       hips_ctrl.Scale.X {type=MovieSceneScriptingFloatChannel, default=1, const [1,368]}
       hips_ctrl.Scale.Y {type=MovieSceneScriptingFloatChannel, default=1, const [1,368]}
       hips_ctrl.Scale.Z {type=MovieSceneScriptingFloatChannel, default=1, const [1,368]}
       hips_tan_ctrl.Location.X {
        type=MovieSceneScriptingFloatChannel
        default=0
        thinned_from=368
        keys [[0,-0.011],[3,-0.0109],[10,-0.0108],[16,-0.0108],[26,-0.0108],[27,-0.0113],[33,-0.0112],[34,-0.0114],[41,-0.0116],[42,-0.0121],[43,-0.0122],[44,-0.013],[45,-0.0113],[46,-0.0107],[47,-0.0124],[48,-0.0121],[49,-0.0138],[50,-0.0113],[51,-0.0131],[52,-0.0136],[53,-0.0137],[54,-0.0135],[56,-0.0141],[57,-0.0136],[59,-0.0142],[62,-0.014],[63,-0.0143],[66,-0.014],[67,-0.0144],[70,-0.0142],[71,-0.0144],[78,-0.0146],[79,-0.0144],[82,-0.0146],[84,-0.0145],[85,-0.0135],[86,-0.0135],[87,-0.0138],[89,-0.0138],[90,-0.0135],[91,-0.0133],[92,-0.0129],[93,-0.0128],[94,-0.0122],[95,-0.0112],[96,-0.0106],[97,-0.0152],[98,-0.0145],[101,-0.0111],[102,-0.0133],[103,-0.0122],[104,-0.0118],[105,-0.0134],[110,-0.0113],[111,-0.013],[112,-0.012],[116,-0.011],[117,-0.0102],[118,-0.0107],[119,-0.0108],[135,-0.0081],[136,-0.0102],[152,-0.0097],[153,-0.01],[168,-0.0102],[169,-0.0095],[170,-0.0095],[171,-0.0089],[172,-0.0091],[173,-0.0084],[174,-0.0107],[175,-0.0073],[176,-0.0094],[177,-0.0084],[178,-0.0089],[179,-0.012],[180,-0.0087],[181,-0.0094],[182,-0.0097],[183,-0.0109],[184,-0.0134],[185,-0.0111],[186,-0.0166],[187,-0.015],[188,-0.015],[189,-0.0164],[190,-0.0106],[191,-0.0148],[192,-0.0112],[193,-0.0176],[194,-0.016],[195,-0.0155],[196,-0.0158],[197,-0.0175],[198,-0.0132],[199,-0.0116],[200,-0.0129],[201,-0.0154],[202,-0.0152],[203,-0.0169],[204,-0.011],[205,-0.0098],[206,-0.0132],[207,-0.0115],[208,-0.0056],[209,-0.0112],[210,-0.0118],[211,-0.0111],[212,-0.0087],[213,-0.014],[214,-0.0062],[215,-0.0095],[216,-0.0078],[217,-0.007],[218,-0.0089],[219,-0.0057],[221,-0.0076],[222,-0.0061],[223,-0.0064],[225,-0.0055],[226,-0.0082],[227,-0.0057],[228,-0.0068],[229,-0.0076],[230,-0.0065],[231,-0.005],[232,-0.0069],[233,-0.0064],[234,-0.0143],[235,-0.0028],[236,-0.01],[237,-0.012],[238,-0.0002],[239,-0.008],[240,0.0013],[241,-0.0073],[242,0.0007],[243,-0.0093],[244,-0.009],[245,-0.0023],[246,-0.0138],[247,-0.0081],[248,-0.0095],[249,-0.0089],[250,-0.0088],[251,-0.0072],[253,-0.0104],[254,-0.0107],[255,-0.0126],[256,-0.0091],[257,-0.0109],[258,-0.0108],[259,-0.0125],[260,-0.0122],[261,-0.0131],[262,-0.0106],[264,-0.0114],[265,-0.0122],[267,-0.0121],[268,-0.0119],[269,-0.011],[271,-0.0108],[274,-0.0101],[275,-0.0091],[278,-0.0083],[279,-0.0071],[280,-0.0069],[281,-0.0115],[285,-0.0106],[286,-0.0085],[289,-0.008],[296,-0.0076],[297,0],[367,0]]
       }
       hips_tan_ctrl.Location.Y {
        type=MovieSceneScriptingFloatChannel
        default=0.0032
        thinned_from=368
        keys [[0,0.047],[1,0.047],[2,0.0468],[3,0.0458],[4,0.0463],[5,0.0452],[6,0.0448],[7,0.045],[10,0.0438],[11,0.043],[12,0.0414],[13,0.0419],[14,0.0414],[15,0.0404],[16,0.0404],[17,0.0395],[19,0.0396],[21,0.0383],[22,0.0362],[23,0.037],[24,0.0368],[25,0.0378],[26,0.0365],[27,0.0373],[28,0.0342],[30,0.0366],[31,0.034],[32,0.0348],[34,0.0342],[35,0.0349],[37,0.0342],[41,0.0337],[42,0.0354],[43,0.0354],[44,0.0345],[46,0.0345],[47,0.0349],[49,0.0348],[50,0.0354],[52,0.0356],[53,0.0361],[54,0.036],[55,0.0364],[56,0.0365],[57,0.0376],[59,0.0389],[60,0.0384],[61,0.0392],[62,0.0394],[63,0.04],[65,0.0419],[67,0.0431],[68,0.0432],[69,0.0435],[70,0.0446],[71,0.0445],[72,0.0448],[73,0.0455],[75,0.0463],[76,0.0463],[77,0.0467],[78,0.0467],[79,0.0473],[80,0.0472],[81,0.0477],[82,0.0477],[83,0.048],[88,0.0487],[92,0.0488],[93,0.046],[96,0.0463],[97,0.0466],[101,0.0469],[102,0.0478],[103,0.0474],[104,0.0475],[106,0.0472],[111,0.0473],[113,0.0471],[135,0.0467],[136,0.047],[168,0.0471],[169,0.0469],[170,0.0471],[174,0.0468],[177,0.046],[179,0.045],[183,0.042],[188,0.0397],[189,0.0399],[191,0.0397],[192,0.0401],[193,0.0401],[194,0.0408],[196,0.0419],[197,0.0416],[198,0.0418],[200,0.0418],[201,0.042],[207,0.042],[216,0.0429],[218,0.0436],[219,0.0437],[220,0.0443],[221,0.0445],[222,0.0457],[224,0.0462],[226,0.0485],[227,0.0498],[229,0.0496],[230,0.0493],[231,0.0485],[233,0.0484],[234,0.048],[236,0.0475],[237,0.0476],[238,0.0468],[239,0.0472],[240,0.0471],[242,0.0475],[243,0.0473],[244,0.0473],[245,0.0471],[247,0.0473],[248,0.0469],[249,0.0473],[251,0.0469],[253,0.0474],[254,0.047],[255,0.0472],[256,0.0471],[257,0.0473],[260,0.0476],[261,0.0471],[280,0.0469],[296,0.0468],[297,0.0032],[367,0.0032]]
       }
       hips_tan_ctrl.Location.Z {
        type=MovieSceneScriptingFloatChannel
        default=-0.0015
        thinned_from=368
        keys [[0,-0.0836],[1,-0.0836],[4,-0.0717],[5,-0.0652],[7,-0.0571],[8,-0.0517],[9,-0.0477],[10,-0.0449],[11,-0.0408],[12,-0.0323],[13,-0.0351],[16,-0.0227],[17,-0.0203],[19,-0.0142],[21,-0.0109],[22,-0.0036],[23,-0.0053],[24,0.0019],[26,-0.0015],[27,0.0058],[28,0.0067],[29,0.0135],[30,0.0118],[31,0.0095],[32,0.019],[33,0.0167],[34,0.0177],[36,0.0181],[37,0.02],[41,0.0202],[43,0.02],[44,0.018],[45,0.0178],[47,0.0169],[48,0.0147],[49,0.0142],[50,0.0119],[51,0.0121],[52,0.0058],[53,0.0078],[55,-0.0027],[56,0.0013],[61,-0.0199],[62,-0.0231],[63,-0.0286],[64,-0.0387],[65,-0.0419],[66,-0.0462],[67,-0.0494],[68,-0.0521],[69,-0.0553],[71,-0.0662],[72,-0.0668],[73,-0.0722],[76,-0.0808],[77,-0.0859],[78,-0.0862],[79,-0.089],[80,-0.0893],[81,-0.0944],[82,-0.0947],[83,-0.0973],[84,-0.0976],[85,-0.0943],[90,-0.0956],[93,-0.0957],[95,-0.0952],[96,-0.0947],[97,-0.0945],[98,-0.094],[99,-0.106],[100,-0.1053],[101,-0.1048],[102,-0.0908],[103,-0.0883],[104,-0.0892],[106,-0.0832],[108,-0.0833],[109,-0.0827],[110,-0.0868],[114,-0.0867],[115,-0.0863],[118,-0.0862],[119,-0.0854],[135,-0.0864],[136,-0.0858],[168,-0.0859],[169,-0.0852],[171,-0.0851],[172,-0.0873],[174,-0.0871],[175,-0.0835],[176,-0.0832],[177,-0.0793],[178,-0.0732],[179,-0.0678],[180,-0.0601],[181,-0.0534],[182,-0.0433],[184,-0.0278],[185,-0.022],[186,-0.0201],[187,-0.0174],[188,-0.0138],[189,-0.0094],[190,-0.0097],[191,-0.0146],[192,-0.0133],[193,-0.0163],[194,-0.0235],[195,-0.0254],[196,-0.0335],[197,-0.0327],[198,-0.0333],[199,-0.0289],[200,-0.0301],[201,-0.0293],[202,-0.0277],[203,-0.027],[204,-0.0254],[206,-0.0238],[207,-0.0238],[208,-0.0233],[209,-0.0232],[210,-0.0226],[211,-0.023],[212,-0.023],[213,-0.0236],[215,-0.027],[217,-0.033],[218,-0.0274],[219,-0.0444],[220,-0.0397],[222,-0.0754],[223,-0.0664],[225,-0.0915],[226,-0.1049],[227,-0.1166],[228,-0.1166],[229,-0.1157],[230,-0.1211],[231,-0.0987],[232,-0.1041],[233,-0.1027],[234,-0.1024],[235,-0.0981],[236,-0.0967],[237,-0.0947],[238,-0.0947],[239,-0.0964],[240,-0.0928],[242,-0.0929],[243,-0.0923],[244,-0.091],[247,-0.0912],[248,-0.0892],[249,-0.0903],[250,-0.0907],[251,-0.0874],[252,-0.0878],[253,-0.0884],[254,-0.0886],[255,-0.0892],[256,-0.086],[259,-0.0874],[260,-0.0872],[261,-0.0877],[262,-0.0845],[263,-0.0845],[264,-0.0842],[265,-0.0855],[296,-0.0852],[297,-0.0015],[367,-0.0015]]
       }
       hips_tan_ctrl.Rotation.X {type=MovieSceneScriptingFloatChannel, default=0, const [0,368]}
       hips_tan_ctrl.Rotation.Y {type=MovieSceneScriptingFloatChannel, default=0, const [0,368]}
       hips_tan_ctrl.Rotation.Z {type=MovieSceneScriptingFloatChannel, default=0, const [0,368]}
       hips_tan_ctrl.Scale.X {type=MovieSceneScriptingFloatChannel, default=1, const [1,368]}
       hips_tan_ctrl.Scale.Y {type=MovieSceneScriptingFloatChannel, default=1, const [1,368]}
       hips_tan_ctrl.Scale.Z {type=MovieSceneScriptingFloatChannel, default=1, const [1,368]}
       spine_01_ctrl.Location.X {
        type=MovieSceneScriptingFloatChannel
        default=0
        thinned_from=368
        keys [[0,-0.0086],[26,-0.0088],[27,-0.0091],[45,-0.0091],[46,-0.008],[49,-0.008],[50,-0.0069],[51,-0.0071],[55,-0.0062],[56,-0.0062],[57,-0.006],[60,-0.006],[61,-0.0057],[64,-0.0057],[65,-0.0055],[101,-0.0049],[102,-0.0062],[104,-0.0062],[105,-0.0073],[110,-0.0073],[111,-0.0084],[118,-0.0084],[152,-0.0092],[168,-0.0094],[169,-0.0089],[173,-0.0089],[174,-0.0113],[175,-0.0113],[177,-0.0208],[178,-0.0279],[179,-0.0374],[180,-0.0445],[184,-0.0824],[185,-0.0885],[186,-0.0979],[189,-0.1121],[190,-0.1121],[191,-0.1168],[192,-0.1168],[193,-0.1216],[197,-0.1216],[200,-0.1074],[201,-0.102],[203,-0.0839],[205,-0.0567],[206,-0.0477],[208,-0.0205],[212,0.0157],[213,0.0202],[214,0.0293],[216,0.0383],[217,0.042],[218,0.0438],[219,0.0475],[220,0.0493],[221,0.0493],[222,0.0512],[225,0.0512],[226,0.0493],[227,0.0493],[231,0.042],[232,0.0383],[233,0.0356],[234,0.029],[235,0.029],[237,0.0157],[238,0.0157],[239,0.009],[240,0.009],[241,0.0024],[242,0.0024],[243,-0.0043],[245,-0.0043],[246,-0.0109],[248,-0.0109],[249,-0.0102],[255,-0.0102],[256,-0.0082],[261,-0.0082],[262,-0.0062],[264,-0.0062],[265,-0.0066],[280,-0.0066],[281,-0.0078],[296,-0.0078],[297,0],[367,0]]
       }
       spine_01_ctrl.Location.Y {
        type=MovieSceneScriptingFloatChannel
        default=0
        thinned_from=368
        keys [[0,0.2436],[2,0.2434],[3,0.2428],[4,0.2426],[11,0.2378],[12,0.2366],[14,0.2354],[15,0.2341],[21,0.2298],[22,0.2283],[23,0.2283],[24,0.2269],[25,0.2269],[26,0.2256],[27,0.2254],[28,0.2242],[30,0.224],[31,0.2227],[41,0.2224],[42,0.2234],[46,0.2235],[47,0.2244],[49,0.2245],[50,0.2255],[59,0.2315],[60,0.2316],[71,0.2391],[85,0.2433],[101,0.2433],[177,0.2438],[188,0.2386],[193,0.2378],[196,0.2382],[217,0.2414],[221,0.243],[227,0.247],[230,0.247],[296,0.2435],[297,0],[367,0]]
       }
       spine_01_ctrl.Location.Z {
        type=MovieSceneScriptingFloatChannel
        default=0
        thinned_from=368
        keys [[0,-0.058],[1,-0.058],[4,-0.0467],[5,-0.0391],[12,-0.0127],[13,-0.0127],[16,-0.0015],[17,0.0005],[21,0.0102],[22,0.015],[23,0.015],[24,0.0198],[26,0.0198],[27,0.0246],[28,0.0246],[29,0.0294],[31,0.0294],[32,0.0342],[33,0.0342],[34,0.0347],[36,0.0347],[37,0.0373],[43,0.0373],[44,0.0348],[47,0.0348],[48,0.0323],[49,0.0323],[50,0.0298],[51,0.0293],[52,0.0256],[53,0.0256],[55,0.0182],[56,0.0182],[63,-0.0077],[64,-0.0151],[67,-0.0262],[68,-0.0293],[71,-0.0414],[72,-0.0441],[73,-0.0482],[77,-0.059],[78,-0.0604],[79,-0.0631],[80,-0.0644],[81,-0.0672],[84,-0.0712],[85,-0.0722],[101,-0.0722],[102,-0.0646],[103,-0.0624],[104,-0.0624],[106,-0.058],[109,-0.058],[110,-0.0602],[118,-0.0602],[119,-0.0598],[168,-0.0605],[169,-0.0598],[171,-0.0598],[172,-0.0615],[174,-0.0615],[176,-0.058],[177,-0.0544],[179,-0.0437],[181,-0.0295],[182,-0.0206],[184,-0.0063],[185,-0.0007],[189,0.0109],[190,0.0109],[191,0.008],[192,0.008],[193,0.0051],[194,-0.0007],[195,-0.0036],[196,-0.0094],[198,-0.0094],[199,-0.0065],[200,-0.0065],[201,-0.0053],[202,-0.0039],[203,-0.0027],[204,-0.0012],[205,-0.0003],[206,0.0003],[208,0.0009],[210,0.0009],[212,0.0003],[213,-0.0006],[214,-0.0023],[215,-0.0044],[217,-0.0103],[218,-0.0103],[219,-0.022],[220,-0.022],[222,-0.0453],[223,-0.0453],[227,-0.0919],[230,-0.0919],[231,-0.0802],[232,-0.0802],[233,-0.0787],[237,-0.0706],[239,-0.0706],[240,-0.0686],[243,-0.0686],[244,-0.0666],[247,-0.0666],[248,-0.0646],[249,-0.0651],[250,-0.0651],[251,-0.0631],[255,-0.0631],[256,-0.061],[261,-0.061],[262,-0.059],[264,-0.059],[265,-0.0595],[296,-0.0594],[297,0],[367,0]]
       }
       spine_01_ctrl.Rotation.X {
        type=MovieSceneScriptingFloatChannel
        default=0
        thinned_from=368
        keys [[0,4.377],[6,4.352],[29,4.0447],[34,4.0196],[48,4.0499],[74,4.3148],[200,4.3716],[208,4.3546],[227,4.4466],[296,4.378],[297,0],[367,0]]
       }
       spine_01_ctrl.Rotation.Y {
        type=MovieSceneScriptingFloatChannel
        default=0
        thinned_from=368
        keys [[0,0.2626],[15,0.2629],[24,0.26],[42,0.2603],[43,0.2599],[44,0.2585],[45,0.256],[46,0.2526],[49,0.24],[51,0.2329],[53,0.2275],[55,0.2237],[56,0.2223],[78,0.2163],[90,0.2151],[92,0.2164],[95,0.2201],[97,0.2237],[110,0.2525],[114,0.2591],[127,0.268],[139,0.2711],[140,0.2719],[151,0.2725],[152,0.2733],[165,0.2734],[166,0.2726],[169,0.2724],[170,0.2731],[171,0.2733],[172,0.2752],[173,0.2764],[174,0.2898],[175,0.3275],[176,0.3905],[177,0.4822],[178,0.5977],[179,0.7358],[180,0.8891],[182,1.2094],[183,1.3634],[184,1.5052],[185,1.6331],[186,1.7463],[187,1.8446],[188,1.9285],[189,1.999],[190,2.0568],[191,2.103],[192,2.1392],[193,2.1654],[194,2.1832],[195,2.1926],[196,2.1941],[197,2.1799],[198,2.1414],[199,2.0762],[200,1.9828],[201,1.8615],[202,1.712],[203,1.5415],[204,1.3498],[205,1.1463],[206,0.938],[207,0.7319],[208,0.5349],[209,0.3494],[210,0.1777],[211,0.0245],[212,-0.1147],[213,-0.2354],[214,-0.3444],[215,-0.4371],[216,-0.5158],[217,-0.5841],[218,-0.6408],[219,-0.6868],[220,-0.7234],[221,-0.7512],[222,-0.7709],[223,-0.7835],[224,-0.789],[225,-0.7879],[226,-0.7814],[227,-0.7699],[228,-0.7517],[229,-0.7274],[230,-0.6957],[231,-0.6562],[232,-0.6086],[233,-0.5518],[234,-0.4877],[235,-0.4169],[236,-0.341],[238,-0.1858],[239,-0.113],[240,-0.0466],[241,0.0148],[242,0.0666],[243,0.112],[244,0.1508],[245,0.1848],[246,0.2108],[247,0.2319],[248,0.2499],[249,0.2627],[250,0.2721],[251,0.2782],[252,0.2812],[253,0.2818],[256,0.2675],[258,0.2592],[260,0.2523],[262,0.2472],[263,0.2455],[265,0.2439],[267,0.244],[268,0.2455],[271,0.2456],[272,0.2472],[273,0.2472],[274,0.2488],[275,0.2488],[277,0.2519],[278,0.2519],[279,0.2534],[280,0.2535],[281,0.2552],[285,0.2585],[292,0.2618],[296,0.2622],[297,0],[367,0]]
       }
       spine_01_ctrl.Rotation.Z {
        type=MovieSceneScriptingFloatChannel
        default=0
        thinned_from=368
        keys [[0,-1.4947],[42,-1.5036],[43,-1.5188],[44,-1.5713],[45,-1.6611],[46,-1.7847],[47,-1.927],[48,-2.076],[49,-2.2183],[50,-2.3453],[51,-2.4552],[52,-2.5472],[53,-2.6234],[54,-2.6848],[55,-2.7335],[56,-2.7694],[60,-2.8541],[66,-2.9386],[77,-3.0217],[89,-3.0497],[90,-3.0454],[91,-3.0326],[93,-2.9773],[94,-2.9389],[96,-2.8409],[97,-2.7855],[99,-2.6577],[101,-2.5171],[107,-2.0573],[109,-1.911],[110,-1.8418],[111,-1.7766],[112,-1.7154],[113,-1.66],[114,-1.6086],[115,-1.5632],[116,-1.5256],[117,-1.494],[120,-1.4196],[123,-1.356],[126,-1.3015],[128,-1.2697],[130,-1.2424],[133,-1.2083],[136,-1.1817],[140,-1.1538],[141,-1.1503],[142,-1.1433],[146,-1.1293],[147,-1.1293],[149,-1.1224],[150,-1.1224],[151,-1.1189],[167,-1.1142],[168,-1.1107],[169,-1.0937],[170,-1.0525],[171,-0.9761],[172,-0.8689],[173,-0.7368],[174,-0.5844],[175,-0.4118],[176,-0.2292],[177,-0.0472],[178,0.1302],[179,0.299],[180,0.4551],[181,0.5946],[182,0.7238],[183,0.8428],[184,0.9517],[185,1.0491],[186,1.1369],[187,1.2147],[188,1.2829],[189,1.3414],[190,1.3918],[191,1.434],[192,1.4694],[193,1.4966],[194,1.5131],[195,1.5159],[196,1.5055],[197,1.4796],[198,1.4348],[199,1.3707],[200,1.2883],[201,1.1845],[202,1.0597],[203,0.9132],[204,0.7449],[205,0.563],[206,0.3702],[207,0.1662],[208,-0.0406],[209,-0.2505],[210,-0.4578],[211,-0.6627],[212,-0.8595],[213,-1.0512],[214,-1.232],[215,-1.4021],[216,-1.5641],[217,-1.7135],[218,-1.8502],[219,-1.978],[220,-2.093],[221,-2.1954],[222,-2.2887],[223,-2.3693],[224,-2.4372],[225,-2.4942],[226,-2.5402],[227,-2.5735],[228,-2.5958],[229,-2.6091],[230,-2.6133],[231,-2.6085],[232,-2.5946],[233,-2.5671],[234,-2.525],[235,-2.465],[236,-2.3906],[238,-2.2168],[239,-2.1335],[240,-2.0572],[241,-1.9898],[242,-1.9313],[243,-1.8799],[244,-1.8356],[246,-1.761],[248,-1.7059],[250,-1.663],[252,-1.6319],[256,-1.591],[260,-1.5564],[262,-1.5422],[266,-1.5237],[267,-1.5237],[268,-1.5201],[269,-1.5201],[270,-1.5165],[271,-1.5165],[272,-1.5129],[273,-1.5129],[274,-1.5093],[276,-1.5093],[277,-1.5058],[279,-1.5058],[280,-1.5022],[296,-1.4944],[297,0],[367,0]]
       }
       spine_01_ctrl.Scale.X {type=MovieSceneScriptingFloatChannel, default=1, const [1,368]}
       spine_01_ctrl.Scale.Y {type=MovieSceneScriptingFloatChannel, default=1, const [1,368]}
       spine_01_ctrl.Scale.Z {type=MovieSceneScriptingFloatChannel, default=1, const [1,368]}
       spine_02_ctrl.Location.X {
        type=MovieSceneScriptingFloatChannel
        default=0
        thinned_from=368
        keys [[0,-0.0564],[33,-0.0568],[44,-0.0566],[45,-0.0588],[47,-0.0588],[48,-0.061],[49,-0.061],[50,-0.0632],[51,-0.0632],[53,-0.0645],[56,-0.0657],[67,-0.067],[84,-0.0678],[85,-0.0687],[96,-0.0687],[97,-0.0638],[101,-0.0638],[111,-0.0583],[112,-0.0583],[115,-0.0566],[117,-0.0566],[118,-0.0561],[135,-0.056],[136,-0.0541],[152,-0.0541],[153,-0.0535],[168,-0.0535],[171,-0.0534],[172,-0.0526],[174,-0.0526],[175,-0.0534],[176,-0.0551],[177,-0.0585],[178,-0.0635],[179,-0.0694],[181,-0.0829],[182,-0.0905],[184,-0.104],[185,-0.1099],[186,-0.1153],[187,-0.1201],[188,-0.1243],[189,-0.1279],[190,-0.1309],[191,-0.1335],[192,-0.1356],[194,-0.1385],[195,-0.1394],[196,-0.1398],[197,-0.1398],[198,-0.1386],[199,-0.1362],[200,-0.1326],[201,-0.1284],[204,-0.1078],[205,-0.0975],[206,-0.0907],[207,-0.0804],[208,-0.0736],[209,-0.0633],[213,-0.0359],[214,-0.0324],[215,-0.0256],[217,-0.0189],[218,-0.0151],[219,-0.0127],[220,-0.0089],[221,-0.0077],[222,-0.0052],[226,-0.0002],[229,-0.0002],[233,-0.0052],[234,-0.0052],[235,-0.011],[236,-0.011],[238,-0.0225],[239,-0.0225],[240,-0.0283],[241,-0.0283],[242,-0.0341],[244,-0.0341],[245,-0.0399],[246,-0.0399],[247,-0.0457],[248,-0.0457],[249,-0.0462],[251,-0.0502],[260,-0.0591],[261,-0.0591],[262,-0.06],[264,-0.06],[265,-0.0603],[280,-0.0603],[281,-0.0579],[296,-0.0579],[297,0],[367,0]]
       }
       spine_02_ctrl.Location.Y {
        type=MovieSceneScriptingFloatChannel
        default=0
        thinned_from=368
        keys [[0,0.3087],[3,0.3081],[4,0.3068],[5,0.3062],[12,0.2975],[13,0.2953],[14,0.294],[16,0.2895],[17,0.2879],[19,0.2834],[23,0.277],[24,0.274],[27,0.2693],[28,0.2693],[30,0.2662],[31,0.266],[32,0.2646],[34,0.2644],[35,0.2635],[43,0.2633],[44,0.2644],[46,0.2644],[47,0.2654],[49,0.2656],[50,0.2667],[54,0.2706],[61,0.2809],[69,0.2943],[70,0.2952],[72,0.2979],[78,0.3032],[79,0.3034],[80,0.3043],[81,0.3043],[82,0.3052],[83,0.3052],[85,0.3065],[92,0.3065],[93,0.309],[98,0.309],[99,0.3082],[173,0.3078],[187,0.3009],[191,0.302],[197,0.3065],[217,0.3074],[227,0.3116],[237,0.3078],[296,0.3086],[297,0],[367,0]]
       }
       spine_02_ctrl.Location.Z {
        type=MovieSceneScriptingFloatChannel
        default=0
        thinned_from=368
        keys [[0,-0.0635],[1,-0.0635],[4,-0.0523],[5,-0.0447],[11,-0.0225],[12,-0.0226],[16,-0.008],[17,-0.0061],[19,-0.0018],[21,0.005],[27,0.0182],[28,0.0182],[31,0.025],[32,0.0249],[33,0.0272],[34,0.0277],[36,0.0276],[37,0.0301],[43,0.0301],[44,0.0277],[47,0.0278],[48,0.0252],[49,0.0252],[50,0.0228],[51,0.0214],[53,0.018],[55,0.0128],[61,-0.008],[62,-0.0123],[63,-0.0158],[65,-0.0245],[66,-0.028],[67,-0.0323],[68,-0.0358],[69,-0.0413],[71,-0.0467],[72,-0.0523],[76,-0.0632],[77,-0.0632],[80,-0.0715],[81,-0.0715],[82,-0.0742],[83,-0.0742],[84,-0.077],[85,-0.0815],[92,-0.0815],[93,-0.0813],[98,-0.0813],[99,-0.0707],[101,-0.0707],[102,-0.0702],[103,-0.0684],[106,-0.0645],[108,-0.0644],[109,-0.0649],[114,-0.0649],[115,-0.0653],[135,-0.0654],[136,-0.0659],[168,-0.066],[169,-0.0657],[171,-0.0658],[172,-0.0666],[174,-0.0667],[175,-0.0658],[176,-0.0633],[177,-0.0599],[178,-0.0556],[179,-0.0496],[180,-0.0427],[182,-0.0272],[184,-0.0135],[185,-0.0081],[186,-0.0037],[187,-0.0004],[188,0.0017],[189,0.0029],[190,0.0033],[191,0.0026],[192,0.0009],[193,-0.0022],[195,-0.011],[196,-0.0144],[197,-0.0157],[198,-0.0152],[199,-0.0143],[202,-0.0105],[203,-0.0091],[204,-0.0079],[205,-0.007],[207,-0.0059],[208,-0.0056],[209,-0.0056],[210,-0.0058],[211,-0.0058],[212,-0.0061],[213,-0.007],[214,-0.0084],[215,-0.0108],[217,-0.0163],[222,-0.043],[227,-0.0964],[229,-0.0964],[230,-0.0911],[231,-0.0911],[232,-0.0858],[233,-0.0838],[234,-0.08],[235,-0.08],[237,-0.0761],[238,-0.0761],[239,-0.0742],[243,-0.0742],[244,-0.0722],[247,-0.0722],[248,-0.0702],[249,-0.0703],[259,-0.0662],[260,-0.0662],[262,-0.0654],[264,-0.0654],[265,-0.0651],[280,-0.0651],[296,-0.065],[297,0],[367,0]]
       }
       spine_02_ctrl.Rotation.X {
        type=MovieSceneScriptingFloatChannel
        default=0
        thinned_from=368
        keys [[0,5.5964],[5,5.5637],[11,5.4379],[24,4.9916],[28,4.8849],[32,4.8172],[41,4.7864],[50,4.818],[54,4.8753],[71,5.3313],[81,5.4664],[174,5.5915],[187,5.5581],[197,5.6824],[201,5.6784],[211,5.6155],[227,5.6376],[243,5.5611],[296,5.5976],[297,0],[367,0]]
       }
       spine_02_ctrl.Rotation.Y {
        type=MovieSceneScriptingFloatChannel
        default=0
        thinned_from=368
        keys [[0,1.2264],[35,1.1984],[42,1.2033],[43,1.2065],[44,1.2154],[45,1.2297],[46,1.2498],[49,1.3248],[51,1.3693],[53,1.4037],[55,1.4301],[56,1.4402],[71,1.5141],[89,1.5479],[91,1.5441],[93,1.5335],[96,1.5055],[101,1.4376],[109,1.312],[112,1.272],[115,1.2408],[118,1.2211],[131,1.1734],[148,1.1498],[164,1.1467],[168,1.15],[169,1.1486],[170,1.1443],[171,1.1348],[172,1.1215],[173,1.1051],[174,1.0945],[175,1.1013],[176,1.1277],[177,1.1757],[178,1.2449],[179,1.3327],[180,1.4349],[183,1.7708],[184,1.8757],[185,1.9729],[186,2.0601],[187,2.1372],[188,2.2061],[189,2.2649],[190,2.3156],[191,2.3601],[192,2.3964],[194,2.4487],[196,2.4795],[197,2.4823],[198,2.467],[199,2.4316],[200,2.3761],[201,2.3005],[202,2.2058],[203,2.0944],[204,1.9698],[207,1.5668],[208,1.4353],[209,1.3096],[210,1.1921],[211,1.0826],[212,0.9818],[213,0.8896],[214,0.8052],[215,0.728],[216,0.6579],[217,0.5945],[218,0.5372],[219,0.4862],[220,0.4412],[221,0.4018],[222,0.3679],[223,0.3397],[224,0.3166],[225,0.2989],[226,0.2852],[227,0.2758],[228,0.2708],[229,0.2714],[230,0.2784],[231,0.2928],[232,0.3144],[233,0.3427],[234,0.3782],[235,0.418],[236,0.4622],[237,0.5094],[238,0.5556],[240,0.6523],[242,0.7457],[243,0.7907],[245,0.8754],[246,0.915],[248,0.989],[249,1.0241],[251,1.0873],[252,1.1164],[253,1.143],[254,1.1673],[256,1.2111],[259,1.2648],[261,1.294],[263,1.3113],[264,1.3163],[268,1.3131],[276,1.2854],[281,1.263],[282,1.2565],[289,1.2339],[290,1.2339],[291,1.2307],[292,1.2306],[293,1.2274],[296,1.2274],[297,0],[367,0]]
       }
       spine_02_ctrl.Rotation.Z {
        type=MovieSceneScriptingFloatChannel
        default=0
        thinned_from=368
        keys [[0,-1.2582],[42,-1.2862],[43,-1.3028],[44,-1.3526],[45,-1.4389],[46,-1.5553],[47,-1.6915],[48,-1.8345],[49,-1.9676],[50,-2.0906],[51,-2.195],[52,-2.2819],[53,-2.3538],[54,-2.4123],[55,-2.4558],[56,-2.4897],[60,-2.5663],[64,-2.62],[77,-2.7086],[89,-2.7301],[90,-2.726],[91,-2.7118],[92,-2.6895],[93,-2.6611],[94,-2.6246],[95,-2.5821],[96,-2.5335],[98,-2.4202],[99,-2.3594],[101,-2.2259],[108,-1.7219],[109,-1.6532],[111,-1.5252],[112,-1.4676],[113,-1.4138],[114,-1.3656],[115,-1.3229],[116,-1.2877],[117,-1.258],[120,-1.1872],[122,-1.1454],[124,-1.1079],[126,-1.0745],[129,-1.0317],[132,-0.9963],[137,-0.9525],[142,-0.9257],[143,-0.923],[144,-0.9176],[147,-0.9095],[148,-0.9095],[150,-0.9041],[151,-0.9041],[153,-0.8993],[162,-0.8992],[163,-0.8959],[167,-0.896],[168,-0.8928],[169,-0.8764],[170,-0.8359],[171,-0.7603],[172,-0.6561],[173,-0.5271],[174,-0.379],[175,-0.2162],[176,-0.0479],[177,0.1167],[178,0.2739],[179,0.4174],[180,0.5465],[181,0.6596],[182,0.7616],[183,0.8555],[184,0.9424],[185,1.0212],[186,1.0922],[187,1.1566],[188,1.2146],[189,1.2643],[190,1.3093],[191,1.3476],[192,1.3794],[193,1.4064],[194,1.4231],[195,1.4296],[196,1.4223],[197,1.4014],[198,1.3656],[199,1.3166],[200,1.2528],[201,1.1716],[202,1.0731],[203,0.9574],[204,0.8259],[205,0.6804],[206,0.5236],[207,0.3572],[208,0.1828],[209,0.0046],[211,-0.3565],[212,-0.5346],[213,-0.7082],[214,-0.8762],[215,-1.0379],[216,-1.1909],[217,-1.3362],[218,-1.4723],[219,-1.5987],[220,-1.7153],[221,-1.8223],[222,-1.9185],[223,-2.004],[224,-2.0788],[225,-2.1428],[226,-2.1961],[227,-2.2376],[228,-2.2693],[229,-2.2913],[230,-2.3066],[231,-2.3142],[232,-2.3133],[233,-2.2982],[234,-2.2693],[235,-2.2244],[236,-2.1652],[239,-1.9512],[240,-1.8858],[241,-1.8267],[242,-1.7725],[243,-1.7246],[245,-1.6415],[247,-1.5726],[249,-1.5125],[251,-1.4616],[254,-1.3999],[257,-1.3468],[260,-1.3],[262,-1.2777],[264,-1.262],[267,-1.2524],[281,-1.2514],[287,-1.2548],[291,-1.2565],[296,-1.2586],[297,0],[367,0]]
       }
       spine_02_ctrl.Scale.X {type=MovieSceneScriptingFloatChannel, default=1, const [1,368]}
       spine_02_ctrl.Scale.Y {type=MovieSceneScriptingFloatChannel, default=1, const [1,368]}
       spine_02_ctrl.Scale.Z {type=MovieSceneScriptingFloatChannel, default=1, const [1,368]}
       spine_03_ctrl.Location.X {
        type=MovieSceneScriptingFloatChannel
        default=0
        thinned_from=368
        keys [[0,-0.1405],[41,-0.1397],[42,-0.1411],[43,-0.1411],[45,-0.1441],[47,-0.1499],[48,-0.1544],[49,-0.1573],[50,-0.1617],[51,-0.1638],[52,-0.1671],[53,-0.1693],[54,-0.1704],[55,-0.1727],[56,-0.1738],[57,-0.1738],[59,-0.176],[60,-0.176],[61,-0.1771],[62,-0.1771],[63,-0.1782],[64,-0.1782],[65,-0.1793],[66,-0.1793],[67,-0.1804],[68,-0.1803],[70,-0.181],[74,-0.1827],[78,-0.1837],[84,-0.1847],[86,-0.1844],[87,-0.1854],[89,-0.1854],[90,-0.1844],[91,-0.1844],[92,-0.1833],[93,-0.1833],[94,-0.1822],[95,-0.1801],[96,-0.1791],[97,-0.177],[98,-0.1759],[101,-0.1696],[102,-0.1684],[103,-0.1645],[115,-0.1416],[116,-0.1416],[117,-0.1397],[118,-0.1397],[128,-0.1342],[136,-0.1317],[152,-0.1317],[153,-0.1295],[168,-0.1295],[170,-0.1297],[173,-0.1243],[174,-0.1207],[175,-0.1207],[176,-0.1189],[177,-0.1189],[178,-0.1207],[181,-0.1315],[182,-0.1369],[183,-0.1405],[184,-0.1459],[185,-0.1496],[186,-0.1537],[188,-0.1611],[190,-0.1672],[192,-0.172],[195,-0.1774],[196,-0.1789],[197,-0.18],[198,-0.1804],[199,-0.1801],[200,-0.179],[201,-0.1775],[202,-0.1736],[203,-0.1717],[206,-0.16],[207,-0.1542],[211,-0.1386],[212,-0.1328],[216,-0.1172],[217,-0.1139],[218,-0.1094],[222,-0.0958],[223,-0.0936],[224,-0.0902],[227,-0.0834],[231,-0.0789],[236,-0.0789],[237,-0.0835],[239,-0.0835],[240,-0.0881],[241,-0.0881],[242,-0.0927],[243,-0.0927],[245,-0.1018],[246,-0.1018],[248,-0.111],[249,-0.1143],[250,-0.1199],[251,-0.1199],[253,-0.131],[254,-0.131],[255,-0.1366],[256,-0.1366],[257,-0.1422],[258,-0.1422],[259,-0.1478],[261,-0.1478],[262,-0.1534],[264,-0.1534],[265,-0.1538],[268,-0.1538],[269,-0.1512],[274,-0.1512],[275,-0.1485],[278,-0.1485],[279,-0.1459],[285,-0.1459],[286,-0.1401],[296,-0.1401],[297,0],[367,0]]
       }
       spine_03_ctrl.Location.Y {
        type=MovieSceneScriptingFloatChannel
        default=0
        thinned_from=368
        keys [[0,0.4476],[3,0.4467],[4,0.4444],[5,0.444],[6,0.4417],[7,0.4412],[9,0.4362],[10,0.4344],[11,0.4302],[12,0.428],[13,0.4238],[14,0.4216],[17,0.4092],[21,0.3949],[22,0.3907],[28,0.3736],[33,0.3663],[39,0.3635],[52,0.3666],[53,0.369],[54,0.3695],[62,0.3917],[64,0.3978],[65,0.403],[67,0.4087],[68,0.4107],[70,0.4173],[78,0.4311],[79,0.4315],[80,0.4332],[81,0.4332],[82,0.435],[83,0.435],[84,0.4367],[171,0.4474],[186,0.4374],[188,0.4374],[191,0.4408],[196,0.4507],[199,0.4535],[221,0.4448],[227,0.4466],[235,0.4415],[239,0.4405],[242,0.4416],[296,0.4472],[297,0],[367,0]]
       }
       spine_03_ctrl.Location.Z {
        type=MovieSceneScriptingFloatChannel
        default=0
        thinned_from=368
        keys [[0,-0.0609],[1,-0.0609],[3,-0.0525],[4,-0.0486],[5,-0.0444],[6,-0.0404],[7,-0.0362],[8,-0.028],[9,-0.024],[10,-0.0242],[11,-0.0204],[12,-0.0164],[13,-0.0126],[14,-0.0086],[16,-0.001],[17,0.0007],[21,0.0123],[28,0.0282],[29,0.0294],[30,0.0318],[33,0.0353],[34,0.0371],[38,0.0391],[41,0.0391],[42,0.0386],[43,0.0386],[45,0.0375],[49,0.0331],[50,0.0314],[51,0.0304],[52,0.0262],[53,0.0264],[54,0.0223],[55,0.0183],[56,0.0186],[62,-0.005],[63,-0.0131],[64,-0.0171],[65,-0.0207],[67,-0.0286],[68,-0.0318],[69,-0.0378],[71,-0.0436],[72,-0.0498],[76,-0.0618],[77,-0.0616],[80,-0.0708],[81,-0.0708],[82,-0.0738],[83,-0.0738],[84,-0.0768],[85,-0.0777],[90,-0.081],[93,-0.0809],[95,-0.0794],[96,-0.0779],[97,-0.0772],[99,-0.0742],[100,-0.0721],[101,-0.0706],[102,-0.0688],[103,-0.0673],[104,-0.0653],[106,-0.0624],[107,-0.0623],[108,-0.0627],[114,-0.0625],[115,-0.063],[168,-0.0634],[169,-0.0643],[174,-0.0645],[175,-0.0606],[176,-0.0607],[178,-0.0529],[184,-0.0056],[186,0.004],[187,0.0074],[188,0.0097],[189,0.011],[190,0.0114],[191,0.0106],[192,0.0088],[193,0.0053],[195,-0.0042],[196,-0.0079],[197,-0.0092],[198,-0.0088],[199,-0.0079],[204,-0.0015],[205,-0.0005],[206,0.0003],[207,0.0005],[208,0.001],[211,0.0008],[212,0.0005],[213,-0.0003],[214,-0.0017],[215,-0.0042],[218,-0.0127],[220,-0.0238],[222,-0.0403],[225,-0.0734],[226,-0.0871],[227,-0.0954],[228,-0.0955],[232,-0.0847],[233,-0.0817],[234,-0.0778],[235,-0.0758],[236,-0.0758],[237,-0.0739],[238,-0.0739],[239,-0.0719],[242,-0.0718],[243,-0.0698],[247,-0.0696],[248,-0.0676],[249,-0.0681],[250,-0.0672],[251,-0.0672],[252,-0.0663],[253,-0.0663],[254,-0.0655],[256,-0.0654],[257,-0.0645],[259,-0.0645],[260,-0.0637],[263,-0.0637],[264,-0.0629],[265,-0.0632],[266,-0.0628],[280,-0.0628],[281,-0.0632],[296,-0.0632],[297,0],[367,0]]
       }
       spine_03_ctrl.Rotation.X {
        type=MovieSceneScriptingFloatChannel
        default=0
        thinned_from=368
        keys [[0,7.5593],[4,7.5231],[8,7.4143],[12,7.2399],[23,6.5492],[27,6.3575],[31,6.2308],[36,6.1548],[46,6.1447],[50,6.1131],[53,6.1428],[56,6.2266],[70,6.9207],[76,7.1128],[116,7.5592],[176,7.5642],[187,7.5119],[189,7.535],[197,7.7893],[200,7.8152],[221,7.5215],[227,7.5134],[238,7.4339],[296,7.5612],[297,0],[367,0]]
       }
       spine_03_ctrl.Rotation.Y {
        type=MovieSceneScriptingFloatChannel
        default=0
        thinned_from=368
        keys [[0,2.6088],[33,2.5331],[42,2.5453],[43,2.554],[44,2.5775],[45,2.6181],[46,2.6734],[49,2.8763],[50,2.9392],[51,2.9957],[52,3.0448],[53,3.0874],[55,3.1543],[56,3.1798],[71,3.3933],[87,3.4893],[90,3.4899],[92,3.4697],[94,3.4312],[97,3.3428],[101,3.1887],[109,2.8443],[112,2.7339],[114,2.6732],[116,2.6267],[123,2.5324],[134,2.4451],[144,2.4063],[166,2.394],[168,2.3977],[169,2.3947],[170,2.3821],[171,2.3555],[172,2.3174],[174,2.2224],[175,2.1823],[176,2.1553],[177,2.1447],[178,2.152],[179,2.1764],[180,2.2156],[181,2.2675],[185,2.5102],[187,2.622],[189,2.72],[191,2.804],[195,2.9336],[197,2.9847],[198,2.9985],[199,3.0011],[200,2.9926],[201,2.9744],[202,2.9463],[203,2.9101],[207,2.7302],[214,2.3654],[219,2.0841],[221,1.9782],[223,1.8824],[225,1.8],[228,1.6947],[229,1.6657],[230,1.6429],[231,1.6263],[232,1.6166],[233,1.6114],[235,1.6133],[237,1.6248],[238,1.6345],[239,1.6512],[240,1.6767],[241,1.7096],[242,1.7491],[243,1.7942],[245,1.8983],[247,2.0155],[254,2.4569],[257,2.6214],[259,2.7133],[260,2.7524],[262,2.811],[263,2.8306],[264,2.8419],[268,2.835],[275,2.7744],[285,2.6598],[291,2.6198],[296,2.6098],[297,0],[367,0]]
       }
       spine_03_ctrl.Rotation.Z {
        type=MovieSceneScriptingFloatChannel
        default=0
        thinned_from=368
        keys [[0,-1.2752],[42,-1.3012],[43,-1.3182],[44,-1.3716],[45,-1.4633],[46,-1.588],[47,-1.7335],[48,-1.8847],[49,-2.0285],[50,-2.1574],[51,-2.2689],[52,-2.3616],[53,-2.4382],[54,-2.499],[55,-2.5462],[56,-2.5814],[59,-2.6452],[64,-2.7187],[76,-2.8053],[89,-2.8303],[90,-2.8259],[91,-2.8107],[92,-2.7869],[93,-2.7566],[94,-2.7178],[96,-2.6229],[98,-2.5046],[100,-2.3691],[102,-2.2232],[107,-1.8391],[109,-1.6926],[110,-1.6234],[111,-1.5581],[112,-1.4968],[113,-1.4395],[114,-1.3881],[115,-1.3447],[116,-1.3052],[118,-1.248],[121,-1.1775],[123,-1.1357],[126,-1.0811],[129,-1.0357],[132,-0.9987],[136,-0.9596],[141,-0.9265],[142,-0.9264],[143,-0.9198],[144,-0.9197],[145,-0.9132],[146,-0.913],[147,-0.9065],[149,-0.9065],[150,-0.8998],[163,-0.8951],[167,-0.8937],[168,-0.889],[169,-0.8711],[170,-0.8276],[171,-0.7468],[172,-0.6341],[173,-0.4964],[174,-0.3395],[175,-0.1696],[176,0.0008],[177,0.1645],[178,0.3162],[179,0.4498],[180,0.5663],[181,0.6631],[182,0.75],[183,0.8297],[184,0.9025],[185,0.9692],[186,1.0308],[187,1.0863],[188,1.1365],[189,1.1822],[190,1.2219],[191,1.2571],[192,1.2886],[193,1.3131],[194,1.3298],[195,1.3355],[196,1.3283],[197,1.3093],[198,1.2776],[199,1.2349],[200,1.1812],[201,1.1149],[202,1.0329],[203,0.9356],[204,0.8241],[205,0.6994],[206,0.5627],[207,0.415],[208,0.2575],[209,0.0923],[210,-0.0772],[212,-0.4227],[213,-0.5943],[214,-0.7625],[215,-0.9252],[216,-1.0812],[217,-1.2309],[218,-1.3725],[219,-1.5052],[220,-1.6291],[221,-1.7442],[222,-1.8481],[223,-1.9422],[224,-2.0252],[225,-2.0972],[226,-2.1581],[227,-2.207],[228,-2.2471],[229,-2.2775],[230,-2.3013],[231,-2.3176],[232,-2.3252],[233,-2.32],[234,-2.2999],[235,-2.2636],[236,-2.2116],[237,-2.1486],[239,-2.0168],[240,-1.9557],[241,-1.9001],[242,-1.8491],[244,-1.7595],[246,-1.6815],[248,-1.6113],[250,-1.548],[252,-1.4916],[257,-1.3712],[260,-1.312],[261,-1.2961],[263,-1.2709],[264,-1.2629],[266,-1.2548],[274,-1.2517],[296,-1.2747],[297,0],[367,0]]
       }
       spine_03_ctrl.Scale.X {type=MovieSceneScriptingFloatChannel, default=1, const [1,368]}
       spine_03_ctrl.Scale.Y {type=MovieSceneScriptingFloatChannel, default=1, const [1,368]}
       spine_03_ctrl.Scale.Z {type=MovieSceneScriptingFloatChannel, default=1, const [1,368]}
       spine_04_ctrl.Location.X {
        type=MovieSceneScriptingFloatChannel
        default=0
        thinned_from=368
        keys [[0,-0.524],[40,-0.5334],[41,-0.5364],[43,-0.5372],[45,-0.5446],[47,-0.5594],[48,-0.57],[49,-0.578],[50,-0.5887],[54,-0.6127],[56,-0.6201],[74,-0.6424],[84,-0.6482],[85,-0.6479],[86,-0.65],[90,-0.6502],[91,-0.6478],[92,-0.6472],[94,-0.6413],[95,-0.6361],[96,-0.6327],[97,-0.6272],[98,-0.6235],[102,-0.601],[103,-0.5914],[104,-0.5892],[105,-0.5796],[106,-0.5775],[107,-0.5679],[108,-0.5658],[109,-0.5563],[110,-0.5545],[111,-0.5451],[112,-0.5435],[113,-0.5344],[115,-0.5317],[116,-0.5232],[118,-0.5217],[119,-0.5189],[120,-0.5183],[121,-0.5157],[122,-0.5151],[123,-0.5126],[124,-0.5121],[125,-0.5095],[127,-0.5087],[128,-0.5063],[129,-0.5059],[130,-0.5035],[133,-0.5025],[134,-0.5002],[140,-0.4994],[141,-0.4943],[167,-0.4935],[168,-0.4942],[169,-0.4959],[172,-0.491],[174,-0.4829],[175,-0.4776],[176,-0.4737],[177,-0.4659],[178,-0.4594],[180,-0.4433],[182,-0.4319],[183,-0.4272],[184,-0.4263],[185,-0.4236],[186,-0.4243],[187,-0.4262],[188,-0.4268],[190,-0.4346],[196,-0.4667],[197,-0.4729],[198,-0.4802],[201,-0.5106],[203,-0.538],[204,-0.5491],[206,-0.579],[207,-0.5928],[208,-0.6043],[210,-0.6169],[211,-0.6177],[212,-0.6162],[213,-0.6125],[214,-0.6032],[215,-0.5956],[218,-0.5564],[219,-0.5409],[220,-0.5232],[223,-0.4742],[224,-0.4566],[226,-0.4259],[228,-0.3978],[230,-0.3721],[231,-0.3619],[232,-0.3504],[233,-0.3399],[235,-0.323],[236,-0.3156],[237,-0.309],[238,-0.3033],[239,-0.299],[240,-0.2982],[242,-0.3038],[243,-0.3115],[244,-0.3207],[245,-0.3312],[246,-0.3428],[247,-0.357],[249,-0.3869],[251,-0.4217],[252,-0.4374],[254,-0.4733],[255,-0.4888],[256,-0.506],[258,-0.5342],[260,-0.5591],[261,-0.5678],[262,-0.5754],[263,-0.5814],[264,-0.586],[265,-0.5875],[268,-0.5864],[269,-0.581],[274,-0.5735],[275,-0.5671],[278,-0.5607],[279,-0.5541],[281,-0.5499],[282,-0.5446],[285,-0.5392],[286,-0.5343],[290,-0.5295],[291,-0.5255],[296,-0.5236],[297,0],[367,0]]
       }
       spine_04_ctrl.Location.Y {
        type=MovieSceneScriptingFloatChannel
        default=0
        thinned_from=368
        keys [[0,0.1428],[1,0.1428],[2,0.1421],[3,0.141],[4,0.1396],[5,0.1377],[6,0.1352],[7,0.1321],[8,0.1285],[9,0.1242],[10,0.1191],[11,0.1132],[12,0.1064],[13,0.0986],[14,0.0899],[15,0.0805],[16,0.0704],[17,0.0609],[19,0.0392],[21,0.0179],[22,0.0094],[23,-0.0006],[24,-0.0086],[25,-0.016],[26,-0.0232],[27,-0.0283],[28,-0.035],[30,-0.044],[32,-0.0518],[33,-0.0556],[34,-0.0584],[35,-0.0605],[36,-0.0624],[37,-0.0639],[38,-0.0651],[39,-0.0661],[41,-0.0662],[42,-0.0661],[43,-0.065],[44,-0.0645],[45,-0.0647],[46,-0.0669],[48,-0.0729],[49,-0.0756],[50,-0.0778],[51,-0.0782],[52,-0.0779],[53,-0.0754],[54,-0.0739],[55,-0.0702],[56,-0.0644],[57,-0.0581],[58,-0.0513],[59,-0.0444],[60,-0.0357],[61,-0.0285],[62,-0.0195],[63,-0.0122],[65,0.0052],[66,0.0124],[67,0.0207],[68,0.0274],[69,0.0348],[70,0.0409],[72,0.052],[74,0.0602],[76,0.0677],[77,0.0712],[79,0.0759],[80,0.079],[82,0.0831],[83,0.085],[84,0.0858],[86,0.0891],[87,0.0897],[88,0.0913],[89,0.0918],[91,0.0949],[92,0.0956],[96,0.1031],[98,0.1073],[99,0.1103],[102,0.1168],[103,0.12],[104,0.1213],[105,0.1244],[106,0.1256],[108,0.1321],[109,0.1333],[110,0.1363],[112,0.1382],[113,0.1408],[116,0.1421],[117,0.1422],[118,0.1441],[119,0.143],[122,0.1433],[123,0.1435],[126,0.1435],[127,0.1438],[130,0.1438],[131,0.1441],[146,0.1445],[147,0.1461],[152,0.1464],[153,0.1458],[168,0.1466],[169,0.1471],[171,0.1458],[172,0.1449],[175,0.1405],[177,0.1368],[180,0.132],[183,0.1279],[185,0.1256],[187,0.124],[188,0.1245],[189,0.1278],[190,0.1324],[191,0.138],[192,0.146],[193,0.1546],[194,0.1617],[195,0.1703],[196,0.177],[197,0.1846],[198,0.1898],[199,0.1925],[200,0.1964],[201,0.1977],[203,0.1989],[204,0.1961],[205,0.1957],[206,0.1948],[207,0.1907],[208,0.1889],[209,0.1838],[210,0.1812],[211,0.1753],[212,0.1721],[214,0.1594],[215,0.1556],[216,0.1493],[217,0.1441],[218,0.1385],[220,0.1292],[222,0.1211],[223,0.1181],[224,0.1154],[225,0.1129],[227,0.1099],[228,0.1075],[229,0.1057],[230,0.1042],[231,0.103],[232,0.1021],[234,0.1015],[235,0.1024],[236,0.103],[237,0.1038],[239,0.107],[242,0.1127],[243,0.1148],[244,0.1174],[249,0.1278],[253,0.135],[255,0.138],[258,0.1413],[261,0.1433],[265,0.1444],[280,0.1438],[296,0.143],[297,0],[367,0]]
       }
       spine_04_ctrl.Location.Z {
        type=MovieSceneScriptingFloatChannel
        default=0
        thinned_from=368
        keys [[0,-0.131],[1,-0.131],[5,-0.1131],[7,-0.1048],[8,-0.1009],[9,-0.0972],[11,-0.0906],[12,-0.0876],[13,-0.085],[14,-0.0826],[16,-0.0786],[21,-0.0707],[23,-0.0686],[26,-0.0646],[28,-0.0614],[29,-0.061],[30,-0.0592],[33,-0.0573],[34,-0.0564],[37,-0.0556],[38,-0.0552],[42,-0.0555],[44,-0.0566],[45,-0.0574],[46,-0.0586],[48,-0.0629],[49,-0.0656],[50,-0.0675],[51,-0.0691],[52,-0.0714],[53,-0.0755],[54,-0.0774],[55,-0.0811],[56,-0.0824],[59,-0.0916],[61,-0.0974],[62,-0.1023],[65,-0.1109],[66,-0.1159],[67,-0.1189],[68,-0.1227],[70,-0.1301],[71,-0.1324],[72,-0.1363],[77,-0.1492],[78,-0.1504],[79,-0.1531],[80,-0.1543],[81,-0.1571],[84,-0.1611],[89,-0.1646],[90,-0.1644],[91,-0.1649],[92,-0.1646],[93,-0.1634],[94,-0.163],[95,-0.1618],[97,-0.1576],[98,-0.1563],[99,-0.1541],[100,-0.1511],[102,-0.147],[103,-0.144],[104,-0.1424],[105,-0.1395],[106,-0.1378],[108,-0.1365],[113,-0.1341],[124,-0.1327],[135,-0.1325],[136,-0.133],[152,-0.1324],[153,-0.1319],[168,-0.1319],[169,-0.1324],[175,-0.1343],[177,-0.1265],[178,-0.1223],[183,-0.0805],[184,-0.0722],[185,-0.0662],[186,-0.0615],[187,-0.0585],[188,-0.0569],[189,-0.0549],[191,-0.0535],[192,-0.054],[195,-0.065],[196,-0.0689],[197,-0.0697],[198,-0.0693],[199,-0.0677],[200,-0.0679],[201,-0.0669],[202,-0.0665],[203,-0.0662],[204,-0.0664],[205,-0.0678],[206,-0.0681],[207,-0.0699],[209,-0.0727],[211,-0.0747],[212,-0.0752],[213,-0.077],[214,-0.0785],[215,-0.0797],[216,-0.0835],[217,-0.0854],[218,-0.0878],[220,-0.0978],[222,-0.1131],[225,-0.1442],[226,-0.1574],[227,-0.1651],[228,-0.1647],[232,-0.1522],[233,-0.1471],[234,-0.1448],[235,-0.1415],[236,-0.1392],[239,-0.135],[241,-0.1343],[242,-0.1332],[243,-0.133],[244,-0.1319],[246,-0.132],[247,-0.1312],[252,-0.1322],[256,-0.1343],[261,-0.1373],[263,-0.138],[264,-0.1381],[265,-0.1378],[279,-0.1357],[280,-0.1347],[281,-0.1349],[296,-0.1333],[297,0],[367,0]]
       }
       spine_04_ctrl.Rotation.X {
        type=MovieSceneScriptingFloatChannel
        default=0
        thinned_from=368
        keys [[0,-7.8993],[6,-7.923],[9,-7.9901],[11,-8.0726],[13,-8.1981],[16,-8.464],[22,-9.0619],[25,-9.3053],[30,-9.5928],[36,-9.7966],[42,-9.8356],[44,-9.8253],[45,-9.8403],[50,-10.075],[52,-10.1096],[54,-10.0904],[56,-10.0139],[68,-9.1782],[72,-8.9807],[80,-8.7651],[93,-8.588],[112,-7.9673],[116,-7.9041],[171,-7.8355],[187,-7.9808],[188,-7.9708],[189,-7.9362],[190,-7.8804],[192,-7.7177],[196,-7.3463],[198,-7.2039],[200,-7.1155],[202,-7.0812],[204,-7.0823],[206,-7.1217],[208,-7.2016],[211,-7.382],[219,-7.9754],[223,-8.2038],[226,-8.3169],[229,-8.3766],[233,-8.3746],[251,-7.9889],[296,-7.8996],[297,0],[367,0]]
       }
       spine_04_ctrl.Rotation.Y {
        type=MovieSceneScriptingFloatChannel
        default=0
        thinned_from=368
        keys [[0,5.1836],[48,5.3253],[56,5.4531],[89,5.624],[95,5.5802],[115,5.2023],[162,5.0536],[167,5.0778],[169,5.1058],[170,5.1309],[171,5.1675],[174,5.3252],[175,5.3564],[176,5.361],[177,5.3352],[178,5.2783],[179,5.1932],[182,4.8628],[183,4.7693],[184,4.6957],[185,4.6429],[186,4.6133],[187,4.6056],[188,4.6145],[189,4.6401],[190,4.6786],[192,4.7837],[196,5.0414],[197,5.1161],[198,5.213],[199,5.3304],[200,5.4722],[201,5.6372],[202,5.8242],[204,6.2427],[206,6.6781],[207,6.8692],[208,7.0169],[209,7.1181],[210,7.1685],[211,7.1711],[212,7.1263],[213,7.0375],[214,6.9095],[215,6.7457],[216,6.5524],[217,6.3346],[218,6.0972],[220,5.5863],[223,4.7952],[225,4.2937],[226,4.0595],[227,3.8395],[228,3.6343],[229,3.4418],[230,3.2602],[231,3.0901],[232,2.9293],[233,2.7801],[234,2.6442],[235,2.5229],[236,2.4184],[237,2.3303],[238,2.2595],[239,2.2163],[240,2.2061],[241,2.2289],[242,2.2849],[243,2.3686],[244,2.48],[245,2.6155],[246,2.7715],[247,2.946],[248,3.1354],[249,3.3373],[251,3.7675],[255,4.6592],[256,4.8702],[257,5.0696],[258,5.2549],[259,5.4248],[260,5.5743],[261,5.7031],[262,5.8089],[263,5.8865],[264,5.9358],[265,5.9534],[268,5.9335],[272,5.8433],[284,5.3866],[288,5.273],[292,5.2051],[296,5.1829],[297,0],[367,0]]
       }
       spine_04_ctrl.Rotation.Z {
        type=MovieSceneScriptingFloatChannel
        default=0
        thinned_from=368
        keys [[0,-2.6348],[42,-2.6534],[43,-2.6712],[44,-2.7276],[45,-2.8237],[46,-2.9529],[47,-3.1045],[48,-3.2642],[49,-3.4154],[50,-3.5506],[51,-3.6698],[52,-3.7703],[53,-3.8529],[54,-3.9203],[55,-3.973],[56,-4.0134],[61,-4.1251],[72,-4.2584],[89,-4.3218],[90,-4.3167],[92,-4.2762],[94,-4.2007],[96,-4.0959],[98,-3.9686],[101,-3.7441],[108,-3.1663],[110,-3.0129],[112,-2.8765],[113,-2.8128],[114,-2.7576],[116,-2.6686],[118,-2.6057],[123,-2.483],[128,-2.3901],[134,-2.3119],[137,-2.2836],[142,-2.2514],[149,-2.2265],[168,-2.2165],[169,-2.201],[170,-2.1606],[171,-2.0828],[172,-1.9746],[173,-1.8416],[175,-1.5506],[176,-1.4182],[177,-1.307],[178,-1.2224],[179,-1.1657],[180,-1.1361],[181,-1.1302],[182,-1.1368],[183,-1.1495],[187,-1.21],[190,-1.2479],[193,-1.278],[194,-1.2914],[195,-1.3102],[198,-1.387],[199,-1.4074],[200,-1.4226],[203,-1.4581],[204,-1.4734],[205,-1.4942],[206,-1.5228],[207,-1.5594],[208,-1.6042],[209,-1.6571],[210,-1.7158],[212,-1.849],[215,-2.0574],[217,-2.187],[219,-2.3027],[221,-2.4024],[223,-2.4849],[225,-2.5486],[226,-2.5725],[228,-2.6032],[232,-2.6458],[233,-2.6508],[234,-2.6489],[235,-2.6383],[236,-2.618],[238,-2.561],[239,-2.5371],[240,-2.5205],[241,-2.5108],[243,-2.5083],[245,-2.523],[249,-2.58],[256,-2.7005],[260,-2.7469],[263,-2.7626],[292,-2.638],[296,-2.6349],[297,0],[367,0]]
       }
       spine_04_ctrl.Scale.X {type=MovieSceneScriptingFloatChannel, default=1, const [1,368]}
       spine_04_ctrl.Scale.Y {type=MovieSceneScriptingFloatChannel, default=1, const [1,368]}
       spine_04_ctrl.Scale.Z {type=MovieSceneScriptingFloatChannel, default=1, const [1,368]}
       head_ctrl.Location.X {type=MovieSceneScriptingFloatChannel, default=0, thinned_from=368, keys [[0,0],[367,0]]}
       head_ctrl.Location.Y {type=MovieSceneScriptingFloatChannel, default=0, thinned_from=368, keys [[0,0],[367,0]]}
       head_ctrl.Location.Z {type=MovieSceneScriptingFloatChannel, default=0, thinned_from=368, keys [[0,0],[367,0]]}
       head_ctrl.Rotation.X {
        type=MovieSceneScriptingFloatChannel
        default=0
        thinned_from=368
        keys [[0,20.3363],[6,20.2157],[12,19.8367],[30,18.0546],[33,17.8199],[35,17.5726],[37,17.2211],[39,16.7858],[41,16.2656],[46,14.7942],[47,14.5611],[48,14.3772],[49,14.258],[50,14.1978],[51,14.2019],[52,14.2519],[53,14.347],[55,14.6239],[58,15.0907],[68,16.1492],[78,16.8053],[86,17.1507],[90,17.485],[95,18.1339],[104,19.5812],[108,20.0652],[111,20.3565],[113,20.6501],[118,21.5993],[121,21.9812],[123,22.1274],[167,22.7621],[168,22.9035],[169,23.2296],[170,23.7061],[171,24.2929],[174,26.2094],[175,26.7535],[176,27.2013],[177,27.5467],[178,27.7932],[179,27.9597],[182,28.1854],[184,28.2116],[186,28.1257],[193,27.3204],[196,27.2106],[199,27.3038],[206,27.8215],[208,27.8181],[210,27.6396],[216,26.616],[218,26.115],[220,25.4407],[222,24.5981],[225,23.218],[226,22.7995],[229,21.7606],[232,20.8859],[233,20.6511],[234,20.4741],[235,20.3456],[236,20.268],[238,20.2433],[240,20.367],[243,20.7503],[251,22.1808],[253,22.4491],[255,22.6223],[257,22.6988],[260,22.6382],[265,22.2373],[278,20.6405],[283,20.2814],[296,20.0422],[297,0],[367,0]]
       }
       head_ctrl.Rotation.Y {
        type=MovieSceneScriptingFloatChannel
        default=0
        thinned_from=368
        keys [[0,12.289],[10,12.1456],[29,11.5682],[33,11.4903],[34,11.4212],[35,11.3072],[36,11.1491],[37,10.9449],[38,10.7004],[41,9.8877],[42,9.6453],[43,9.439],[44,9.2724],[45,9.1432],[46,9.0389],[50,8.7467],[52,8.6408],[54,8.5789],[56,8.5593],[59,8.5859],[63,8.7175],[72,9.1342],[79,9.3178],[84,9.3757],[85,9.3743],[87,9.314],[89,9.1878],[91,8.9894],[93,8.7285],[95,8.4073],[97,8.035],[100,7.4425],[102,7.0852],[103,6.9268],[104,6.7936],[105,6.7043],[106,6.6518],[107,6.6423],[108,6.6664],[109,6.7255],[110,6.8197],[111,6.9458],[115,7.5943],[117,7.8672],[119,8.0596],[122,8.2891],[124,8.4029],[134,8.6359],[166,8.8742],[167,8.8647],[168,8.6571],[169,8.1541],[170,7.3692],[171,6.3176],[172,5.0228],[173,3.5175],[174,1.8489],[175,0.074],[176,-1.7367],[177,-3.5029],[178,-5.1544],[179,-6.622],[180,-7.8493],[181,-8.821],[182,-9.6412],[183,-10.346],[184,-10.9898],[185,-11.5616],[186,-12.0563],[187,-12.46],[188,-12.7988],[189,-13.0916],[190,-13.3427],[191,-13.5373],[192,-13.6759],[193,-13.76],[194,-13.7872],[195,-13.7691],[196,-13.7107],[201,-13.2663],[209,-12.8784],[211,-12.7279],[212,-12.5592],[213,-12.2291],[214,-11.7573],[215,-11.1599],[216,-10.4521],[217,-9.6584],[218,-8.7916],[221,-6.0467],[222,-5.164],[223,-4.3344],[224,-3.5887],[225,-2.9417],[226,-2.4104],[227,-2.0146],[228,-1.6991],[229,-1.3984],[230,-1.1125],[231,-0.8533],[232,-0.619],[233,-0.4202],[234,-0.2379],[235,-0.0854],[236,0.0532],[237,0.1756],[238,0.3431],[239,0.6321],[240,1.0578],[241,1.6264],[242,2.3288],[243,3.1443],[244,4.0365],[246,5.9051],[247,6.7966],[248,7.6319],[249,8.391],[250,9.0655],[251,9.6603],[252,10.1862],[253,10.6413],[254,11.0387],[255,11.3718],[256,11.6495],[257,11.8759],[258,12.0523],[259,12.1847],[261,12.3615],[263,12.4699],[266,12.5215],[272,12.4367],[279,12.2553],[296,12.2896],[297,0],[367,0]]
       }
       head_ctrl.Rotation.Z {
        type=MovieSceneScriptingFloatChannel
        default=0
        thinned_from=368
        keys [[0,-20.5242],[32,-20.9492],[33,-20.9937],[34,-21.1181],[35,-21.3279],[36,-21.6375],[37,-22.0594],[38,-22.6001],[39,-23.2719],[40,-24.0883],[41,-25.0505],[42,-26.1637],[43,-27.4274],[44,-28.8178],[47,-33.3298],[48,-34.7176],[49,-35.9537],[50,-37.0217],[51,-37.9106],[52,-38.6271],[53,-39.1839],[54,-39.5955],[55,-39.8788],[58,-40.2805],[68,-40.4823],[83,-41.0098],[84,-40.9978],[85,-40.8988],[86,-40.6472],[87,-40.2465],[88,-39.713],[89,-39.0582],[90,-38.2965],[91,-37.4346],[93,-35.4636],[95,-33.2489],[100,-27.3037],[102,-25.1034],[103,-24.0941],[104,-23.152],[105,-22.2901],[106,-21.5106],[107,-20.8276],[108,-20.2518],[109,-19.7908],[110,-19.4581],[111,-19.2611],[112,-19.2117],[123,-19.8428],[166,-19.6419],[167,-19.5851],[168,-18.929],[169,-17.3404],[170,-14.908],[171,-11.7475],[172,-7.991],[173,-3.7954],[174,0.7022],[175,5.3619],[176,9.9938],[177,14.4003],[178,18.3983],[179,21.8056],[180,24.4639],[181,26.3458],[182,27.8191],[184,30.1891],[186,32.3494],[188,34.3067],[190,35.837],[192,37.1933],[194,38.3323],[196,39.231],[198,39.8897],[200,40.3396],[202,40.5713],[204,40.5387],[206,40.2605],[208,39.7467],[210,39.0421],[211,38.6292],[212,37.8494],[213,36.4067],[214,34.398],[215,31.911],[216,29.0311],[217,25.851],[218,22.4418],[219,18.8894],[221,11.6833],[222,8.1789],[223,4.8343],[224,1.7483],[225,-1.0253],[226,-3.3931],[227,-5.2836],[228,-6.8746],[229,-8.3874],[230,-9.831],[231,-11.1911],[232,-12.4738],[233,-13.6587],[234,-14.7489],[235,-15.7274],[236,-16.5981],[237,-17.3688],[238,-18.055],[239,-18.6591],[240,-19.1916],[241,-19.657],[242,-20.0505],[243,-20.3738],[244,-20.6298],[245,-20.8197],[246,-20.9501],[247,-21.0255],[248,-21.0447],[249,-21.0068],[254,-20.5023],[257,-20.3199],[296,-20.5215],[297,0],[367,0]]
       }
       head_ctrl.Scale.X {type=MovieSceneScriptingFloatChannel, default=1, const [1,368]}
       head_ctrl.Scale.Y {type=MovieSceneScriptingFloatChannel, default=1, const [1,368]}
       head_ctrl.Scale.Z {type=MovieSceneScriptingFloatChannel, default=1, const [1,368]}
       clavicle_l_ctrl.Location.X {type=MovieSceneScriptingFloatChannel, default=0, thinned_from=368, keys [[0,0],[367,0]]}
       clavicle_l_ctrl.Location.Y {type=MovieSceneScriptingFloatChannel, default=0, thinned_from=368, keys [[0,0],[367,0]]}
       clavicle_l_ctrl.Location.Z {type=MovieSceneScriptingFloatChannel, default=0, thinned_from=368, keys [[0,0],[367,0]]}
       clavicle_l_ctrl.Rotation.X {
        type=MovieSceneScriptingFloatChannel
        default=0
        thinned_from=368
        keys [[0,-0.0047],[1,-0.0047],[2,-0.0052],[3,-0.0062],[4,-0.0085],[5,-0.0111],[6,-0.0142],[7,-0.0179],[8,-0.0221],[9,-0.0267],[11,-0.0369],[12,-0.0423],[13,-0.0479],[16,-0.0655],[17,-0.0718],[19,-0.0831],[21,-0.094],[22,-0.0986],[23,-0.102],[24,-0.1052],[26,-0.1097],[27,-0.1118],[28,-0.1131],[30,-0.115],[31,-0.1156],[32,-0.1156],[34,-0.1162],[36,-0.1156],[38,-0.1142],[39,-0.1141],[41,-0.1133],[42,-0.1125],[43,-0.1106],[45,-0.1075],[46,-0.1055],[47,-0.1051],[48,-0.1024],[49,-0.1005],[50,-0.0972],[51,-0.0946],[52,-0.0916],[53,-0.0884],[55,-0.0792],[56,-0.0753],[57,-0.0687],[59,-0.0589],[60,-0.0544],[61,-0.0496],[63,-0.0406],[64,-0.0369],[65,-0.0341],[66,-0.0304],[67,-0.0288],[68,-0.0262],[71,-0.0209],[72,-0.0195],[76,-0.015],[79,-0.0124],[82,-0.0104],[83,-0.01],[84,-0.0093],[88,-0.0077],[89,-0.0071],[93,-0.0061],[94,-0.0059],[95,-0.0056],[97,-0.0053],[99,-0.0051],[100,-0.0049],[101,-0.005],[296,-0.0047],[297,0],[367,0]]
       }
       clavicle_l_ctrl.Rotation.Y {
        type=MovieSceneScriptingFloatChannel
        default=0
        thinned_from=368
        keys [[0,5.1755],[3,5.1836],[5,5.217],[8,5.31],[11,5.4513],[14,5.6349],[17,5.8565],[21,6.1855],[25,6.4475],[28,6.5934],[30,6.6583],[43,6.8011],[57,6.8076],[59,6.7813],[61,6.7217],[63,6.616],[66,6.4054],[74,5.7449],[76,5.6106],[79,5.4701],[83,5.3444],[89,5.2414],[296,5.1757],[297,0],[367,0]]
       }
       clavicle_l_ctrl.Rotation.Z {
        type=MovieSceneScriptingFloatChannel
        default=0
        thinned_from=368
        keys [[0,-0.0789],[1,-0.0789],[2,-0.082],[3,-0.0937],[4,-0.1128],[5,-0.1383],[6,-0.1695],[7,-0.2056],[8,-0.2468],[9,-0.2918],[10,-0.3403],[11,-0.3922],[12,-0.4469],[13,-0.5039],[15,-0.6201],[16,-0.6752],[17,-0.7279],[19,-0.8234],[21,-0.9061],[22,-0.9422],[23,-0.9752],[24,-1.0043],[25,-1.0298],[26,-1.052],[27,-1.0705],[28,-1.0855],[29,-1.0969],[30,-1.1048],[33,-1.1148],[36,-1.1191],[39,-1.1164],[41,-1.1096],[43,-1.0979],[45,-1.0797],[47,-1.056],[49,-1.0257],[51,-0.9889],[52,-0.9676],[54,-0.9193],[55,-0.8924],[56,-0.8619],[57,-0.8291],[59,-0.754],[61,-0.6707],[62,-0.625],[63,-0.5811],[64,-0.5358],[65,-0.4892],[67,-0.4022],[68,-0.3604],[69,-0.3207],[70,-0.284],[71,-0.2505],[72,-0.221],[73,-0.1955],[74,-0.175],[75,-0.1597],[76,-0.1499],[77,-0.1429],[78,-0.1363],[79,-0.1304],[81,-0.1199],[82,-0.1152],[83,-0.1108],[84,-0.1071],[85,-0.103],[87,-0.0969],[88,-0.0944],[89,-0.0917],[91,-0.0878],[93,-0.0846],[94,-0.0833],[95,-0.0825],[96,-0.0815],[97,-0.0809],[99,-0.0798],[101,-0.0793],[102,-0.0794],[103,-0.0792],[118,-0.079],[296,-0.0788],[297,0],[367,0]]
       }
       clavicle_l_ctrl.Scale.X {type=MovieSceneScriptingFloatChannel, default=1, const [1,368]}
       clavicle_l_ctrl.Scale.Y {type=MovieSceneScriptingFloatChannel, default=1, const [1,368]}
       clavicle_l_ctrl.Scale.Z {type=MovieSceneScriptingFloatChannel, default=1, const [1,368]}
       clavicle_r_ctrl.Location.X {type=MovieSceneScriptingFloatChannel, default=0, thinned_from=368, keys [[0,0],[367,0]]}
       clavicle_r_ctrl.Location.Y {type=MovieSceneScriptingFloatChannel, default=0, thinned_from=368, keys [[0,0],[367,0]]}
       clavicle_r_ctrl.Location.Z {type=MovieSceneScriptingFloatChannel, default=0, thinned_from=368, keys [[0,-0.0001],[367,0]]}
       clavicle_r_ctrl.Rotation.X {
        type=MovieSceneScriptingFloatChannel
        default=0
        thinned_from=368
        keys [[0,-0.8091],[3,-0.8107],[7,-0.8236],[11,-0.8458],[17,-0.8826],[22,-0.9061],[24,-0.912],[30,-0.9211],[32,-0.9198],[34,-0.922],[38,-0.9197],[42,-0.9199],[43,-0.9171],[44,-0.9178],[50,-0.9096],[53,-0.9013],[56,-0.8887],[59,-0.868],[60,-0.8598],[62,-0.8475],[64,-0.8378],[67,-0.8268],[71,-0.8192],[72,-0.8184],[74,-0.814],[75,-0.8133],[76,-0.8162],[77,-0.8134],[79,-0.8138],[80,-0.8107],[83,-0.8095],[84,-0.8111],[86,-0.8108],[87,-0.8092],[98,-0.8089],[99,-0.808],[296,-0.8088],[297,0],[367,0]]
       }
       clavicle_r_ctrl.Rotation.Y {
        type=MovieSceneScriptingFloatChannel
        default=0
        thinned_from=368
        keys [[0,7.0059],[2,7.0123],[4,7.0555],[7,7.1731],[12,7.4578],[23,8.258],[26,8.4188],[29,8.5226],[56,8.6422],[59,8.602],[61,8.5362],[64,8.3638],[75,7.5039],[77,7.3789],[80,7.2432],[85,7.1086],[296,7.0061],[297,0],[367,0]]
       }
       clavicle_r_ctrl.Rotation.Z {
        type=MovieSceneScriptingFloatChannel
        default=0
        thinned_from=368
        keys [[0,4.3405],[2,4.3374],[4,4.313],[7,4.2404],[10,4.1311],[14,3.9459],[26,3.2972],[29,3.1607],[31,3.0912],[33,3.0391],[34,3.0195],[36,2.9947],[39,2.9826],[41,2.9906],[44,3.0269],[47,3.0902],[50,3.177],[53,3.2905],[64,3.7733],[69,3.9628],[74,4.1154],[81,4.2357],[95,4.3324],[296,4.3408],[297,0],[367,0]]
       }
       clavicle_r_ctrl.Scale.X {type=MovieSceneScriptingFloatChannel, default=1, const [1,368]}
       clavicle_r_ctrl.Scale.Y {type=MovieSceneScriptingFloatChannel, default=1, const [1,368]}
       clavicle_r_ctrl.Scale.Z {type=MovieSceneScriptingFloatChannel, default=1, const [1,368]}
       neck_01_ctrl.Location.X {type=MovieSceneScriptingFloatChannel, default=0, thinned_from=368, keys [[0,0],[367,0]]}
       neck_01_ctrl.Location.Y {type=MovieSceneScriptingFloatChannel, default=0, thinned_from=368, keys [[0,0],[367,0]]}
       neck_01_ctrl.Location.Z {type=MovieSceneScriptingFloatChannel, default=0, thinned_from=368, keys [[0,0],[367,0]]}
       neck_01_ctrl.Rotation.X {
        type=MovieSceneScriptingFloatChannel
        default=0
        thinned_from=368
        keys [[0,-1.0722],[16,-1.0722],[32,-1.0734],[33,-1.0802],[34,-1.1094],[35,-1.1504],[36,-1.2051],[37,-1.2735],[38,-1.3555],[39,-1.4444],[40,-1.5469],[41,-1.6563],[43,-1.8885],[44,-2.0183],[45,-2.1412],[46,-2.2709],[50,-2.7622],[52,-2.9783],[53,-3.0733],[54,-3.1601],[55,-3.2386],[57,-3.3583],[60,-3.5111],[63,-3.6432],[65,-3.7217],[70,-3.8757],[74,-3.9634],[79,-4.0277],[82,-4.0453],[85,-4.0394],[86,-4.0147],[87,-3.9745],[88,-3.922],[89,-3.8572],[90,-3.783],[91,-3.6965],[93,-3.4988],[95,-3.2731],[98,-2.9051],[102,-2.4047],[103,-2.285],[105,-2.062],[106,-1.9608],[107,-1.8678],[108,-1.7831],[109,-1.7088],[110,-1.6468],[111,-1.5972],[112,-1.562],[114,-1.5083],[117,-1.4401],[118,-1.4215],[119,-1.3998],[122,-1.3495],[123,-1.3369],[124,-1.3202],[130,-1.2447],[132,-1.2279],[133,-1.2153],[135,-1.1986],[136,-1.1927],[137,-1.1825],[138,-1.1774],[139,-1.1671],[142,-1.1518],[143,-1.1416],[146,-1.1262],[147,-1.1262],[150,-1.1109],[151,-1.1109],[152,-1.1058],[154,-1.0999],[157,-1.0999],[158,-1.0966],[164,-1.0966],[165,-1.0933],[166,-1.0933],[167,-1.087],[168,-0.8132],[169,-0.0264],[170,1.387],[171,3.232],[172,4.9851],[173,6.3379],[174,7.295],[175,7.9327],[176,8.3315],[177,8.5331],[178,8.5352],[179,8.3123],[180,7.8799],[181,7.2366],[182,6.3961],[183,5.3853],[184,4.2547],[185,3.0823],[186,1.9921],[187,1.179],[188,0.4428],[189,-0.2924],[190,-1.2645],[191,-2.3181],[192,-3.3344],[193,-4.1952],[194,-4.7789],[195,-4.9918],[196,-4.9838],[197,-4.9641],[199,-4.8931],[201,-4.7838],[203,-4.6529],[205,-4.4982],[208,-4.2333],[210,-4.04],[213,-3.7215],[216,-3.3752],[219,-3.0023],[222,-2.6074],[226,-2.0621],[228,-1.7966],[229,-1.6738],[230,-1.5838],[231,-1.5303],[232,-1.481],[234,-1.3963],[236,-1.3279],[237,-1.2984],[239,-1.2535],[241,-1.2186],[242,-1.2053],[243,-1.196],[246,-1.1812],[249,-1.1809],[256,-1.1974],[266,-1.191],[272,-1.1756],[291,-1.0822],[295,-1.0726],[296,-1.0724],[297,0],[367,0]]
       }
       neck_01_ctrl.Rotation.Y {
        type=MovieSceneScriptingFloatChannel
        default=0
        thinned_from=368
        keys [[0,-8.6133],[167,-8.6058],[168,-8.429],[169,-8.0409],[170,-7.4953],[171,-6.8384],[173,-5.3907],[174,-4.6982],[175,-4.0909],[176,-3.6188],[177,-3.3296],[178,-3.1483],[179,-2.9788],[180,-2.8227],[181,-2.6769],[182,-2.5447],[183,-2.4232],[184,-2.3132],[185,-2.2136],[186,-2.1234],[187,-2.0427],[188,-1.9721],[190,-1.8806],[191,-1.8389],[192,-1.8016],[193,-1.7685],[194,-1.7392],[196,-1.6916],[198,-1.6566],[200,-1.6323],[202,-1.6174],[204,-1.6082],[207,-1.6027],[208,-1.6037],[209,-1.6259],[210,-1.688],[211,-1.7953],[212,-1.945],[213,-2.14],[214,-2.3719],[215,-2.6464],[216,-2.9525],[217,-3.2877],[218,-3.6441],[219,-4.0191],[222,-5.1759],[223,-5.5503],[224,-5.9105],[225,-6.2566],[226,-6.5791],[227,-6.883],[228,-7.1633],[229,-7.4154],[230,-7.6442],[231,-7.8498],[232,-8.0271],[233,-8.1862],[234,-8.3165],[235,-8.4296],[236,-8.5164],[237,-8.5816],[238,-8.6294],[239,-8.6598],[240,-8.6685],[241,-8.6597],[242,-8.6335],[244,-8.5335],[245,-8.4596],[247,-8.2683],[255,-7.2638],[257,-7.0468],[259,-6.8647],[261,-6.722],[263,-6.6171],[265,-6.5508],[267,-6.5172],[268,-6.5139],[270,-6.544],[272,-6.6279],[274,-6.7553],[277,-7.0103],[288,-8.1681],[291,-8.4184],[293,-8.5375],[295,-8.6032],[296,-8.6129],[297,0],[367,0]]
       }
       neck_01_ctrl.Rotation.Z {
        type=MovieSceneScriptingFloatChannel
        default=0
        thinned_from=368
        keys [[0,-1.3047],[166,-1.3048],[167,-1.3026],[168,-1.2418],[169,-1.1089],[170,-0.9226],[171,-0.699],[172,-0.4554],[173,-0.2086],[174,0.025],[175,0.2296],[176,0.3881],[177,0.4852],[178,0.5462],[179,0.6032],[180,0.6555],[181,0.7041],[182,0.7488],[183,0.7894],[184,0.8263],[185,0.8598],[186,0.89],[187,0.917],[188,0.9407],[191,0.9853],[193,1.0089],[195,1.0272],[198,1.0464],[201,1.0576],[208,1.064],[209,1.0571],[210,1.0358],[211,0.9996],[212,0.9493],[213,0.8845],[214,0.8064],[215,0.715],[216,0.6118],[217,0.4997],[218,0.3795],[219,0.2535],[220,0.1236],[221,-0.0067],[222,-0.1355],[223,-0.2624],[224,-0.3847],[225,-0.5013],[226,-0.6113],[227,-0.7138],[228,-0.8092],[229,-0.8956],[230,-0.9737],[231,-1.0429],[232,-1.1049],[233,-1.1579],[234,-1.2033],[235,-1.2412],[236,-1.272],[237,-1.2939],[238,-1.3103],[239,-1.3206],[240,-1.323],[241,-1.3202],[242,-1.312],[243,-1.2975],[244,-1.2771],[245,-1.2515],[246,-1.2208],[247,-1.1858],[248,-1.1472],[250,-1.0617],[253,-0.9277],[254,-0.8845],[255,-0.844],[256,-0.8056],[257,-0.7698],[258,-0.7374],[259,-0.7081],[260,-0.6821],[261,-0.6594],[262,-0.6403],[263,-0.6241],[264,-0.6112],[265,-0.6014],[266,-0.5947],[267,-0.5902],[268,-0.5892],[269,-0.592],[270,-0.5997],[271,-0.6116],[272,-0.6277],[273,-0.6484],[274,-0.6713],[275,-0.6981],[276,-0.7268],[277,-0.7578],[279,-0.8259],[281,-0.8984],[285,-1.0484],[287,-1.1192],[289,-1.1836],[290,-1.2122],[291,-1.2382],[292,-1.2604],[293,-1.2789],[294,-1.2926],[295,-1.3016],[296,-1.3044],[297,0],[367,0]]
       }
       neck_01_ctrl.Scale.X {type=MovieSceneScriptingFloatChannel, default=1, const [1,368]}
       neck_01_ctrl.Scale.Y {type=MovieSceneScriptingFloatChannel, default=1, const [1,368]}
       neck_01_ctrl.Scale.Z {type=MovieSceneScriptingFloatChannel, default=1, const [1,368]}
       neck_02_ctrl.Location.X {type=MovieSceneScriptingFloatChannel, default=0, thinned_from=368, keys [[0,0],[367,0]]}
       neck_02_ctrl.Location.Y {
        type=MovieSceneScriptingFloatChannel
        default=0
        thinned_from=368
        keys [[0,0.1094],[3,0.1094],[4,0.1104],[6,0.1104],[7,0.1164],[8,0.1114],[9,0.1174],[10,0.1124],[11,0.1184],[12,0.1134],[13,0.1144],[14,0.1204],[15,0.1153],[16,0.1163],[17,0.1167],[19,0.1194],[21,0.1152],[22,0.1149],[23,0.1179],[27,0.1167],[28,0.1131],[30,0.1125],[31,0.1088],[32,0.1122],[33,0.1122],[34,0.1115],[36,0.1115],[37,0.1131],[42,0.1131],[43,0.1148],[44,0.1131],[45,0.1148],[46,0.1132],[51,0.1131],[52,0.1124],[53,0.1121],[54,0.1114],[55,0.1096],[56,0.1104],[65,0.1044],[66,0.1022],[67,0.1027],[68,0.1021],[69,0.1019],[70,0.099],[80,0.0979],[81,0.1006],[84,0.1004],[85,0.1002],[86,0.1009],[87,0.1009],[88,0.1017],[90,0.1016],[91,0.1024],[93,0.1024],[95,0.104],[96,0.104],[97,0.1047],[98,0.1048],[100,0.1063],[101,0.1063],[102,0.1068],[118,0.1068],[119,0.1082],[216,0.1082],[217,0.1077],[232,0.1077],[233,0.1082],[296,0.1082],[297,0],[367,0]]
       }
       neck_02_ctrl.Location.Z {
        type=MovieSceneScriptingFloatChannel
        default=0
        thinned_from=368
        keys [[0,-0.3676],[3,-0.3676],[4,-0.3846],[6,-0.3846],[7,-0.3866],[8,-0.4016],[9,-0.4036],[10,-0.4186],[11,-0.4206],[12,-0.4355],[13,-0.4525],[14,-0.4545],[15,-0.4695],[16,-0.4865],[17,-0.4968],[19,-0.522],[21,-0.5571],[27,-0.6306],[31,-0.6767],[33,-0.6778],[34,-0.6814],[40,-0.6819],[41,-0.6765],[42,-0.6765],[43,-0.6716],[44,-0.6711],[45,-0.6662],[46,-0.6656],[50,-0.6439],[51,-0.6401],[52,-0.6321],[53,-0.6281],[54,-0.6201],[55,-0.6156],[66,-0.5277],[67,-0.5162],[68,-0.5096],[80,-0.3987],[81,-0.3905],[82,-0.3813],[83,-0.3813],[84,-0.3722],[85,-0.3708],[86,-0.366],[89,-0.3586],[90,-0.3586],[91,-0.3564],[97,-0.3571],[98,-0.3597],[100,-0.3602],[101,-0.3628],[102,-0.3635],[118,-0.3635],[119,-0.3661],[216,-0.3661],[217,-0.367],[232,-0.367],[233,-0.3661],[296,-0.3661],[297,0],[367,0]]
       }
       neck_02_ctrl.Rotation.X {
        type=MovieSceneScriptingFloatChannel
        default=0
        thinned_from=368
        keys [[0,1.2783],[16,1.2783],[21,1.2767],[27,1.2744],[31,1.2721],[36,1.2694],[37,1.2638],[42,1.2638],[43,1.2582],[47,1.2582],[48,1.2526],[52,1.2516],[58,1.247],[71,1.2457],[72,1.2507],[76,1.2507],[77,1.2556],[81,1.2556],[82,1.2606],[87,1.2613],[88,1.2659],[92,1.2659],[93,1.2706],[98,1.2706],[99,1.2752],[104,1.2758],[296,1.2781],[297,0],[367,0]]
       }
       neck_02_ctrl.Rotation.Y {type=MovieSceneScriptingFloatChannel, default=0, thinned_from=368, keys [[0,0],[367,0]]}
       neck_02_ctrl.Rotation.Z {type=MovieSceneScriptingFloatChannel, default=0, thinned_from=368, keys [[0,0],[367,0]]}
       neck_02_ctrl.Scale.X {type=MovieSceneScriptingFloatChannel, default=1, const [1,368]}
       neck_02_ctrl.Scale.Y {type=MovieSceneScriptingFloatChannel, default=1, const [1,368]}
       neck_02_ctrl.Scale.Z {type=MovieSceneScriptingFloatChannel, default=1, const [1,368]}
       head_ik_ctrl.Location.X {
        type=MovieSceneScriptingFloatChannel
        default=0
        thinned_from=368
        keys [[0,1.4921],[37,1.4966],[45,1.4774],[55,1.4182],[79,1.3598],[86,1.3576],[90,1.3666],[106,1.4611],[166,1.4934],[167,1.4922],[168,1.4658],[169,1.4105],[170,1.3367],[171,1.2468],[172,1.1398],[173,1.0199],[174,0.897],[175,0.7827],[176,0.6887],[177,0.6258],[178,0.582],[179,0.5415],[180,0.5052],[181,0.4727],[182,0.4438],[183,0.4179],[184,0.3952],[185,0.3758],[186,0.3593],[187,0.3457],[188,0.3344],[189,0.3281],[190,0.3227],[191,0.319],[194,0.3125],[195,0.309],[200,0.2888],[203,0.2773],[208,0.2559],[209,0.2552],[210,0.2616],[211,0.2759],[212,0.2979],[213,0.328],[214,0.3651],[215,0.4098],[216,0.4606],[217,0.5169],[218,0.5773],[219,0.6414],[222,0.8412],[223,0.9065],[224,0.9697],[225,1.0306],[226,1.0878],[227,1.1418],[228,1.1921],[229,1.2374],[230,1.2781],[231,1.314],[232,1.3453],[233,1.3736],[235,1.4182],[236,1.435],[237,1.4482],[238,1.4584],[239,1.4657],[240,1.4691],[241,1.4696],[242,1.4671],[244,1.4535],[245,1.4426],[247,1.413],[250,1.3545],[254,1.2696],[257,1.2144],[259,1.1845],[261,1.1611],[263,1.1438],[265,1.1327],[267,1.1269],[268,1.1263],[270,1.1315],[272,1.146],[274,1.1682],[277,1.2124],[288,1.4141],[291,1.4579],[293,1.4787],[295,1.4902],[296,1.4919],[297,0],[367,0]]
       }
       head_ik_ctrl.Location.Y {
        type=MovieSceneScriptingFloatChannel
        default=0
        thinned_from=368
        keys [[0,-0.1688],[3,-0.1687],[4,-0.17],[6,-0.1695],[7,-0.1636],[8,-0.1702],[9,-0.164],[10,-0.1703],[11,-0.1639],[12,-0.1699],[13,-0.1699],[14,-0.1629],[15,-0.1682],[17,-0.1671],[19,-0.1634],[21,-0.1674],[22,-0.1672],[23,-0.1637],[27,-0.1627],[28,-0.166],[30,-0.1661],[31,-0.1695],[32,-0.1656],[33,-0.1659],[34,-0.1706],[35,-0.1762],[36,-0.1839],[37,-0.1928],[38,-0.2049],[39,-0.2182],[40,-0.2336],[41,-0.2498],[43,-0.2853],[44,-0.3119],[45,-0.3376],[46,-0.3706],[47,-0.4025],[48,-0.4355],[49,-0.4672],[50,-0.4977],[51,-0.5248],[52,-0.5505],[53,-0.5728],[55,-0.6115],[58,-0.6478],[62,-0.6887],[66,-0.7225],[69,-0.7395],[70,-0.7475],[74,-0.7621],[79,-0.7714],[80,-0.7722],[81,-0.7701],[85,-0.7694],[86,-0.7647],[87,-0.7585],[89,-0.7392],[90,-0.7276],[92,-0.6957],[93,-0.6772],[95,-0.6342],[96,-0.6114],[98,-0.5629],[100,-0.5106],[103,-0.4349],[105,-0.3869],[106,-0.3645],[107,-0.3435],[108,-0.3239],[109,-0.3059],[110,-0.2901],[111,-0.2766],[112,-0.2655],[113,-0.2561],[114,-0.2473],[115,-0.2397],[116,-0.2326],[117,-0.2262],[118,-0.2209],[119,-0.2142],[122,-0.2],[123,-0.1962],[124,-0.1918],[128,-0.1777],[130,-0.1713],[132,-0.1665],[133,-0.1635],[135,-0.1592],[136,-0.1574],[137,-0.1552],[138,-0.1539],[139,-0.1517],[142,-0.1478],[143,-0.146],[146,-0.1427],[147,-0.1425],[150,-0.1396],[151,-0.1395],[154,-0.1376],[157,-0.1375],[158,-0.137],[164,-0.1369],[165,-0.1364],[166,-0.1363],[167,-0.1352],[168,-0.0891],[169,0.0399],[170,0.2658],[171,0.5556],[172,0.8285],[173,1.0397],[174,1.1905],[175,1.2925],[176,1.3575],[177,1.3926],[178,1.3987],[179,1.3726],[180,1.3164],[181,1.2293],[182,1.1133],[183,0.9715],[184,0.8106],[185,0.6413],[186,0.4819],[187,0.3619],[188,0.2522],[189,0.1419],[190,-0.0051],[191,-0.1657],[192,-0.3217],[193,-0.4547],[194,-0.5453],[195,-0.5788],[196,-0.5782],[197,-0.576],[199,-0.5669],[201,-0.5524],[203,-0.5352],[205,-0.5151],[208,-0.481],[210,-0.456],[212,-0.4295],[214,-0.4046],[215,-0.3934],[216,-0.3831],[218,-0.3664],[219,-0.3594],[220,-0.3534],[225,-0.3307],[227,-0.3199],[229,-0.3073],[230,-0.3048],[231,-0.3065],[232,-0.3072],[233,-0.3065],[234,-0.3042],[235,-0.3007],[236,-0.2948],[238,-0.2799],[242,-0.2486],[249,-0.1991],[251,-0.1859],[253,-0.1738],[254,-0.1681],[256,-0.1577],[257,-0.1529],[258,-0.1484],[259,-0.1443],[260,-0.1406],[261,-0.1375],[262,-0.1347],[263,-0.1324],[264,-0.1306],[265,-0.1293],[267,-0.1277],[269,-0.1271],[270,-0.1272],[272,-0.1284],[274,-0.1305],[277,-0.1352],[280,-0.1414],[288,-0.1598],[290,-0.1637],[292,-0.1669],[294,-0.1691],[296,-0.1698],[297,0],[367,0]]
       }
       head_ik_ctrl.Location.Z {
        type=MovieSceneScriptingFloatChannel
        default=0
        thinned_from=368
        keys [[0,-0.1845],[3,-0.1847],[4,-0.2017],[6,-0.202],[7,-0.2049],[8,-0.2193],[9,-0.2223],[10,-0.2367],[11,-0.2397],[12,-0.2542],[13,-0.2714],[14,-0.2745],[15,-0.289],[16,-0.3063],[17,-0.3167],[19,-0.3427],[21,-0.3777],[22,-0.3899],[23,-0.4035],[27,-0.4522],[28,-0.4628],[30,-0.487],[31,-0.4976],[32,-0.4989],[33,-0.4984],[34,-0.5001],[36,-0.4942],[37,-0.4905],[39,-0.48],[40,-0.4738],[41,-0.4617],[42,-0.4547],[43,-0.4423],[44,-0.4327],[45,-0.4184],[46,-0.4072],[50,-0.3421],[51,-0.3287],[52,-0.3115],[53,-0.2991],[54,-0.2836],[55,-0.272],[60,-0.2066],[63,-0.1685],[66,-0.1311],[67,-0.1156],[68,-0.1053],[69,-0.0925],[70,-0.0788],[71,-0.0665],[72,-0.0546],[73,-0.043],[74,-0.0312],[75,-0.0202],[76,-0.0089],[79,0.0234],[80,0.0337],[81,0.0428],[82,0.0528],[83,0.0534],[84,0.0632],[85,0.0647],[86,0.0684],[87,0.069],[88,0.0682],[89,0.0671],[90,0.0627],[91,0.0593],[92,0.0528],[93,0.0455],[94,0.0371],[95,0.0282],[96,0.0188],[97,0.0087],[98,-0.0038],[100,-0.0259],[101,-0.0392],[102,-0.0509],[103,-0.0616],[105,-0.0826],[106,-0.0924],[107,-0.1017],[108,-0.1104],[109,-0.1184],[110,-0.1254],[111,-0.1314],[112,-0.1363],[114,-0.1444],[116,-0.1509],[117,-0.1538],[118,-0.1562],[119,-0.1615],[122,-0.1677],[130,-0.1805],[135,-0.186],[139,-0.1896],[142,-0.1914],[146,-0.1941],[150,-0.1957],[151,-0.1958],[154,-0.1969],[157,-0.1971],[165,-0.1973],[166,-0.1971],[167,-0.1973],[168,-0.2184],[169,-0.2803],[170,-0.3954],[171,-0.5545],[172,-0.7164],[173,-0.8506],[174,-0.9521],[175,-1.0238],[176,-1.071],[177,-1.0955],[178,-1.0983],[179,-1.0787],[180,-1.0387],[181,-0.9791],[182,-0.9021],[183,-0.8116],[185,-0.6144],[186,-0.5258],[187,-0.4619],[188,-0.4055],[189,-0.3501],[190,-0.2786],[191,-0.2032],[192,-0.1325],[193,-0.0739],[194,-0.0345],[195,-0.0194],[196,-0.0188],[197,-0.0192],[198,-0.0206],[199,-0.023],[200,-0.0266],[201,-0.0308],[202,-0.0357],[203,-0.0411],[205,-0.0536],[207,-0.0677],[208,-0.0753],[209,-0.0826],[210,-0.089],[211,-0.0947],[212,-0.0998],[214,-0.1085],[215,-0.1126],[216,-0.117],[218,-0.1286],[219,-0.1354],[220,-0.1432],[221,-0.1521],[222,-0.162],[223,-0.173],[224,-0.1849],[225,-0.1974],[228,-0.2371],[229,-0.2498],[230,-0.2599],[231,-0.267],[232,-0.2733],[234,-0.2822],[235,-0.2857],[236,-0.2884],[237,-0.2903],[238,-0.2912],[239,-0.2911],[240,-0.2895],[241,-0.2869],[242,-0.2831],[243,-0.2784],[244,-0.2727],[246,-0.26],[251,-0.225],[253,-0.2124],[254,-0.2067],[256,-0.1967],[257,-0.1925],[258,-0.1887],[259,-0.1854],[260,-0.1826],[261,-0.1804],[262,-0.1787],[263,-0.1776],[264,-0.1771],[271,-0.1804],[296,-0.1829],[297,0],[367,0]]
       }
       head_ik_ctrl.Rotation.X {
        type=MovieSceneScriptingFloatChannel
        default=0
        thinned_from=368
        keys [[0,20.3363],[6,20.2157],[12,19.8367],[30,18.0546],[33,17.8199],[35,17.5726],[37,17.2211],[39,16.7858],[41,16.2656],[46,14.7942],[47,14.5611],[48,14.3772],[49,14.258],[50,14.1978],[51,14.2019],[52,14.2519],[53,14.347],[55,14.6239],[58,15.0907],[68,16.1492],[78,16.8053],[86,17.1507],[90,17.485],[95,18.1339],[104,19.5812],[108,20.0652],[111,20.3565],[113,20.6501],[118,21.5993],[121,21.9812],[123,22.1274],[167,22.7621],[168,22.9035],[169,23.2296],[170,23.7061],[171,24.2929],[174,26.2094],[175,26.7535],[176,27.2013],[177,27.5467],[178,27.7932],[179,27.9597],[182,28.1854],[184,28.2116],[186,28.1257],[193,27.3204],[196,27.2106],[199,27.3038],[206,27.8215],[208,27.8181],[210,27.6396],[216,26.616],[218,26.115],[220,25.4407],[222,24.5981],[225,23.218],[226,22.7995],[229,21.7606],[232,20.8859],[233,20.6511],[234,20.4741],[235,20.3456],[236,20.268],[238,20.2433],[240,20.367],[243,20.7503],[251,22.1808],[253,22.4491],[255,22.6223],[257,22.6988],[260,22.6382],[265,22.2373],[278,20.6405],[283,20.2814],[296,20.0422],[297,0],[367,0]]
       }
       head_ik_ctrl.Rotation.Y {
        type=MovieSceneScriptingFloatChannel
        default=0
        thinned_from=368
        keys [[0,12.289],[10,12.1456],[29,11.5682],[33,11.4903],[34,11.4212],[35,11.3072],[36,11.1491],[37,10.9449],[38,10.7004],[41,9.8877],[42,9.6453],[43,9.439],[44,9.2724],[45,9.1432],[46,9.0389],[50,8.7467],[52,8.6408],[54,8.5789],[56,8.5593],[59,8.5859],[63,8.7175],[72,9.1342],[79,9.3178],[84,9.3757],[85,9.3743],[87,9.314],[89,9.1878],[91,8.9894],[93,8.7285],[95,8.4073],[97,8.035],[100,7.4425],[102,7.0852],[103,6.9268],[104,6.7936],[105,6.7043],[106,6.6518],[107,6.6423],[108,6.6664],[109,6.7255],[110,6.8197],[111,6.9458],[115,7.5943],[117,7.8672],[119,8.0596],[122,8.2891],[124,8.4029],[134,8.6359],[166,8.8742],[167,8.8647],[168,8.6571],[169,8.1541],[170,7.3692],[171,6.3176],[172,5.0228],[173,3.5175],[174,1.8489],[175,0.074],[176,-1.7367],[177,-3.5029],[178,-5.1544],[179,-6.622],[180,-7.8493],[181,-8.821],[182,-9.6412],[183,-10.346],[184,-10.9898],[185,-11.5616],[186,-12.0563],[187,-12.46],[188,-12.7988],[189,-13.0916],[190,-13.3427],[191,-13.5373],[192,-13.6759],[193,-13.76],[194,-13.7872],[195,-13.7691],[196,-13.7107],[201,-13.2663],[209,-12.8784],[211,-12.7279],[212,-12.5592],[213,-12.2291],[214,-11.7573],[215,-11.1599],[216,-10.4521],[217,-9.6584],[218,-8.7916],[221,-6.0467],[222,-5.164],[223,-4.3344],[224,-3.5887],[225,-2.9417],[226,-2.4104],[227,-2.0146],[228,-1.6991],[229,-1.3984],[230,-1.1125],[231,-0.8533],[232,-0.619],[233,-0.4202],[234,-0.2379],[235,-0.0854],[236,0.0532],[237,0.1756],[238,0.3431],[239,0.6321],[240,1.0578],[241,1.6264],[242,2.3288],[243,3.1443],[244,4.0365],[246,5.9051],[247,6.7966],[248,7.6319],[249,8.391],[250,9.0655],[251,9.6603],[252,10.1862],[253,10.6413],[254,11.0387],[255,11.3718],[256,11.6495],[257,11.8759],[258,12.0523],[259,12.1847],[261,12.3615],[263,12.4699],[266,12.5215],[272,12.4367],[279,12.2553],[296,12.2896],[297,0],[367,0]]
       }
       head_ik_ctrl.Rotation.Z {
        type=MovieSceneScriptingFloatChannel
        default=0
        thinned_from=368
        keys [[0,-20.5242],[32,-20.9492],[33,-20.9937],[34,-21.1181],[35,-21.3279],[36,-21.6375],[37,-22.0594],[38,-22.6001],[39,-23.2719],[40,-24.0883],[41,-25.0505],[42,-26.1637],[43,-27.4274],[44,-28.8178],[47,-33.3298],[48,-34.7176],[49,-35.9537],[50,-37.0217],[51,-37.9106],[52,-38.6271],[53,-39.1839],[54,-39.5955],[55,-39.8788],[58,-40.2805],[68,-40.4823],[83,-41.0098],[84,-40.9978],[85,-40.8988],[86,-40.6472],[87,-40.2465],[88,-39.713],[89,-39.0582],[90,-38.2965],[91,-37.4346],[93,-35.4636],[95,-33.2489],[100,-27.3037],[102,-25.1034],[103,-24.0941],[104,-23.152],[105,-22.2901],[106,-21.5106],[107,-20.8276],[108,-20.2518],[109,-19.7908],[110,-19.4581],[111,-19.2611],[112,-19.2117],[123,-19.8428],[166,-19.6419],[167,-19.5851],[168,-18.929],[169,-17.3404],[170,-14.908],[171,-11.7475],[172,-7.991],[173,-3.7954],[174,0.7022],[175,5.3619],[176,9.9938],[177,14.4003],[178,18.3983],[179,21.8056],[180,24.4639],[181,26.3458],[182,27.8191],[184,30.1891],[186,32.3494],[188,34.3067],[190,35.837],[192,37.1933],[194,38.3323],[196,39.231],[198,39.8897],[200,40.3396],[202,40.5713],[204,40.5387],[206,40.2605],[208,39.7467],[210,39.0421],[211,38.6292],[212,37.8494],[213,36.4067],[214,34.398],[215,31.911],[216,29.0311],[217,25.851],[218,22.4418],[219,18.8894],[221,11.6833],[222,8.1789],[223,4.8343],[224,1.7483],[225,-1.0253],[226,-3.3931],[227,-5.2836],[228,-6.8746],[229,-8.3874],[230,-9.831],[231,-11.1911],[232,-12.4738],[233,-13.6587],[234,-14.7489],[235,-15.7274],[236,-16.5981],[237,-17.3688],[238,-18.055],[239,-18.6591],[240,-19.1916],[241,-19.657],[242,-20.0505],[243,-20.3738],[244,-20.6298],[245,-20.8197],[246,-20.9501],[247,-21.0255],[248,-21.0447],[249,-21.0068],[254,-20.5023],[257,-20.3199],[296,-20.5215],[297,0],[367,0]]
       }
       head_ik_ctrl.Scale.X {type=MovieSceneScriptingFloatChannel, default=1, const [1,368]}
       head_ik_ctrl.Scale.Y {type=MovieSceneScriptingFloatChannel, default=1, const [1,368]}
       head_ik_ctrl.Scale.Z {type=MovieSceneScriptingFloatChannel, default=1, const [1,368]}
       chest_ctrl.Location.X {
        type=MovieSceneScriptingFloatChannel
        default=0
        thinned_from=368
        keys [[0,-0.3736],[41,-0.3845],[42,-0.386],[43,-0.386],[44,-0.387],[45,-0.39],[46,-0.391],[47,-0.3929],[48,-0.3986],[49,-0.4006],[50,-0.4052],[51,-0.4068],[52,-0.4098],[53,-0.4117],[56,-0.415],[57,-0.4143],[59,-0.4154],[60,-0.4148],[61,-0.4152],[62,-0.4146],[63,-0.4151],[66,-0.4139],[67,-0.4144],[70,-0.4131],[84,-0.4122],[86,-0.4126],[87,-0.4136],[89,-0.4135],[90,-0.4123],[93,-0.4117],[94,-0.4109],[95,-0.4092],[96,-0.4087],[97,-0.402],[98,-0.4017],[101,-0.3972],[102,-0.3975],[103,-0.3938],[104,-0.392],[105,-0.3915],[110,-0.3824],[112,-0.3802],[115,-0.3742],[116,-0.3745],[117,-0.3729],[135,-0.368],[136,-0.3659],[152,-0.3664],[153,-0.3638],[169,-0.3639],[170,-0.3647],[172,-0.363],[173,-0.3631],[174,-0.3679],[175,-0.3829],[176,-0.4101],[177,-0.45],[178,-0.5036],[179,-0.5696],[180,-0.6391],[181,-0.7133],[182,-0.7902],[183,-0.8625],[184,-0.933],[186,-1.0512],[187,-1.0989],[188,-1.1411],[189,-1.1777],[191,-1.2316],[193,-1.2678],[194,-1.2779],[195,-1.2844],[196,-1.2877],[197,-1.2852],[198,-1.2679],[199,-1.2397],[200,-1.2],[201,-1.1492],[202,-1.0803],[203,-1.0059],[204,-0.9179],[205,-0.8224],[206,-0.733],[207,-0.6339],[208,-0.5425],[209,-0.4555],[210,-0.3761],[211,-0.302],[212,-0.2304],[213,-0.1708],[214,-0.114],[215,-0.0632],[216,-0.0202],[217,0.0177],[218,0.0515],[219,0.0813],[220,0.1072],[222,0.1444],[223,0.1562],[224,0.1665],[225,0.1731],[226,0.1759],[227,0.1773],[228,0.1736],[229,0.1676],[230,0.1578],[231,0.1449],[232,0.1262],[233,0.1051],[234,0.0784],[235,0.0499],[236,0.0186],[237,-0.0242],[239,-0.0871],[240,-0.1205],[241,-0.1487],[242,-0.178],[243,-0.2019],[244,-0.2216],[245,-0.2459],[246,-0.2641],[247,-0.2846],[248,-0.2984],[249,-0.3094],[250,-0.3234],[251,-0.3311],[252,-0.3422],[253,-0.3524],[254,-0.3556],[255,-0.3639],[256,-0.3647],[257,-0.3728],[258,-0.3753],[259,-0.3834],[260,-0.3856],[261,-0.3869],[262,-0.3923],[265,-0.3946],[268,-0.3948],[269,-0.3919],[274,-0.3909],[275,-0.3877],[278,-0.3868],[279,-0.384],[282,-0.3817],[285,-0.3812],[286,-0.3751],[296,-0.374],[297,0],[367,0]]
       }
       chest_ctrl.Location.Y {
        type=MovieSceneScriptingFloatChannel
        default=0
        thinned_from=368
        keys [[0,2.6975],[3,2.6948],[9,2.6576],[14,2.6023],[22,2.4925],[28,2.4268],[33,2.3975],[43,2.3961],[49,2.4178],[54,2.4551],[64,2.5581],[71,2.6295],[84,2.6859],[173,2.6871],[188,2.639],[220,2.7032],[296,2.6968],[297,0],[367,0]]
       }
       chest_ctrl.Location.Z {
        type=MovieSceneScriptingFloatChannel
        default=0
        thinned_from=368
        keys [[0,-0.2457],[1,-0.2457],[4,-0.2097],[5,-0.19],[7,-0.1654],[8,-0.1488],[9,-0.1364],[10,-0.1281],[11,-0.1154],[13,-0.0976],[16,-0.0594],[17,-0.052],[19,-0.0341],[21,-0.0139],[22,-0.0032],[23,0.0025],[24,0.013],[26,0.0243],[27,0.0347],[28,0.0379],[29,0.0469],[30,0.0522],[31,0.0562],[32,0.0626],[33,0.0665],[34,0.0695],[36,0.0708],[37,0.0765],[38,0.0771],[41,0.077],[42,0.0763],[43,0.0762],[44,0.0704],[45,0.0696],[47,0.0667],[48,0.0601],[49,0.0586],[50,0.0513],[51,0.0478],[52,0.0376],[53,0.0351],[55,0.0126],[56,0.0083],[61,-0.0527],[62,-0.0659],[63,-0.0824],[64,-0.0993],[65,-0.1126],[66,-0.1249],[67,-0.1381],[68,-0.1488],[69,-0.1656],[71,-0.1871],[72,-0.2023],[73,-0.2128],[76,-0.2402],[77,-0.2433],[78,-0.251],[79,-0.26],[80,-0.2676],[81,-0.2706],[82,-0.2782],[83,-0.2797],[84,-0.2873],[85,-0.2939],[90,-0.2981],[93,-0.2982],[95,-0.2968],[96,-0.2953],[97,-0.2946],[98,-0.2931],[99,-0.281],[100,-0.2787],[101,-0.2772],[102,-0.2674],[103,-0.2618],[104,-0.2585],[106,-0.2483],[109,-0.2491],[110,-0.2512],[114,-0.2509],[115,-0.2517],[118,-0.2516],[135,-0.2516],[136,-0.2521],[171,-0.2521],[172,-0.2546],[174,-0.2542],[175,-0.2477],[176,-0.2436],[177,-0.2334],[178,-0.221],[179,-0.2035],[180,-0.1839],[181,-0.1638],[182,-0.1424],[184,-0.1053],[185,-0.0926],[186,-0.0837],[187,-0.077],[188,-0.0724],[189,-0.0696],[190,-0.0713],[191,-0.0779],[192,-0.0834],[193,-0.0946],[194,-0.111],[195,-0.1243],[196,-0.138],[197,-0.1408],[198,-0.1392],[199,-0.1327],[200,-0.1277],[201,-0.1202],[203,-0.1033],[204,-0.0945],[205,-0.087],[206,-0.0808],[207,-0.0758],[208,-0.0715],[209,-0.069],[210,-0.0673],[211,-0.0661],[212,-0.0658],[213,-0.0677],[214,-0.0717],[215,-0.0783],[216,-0.0864],[217,-0.095],[218,-0.1031],[219,-0.1256],[220,-0.1365],[222,-0.1871],[223,-0.2089],[225,-0.2757],[226,-0.3118],[227,-0.3424],[228,-0.3422],[229,-0.3393],[230,-0.331],[231,-0.3164],[232,-0.3081],[233,-0.3014],[234,-0.2912],[236,-0.2829],[237,-0.2769],[238,-0.2769],[239,-0.2731],[240,-0.2713],[242,-0.2719],[243,-0.2702],[244,-0.2667],[247,-0.2679],[248,-0.2623],[249,-0.2639],[250,-0.2631],[251,-0.2609],[252,-0.2601],[253,-0.2601],[254,-0.259],[255,-0.2589],[256,-0.2566],[257,-0.2557],[259,-0.2552],[261,-0.2542],[262,-0.252],[263,-0.252],[264,-0.2512],[265,-0.2518],[296,-0.2508],[297,0],[367,0]]
       }
       chest_ctrl.Rotation.X {
        type=MovieSceneScriptingFloatChannel
        default=0
        thinned_from=368
        keys [[0,17.5147],[4,17.4473],[9,17.1679],[14,16.6987],[23,15.669],[28,15.2376],[32,15.0294],[41,14.9282],[50,14.9429],[53,15.0303],[60,15.5328],[70,16.4389],[79,16.8841],[187,17.6004],[197,18.0691],[200,18.059],[210,17.6686],[230,17.5776],[242,17.3741],[296,17.5188],[297,0],[367,0]]
       }
       chest_ctrl.Rotation.Y {
        type=MovieSceneScriptingFloatChannel
        default=0
        thinned_from=368
        keys [[0,4.3714],[33,4.2481],[42,4.2652],[43,4.2805],[44,4.3231],[45,4.3952],[46,4.494],[49,4.8542],[50,4.9647],[51,5.0639],[52,5.1494],[53,5.2232],[55,5.339],[56,5.3827],[70,5.7188],[87,5.8851],[90,5.8858],[92,5.8511],[94,5.7845],[96,5.6906],[99,5.5095],[109,4.7784],[112,4.5876],[114,4.4827],[116,4.402],[123,4.2377],[133,4.0955],[144,4.0217],[166,3.998],[168,4.003],[169,3.994],[170,3.9673],[171,3.9121],[172,3.8357],[173,3.7405],[174,3.6577],[175,3.6208],[176,3.6412],[177,3.7294],[178,3.8829],[179,4.0983],[180,4.3623],[181,4.659],[183,5.2715],[184,5.5592],[185,5.8257],[186,6.0659],[187,6.2796],[188,6.4688],[189,6.6318],[190,6.7731],[191,6.8927],[192,6.9922],[194,7.1391],[196,7.2367],[197,7.2528],[198,7.2211],[199,7.135],[200,6.9929],[201,6.797],[202,6.5482],[203,6.2578],[204,5.9306],[207,4.8836],[208,4.5474],[209,4.229],[210,3.9328],[211,3.6627],[212,3.4145],[213,3.1919],[214,2.9873],[215,2.8036],[216,2.6385],[217,2.4887],[218,2.355],[219,2.2362],[220,2.133],[221,2.0431],[222,1.9667],[223,1.9041],[224,1.8539],[225,1.8166],[226,1.7884],[227,1.7682],[228,1.7575],[229,1.7606],[230,1.7818],[231,1.8222],[232,1.8825],[233,1.9596],[234,2.0522],[235,2.1544],[238,2.4937],[245,3.3494],[248,3.6781],[250,3.8782],[252,4.0609],[254,4.2168],[257,4.4117],[259,4.5197],[261,4.6036],[262,4.6353],[264,4.6708],[270,4.6417],[289,4.397],[296,4.373],[297,0],[367,0]]
       }
       chest_ctrl.Rotation.Z {
        type=MovieSceneScriptingFloatChannel
        default=0
        thinned_from=368
        keys [[0,-3.4612],[36,-3.6003],[42,-3.5995],[43,-3.6462],[44,-3.7967],[45,-4.0557],[46,-4.4087],[47,-4.8189],[48,-5.2473],[49,-5.652],[50,-6.0171],[51,-6.33],[52,-6.5903],[53,-6.8044],[54,-6.9758],[55,-7.1064],[56,-7.2037],[60,-7.4229],[66,-7.6241],[80,-7.8249],[89,-7.851],[90,-7.8383],[91,-7.7974],[92,-7.7283],[93,-7.6432],[94,-7.5341],[96,-7.2618],[98,-6.9253],[100,-6.5471],[107,-5.051],[109,-4.6373],[110,-4.4425],[111,-4.257],[112,-4.0847],[113,-3.9254],[114,-3.7809],[115,-3.6548],[116,-3.5474],[117,-3.4603],[120,-3.2493],[122,-3.1258],[125,-2.9643],[128,-2.8261],[130,-2.7488],[133,-2.6522],[136,-2.5743],[140,-2.4994],[143,-2.4609],[145,-2.44],[150,-2.4076],[167,-2.3852],[168,-2.3735],[169,-2.3232],[170,-2.2012],[171,-1.9752],[172,-1.6605],[173,-1.2735],[174,-0.8278],[175,-0.3319],[176,0.1836],[177,0.6928],[178,1.1831],[179,1.6379],[180,2.0526],[181,2.4185],[182,2.7548],[183,3.0663],[184,3.3536],[185,3.6144],[186,3.852],[187,4.0655],[188,4.2565],[189,4.424],[190,4.5716],[191,4.6986],[192,4.8074],[193,4.8952],[194,4.9534],[195,4.9758],[196,4.9578],[197,4.8967],[198,4.7856],[199,4.627],[200,4.4205],[201,4.1591],[202,3.8404],[203,3.4649],[204,3.0361],[205,2.5658],[206,2.0616],[207,1.5263],[208,0.9707],[209,0.4015],[210,-0.1707],[211,-0.7434],[212,-1.3046],[213,-1.8545],[214,-2.3839],[215,-2.8904],[216,-3.3732],[217,-3.8291],[218,-4.2548],[219,-4.6525],[220,-5.0183],[221,-5.3523],[222,-5.6549],[223,-5.9233],[224,-6.1564],[225,-6.3561],[226,-6.5225],[227,-6.6519],[228,-6.752],[229,-6.8223],[230,-6.8691],[231,-6.8902],[232,-6.8833],[233,-6.835],[234,-6.7422],[235,-6.5989],[236,-6.411],[239,-5.7338],[240,-5.5245],[241,-5.3346],[242,-5.1618],[243,-5.006],[244,-4.8643],[245,-4.7336],[246,-4.6128],[248,-4.3977],[250,-4.2091],[252,-4.0493],[254,-3.9135],[257,-3.7303],[260,-3.5719],[262,-3.4949],[264,-3.442],[266,-3.4188],[272,-3.4046],[296,-3.4604],[297,0],[367,0]]
       }
       chest_ctrl.Scale.X {type=MovieSceneScriptingFloatChannel, default=1, const [1,368]}
       chest_ctrl.Scale.Y {type=MovieSceneScriptingFloatChannel, default=1, const [1,368]}
       chest_ctrl.Scale.Z {type=MovieSceneScriptingFloatChannel, default=1, const [1,368]}
       chest_tan_ctrl.Location.X {
        type=MovieSceneScriptingFloatChannel
        default=0
        thinned_from=368
        keys [[0,-0.0171],[4,-0.0171],[5,-0.0168],[7,-0.0167],[8,-0.0171],[9,-0.0172],[10,-0.0167],[11,-0.0167],[12,-0.0171],[13,-0.0169],[27,-0.0168],[28,-0.017],[31,-0.017],[32,-0.0172],[41,-0.018],[42,-0.0155],[43,-0.0162],[44,-0.0154],[45,-0.0185],[46,-0.0178],[47,-0.0175],[50,-0.0174],[51,-0.018],[52,-0.0164],[53,-0.0166],[54,-0.0178],[55,-0.0163],[56,-0.017],[57,-0.0185],[59,-0.0169],[60,-0.0181],[61,-0.0173],[62,-0.0184],[63,-0.017],[64,-0.0179],[65,-0.0168],[66,-0.0178],[67,-0.0167],[68,-0.0178],[70,-0.0181],[71,-0.0177],[72,-0.0179],[74,-0.0178],[75,-0.0183],[78,-0.018],[79,-0.0184],[80,-0.0182],[81,-0.0184],[82,-0.0181],[83,-0.0185],[84,-0.0182],[85,-0.0204],[86,-0.0205],[87,-0.0188],[89,-0.0188],[90,-0.0204],[91,-0.0198],[92,-0.0207],[93,-0.0196],[94,-0.02],[95,-0.0219],[96,-0.0215],[97,-0.017],[98,-0.0164],[99,-0.0163],[101,-0.018],[102,-0.0162],[103,-0.0192],[104,-0.0187],[105,-0.0179],[108,-0.0164],[111,-0.0157],[112,-0.0164],[113,-0.0168],[114,-0.0174],[115,-0.0184],[116,-0.0167],[117,-0.019],[118,-0.0172],[119,-0.0171],[135,-0.0192],[136,-0.0172],[137,-0.0165],[144,-0.0147],[145,-0.0148],[147,-0.0142],[149,-0.0143],[150,-0.014],[151,-0.014],[152,-0.0137],[153,-0.0169],[155,-0.0167],[156,-0.0168],[160,-0.0165],[168,-0.0173],[169,-0.0168],[170,-0.0159],[171,-0.0174],[172,-0.017],[173,-0.017],[174,-0.0193],[175,-0.0176],[177,-0.0201],[178,-0.0204],[179,-0.0189],[181,-0.0214],[182,-0.0211],[183,-0.0232],[184,-0.0222],[185,-0.0243],[186,-0.0238],[187,-0.0247],[188,-0.0253],[189,-0.0256],[190,-0.0273],[191,-0.0267],[192,-0.0279],[193,-0.0271],[195,-0.0282],[197,-0.0281],[198,-0.0289],[199,-0.0292],[200,-0.0287],[201,-0.0279],[202,-0.029],[203,-0.0261],[204,-0.028],[205,-0.0255],[206,-0.0255],[207,-0.0263],[208,-0.0274],[209,-0.0224],[211,-0.02],[212,-0.0218],[213,-0.0184],[214,-0.0203],[215,-0.0164],[216,-0.0164],[217,-0.015],[218,-0.0149],[219,-0.0143],[220,-0.012],[221,-0.0117],[222,-0.0108],[223,-0.0092],[224,-0.0098],[225,-0.0086],[226,-0.0068],[227,-0.0075],[228,-0.0058],[229,-0.0046],[230,-0.005],[231,-0.0063],[232,-0.0051],[233,-0.0046],[234,-0.0017],[235,-0.0083],[236,-0.0056],[237,-0.0018],[238,-0.0091],[239,-0.0076],[241,-0.0079],[242,-0.0093],[243,-0.0105],[244,-0.006],[245,-0.0089],[246,-0.0111],[247,-0.0147],[248,-0.0115],[249,-0.0115],[250,-0.0093],[251,-0.017],[252,-0.0135],[253,-0.0097],[254,-0.0161],[255,-0.0119],[256,-0.0185],[257,-0.0142],[258,-0.0196],[259,-0.0145],[260,-0.0191],[261,-0.0214],[262,-0.0157],[263,-0.0174],[264,-0.0184],[265,-0.018],[266,-0.018],[268,-0.0174],[269,-0.0217],[272,-0.0197],[274,-0.018],[275,-0.0218],[278,-0.0187],[279,-0.0225],[280,-0.0214],[281,-0.0171],[282,-0.0165],[283,-0.0154],[285,-0.0136],[286,-0.0234],[287,-0.0227],[290,-0.0211],[292,-0.0204],[295,-0.0199],[296,-0.0199],[297,0],[367,0]]
       }
       chest_tan_ctrl.Location.Y {
        type=MovieSceneScriptingFloatChannel
        default=0.0049
        thinned_from=368
        keys [[0,0.0588],[3,0.059],[4,0.0608],[5,0.0589],[6,0.0602],[7,0.0579],[8,0.0602],[9,0.0602],[10,0.0584],[11,0.0611],[12,0.0619],[13,0.0621],[14,0.0606],[15,0.0609],[16,0.0607],[17,0.0614],[19,0.0593],[21,0.0584],[22,0.0601],[23,0.0587],[24,0.0569],[25,0.0561],[26,0.0563],[27,0.0567],[28,0.0598],[29,0.0581],[30,0.0569],[31,0.0573],[32,0.0574],[33,0.0579],[34,0.0581],[35,0.0575],[36,0.0576],[37,0.0574],[38,0.0577],[41,0.0576],[42,0.0569],[43,0.0568],[44,0.0581],[45,0.0577],[46,0.0569],[47,0.0574],[48,0.0573],[49,0.0566],[51,0.0572],[52,0.0579],[53,0.0566],[54,0.0582],[55,0.0564],[56,0.0574],[57,0.0565],[58,0.0563],[61,0.0573],[62,0.0585],[63,0.0574],[64,0.0577],[65,0.0548],[66,0.0553],[67,0.056],[68,0.0579],[69,0.0574],[70,0.0559],[71,0.0575],[72,0.0579],[74,0.0578],[77,0.0571],[78,0.0566],[79,0.0574],[80,0.0566],[81,0.0575],[82,0.0566],[83,0.0573],[84,0.0561],[85,0.0579],[87,0.0571],[88,0.0566],[89,0.0568],[90,0.0561],[91,0.0558],[92,0.0563],[93,0.059],[95,0.0589],[98,0.0602],[99,0.0566],[100,0.0571],[101,0.0574],[102,0.0584],[103,0.0587],[104,0.0585],[106,0.0594],[107,0.059],[108,0.0589],[109,0.0592],[111,0.059],[112,0.0595],[114,0.059],[116,0.0593],[138,0.0592],[149,0.0594],[152,0.0596],[153,0.0592],[168,0.0595],[169,0.0588],[170,0.0585],[171,0.0586],[172,0.0591],[173,0.0587],[174,0.0586],[175,0.06],[176,0.059],[178,0.0592],[179,0.0605],[181,0.0619],[182,0.0624],[184,0.0638],[188,0.0649],[190,0.0655],[191,0.0654],[192,0.0658],[194,0.0655],[195,0.0655],[196,0.0651],[198,0.0656],[199,0.066],[201,0.066],[204,0.0668],[206,0.067],[207,0.0667],[208,0.0667],[209,0.0662],[213,0.0658],[216,0.0646],[217,0.064],[218,0.0649],[219,0.0641],[220,0.0639],[221,0.0619],[222,0.0594],[223,0.0598],[224,0.0586],[225,0.0578],[226,0.0557],[227,0.0552],[228,0.0555],[229,0.0562],[230,0.0552],[231,0.0572],[232,0.0561],[233,0.0563],[234,0.0574],[235,0.0572],[236,0.0566],[237,0.0566],[238,0.0571],[239,0.0574],[240,0.0565],[241,0.0571],[242,0.0564],[243,0.0581],[245,0.0572],[246,0.0573],[248,0.0584],[249,0.058],[250,0.058],[251,0.0587],[252,0.0587],[253,0.0584],[254,0.0592],[255,0.0587],[256,0.0591],[257,0.0589],[259,0.0594],[260,0.0588],[261,0.0589],[262,0.0593],[263,0.0593],[264,0.0597],[265,0.0595],[268,0.0597],[278,0.0593],[280,0.0591],[281,0.0596],[285,0.0595],[286,0.0592],[296,0.0591],[297,0.0049],[367,0.0049]]
       }
       chest_tan_ctrl.Location.Z {
        type=MovieSceneScriptingFloatChannel
        default=1.2539
        thinned_from=368
        keys [[0,1.2976],[1,1.2976],[4,1.2851],[5,1.2839],[7,1.2754],[8,1.2642],[9,1.2603],[10,1.2639],[11,1.2608],[12,1.2527],[16,1.2396],[30,1.2095],[31,1.2105],[32,1.2066],[33,1.2075],[36,1.2029],[45,1.2036],[47,1.207],[51,1.2087],[52,1.2152],[53,1.2122],[55,1.2227],[56,1.2178],[62,1.2405],[63,1.2519],[68,1.2675],[69,1.2728],[71,1.2791],[76,1.2963],[82,1.3082],[84,1.311],[85,1.3078],[93,1.314],[98,1.305],[99,1.3149],[101,1.3088],[102,1.3037],[107,1.2975],[168,1.3009],[174,1.3034],[175,1.2972],[176,1.2999],[178,1.2924],[184,1.2432],[188,1.2287],[190,1.2263],[192,1.2287],[196,1.245],[211,1.2347],[215,1.2406],[217,1.2468],[218,1.246],[219,1.254],[220,1.2578],[222,1.2831],[223,1.2907],[225,1.3132],[226,1.3293],[227,1.3356],[228,1.3359],[229,1.3312],[230,1.3327],[231,1.3242],[232,1.3256],[235,1.3149],[236,1.3164],[242,1.312],[243,1.3084],[247,1.3089],[248,1.3065],[264,1.2981],[296,1.3007],[297,1.2539],[367,1.2539]]
       }
       chest_tan_ctrl.Rotation.X {
        type=MovieSceneScriptingFloatChannel
        default=0.6533
        thinned_from=368
        keys [[0,9.9683],[5,9.9179],[12,9.6843],[27,8.9687],[32,8.8372],[49,8.8586],[54,8.9706],[71,9.6128],[189,10.0002],[197,10.1261],[209,9.9775],[227,10.1106],[296,9.9704],[297,0],[367,0]]
       }
       chest_tan_ctrl.Rotation.Y {
        type=MovieSceneScriptingFloatChannel
        default=0
        thinned_from=368
        keys [[0,1.5814],[35,1.5463],[42,1.5509],[43,1.5549],[44,1.5659],[45,1.5838],[46,1.6087],[49,1.7005],[51,1.7544],[53,1.7955],[55,1.8263],[56,1.8379],[71,1.9283],[89,1.9659],[92,1.9563],[95,1.928],[98,1.8863],[110,1.6678],[113,1.6223],[115,1.5985],[123,1.5469],[137,1.5014],[164,1.4851],[168,1.4875],[169,1.4845],[170,1.4779],[171,1.4628],[172,1.4436],[173,1.4184],[174,1.4101],[175,1.4421],[176,1.5186],[177,1.6455],[178,1.8181],[179,2.0327],[180,2.2781],[182,2.8061],[183,3.0636],[184,3.3033],[185,3.5221],[186,3.7167],[187,3.8868],[188,4.0349],[189,4.1602],[190,4.265],[191,4.3526],[192,4.4225],[193,4.4749],[194,4.5152],[195,4.5416],[196,4.5568],[197,4.5471],[198,4.4961],[199,4.3996],[200,4.2559],[201,4.0656],[202,3.8294],[203,3.5567],[204,3.2509],[205,2.9266],[207,2.267],[208,1.9521],[209,1.6548],[210,1.3797],[211,1.131],[212,0.905],[213,0.7057],[214,0.5255],[215,0.3682],[216,0.2315],[217,0.1113],[218,0.0082],[219,0],[233,0],[234,0.065],[235,0.1718],[236,0.2869],[237,0.4059],[238,0.5236],[239,0.6391],[240,0.7484],[241,0.8519],[242,0.9458],[243,1.0323],[244,1.11],[245,1.1831],[246,1.2458],[247,1.3013],[248,1.3537],[249,1.3993],[250,1.4382],[251,1.4739],[252,1.5041],[253,1.5298],[256,1.5791],[259,1.6172],[261,1.6383],[264,1.6534],[272,1.6402],[281,1.61],[289,1.5869],[290,1.5872],[291,1.5844],[292,1.5848],[293,1.5816],[296,1.582],[297,0],[367,0]]
       }
       chest_tan_ctrl.Rotation.Z {
        type=MovieSceneScriptingFloatChannel
        default=0
        thinned_from=368
        keys [[0,0],[176,0],[177,0.159],[178,0.4985],[179,0.8171],[180,1.1099],[181,1.3708],[182,1.6106],[183,1.8319],[184,2.0357],[185,2.2193],[186,2.385],[187,2.5333],[188,2.6648],[189,2.7778],[190,2.8771],[191,2.9612],[192,3.0313],[193,3.0877],[194,3.1229],[195,3.1336],[196,3.1169],[197,3.0702],[198,2.9883],[199,2.8722],[200,2.7215],[201,2.5303],[202,2.2995],[203,2.0285],[204,1.7191],[205,1.3816],[206,1.0219],[207,0.6415],[208,0.2508],[209,0],[367,0]]
       }
       chest_tan_ctrl.Scale.X {type=MovieSceneScriptingFloatChannel, default=1, const [1,368]}
       chest_tan_ctrl.Scale.Y {type=MovieSceneScriptingFloatChannel, default=1, const [1,368]}
       chest_tan_ctrl.Scale.Z {type=MovieSceneScriptingFloatChannel, default=1, const [1,368]}
       ParentConstraint_0 {type=MovieSceneScriptingConstraintChannel, const [true,1]}
      ]
     }]
   }
  ]
 }
]
display_rate {denominator=1, numerator=30}
tick_resolution {denominator=1, numerator=24000}
tracks []
